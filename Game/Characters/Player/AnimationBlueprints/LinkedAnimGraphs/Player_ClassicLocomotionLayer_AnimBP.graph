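# Linked anim layer Blueprint for the Fortnite player, supplying classic locomotion, idle and jump/fall additive poses.
# Jump and fall layers select additives by weapon core animation, targeting and DBNO state.
# LocomotionLayer_Additives runs an idle, start, move, stop, pivot and crouch-turn state machine from cached poses and cardinal-direction animation sets.
# Curve-driven per-bone blends mask arm and hand-IK additives, with a left-arm override hack for one-handed transitions.
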AnimBlueprint /Game/Characters/Player/AnimationBlueprints/LinkedAnimGraphs/Player_ClassicLocomotionLayer_AnimBP
target_skeleton=&Fortnite_M_Avg_Player_Skeleton
graphs [
 AnimGraph {nodes [
   Input Pose() {name=UpperBody}
   Output Pose() {layerGroup=None}
  ]}
 EventGraph {links [], nodes []}
 FallAdditive {nodes [
   n1 = Break Fort Legacy Weapon Anim Set!(FortLegacyWeaponAnimSet:"Fort Legacy Weapon Anim Set Structure (by ref)"=Get LegacyWeaponAnimSet() -> (LegacyWeaponAnimSet:Fort Legacy Weapon Anim Set Structure)) -> (
    RelaxedPose:Anim Sequence Object Reference
    TargetingPose:Anim Sequence Object Reference
    NonTargetingPose:Anim Sequence Object Reference
    CrouchNonTargetingPose:Anim Sequence Object Reference
    CrouchTargetingPose:Anim Sequence Object Reference
    CrouchRelaxedPose:Anim Sequence Object Reference
    CrouchCoreTargetingBlendSpace:Blend Space Object Reference
    CrouchCoreNonTargetingBlendSpace:Blend Space Object Reference
    IdleNoise:Anim Sequence Object Reference
    IdleNoise_AR_DownSights:Anim Sequence Object Reference
    JogAdditiveBlendSpace:Blend Space Object Reference
    JogAdditiveBlendSpaceRelaxed:Blend Space Object Reference
    JogAdditiveBlendSpaceMeleeGuarding:Blend Space Object Reference
    CrouchWalkAdditiveBlendSpace:Blend Space Object Reference
    CrouchJogAdditiveBlendSpace:Blend Space Object Reference
    CrouchJogAdditiveBlendSpaceRelaxed:Blend Space Object Reference
    SprintAnimation:Anim Sequence Object Reference
    SprintTargetingAnimation:Anim Sequence Object Reference
    CrouchSprintAnimation:Anim Sequence Object Reference
    CrouchSprintTargetingAnimation:Anim Sequence Object Reference
    FallAdditive:Anim Sequence Object Reference
    JumpLoopAdditive:Anim Sequence Object Reference
    JumpUpAdditive:Anim Sequence Object Reference
    FullBodyJumpAdditive:Anim Sequence Object Reference
   ) {is=title K2Node_BreakStruct}
   n2 = Get FallPlayRate() -> (FallPlayRate:float)
   n3 = "Sequence Player (additive) 'M_JumpFall_ARDS_Additive'"!(PlayRate:float=@n2) -> (Pose) {bLoopAnimation=false, sequence=/Game/Animation/Game/MainPlayer/Locomotion/Medium/Male/Jump/AssaultRifle/M_JumpFall_ARDS_Additive}
   n4 = "Sequence Player (additive) 'M_JumpFall_ARDS_Additive'"(PlayRate=@n2) {bLoopAnimation=false, sequence=&M_JumpFall_ARDS_Additive}
   n5 = Get bIsTargeting()
   n6 = Get ADSToNonADSBlendTime() -> (ADSToNonADSBlendTime:float)
   n7 = "Sequence Player (additive) 'Balloons_JumpFall_Base_Additive_CMM'"!(PlayRate:float=Get FallPlayRate()) -> (Pose) {bLoopAnimation=false, sequence=/Game/Animation/Game/MainPlayer/Combat/Gadgets/Medium/Male/Balloons/Balloons_JumpFall_Base_Additive_CMM}
   n8 = "Blend Poses (EFortWeaponCoreAnimation)"!(
    BlendPose_0:Pose Link Structure=Blend Poses by bool(
     BlendPose_0=Additive Identity Pose()
     BlendPose_1=Sequence Player!(Sequence:Anim Sequence Base Object Reference=@n1.FallAdditive, PlayRate:float=@n2) -> (Pose) {bLoopAnimation=false}
     bActiveValue=Get bIsDBNO()
    ) {blendType=Linear}
    BlendPose_1:Pose Link Structure=Blend Poses by bool(
     BlendPose_0="Sequence Player (additive) 'M_JumpFall_ARDS_Additive'"(PlayRate=@n2) {bLoopAnimation=false, sequence=&M_JumpFall_ARDS_Additive}
     BlendPose_1=Blend Poses by bool(BlendPose_0=@n3, BlendPose_1=@n4, BlendTime_1=@n6, bActiveValue=@n5) {blendType=Linear}
     BlendTime_0=.2
     BlendTime_1=.2
     bActiveValue=Get bIsCrouching()
    ) {blendType=Linear}
    BlendPose_2:Pose Link Structure=@n7
    BlendTime_0:float=0
    BlendTime_1:float=0
    BlendTime_2:float=0
    ActiveEnumValue:Byte=Get WeaponCoreAnim!() -> (WeaponCoreAnim)
   ) -> (Pose:Pose Link Structure) {
    is=title AnimGraphNode_BlendListByEnum
    blendType=Linear
    blendPose [{linkId=-1, sourceLinkId=-1}, {linkId=-1, sourceLinkId=-1}, {linkId=-1, sourceLinkId=-1}]
    blendTime [0.1,0.1,0.1]
   }
   n9 = Get FallPlayRate()
   Output Pose(Result=Blend Poses by bool(
     BlendPose_0="Sequence Player (additive) 'M_JumpFall_ARDS_Additive'"(PlayRate=@n9) {bLoopAnimation=false, sequence=&M_JumpFall_ARDS_Additive}
     BlendPose_1=@n8
     BlendTime_0=0
     BlendTime_1=0
     bActiveValue=Get bIsDBNOCarrying() -> (bIsDBNOCarrying:bool)
    ) {blendType=Linear})
  ]}
 IdleAdditiveLayer {nodes [
   AnimGraphNode_StateMachine_18 = IdleAdditive() {is=title AnimGraphNode_StateMachine, bAllowConduitEntryStates=true}
   Output Pose(Result=Two Way Blend!(A=@AnimGraphNode_StateMachine_18, B=Additive Identity Pose(), AlphaCurveName:Name=DisableIdleNoiseAdditive) -> (Pose) {blendNode {alphaInputType=Curve, alphaBoolBlend {blendOption=Linear}}})
  ]}
 IdleAdditive {
  entry=Ground
  path=IdleAdditiveLayer.AnimGraphNode_StateMachine_18.IdleAdditive
  ids=id
  nodes [AnimStateNode_2 = Ground() {is=title AnimStateNode}]
 }
 Ground {
  path=IdleAdditiveLayer.AnimGraphNode_StateMachine_18.IdleAdditive.AnimStateNode_2.Ground
  nodes [
   n1 = Break Fort Legacy Weapon Anim Set(FortLegacyWeaponAnimSet=Get LegacyWeaponAnimSet())
   Output Animation Pose(Result=Two Way Blend!(
     A=Sequence Player!(Sequence=@n1.IdleNoise) -> (Pose)
     B=Sequence Player(Sequence=@n1.IdleNoise_AR_DownSights)
     Alpha:float=Get LeftHandIKAlpha() -> (LeftHandIKAlpha:float)
    ) -> (Pose) {blendNode {alphaBoolBlend {blendOption=Linear}}})
  ]
 }
 JumpLoopAdditive {nodes [
   n1 = Break Fort Legacy Weapon Anim Set(FortLegacyWeaponAnimSet=Get LegacyWeaponAnimSet())
   n2 = "Blend Poses (EFortWeaponCoreAnimation)"(
    BlendPose_0=Blend Poses by bool(BlendPose_0=Additive Identity Pose(), BlendPose_1=Sequence Player(Sequence=@n1.JumpLoopAdditive) {playRate=0.5}, bActiveValue=Get bIsDBNO()) {blendType=Linear}
    BlendPose_1=Blend Poses by bool(
     BlendPose_0="Sequence Player (additive) 'M_JumpApex_ARDS_Loop_Additive'"() {playRate=0.5, sequence=/Game/Animation/Game/MainPlayer/Locomotion/Medium/Male/Jump/AssaultRifle/M_JumpApex_ARDS_Loop_Additive}
     BlendPose_1="Sequence Player (additive) 'M_JumpApex_ARDS_Loop_Additive'"() {playRate=0.5, sequence=&M_JumpApex_ARDS_Loop_Additive}
     BlendTime_1=Get ADSToNonADSBlendTime()
     bActiveValue=Get bIsTargeting()
    ) {blendType=Linear}
    BlendPose_2="Sequence Player (additive) 'Balloons_JumpApex_Base_Additive_CMM'"() {sequence=/Game/Animation/Game/MainPlayer/Combat/Gadgets/Medium/Male/Balloons/Balloons_JumpApex_Base_Additive_CMM}
    BlendTime_0=0
    BlendTime_1=0
    BlendTime_2=0
    ActiveEnumValue=Get WeaponCoreAnim()
   ) {
    blendType=Linear
    blendPose [{linkId=-1, sourceLinkId=-1}, {linkId=-1, sourceLinkId=-1}, {linkId=-1, sourceLinkId=-1}]
    blendTime [0.1,0.1,0.1]
   }
   Output Pose(Result=Blend Poses by bool(
     BlendPose_0="Sequence Player (additive) 'M_JumpApex_ARDS_Loop_Additive'"() {playRate=0.5, sequence=&M_JumpApex_ARDS_Loop_Additive}
     BlendPose_1=@n2
     BlendTime_0=0
     BlendTime_1=0
     bActiveValue=Get bIsDBNOCarrying()
    ) {blendType=Linear})
  ]}
 JumpUpAdditive {nodes [
   n1 = Break Fort Legacy Weapon Anim Set(FortLegacyWeaponAnimSet=Get LegacyWeaponAnimSet())
   n2 = "Sequence Player (additive) 'Balloons_JumpUp_Base_Additive_CMM'"() {bLoopAnimation=false, sequence=/Game/Animation/Game/MainPlayer/Combat/Gadgets/Medium/Male/Balloons/Balloons_JumpUp_Base_Additive_CMM}
   n3 = "Blend Poses (EFortWeaponCoreAnimation)"(
    BlendPose_0=Blend Poses by bool(
     BlendPose_0=Additive Identity Pose()
     BlendPose_1=Sequence Player(Sequence=@n1.JumpUpAdditive) {bLoopAnimation=false}
     bActiveValue=Get bIsDBNO()
    ) {blendType=Linear}
    BlendPose_1=Blend Poses by bool(
     BlendPose_0="Sequence Player (additive) 'M_JumpUp_ARDS_Additive'"() {bLoopAnimation=false, sequence=/Game/Animation/Game/MainPlayer/Locomotion/Medium/Male/Jump/AssaultRifle/M_JumpUp_ARDS_Additive}
     BlendPose_1="Sequence Player (additive) 'M_JumpUp_ARDS_Additive'"() {bLoopAnimation=false, sequence=&M_JumpUp_ARDS_Additive}
     BlendTime_1=Get ADSToNonADSBlendTime()
     bActiveValue=Get bIsTargeting()
    ) {blendType=Linear}
    BlendPose_2=@n2
    BlendTime_0=0
    BlendTime_1=0
    BlendTime_2=0
    ActiveEnumValue=Get WeaponCoreAnim()
   ) {
    blendType=Linear
    blendPose [{linkId=-1, sourceLinkId=-1}, {linkId=-1, sourceLinkId=-1}, {linkId=-1, sourceLinkId=-1}]
    blendTime [0.1,0.1,0.1]
   }
   Output Pose(Result=Blend Poses by bool(
     BlendPose_0="Sequence Player (additive) 'M_JumpUp_ARDS_Additive'"() {bLoopAnimation=false, sequence=&M_JumpUp_ARDS_Additive}
     BlendPose_1=@n3
     BlendTime_0=0
     BlendTime_1=0
     bActiveValue=Get bIsDBNOCarrying()
    ) {blendType=Linear})
  ]}
 LocomotionLayer_Additives {nodes [
   AnimGraphNode_StateMachine_0 = LocomotionAdditives() {is=title AnimGraphNode_StateMachine}
   Output Pose(Result=@AnimGraphNode_StateMachine_0) {layerGroup=LocomotionOverride}
   Save cached pose 'IdleAdditivePose'(Pose=IdleAdditiveLayer!() -> (Pose:Pose Link Structure) {is=title AnimGraphNode_LinkedAnimLayer, layer=IdleAdditiveLayer}) {cacheName=IdleAdditivePose}
   AnimGraphNode_StateMachine_10 = MovementLayerAdditivePose() {is=title AnimGraphNode_StateMachine, bAllowConduitEntryStates=true}
   Save cached pose 'MovementAdditiveLayerPose'(Pose=@AnimGraphNode_StateMachine_10) {cacheName=MovementAdditiveLayerPose}
   AnimGraphNode_StateMachine_2 = "Movement TransitionAdditiveLayerPose(Hack) "() {is=title AnimGraphNode_StateMachine}
   Save cached pose 'MovementTransitionAdditiveLayerPose'(Pose=@AnimGraphNode_StateMachine_2) {cacheName=MovementTransitionAdditiveLayerPose}
  ]}
 LocomotionAdditives {
  entry=State
  path=LocomotionLayer_Additives.AnimGraphNode_StateMachine_0.LocomotionAdditives
  ids=id
  nodes [AnimStateNode_0 = State() {is=title AnimStateNode}]
 }
 State {
  path=LocomotionLayer_Additives.AnimGraphNode_StateMachine_0.LocomotionAdditives.AnimStateNode_0.State
  nodes [
   AnimGraphNode_StateMachine_1 = LocomotionAdditive() {is=title AnimGraphNode_StateMachine}
   n1 = Layered blend per bone(
    BasePose=@AnimGraphNode_StateMachine_1
    BlendPoses_0=Additive Identity Pose()
    BlendWeights_0=Get DisableArmsHeadAdditiveCurveAlpha() -> (DisableArmsHeadAdditiveCurveAlpha:float)
   ) {
    bMeshSpaceRotationBlend=true
    lODThreshold=2
    layerSetup [{branchFilters [{boneName=clavicle_l}, {boneName=clavicle_r}, {boneName=ik_hand_root}]}]
   }
   n2 = Layered blend per bone(
    BasePose=@n1
    BlendPoses_0=Additive Identity Pose()
    BlendWeights_0=Get DisableRightArmAdditiveCurveAlpha() -> (DisableRightArmAdditiveCurveAlpha:float)
   ) {bMeshSpaceRotationBlend=true, lODThreshold=2, layerSetup [{branchFilters [{boneName=clavicle_r}, {boneName=ik_hand_root}]}]}
   Output Animation Pose(Result=Layered blend per bone(
     BasePose=@n2
     BlendPoses_0=Additive Identity Pose()
     BlendWeights_0=Get DisableIKRootAdditiveCurveAlpha() -> (DisableIKRootAdditiveCurveAlpha:float)
    ) {bMeshSpaceRotationBlend=true, lODThreshold=2, layerSetup [{branchFilters [{boneName=ik_hand_root}]}]})
  ]
 }
 LocomotionAdditive {
  entry=IdleAdditive
  path=LocomotionLayer_Additives.AnimGraphNode_StateMachine_0.LocomotionAdditives.AnimStateNode_0.State.AnimGraphNode_StateMachine_1.LocomotionAdditive
  ids=id
  nodes [
   AnimStateNode_0 = IdleAdditive() {is=title AnimStateNode}
   AnimStateNode_1 = JogStartAdditive() {is=title AnimStateNode}
   AnimStateNode_2 = JogStopAdditive() {is=title AnimStateNode}
   AnimStateNode_3 = JogPostPivotAdditive() {is=title AnimStateNode, bAlwaysResetOnEntry=true}
   AnimStateNode_4 = CrouchTurning() {is=title AnimStateNode}
   AnimStateNode_6 = JogPrePivotAdditive() {is=title AnimStateNode, bAlwaysResetOnEntry=true}
   AnimStateNode_7 = MovementAdditive() {is=title AnimStateNode}
   AnimStateTransitionNode_0 = AnimStateTransitionNode()
   AnimStateTransitionNode_10 = AnimStateTransitionNode() {
    crossfadeDuration=0.4
    priorityOrder=2
    blendProfileWrapper {blendProfile=&Fortnite_M_Avg_Player_Skeleton:FastFeet}
   }
   AnimStateTransitionNode_11 = AnimStateTransitionNode() {crossfadeDuration=0.25, priorityOrder=2}
   AnimStateTransitionNode_12 = AnimStateTransitionNode() {crossfadeDuration=0.25}
   AnimStateTransitionNode_13 = AnimStateTransitionNode() {bAutomaticRuleBasedOnSequencePlayerInState=true, crossfadeDuration=0.5, priorityOrder=2}
   AnimStateTransitionNode_14 = AnimStateTransitionNode() {crossfadeDuration=0.3, priorityOrder=3}
   AnimStateTransitionNode_15 = AnimStateTransitionNode() {crossfadeDuration=0.35}
   AnimStateTransitionNode_2 = AnimStateTransitionNode() {bAutomaticRuleBasedOnSequencePlayerInState=true, crossfadeDuration=0.5}
   AnimStateTransitionNode_25 = AnimStateTransitionNode() {crossfadeDuration=0.0667}
   AnimStateTransitionNode_26 = AnimStateTransitionNode()
   AnimStateTransitionNode_27 = AnimStateTransitionNode() {bAutomaticRuleBasedOnSequencePlayerInState=true, crossfadeDuration=0.4}
   AnimStateTransitionNode_28 = AnimStateTransitionNode() {
    crossfadeDuration=0.15
    blendProfileWrapper {blendProfile=&Fortnite_M_Avg_Player_Skeleton:FastFeet}
   }
   AnimStateTransitionNode_3 = AnimStateTransitionNode()
   AnimStateTransitionNode_4 = AnimStateTransitionNode() {crossfadeDuration=0.25}
   AnimStateTransitionNode_5 = AnimStateTransitionNode()
   AnimStateTransitionNode_6 = AnimStateTransitionNode() {blendProfileWrapper {blendProfile=&Fortnite_M_Avg_Player_Skeleton:FastFeet}}
   AnimStateTransitionNode_7 = AnimStateTransitionNode() {crossfadeDuration=0.25, priorityOrder=2}
   AnimStateTransitionNode_8 = AnimStateTransitionNode() {
    bAutomaticRuleBasedOnSequencePlayerInState=true
    crossfadeDuration=0.75
    blendProfileWrapper {blendProfile=&Fortnite_M_Avg_Player_Skeleton:FastFeet}
   }
   AnimStateTransitionNode_9 = AnimStateTransitionNode() {priorityOrder=3, blendProfileWrapper {blendProfile=&Fortnite_M_Avg_Player_Skeleton:FastFeet}}
  ]
  links [
   {
    dst=IdleAdditive
    rule=LocomotionLayer_Additives.AnimGraphNode_StateMachine_0.LocomotionAdditives.AnimStateNode_0.State.AnimGraphNode_StateMachine_1.LocomotionAdditive.AnimStateTransitionNode_15.Transition
    src=CrouchTurning
    via=AnimStateTransitionNode_15
   }
   {
    dst=CrouchTurning
    rule=LocomotionLayer_Additives.AnimGraphNode_StateMachine_0.LocomotionAdditives.AnimStateNode_0.State.AnimGraphNode_StateMachine_1.LocomotionAdditive.AnimStateTransitionNode_14.Transition
    src=IdleAdditive
    via=AnimStateTransitionNode_14
   }
   {
    dst=JogPrePivotAdditive
    rule=LocomotionLayer_Additives.AnimGraphNode_StateMachine_0.LocomotionAdditives.AnimStateNode_0.State.AnimGraphNode_StateMachine_1.LocomotionAdditive.AnimStateTransitionNode_4.Transition
    src=IdleAdditive
    via=AnimStateTransitionNode_4
   }
   {
    dst=JogStartAdditive
    rule=LocomotionLayer_Additives.AnimGraphNode_StateMachine_0.LocomotionAdditives.AnimStateNode_0.State.AnimGraphNode_StateMachine_1.LocomotionAdditive.AnimStateTransitionNode_0.Transition
    src=IdleAdditive
    via=AnimStateTransitionNode_0
   }
   {
    dst=JogStopAdditive
    rule=LocomotionLayer_Additives.AnimGraphNode_StateMachine_0.LocomotionAdditives.AnimStateNode_0.State.AnimGraphNode_StateMachine_1.LocomotionAdditive.AnimStateTransitionNode_12.Transition
    src=IdleAdditive
    via=AnimStateTransitionNode_12
   }
   {
    dst=MovementAdditive
    rule=LocomotionLayer_Additives.AnimGraphNode_StateMachine_0.LocomotionAdditives.AnimStateNode_0.State.AnimGraphNode_StateMachine_1.LocomotionAdditive.AnimStateTransitionNode_10.AnimationTransitionGraph_114
    src=IdleAdditive
    via=AnimStateTransitionNode_10
   }
   {
    dst=JogPrePivotAdditive
    rule=LocomotionLayer_Additives.AnimGraphNode_StateMachine_0.LocomotionAdditives.AnimStateNode_0.State.AnimGraphNode_StateMachine_1.LocomotionAdditive.AnimStateTransitionNode_28.AnimationTransitionGraph_0
    src=JogPostPivotAdditive
    via=AnimStateTransitionNode_28
   }
   {
    dst=MovementAdditive
    rule=LocomotionLayer_Additives.AnimGraphNode_StateMachine_0.LocomotionAdditives.AnimStateNode_0.State.AnimGraphNode_StateMachine_1.LocomotionAdditive.AnimStateTransitionNode_26.Transition
    src=JogPostPivotAdditive
    via=AnimStateTransitionNode_26
   }
   {
    dst=MovementAdditive
    rule=LocomotionLayer_Additives.AnimGraphNode_StateMachine_0.LocomotionAdditives.AnimStateNode_0.State.AnimGraphNode_StateMachine_1.LocomotionAdditive.AnimStateTransitionNode_27.Transition
    src=JogPostPivotAdditive
    via=AnimStateTransitionNode_27
   }
   {
    dst=JogPostPivotAdditive
    rule=LocomotionLayer_Additives.AnimGraphNode_StateMachine_0.LocomotionAdditives.AnimStateNode_0.State.AnimGraphNode_StateMachine_1.LocomotionAdditive.AnimStateTransitionNode_25.AnimationTransitionGraph_0
    src=JogPrePivotAdditive
    via=AnimStateTransitionNode_25
   }
   {
    dst=MovementAdditive
    rule=LocomotionLayer_Additives.AnimGraphNode_StateMachine_0.LocomotionAdditives.AnimStateNode_0.State.AnimGraphNode_StateMachine_1.LocomotionAdditive.AnimStateTransitionNode_13.Transition
    src=JogPrePivotAdditive
    via=AnimStateTransitionNode_13
   }
   {
    dst=MovementAdditive
    rule=LocomotionLayer_Additives.AnimGraphNode_StateMachine_0.LocomotionAdditives.AnimStateNode_0.State.AnimGraphNode_StateMachine_1.LocomotionAdditive.AnimStateTransitionNode_5.Transition
    src=JogPrePivotAdditive
    via=AnimStateTransitionNode_5
   }
   {
    dst=MovementAdditive
    rule=LocomotionLayer_Additives.AnimGraphNode_StateMachine_0.LocomotionAdditives.AnimStateNode_0.State.AnimGraphNode_StateMachine_1.LocomotionAdditive.AnimStateTransitionNode_2.Transition
    src=JogStartAdditive
    via=AnimStateTransitionNode_2
   }
   {
    dst=MovementAdditive
    rule=LocomotionLayer_Additives.AnimGraphNode_StateMachine_0.LocomotionAdditives.AnimStateNode_0.State.AnimGraphNode_StateMachine_1.LocomotionAdditive.AnimStateTransitionNode_3.Transition
    src=JogStartAdditive
    via=AnimStateTransitionNode_3
   }
   {
    dst=IdleAdditive
    rule=LocomotionLayer_Additives.AnimGraphNode_StateMachine_0.LocomotionAdditives.AnimStateNode_0.State.AnimGraphNode_StateMachine_1.LocomotionAdditive.AnimStateTransitionNode_11.Transition
    src=JogStopAdditive
    via=AnimStateTransitionNode_11
   }
   {
    dst=IdleAdditive
    rule=LocomotionLayer_Additives.AnimGraphNode_StateMachine_0.LocomotionAdditives.AnimStateNode_0.State.AnimGraphNode_StateMachine_1.LocomotionAdditive.AnimStateTransitionNode_8.Transition
    src=JogStopAdditive
    via=AnimStateTransitionNode_8
   }
   {
    dst=IdleAdditive
    rule=LocomotionLayer_Additives.AnimGraphNode_StateMachine_0.LocomotionAdditives.AnimStateNode_0.State.AnimGraphNode_StateMachine_1.LocomotionAdditive.AnimStateTransitionNode_9.AnimationTransitionGraph_113
    src=MovementAdditive
    via=AnimStateTransitionNode_9
   }
   {
    dst=JogPrePivotAdditive
    rule=LocomotionLayer_Additives.AnimGraphNode_StateMachine_0.LocomotionAdditives.AnimStateNode_0.State.AnimGraphNode_StateMachine_1.LocomotionAdditive.AnimStateTransitionNode_6.Transition
    src=MovementAdditive
    via=AnimStateTransitionNode_6
   }
   {
    dst=JogStopAdditive
    rule=LocomotionLayer_Additives.AnimGraphNode_StateMachine_0.LocomotionAdditives.AnimStateNode_0.State.AnimGraphNode_StateMachine_1.LocomotionAdditive.AnimStateTransitionNode_7.Transition
    src=MovementAdditive
    via=AnimStateTransitionNode_7
   }
  ]
 }
 IdleAdditive {
  path=LocomotionLayer_Additives.AnimGraphNode_StateMachine_0.LocomotionAdditives.AnimStateNode_0.State.AnimGraphNode_StateMachine_1.LocomotionAdditive.AnimStateNode_0.IdleAdditive
  nodes [Output Animation Pose(Result=Use cached pose 'IdleAdditivePose'())]
 }
 JogStartAdditive {
  path=LocomotionLayer_Additives.AnimGraphNode_StateMachine_0.LocomotionAdditives.AnimStateNode_0.State.AnimGraphNode_StateMachine_1.LocomotionAdditive.AnimStateNode_1.JogStartAdditive
  nodes [
   n1 = Break Fort Locomotion Anim Set!(FortLocomotionAnimSet:"Fort Locomotion Anim Set Structure (by ref)"=Get LocomotionAnimSet() -> (LocomotionAnimSet:Fort Locomotion Anim Set Structure)) -> (
    EnableAdditiveLeans:bool=~
    UseMovementAdditiveForOneHandedTransitions:bool=~
    PlayMovementTransitions:bool=~
    MovementBlendSpaceTargeting:Blend Space Object Reference
    MovementBlendSpaceTargetingRightFootForward:Blend Space Object Reference
    MovementBlendSpace:Blend Space Object Reference
    MovementBlendSpaceNo45s:Blend Space Object Reference
    Start_N:Anim Sequence Base Object Reference
    Start_S:Anim Sequence Base Object Reference
    Start_E:Anim Sequence Base Object Reference
    Start_W:Anim Sequence Base Object Reference
    Stop_N:Anim Sequence Base Object Reference
    Stop_S:Anim Sequence Base Object Reference
    Stop_E:Anim Sequence Base Object Reference
    Stop_W:Anim Sequence Base Object Reference
    Pivot_N:Anim Sequence Base Object Reference
    Pivot_S:Anim Sequence Base Object Reference
    Pivot_E:Anim Sequence Base Object Reference
    Pivot_W:Anim Sequence Base Object Reference
    CrouchAdditiveLean:Blend Space Object Reference
    AdditiveLean:Blend Space Object Reference
    AdditiveStart_N:Anim Sequence Base Object Reference
    AdditiveStart_S:Anim Sequence Base Object Reference
    AdditiveStart_E:Anim Sequence Base Object Reference
    AdditiveStart_W:Anim Sequence Base Object Reference
    AdditiveStop_N:Anim Sequence Base Object Reference
    AdditiveStop_S:Anim Sequence Base Object Reference
    AdditiveStop_E:Anim Sequence Base Object Reference
    AdditiveStop_W:Anim Sequence Base Object Reference
    AdditivePivot_N:Anim Sequence Base Object Reference
    AdditivePivot_S:Anim Sequence Base Object Reference
    AdditivePivot_E:Anim Sequence Base Object Reference
    AdditivePivot_W:Anim Sequence Base Object Reference
   ) {is=title K2Node_BreakStruct}
   n2 = Break Fort Anim Input Speed Warping!(FortAnimInput_SpeedWarping:"Fort Anim Input Speed Warping Structure (by ref)"=Get StartSpeedWarping() -> (StartSpeedWarping:Fort Anim Input Speed Warping Structure)) -> (SpeedWarpingAmount:float=~, PlayRate:float=~) {is=title K2Node_BreakStruct}
   n3 = Get StartAnimBlendTime() -> (StartAnimBlendTime:double)
   n4 = "Blend Poses (EFortCardinalDirection)"!(
    BlendPose_0:Pose Link Structure=Sequence Player!(Sequence=@n1.AdditiveStart_N, PlayRate:float=@n2.PlayRate) -> (Pose) {bLoopAnimation=false, groupName=Locomotion, groupRole=TransitionFollower, method=SyncGroup}
    BlendPose_1:Pose Link Structure=Sequence Player(Sequence=@n1.AdditiveStart_E, PlayRate=@n2.PlayRate) {bLoopAnimation=false, groupName=Locomotion, groupRole=TransitionFollower, method=SyncGroup}
    BlendPose_2:Pose Link Structure=Sequence Player(Sequence=@n1.AdditiveStart_S, PlayRate=@n2.PlayRate) {bLoopAnimation=false, groupName=Locomotion, groupRole=TransitionFollower, method=SyncGroup}
    BlendPose_3:Pose Link Structure=Sequence Player(Sequence=@n1.AdditiveStart_W, PlayRate=@n2.PlayRate) {bLoopAnimation=false, groupName=Locomotion, groupRole=TransitionFollower, method=SyncGroup}
    BlendTime_0:float=@n3
    BlendTime_1:float=@n3
    BlendTime_2:float=@n3
    BlendTime_3:float=@n3
    ActiveEnumValue:Byte=Get LocomotionCardinalDirection() -> (LocomotionCardinalDirection:EFortCardinalDirection Enum=North)
   ) -> (Pose:Pose Link Structure) {
    is=title AnimGraphNode_BlendListByEnum
    blendType=Linear
    blendPose [{linkId=-1, sourceLinkId=-1}, {linkId=-1, sourceLinkId=-1}, {linkId=-1, sourceLinkId=-1}, {linkId=-1, sourceLinkId=-1}]
    blendTime [0.1,0.1,0.1,0.1]
   }
   n5 = Break Fort Legacy Weapon Anim Set(FortLegacyWeaponAnimSet=Get LegacyWeaponAnimSet())
   n6 = Layered blend per bone(
    BasePose=@n4
    BlendPoses_0=Two Way Blend(
     A=Sequence Player!(Sequence=@n5.IdleNoise) -> (Pose) {bIgnoreForRelevancyTest=true}
     B=Use cached pose 'MovementTransitionAdditiveLayerPose'()
     Alpha=Get Speed2D() -> (Speed2D:float)
    ) {blendNode {alphaBoolBlend {blendOption=Linear}, alphaScaleBias {scale=0.0033}}}
    BlendWeights_0=Get OneHandedTransitionHackAlpha() -> (OneHandedTransitionHackAlpha:float)
   ) {bMeshSpaceRotationBlend=true, layerSetup [{branchFilters [{blendDepth=2, boneName=Clavicle_L}, {boneName=IK_Hand_L}]}]}
   Output Animation Pose(Result=@n6)
   Use cached pose 'MovementAdditiveLayerPose'()
  ]
 }
 JogStopAdditive {
  path=LocomotionLayer_Additives.AnimGraphNode_StateMachine_0.LocomotionAdditives.AnimStateNode_0.State.AnimGraphNode_StateMachine_1.LocomotionAdditive.AnimStateNode_2.JogStopAdditive
  nodes [
   n1 = Break Fort Legacy Weapon Anim Set(FortLegacyWeaponAnimSet=Get LegacyWeaponAnimSet())
   n2 = Break Fort Locomotion Anim Set(FortLocomotionAnimSet=Get LocomotionAnimSet()) -> (EnableAdditiveLeans=~, UseMovementAdditiveForOneHandedTransitions=~, PlayMovementTransitions=~)
   n3 = "Blend Poses (EFortCardinalDirection)"(
    BlendPose_0=Sequence Player(Sequence=@n2.AdditiveStop_N) {bLoopAnimation=false, groupName=StopSync, groupRole=AlwaysFollower, method=SyncGroup}
    BlendPose_1=Sequence Player(Sequence=@n2.AdditiveStop_E) {bLoopAnimation=false, groupName=StopSync, groupRole=AlwaysFollower, method=SyncGroup}
    BlendPose_2=Sequence Player(Sequence=@n2.AdditiveStop_S) {bLoopAnimation=false, groupName=StopSync, groupRole=AlwaysFollower, method=SyncGroup}
    BlendPose_3=Sequence Player(Sequence=@n2.AdditiveStop_W) {groupName=StopSync, groupRole=AlwaysFollower, method=SyncGroup}
    BlendTime_0=0.1
    BlendTime_1=0.1
    BlendTime_2=0.1
    BlendTime_3=0.1
    ActiveEnumValue=Get StopCardinalDirection() -> (StopCardinalDirection:EFortCardinalDirection Enum=North)
   ) {
    blendType=Linear
    blendPose [{linkId=-1, sourceLinkId=-1}, {linkId=-1, sourceLinkId=-1}, {linkId=-1, sourceLinkId=-1}, {linkId=-1, sourceLinkId=-1}]
    blendTime [0.1,0.1,0.1,0.1]
   }
   n4 = Layered blend per bone(
    BasePose=@n3
    BlendPoses_0=Blend Poses by bool(
     BlendPose_0=Use cached pose 'MovementTransitionAdditiveLayerPose'()
     BlendPose_1=Sequence Player(Sequence=@n1.IdleNoise) {bIgnoreForRelevancyTest=true}
     BlendTime_1=.6
     bActiveValue=Get bIsMoving2D()
    ) {blendType=Linear}
    BlendWeights_0=Get OneHandedTransitionHackAlpha()
   ) {bMeshSpaceRotationBlend=true, layerSetup [{branchFilters [{blendDepth=2, boneName=Clavicle_L}, {boneName=IK_Hand_L}]}]}
   Output Animation Pose(Result=Modify Curve(SourcePose=@n4, CurveValues_0=0, Alpha=1) {applyMode=Scale, curveNames [Interruptible], curveValues [0]})
  ]
 }
 JogPostPivotAdditive {
  path=LocomotionLayer_Additives.AnimGraphNode_StateMachine_0.LocomotionAdditives.AnimStateNode_0.State.AnimGraphNode_StateMachine_1.LocomotionAdditive.AnimStateNode_3.JogPostPivotAdditive
  nodes [
   n1 = Break Fort Locomotion Anim Set(FortLocomotionAnimSet=Get LocomotionAnimSet()) -> (EnableAdditiveLeans=~, UseMovementAdditiveForOneHandedTransitions=~, PlayMovementTransitions=~)
   n2 = Get PrePivotTimer() -> (PrePivotTimer:float)
   n3 = Get TransitionPlayRate() -> (TransitionPlayRate:float)
   n4 = Get PivotAnimBlendTime() -> (PivotAnimBlendTime:double)
   n5 = "Blend Poses (EFortCardinalDirection)"(
    BlendPose_0=Sequence Player!(Sequence=@n1.AdditivePivot_S, PlayRate:float=@n3, StartPosition:float=@n2) -> (Pose) {bLoopAnimation=false}
    BlendPose_1=Sequence Player(Sequence=@n1.AdditivePivot_W, PlayRate=@n3, StartPosition=@n2) {bLoopAnimation=false}
    BlendPose_2=Sequence Player(Sequence=@n1.AdditivePivot_N, PlayRate=@n3, StartPosition=@n2) {bLoopAnimation=false}
    BlendPose_3=Sequence Player(Sequence=@n1.AdditivePivot_E, PlayRate=@n3, StartPosition=@n2) {bLoopAnimation=false}
    BlendTime_0=@n4
    BlendTime_1=@n4
    BlendTime_2=@n4
    BlendTime_3=@n4
    ActiveEnumValue=Get PostPivotCardinalDirection() -> (PostPivotCardinalDirection:EFortCardinalDirection Enum=North)
   ) {
    blendType=Linear
    blendPose [{linkId=-1, sourceLinkId=-1}, {linkId=-1, sourceLinkId=-1}, {linkId=-1, sourceLinkId=-1}, {linkId=-1, sourceLinkId=-1}]
    blendTime [0.1,0.1,0.1,0.1]
   }
   n6 = Layered blend per bone(BasePose=@n5, BlendPoses_0=Use cached pose 'MovementTransitionAdditiveLayerPose'(), BlendWeights_0=Get OneHandedTransitionHackAlpha()) {bMeshSpaceRotationBlend=true, layerSetup [{branchFilters [{blendDepth=2, boneName=Clavicle_L}, {boneName=IK_Hand_L}]}]}
   Output Animation Pose(Result=@n6)
   Use cached pose 'MovementAdditiveLayerPose'()
  ]
 }
 CrouchTurning {
  path=LocomotionLayer_Additives.AnimGraphNode_StateMachine_0.LocomotionAdditives.AnimStateNode_0.State.AnimGraphNode_StateMachine_1.LocomotionAdditive.AnimStateNode_4.CrouchTurning
  nodes [
   n1 = Get DBNOTurnPlayRate() -> (DBNOTurnPlayRate:float)
   n2 = "Sequence Player (additive) 'Crouch_PistolNonTargeted_E'"!(PlayRate:float=@n1) -> (Pose) {
    groupName=TurnInPlaceSync
    groupRole=AlwaysFollower
    method=SyncGroup
    sequence=/Game/Animation/Game/MainPlayer/Locomotion/Medium/Male/Crouch/Pistol_NonTargeted/Crouch_PistolNonTargeted_E
   }
   n3 = "Sequence Player (additive) 'Crouch_PistolNonTargeted_E'"(PlayRate=@n1) {
    groupName=TurnInPlaceSync
    groupRole=AlwaysFollower
    method=SyncGroup
    sequence=&Crouch_PistolNonTargeted_E
   }
   n4 = "Sequence Player (additive) 'Crouch_THMelee_E'"!(PlayRate:float=@n1) -> (Pose) {
    groupName=TurnInPlaceSync
    groupRole=AlwaysFollower
    method=SyncGroup
    sequence=/Game/Animation/Game/MainPlayer/Locomotion/Medium/Male/Crouch/TwoHandedMelee/Crouch_THMelee_E
   }
   n5 = "Sequence Player (additive) 'Crouch_PistolNonTargeted_E'"(PlayRate=@n1) {
    groupName=TurnInPlaceSync
    groupRole=AlwaysFollower
    method=SyncGroup
    sequence=&Crouch_PistolNonTargeted_E
   }
   n6 = "Sequence Player (additive) 'Crouch_ARNonTargeted_E'"!(PlayRate:float=@n1) -> (Pose) {
    groupName=TurnInPlaceSync
    groupRole=AlwaysFollower
    method=SyncGroup
    sequence=/Game/Animation/Game/MainPlayer/Locomotion/Medium/Male/Crouch/AssaultRifle_NonTargeted/Crouch_ARNonTargeted_E
   }
   n7 = "Sequence Player (additive) 'Crouch_PistolNonTargeted_E'"(PlayRate=@n1) {
    groupName=TurnInPlaceSync
    groupRole=AlwaysFollower
    method=SyncGroup
    sequence=&Crouch_PistolNonTargeted_E
   }
   n8 = "Sequence Player (additive) 'Crouch_TacticalShotgun02_E'"!(PlayRate:float=@n1) -> (Pose) {
    groupName=TurnInPlaceSync
    method=SyncGroup
    sequence=/Game/Animation/Game/MainPlayer/Locomotion/Medium/Male/Crouch/TacticalShotgun02/Crouch_TacticalShotgun02_E
   }
   n9 = "Sequence Player (additive) 'DualWield_Crouch_E'"!(PlayRate:float=@n1) -> (Pose) {
    groupName=TurnInPlaceSync
    method=SyncGroup
    sequence=/Game/Animation/Game/MainPlayer/Locomotion/Medium/Male/Crouch/DualWield/DualWield_Crouch_E
   }
   n10 = "Blend Poses (EFortWeaponCoreAnimation)"!(
    BlendPose_0=@n6
    BlendPose_1=@n4
    BlendPose_2=@n2
    BlendPose_3:Pose Link Structure=@n5
    BlendPose_4:Pose Link Structure=@n7
    BlendPose_5:Pose Link Structure=@n8
    BlendPose_6:Pose Link Structure=@n9
    BlendPose_7:Pose Link Structure=@n3
    BlendTime_0=0.1
    BlendTime_1=0.1
    BlendTime_2=0.1
    BlendTime_3:float=0.1
    BlendTime_4:float=0.1
    BlendTime_5:float=0.1
    BlendTime_6:float=0.1
    BlendTime_7:float=0.1
    ActiveEnumValue=Get WeaponCoreAnim()
   ) -> (Pose) {
    blendType=Linear
    blendPose [
     {linkId=-1, sourceLinkId=-1}
     {linkId=-1, sourceLinkId=-1}
     {linkId=-1, sourceLinkId=-1}
     {linkId=-1, sourceLinkId=-1}
     {linkId=-1, sourceLinkId=-1}
     {linkId=-1, sourceLinkId=-1}
     {linkId=-1, sourceLinkId=-1}
     {linkId=-1, sourceLinkId=-1}
    ]
    blendTime [0.1,0.1,0.1,0.1,0.1,0.1,0.1,0.1]
   }
   Output Animation Pose(Result=@n10)
  ]
 }
 JogPrePivotAdditive {
  path=LocomotionLayer_Additives.AnimGraphNode_StateMachine_0.LocomotionAdditives.AnimStateNode_0.State.AnimGraphNode_StateMachine_1.LocomotionAdditive.AnimStateNode_6.JogPrePivotAdditive
  nodes [
   n1 = Break Fort Locomotion Anim Set(FortLocomotionAnimSet=Get LocomotionAnimSet()) -> (EnableAdditiveLeans=~, UseMovementAdditiveForOneHandedTransitions=~, PlayMovementTransitions=~)
   n2 = Get TransitionPlayRate()
   n3 = Get PivotAnimBlendTime()
   n4 = "Blend Poses (EFortCardinalDirection)"(
    BlendPose_0=Sequence Player!(Sequence=@n1.AdditivePivot_S, PlayRate=@n2) -> (Pose) {bLoopAnimation=false, groupName=PrePivotHack, groupRole=AlwaysFollower, method=SyncGroup}
    BlendPose_1=Sequence Player(Sequence=@n1.AdditivePivot_W, PlayRate=@n2) {bLoopAnimation=false, groupName=PrePivotHack, groupRole=AlwaysFollower, method=SyncGroup}
    BlendPose_2=Sequence Player(Sequence=@n1.AdditivePivot_N, PlayRate=@n2) {bLoopAnimation=false, groupName=PrePivotHack, groupRole=AlwaysFollower, method=SyncGroup}
    BlendPose_3=Sequence Player(Sequence=@n1.AdditivePivot_E, PlayRate=@n2) {bLoopAnimation=false, groupName=PrePivotHack, groupRole=AlwaysFollower, method=SyncGroup}
    BlendTime_0=@n3
    BlendTime_1=@n3
    BlendTime_2=@n3
    BlendTime_3=@n3
    ActiveEnumValue=Get PrePivotCardinalDirection() -> (PrePivotCardinalDirection:EFortCardinalDirection Enum=North)
   ) {
    blendType=Linear
    blendPose [{linkId=-1, sourceLinkId=-1}, {linkId=-1, sourceLinkId=-1}, {linkId=-1, sourceLinkId=-1}, {linkId=-1, sourceLinkId=-1}]
    blendTime [0.1,0.1,0.1,0.1]
   }
   n5 = Layered blend per bone(BasePose=@n4, BlendPoses_0=Use cached pose 'MovementTransitionAdditiveLayerPose'(), BlendWeights_0=Get OneHandedTransitionHackAlpha()) {bMeshSpaceRotationBlend=true, layerSetup [{branchFilters [{blendDepth=2, boneName=Clavicle_L}, {boneName=IK_Hand_L}]}]}
   Output Animation Pose(Result=@n5)
   Use cached pose 'MovementAdditiveLayerPose'()
  ]
 }
 MovementAdditive {
  path=LocomotionLayer_Additives.AnimGraphNode_StateMachine_0.LocomotionAdditives.AnimStateNode_0.State.AnimGraphNode_StateMachine_1.LocomotionAdditive.AnimStateNode_7.MovementAdditive
  nodes [Output Animation Pose(Result=Use cached pose 'MovementAdditiveLayerPose'())]
 }
 Transition {
  path=LocomotionLayer_Additives.AnimGraphNode_StateMachine_0.LocomotionAdditives.AnimStateNode_0.State.AnimGraphNode_StateMachine_1.LocomotionAdditive.AnimStateTransitionNode_0.Transition
  nodes [Result(bCanEnterTransition=Get bShouldPlayStartTransition() -> (bShouldPlayStartTransition:bool))]
 }
 AnimationTransitionGraph_114 {
  path=LocomotionLayer_Additives.AnimGraphNode_StateMachine_0.LocomotionAdditives.AnimStateNode_0.State.AnimGraphNode_StateMachine_1.LocomotionAdditive.AnimStateTransitionNode_10.AnimationTransitionGraph_114
  nodes [Result(bCanEnterTransition=Get bTransition_LowerBody_Movement() -> (bTransition_LowerBody_Movement:bool))]
 }
 Transition {
  path=LocomotionLayer_Additives.AnimGraphNode_StateMachine_0.LocomotionAdditives.AnimStateNode_0.State.AnimGraphNode_StateMachine_1.LocomotionAdditive.AnimStateTransitionNode_11.Transition
  nodes [Result(bCanEnterTransition=Get bShouldEarlyOutStopState() -> (bShouldEarlyOutStopState:bool))]
 }
 Transition {
  path=LocomotionLayer_Additives.AnimGraphNode_StateMachine_0.LocomotionAdditives.AnimStateNode_0.State.AnimGraphNode_StateMachine_1.LocomotionAdditive.AnimStateTransitionNode_12.Transition
  nodes [Result(bCanEnterTransition=Get bShouldPlayStopTransition() -> (bShouldPlayStopTransition:bool))]
 }
 Transition {
  path=LocomotionLayer_Additives.AnimGraphNode_StateMachine_0.LocomotionAdditives.AnimStateNode_0.State.AnimGraphNode_StateMachine_1.LocomotionAdditive.AnimStateTransitionNode_13.Transition
  nodes [Result()]
 }
 Transition {
  path=LocomotionLayer_Additives.AnimGraphNode_StateMachine_0.LocomotionAdditives.AnimStateNode_0.State.AnimGraphNode_StateMachine_1.LocomotionAdditive.AnimStateTransitionNode_14.Transition
  nodes [Result(bCanEnterTransition=Get bTransition_LocomotionAdditive_CrouchTurning() -> (bTransition_LocomotionAdditive_CrouchTurning:bool))]
 }
 Transition {
  path=LocomotionLayer_Additives.AnimGraphNode_StateMachine_0.LocomotionAdditives.AnimStateNode_0.State.AnimGraphNode_StateMachine_1.LocomotionAdditive.AnimStateTransitionNode_15.Transition
  nodes [Result(bCanEnterTransition=NOT Boolean(A=Get bTransition_LocomotionAdditive_CrouchTurning()))]
 }
 Transition {
  path=LocomotionLayer_Additives.AnimGraphNode_StateMachine_0.LocomotionAdditives.AnimStateNode_0.State.AnimGraphNode_StateMachine_1.LocomotionAdditive.AnimStateTransitionNode_2.Transition
  nodes [Result()]
 }
 AnimationTransitionGraph_0 {
  path=LocomotionLayer_Additives.AnimGraphNode_StateMachine_0.LocomotionAdditives.AnimStateNode_0.State.AnimGraphNode_StateMachine_1.LocomotionAdditive.AnimStateTransitionNode_25.AnimationTransitionGraph_0
  nodes [Result(bCanEnterTransition=Get bShouldPlayPostPivotTransition() -> (bShouldPlayPostPivotTransition:bool))]
 }
 Transition {
  path=LocomotionLayer_Additives.AnimGraphNode_StateMachine_0.LocomotionAdditives.AnimStateNode_0.State.AnimGraphNode_StateMachine_1.LocomotionAdditive.AnimStateTransitionNode_26.Transition
  nodes [Result(bCanEnterTransition=Get bShouldEarlyOutPivotState() -> (bShouldEarlyOutPivotState:bool))]
 }
 Transition {
  path=LocomotionLayer_Additives.AnimGraphNode_StateMachine_0.LocomotionAdditives.AnimStateNode_0.State.AnimGraphNode_StateMachine_1.LocomotionAdditive.AnimStateTransitionNode_27.Transition
  nodes [Result()]
 }
 AnimationTransitionGraph_0 {
  path=LocomotionLayer_Additives.AnimGraphNode_StateMachine_0.LocomotionAdditives.AnimStateNode_0.State.AnimGraphNode_StateMachine_1.LocomotionAdditive.AnimStateTransitionNode_28.AnimationTransitionGraph_0
  nodes [Result(bCanEnterTransition=Get bShouldPlayPivotTransition() -> (bShouldPlayPivotTransition:bool))]
 }
 Transition {
  path=LocomotionLayer_Additives.AnimGraphNode_StateMachine_0.LocomotionAdditives.AnimStateNode_0.State.AnimGraphNode_StateMachine_1.LocomotionAdditive.AnimStateTransitionNode_3.Transition
  nodes [Result(bCanEnterTransition=Get bShouldEarlyOutStartState() -> (bShouldEarlyOutStartState:bool))]
 }
 Transition {
  path=LocomotionLayer_Additives.AnimGraphNode_StateMachine_0.LocomotionAdditives.AnimStateNode_0.State.AnimGraphNode_StateMachine_1.LocomotionAdditive.AnimStateTransitionNode_4.Transition
  nodes [Result(bCanEnterTransition=Get bShouldPlayPivotTransition())]
 }
 Transition {
  path=LocomotionLayer_Additives.AnimGraphNode_StateMachine_0.LocomotionAdditives.AnimStateNode_0.State.AnimGraphNode_StateMachine_1.LocomotionAdditive.AnimStateTransitionNode_5.Transition
  nodes [Result(bCanEnterTransition=Get bShouldEarlyOutPivotState())]
 }
 Transition {
  path=LocomotionLayer_Additives.AnimGraphNode_StateMachine_0.LocomotionAdditives.AnimStateNode_0.State.AnimGraphNode_StateMachine_1.LocomotionAdditive.AnimStateTransitionNode_6.Transition
  nodes [Result(bCanEnterTransition=Get bShouldPlayPivotTransition())]
 }
 Transition {
  path=LocomotionLayer_Additives.AnimGraphNode_StateMachine_0.LocomotionAdditives.AnimStateNode_0.State.AnimGraphNode_StateMachine_1.LocomotionAdditive.AnimStateTransitionNode_7.Transition
  nodes [Result(bCanEnterTransition=Get bShouldPlayStopTransition())]
 }
 Transition {
  path=LocomotionLayer_Additives.AnimGraphNode_StateMachine_0.LocomotionAdditives.AnimStateNode_0.State.AnimGraphNode_StateMachine_1.LocomotionAdditive.AnimStateTransitionNode_8.Transition
  nodes [Result()]
 }
 AnimationTransitionGraph_113 {
  path=LocomotionLayer_Additives.AnimGraphNode_StateMachine_0.LocomotionAdditives.AnimStateNode_0.State.AnimGraphNode_StateMachine_1.LocomotionAdditive.AnimStateTransitionNode_9.AnimationTransitionGraph_113
  nodes [Result(bCanEnterTransition=NOT Boolean(A=Get bTransition_LowerBody_Movement()))]
 }
 MovementLayerAdditivePose {
  entry=Conduit
  path=LocomotionLayer_Additives.AnimGraphNode_StateMachine_10.MovementLayerAdditivePose
  ids=id
  nodes [
   AnimStateConduitNode_0 = Conduit()
   AnimStateNode_1 = JogAdditivePose() {is=title AnimStateNode}
   AnimStateNode_2 = CrouchAdditivePose() {is=title AnimStateNode}
   AnimStateTransitionNode_4 = AnimStateTransitionNode() {crossfadeDuration=0.25, logicType=TLT_Inertialization, priorityOrder=100}
   AnimStateTransitionNode_5 = AnimStateTransitionNode() {crossfadeDuration=0.25, logicType=TLT_Inertialization}
   AnimStateTransitionNode_6 = AnimStateTransitionNode()
   AnimStateTransitionNode_7 = AnimStateTransitionNode()
  ]
  links [
   {
    dst=CrouchAdditivePose
    rule=LocomotionLayer_Additives.AnimGraphNode_StateMachine_10.MovementLayerAdditivePose.AnimStateTransitionNode_5.Transition
    src=Conduit
    via=AnimStateTransitionNode_5
   }
   {
    dst=JogAdditivePose
    rule=LocomotionLayer_Additives.AnimGraphNode_StateMachine_10.MovementLayerAdditivePose.AnimStateTransitionNode_4.Transition
    src=Conduit
    via=AnimStateTransitionNode_4
   }
   {
    dst=Conduit
    rule=LocomotionLayer_Additives.AnimGraphNode_StateMachine_10.MovementLayerAdditivePose.AnimStateTransitionNode_6.Transition
    src=CrouchAdditivePose
    via=AnimStateTransitionNode_6
   }
   {
    dst=Conduit
    rule=LocomotionLayer_Additives.AnimGraphNode_StateMachine_10.MovementLayerAdditivePose.AnimStateTransitionNode_7.Transition
    src=JogAdditivePose
    via=AnimStateTransitionNode_7
   }
  ]
 }
 Conduit {
  path=LocomotionLayer_Additives.AnimGraphNode_StateMachine_10.MovementLayerAdditivePose.AnimStateConduitNode_0.Conduit
  nodes [Result(bCanEnterTransition=true)]
 }
 JogAdditivePose {
  path=LocomotionLayer_Additives.AnimGraphNode_StateMachine_10.MovementLayerAdditivePose.AnimStateNode_1.JogAdditivePose
  nodes [
   n1 = Break Fort Legacy Weapon Anim Set(FortLegacyWeaponAnimSet=Get LegacyWeaponAnimSet())
   n2 = Break Fort Anim Input Speed Warping(FortAnimInput_SpeedWarping=Get SpeedWarping() -> (SpeedWarping:Fort Anim Input Speed Warping Structure)) -> (SpeedWarpingAmount=~, PlayRate=~)
   n3 = Get Speed2D()
   n4 = Get LocalVelocityYawAngleMinusJogBlendSpaceRotation() -> (LocalVelocityYawAngleMinusJogBlendSpaceRotation:float)
   n5 = Blend Poses by bool(
    BlendPose_0=Blendspace Player!(X=@n4, Y=@n3, PlayRate:float=@n2.PlayRate, BlendSpace:Blend Space Object Reference=@n1.JogAdditiveBlendSpaceRelaxed) -> (Pose) {groupName=Locomotion, groupRole=AlwaysFollower, method=SyncGroup}
    BlendPose_1=Blendspace Player(X=@n4, Y=@n3, PlayRate=@n2.PlayRate, BlendSpace=@n1.JogAdditiveBlendSpace) {groupName=Locomotion, groupRole=AlwaysFollower, method=SyncGroup}
    BlendTime_0=.4
    BlendTime_1=0
    bActiveValue=Get bIsRelaxedLevel1() -> (bIsRelaxedLevel1:bool)
   ) {blendType=Linear}
   n6 = Get Speed2D()
   n7 = Get LocalVelocityYawAngle() -> (LocalVelocityYawAngle:float)
   n8 = Blendspace Player 'M_Locomotion2D_Walk_AR_Additive'(X=@n7, Y=@n6) {
    blendSpace=/Game/Animation/Game/MainPlayer/Locomotion/Medium/Male/Walk/AssaultRifle_Targeted/M_Locomotion2D_Walk_AR_Additive
    groupName=Locomotion
    groupRole=AlwaysFollower
    method=SyncGroup
   }
   n9 = Blendspace Player 'M_Locomotion2D_Walk_Pistol_Additive'(X=@n7, Y=@n6) {
    blendSpace=/Game/Animation/Game/MainPlayer/Locomotion/Medium/Male/Walk/Pistol/M_Locomotion2D_Walk_Pistol_Additive
    groupName=Locomotion
    groupRole=AlwaysFollower
    method=SyncGroup
   }
   n10 = "Blend Poses (EFortWeaponCoreAnimation)"!(
    BlendPose_0=@n8
    BlendPose_1=@n9
    BlendPose_2=Blendspace Player(X=@n4, Y=@n3, PlayRate=@n2.PlayRate, BlendSpace=@n1.JogAdditiveBlendSpace) {groupName=Locomotion, groupRole=AlwaysFollower, method=SyncGroup}
    BlendTime_0=0
    BlendTime_1=0
    BlendTime_2=0.1
    ActiveEnumValue=Get WeaponCoreAnim()
   ) -> (Pose) {
    blendType=Linear
    blendPose [{linkId=-1, sourceLinkId=-1}, {linkId=-1, sourceLinkId=-1}, {linkId=-1, sourceLinkId=-1}]
    blendTime [0.1,0.1,0.1]
   }
   n11 = Get bIsTargeting()
   Output Animation Pose(Result=Blend Poses by bool(
     BlendPose_0=Blendspace Player(X=@n4, Y=@n3, PlayRate=@n2.PlayRate, BlendSpace=@n1.JogAdditiveBlendSpaceMeleeGuarding) {groupName=Locomotion, groupRole=AlwaysFollower, method=SyncGroup}
     BlendPose_1=Blend Poses by bool(BlendPose_0=@n10, BlendPose_1=@n5, BlendTime_1=.2, bActiveValue=@n11) {blendType=Linear}
     bActiveValue=Get bIsMeleeGuarding() -> (bIsMeleeGuarding:bool)
    ) {blendType=Linear})
  ]
 }
 CrouchAdditivePose {
  path=LocomotionLayer_Additives.AnimGraphNode_StateMachine_10.MovementLayerAdditivePose.AnimStateNode_2.CrouchAdditivePose
  nodes [
   n1 = Break Fort Legacy Weapon Anim Set(FortLegacyWeaponAnimSet=Get LegacyWeaponAnimSet())
   n2 = Get CrouchPlayRate() -> (CrouchPlayRate:double)
   n3 = Get Speed2D()
   n4 = Get LocalVelocityYawAngle()
   n5 = Blendspace Player(X=@n4, Y=@n3, PlayRate=@n2, BlendSpace=@n1.CrouchJogAdditiveBlendSpace) {groupName=Locomotion, groupRole=AlwaysFollower, method=SyncGroup}
   n6 = Blendspace Player(X=@n4, Y=@n3, PlayRate=@n2, BlendSpace=@n1.CrouchJogAdditiveBlendSpaceRelaxed) {groupName=Locomotion, groupRole=AlwaysFollower, method=SyncGroup}
   n7 = Get bIsRelaxedLevel1()
   Output Animation Pose(Result=Blend Poses by bool(
     BlendPose_0=Blendspace Player(X=@n4, Y=@n3, PlayRate=@n2, BlendSpace=@n1.CrouchWalkAdditiveBlendSpace) {groupName=Locomotion, groupRole=AlwaysFollower, method=SyncGroup}
     BlendPose_1=Blend Poses by bool(BlendPose_0=@n6, BlendPose_1=@n5, BlendTime_0=.6, BlendTime_1=.0667, bActiveValue=@n7) {blendType=Linear}
     BlendTime_1=.2
     bActiveValue=Get bIsTargeting()
    ) {blendType=Linear})
  ]
 }
 Transition {
  path=LocomotionLayer_Additives.AnimGraphNode_StateMachine_10.MovementLayerAdditivePose.AnimStateTransitionNode_4.Transition
  nodes [Result(bCanEnterTransition=true)]
 }
 Transition {
  path=LocomotionLayer_Additives.AnimGraphNode_StateMachine_10.MovementLayerAdditivePose.AnimStateTransitionNode_5.Transition
  nodes [Result(bCanEnterTransition=Get bIsCrouching())]
 }
 Transition {
  path=LocomotionLayer_Additives.AnimGraphNode_StateMachine_10.MovementLayerAdditivePose.AnimStateTransitionNode_6.Transition
  nodes [Result(bCanEnterTransition=true)]
 }
 Transition {
  path=LocomotionLayer_Additives.AnimGraphNode_StateMachine_10.MovementLayerAdditivePose.AnimStateTransitionNode_7.Transition
  nodes [Result(bCanEnterTransition=true)]
 }
 "Movement TransitionAdditiveLayerPose(Hack) " {
  entry=LayerPose
  path="LocomotionLayer_Additives.AnimGraphNode_StateMachine_2.Movement TransitionAdditiveLayerPose(Hack) "
  ids=id
  nodes [AnimStateNode_0 = LayerPose() {is=title AnimStateNode}]
 }
 LayerPose {
  path="LocomotionLayer_Additives.AnimGraphNode_StateMachine_2.Movement TransitionAdditiveLayerPose(Hack) .AnimStateNode_0.LayerPose"
  nodes [
   n1 = Break Fort Legacy Weapon Anim Set(FortLegacyWeaponAnimSet=Get LegacyWeaponAnimSet())
   n2 = Get Speed2D()
   n3 = Get LocalVelocityYawAngle()
   n4 = Get CrouchPlayRate()
   n5 = Blend Poses by bool(
    BlendPose_0=Blendspace Player(X=@n3, Y=@n2, PlayRate=@n4, BlendSpace=@n1.CrouchJogAdditiveBlendSpaceRelaxed) {groupName=LocoAddHack, method=SyncGroup}
    BlendPose_1=Blendspace Player(X=@n3, Y=@n2, PlayRate=@n4, BlendSpace=@n1.CrouchJogAdditiveBlendSpace) {groupName=LocoAddHack, method=SyncGroup}
    BlendTime_0=.6
    BlendTime_1=.0667
    bActiveValue=Get bIsRelaxedLevel1()
   ) {blendType=Linear}
   n6 = Blendspace Player(X=@n3, Y=@n2, PlayRate=@n4, BlendSpace=@n1.CrouchWalkAdditiveBlendSpace) {groupName=LocoAddHack, method=SyncGroup}
   n7 = Get bIsTargeting()
   n8 = Break Fort Legacy Weapon Anim Set(FortLegacyWeaponAnimSet=Get LegacyWeaponAnimSet())
   n9 = Break Fort Anim Input Speed Warping(FortAnimInput_SpeedWarping=Get SpeedWarping()) -> (SpeedWarpingAmount=~, PlayRate=~)
   n10 = Get Speed2D()
   n11 = Get LocalVelocityYawAngleMinusJogBlendSpaceRotation()
   n12 = Blend Poses by bool(
    BlendPose_0=Blendspace Player(X=@n11, Y=@n10, PlayRate=@n9.PlayRate, BlendSpace=@n8.JogAdditiveBlendSpaceRelaxed) {groupName=LocoAddHack, method=SyncGroup}
    BlendPose_1=Blendspace Player(X=@n11, Y=@n10, PlayRate=@n9.PlayRate, BlendSpace=@n8.JogAdditiveBlendSpace) {groupName=LocoAddHack, method=SyncGroup}
    BlendTime_0=.4
    BlendTime_1=0
    bActiveValue=Get bIsRelaxedLevel1()
   ) {blendType=Linear}
   n13 = Get Speed2D()
   n14 = Get LocalVelocityYawAngle()
   n15 = Blendspace Player 'M_Locomotion2D_Walk_AR_Additive'(X=@n14, Y=@n13) {
    blendSpace=&M_Locomotion2D_Walk_AR_Additive
    groupName=LocoAddHack
    method=SyncGroup
   }
   n16 = Blendspace Player 'M_Locomotion2D_Walk_Pistol_Additive'(X=@n14, Y=@n13) {
    blendSpace=&M_Locomotion2D_Walk_Pistol_Additive
    groupName=LocoAddHack
    method=SyncGroup
   }
   n17 = "Blend Poses (EFortWeaponCoreAnimation)"!(BlendPose_0=@n15, BlendPose_1=@n16, BlendTime_0=0, BlendTime_1=0, ActiveEnumValue=Get WeaponCoreAnim()) -> (Pose) {blendType=Linear, blendPose [{linkId=-1, sourceLinkId=-1}, {linkId=-1, sourceLinkId=-1}], blendTime [0.1,0.1]}
   n18 = Get bIsTargeting()
   Output Animation Pose(Result=Blend Poses by bool(
     BlendPose_0=Blend Poses by bool(BlendPose_0=@n6, BlendPose_1=@n5, BlendTime_1=.2, bActiveValue=@n7) {blendType=Linear}
     BlendPose_1=Blend Poses by bool(BlendPose_0=@n17, BlendPose_1=@n12, BlendTime_1=.2, bActiveValue=@n18) {blendType=Linear}
     BlendTime_0=.25
     BlendTime_1=.35
     bActiveValue=Get bIsCrouching()
    ))
  ]
 }
 LocomotionLayer_BaseLocomotion {nodes [
   Input Pose() {name=InPose_BaseLocomotion}
   n1 = Break Fort Legacy Weapon Anim Set(FortLegacyWeaponAnimSet=Get LegacyWeaponAnimSet())
   n2 = Break Fort Legacy Weapon Anim Set(FortLegacyWeaponAnimSet=Get LegacyWeaponAnimSet())
   n3 = "Sequence Evaluator (by time)"!(Sequence:Anim Sequence Base Object Reference=@n1.CrouchRelaxedPose) -> (Pose:Pose Link Structure) {is=title AnimGraphNode_SequenceEvaluator}
   n4 = "Sequence Evaluator (by time)"(Sequence=@n1.CrouchNonTargetingPose)
   n5 = Get bIsRelaxedLevel1()
   n6 = "Sequence Evaluator (by time)"(Sequence=@n2.NonTargetingPose)
   n7 = "Sequence Evaluator (by time)"(Sequence=@n2)
   n8 = Get WeaponFireBlendTime() -> (WeaponFireBlendTime:float)
   n9 = Get bIsRelaxedLevel1()
   n10 = Blend Poses by bool(
    BlendPose_0=Blend Poses by bool(
     BlendPose_0="Sequence Evaluator (by time)"(Sequence=@n1.CrouchTargetingPose)
     BlendPose_1="Sequence Evaluator (by time)"(Sequence=@n2.TargetingPose)
     BlendTime_0=.2
     BlendTime_1=.2
     bActiveValue=Get bIsCrouching()
    ) {blendType=Linear}
    BlendPose_1=Blend Poses by bool(
     BlendPose_0=Blend Poses by bool(BlendPose_0=@n3, BlendPose_1=@n4, BlendTime_0=.4, BlendTime_1=.05, bActiveValue=@n5) {blendType=Linear}
     BlendPose_1=Blend Poses by bool(BlendPose_0=@n7, BlendPose_1=@n6, BlendTime_0=.4, BlendTime_1=@n8, bActiveValue=@n9) {blendType=Linear, childUpateMode=ResetChildOnActivate}
     BlendTime_0=.2
     BlendTime_1=.2
     bActiveValue=Get bIsCrouching()
    ) {blendType=Linear}
    BlendTime_0=Get NonADSToADSBlendTime() -> (NonADSToADSBlendTime:float)
    BlendTime_1=Get ADSToNonADSBlendTime()
    bActiveValue=Get bPlayUpperBodyTargeting() -> (bPlayUpperBodyTargeting:bool)
   ) {
    becomeRelevantFunction {bSelfContext=true, memberGuid=#1, memberName=ResetBlendListNode}
    node {blendType=Linear, childUpateMode=ResetChildOnActivate, becomeRelevantFunction {functionName=OnNewFunction}}
   }
   Output Pose(Result=@n10) {layerGroup=LocomotionOverride}
  ]}
 LocomotionLayer_FinalPose {nodes [
   Input Pose() {name=InPoseFinalPose}
   Output Pose() {layerGroup=LocomotionOverride}
  ]}
 LocomotionLayer_FullBody {nodes [
   AnimGraphNode_StateMachine_16 = SprintingPoseAdditives() {is=title AnimGraphNode_StateMachine}
   Output Pose(Result=Blend Poses by bool(
     BlendPose_0=Inertialization!(Source:Pose Link Structure=@AnimGraphNode_StateMachine_16) -> (Pose:Pose Link Structure) {is=title AnimGraphNode_Inertialization}
     BlendPose_1=Input Pose() {name=InPose_FullBody}
     BlendTime_0=0.3
     BlendTime_1=0.3
     bActiveValue=Get bTransition_FullBody_Sprinting() -> (bTransition_FullBody_Sprinting:bool)
    ) {blendType=Linear}) {layerGroup=LocomotionOverride}
   n1 = Get SprintYawAngle() -> (SprintYawAngle:float)
   AnimGraphNode_StateMachine_15 = SprintingPoseBase() {is=title AnimGraphNode_StateMachine, bAllowConduitEntryStates=true}
   n2 = Get SprintYawAngle()
   Save cached pose 'SprintingPose_PreMask'(Pose=Apply Additive(
     Base=Rotate Root Bone!(BasePose:Pose Link Structure=@AnimGraphNode_StateMachine_15, Yaw:float=@n2) -> (Pose:Pose Link Structure) {is=title AnimGraphNode_RotateRootBone}
     Additive=Blendspace Player 'Sprint_LookAt_BS_1D'!(X=@n1) -> (Pose) {blendSpace=/Game/Animation/Game/MainPlayer/Locomotion/Medium/Male/Sprint/Sprint_LookAt_BS_1D}
    ) {lODThreshold=1, alphaBoolBlend {blendOption=Linear}}) {cacheName=SprintingPose_PreMask}
  ]}
 SprintingPoseBase {
  entry=Conduit
  path=LocomotionLayer_FullBody.AnimGraphNode_StateMachine_15.SprintingPoseBase
  ids=id
  nodes [
   AnimStateConduitNode_0 = Conduit()
   AnimStateNode_1 = CrouchSprint_State() {is=title AnimStateNode}
   AnimStateNode_2 = Sprint_State() {is=title AnimStateNode}
   AnimStateTransitionNode_0 = AnimStateTransitionNode() {logicType=TLT_Inertialization}
   AnimStateTransitionNode_1 = AnimStateTransitionNode() {logicType=TLT_Inertialization, priorityOrder=100}
   AnimStateTransitionNode_2 = AnimStateTransitionNode()
   AnimStateTransitionNode_3 = AnimStateTransitionNode()
  ]
  links [
   {
    dst=CrouchSprint_State
    rule=LocomotionLayer_FullBody.AnimGraphNode_StateMachine_15.SprintingPoseBase.AnimStateTransitionNode_0.Transition
    src=Conduit
    via=AnimStateTransitionNode_0
   }
   {
    dst=Sprint_State
    rule=LocomotionLayer_FullBody.AnimGraphNode_StateMachine_15.SprintingPoseBase.AnimStateTransitionNode_1.Transition
    src=Conduit
    via=AnimStateTransitionNode_1
   }
   {
    dst=Conduit
    rule=LocomotionLayer_FullBody.AnimGraphNode_StateMachine_15.SprintingPoseBase.AnimStateTransitionNode_3.Transition
    src=CrouchSprint_State
    via=AnimStateTransitionNode_3
   }
   {
    dst=Conduit
    rule=LocomotionLayer_FullBody.AnimGraphNode_StateMachine_15.SprintingPoseBase.AnimStateTransitionNode_2.Transition
    src=Sprint_State
    via=AnimStateTransitionNode_2
   }
  ]
 }
 Conduit {
  path=LocomotionLayer_FullBody.AnimGraphNode_StateMachine_15.SprintingPoseBase.AnimStateConduitNode_0.Conduit
  nodes [Result(bCanEnterTransition=true)]
 }
 CrouchSprint_State {
  path=LocomotionLayer_FullBody.AnimGraphNode_StateMachine_15.SprintingPoseBase.AnimStateNode_1.CrouchSprint_State
  nodes [
   n1 = Break Fort Legacy Weapon Anim Set(FortLegacyWeaponAnimSet=Get LegacyWeaponAnimSet())
   Output Animation Pose(Result=Blend Poses by bool(
     BlendPose_0=Sequence Player(Sequence=@n1.CrouchSprintTargetingAnimation, PlayRate=1) {groupName=Locomotion, method=SyncGroup}
     BlendPose_1=Sequence Player(Sequence=@n1.CrouchSprintAnimation, PlayRate=1) {groupName=Locomotion, method=SyncGroup}
     BlendTime_0=0.2
     BlendTime_1=0.2
     bActiveValue=Get bPlayUpperBodyTargeting()
    ) {blendType=Linear, transitionType=Inertialization})
  ]
 }
 Sprint_State {
  path=LocomotionLayer_FullBody.AnimGraphNode_StateMachine_15.SprintingPoseBase.AnimStateNode_2.Sprint_State
  nodes [
   n1 = Break Fort Legacy Weapon Anim Set(FortLegacyWeaponAnimSet=Get LegacyWeaponAnimSet())
   n2 = Get SpeedAdjustedPlayrate() -> (SpeedAdjustedPlayrate:float)
   n3 = Get SpeedAdjustedPlayrate()
   Output Animation Pose(Result=Blend Poses by bool(
     BlendPose_0=Sequence Player(Sequence=@n1.SprintTargetingAnimation, PlayRate=@n3) {groupName=Locomotion, method=SyncGroup}
     BlendPose_1=Sequence Player(Sequence=@n1.SprintAnimation, PlayRate=@n2) {groupName=Locomotion, method=SyncGroup}
     BlendTime_0=0.2
     BlendTime_1=0.2
     bActiveValue=Get bPlayUpperBodyTargeting()
    ) {blendType=Linear, transitionType=Inertialization})
  ]
 }
 Transition {
  path=LocomotionLayer_FullBody.AnimGraphNode_StateMachine_15.SprintingPoseBase.AnimStateTransitionNode_0.Transition
  nodes [Result(bCanEnterTransition=Get bIsCrouching())]
 }
 Transition {
  path=LocomotionLayer_FullBody.AnimGraphNode_StateMachine_15.SprintingPoseBase.AnimStateTransitionNode_1.Transition
  nodes [Result(bCanEnterTransition=true)]
 }
 Transition {
  path=LocomotionLayer_FullBody.AnimGraphNode_StateMachine_15.SprintingPoseBase.AnimStateTransitionNode_2.Transition
  nodes [Result(bCanEnterTransition=true)]
 }
 Transition {
  path=LocomotionLayer_FullBody.AnimGraphNode_StateMachine_15.SprintingPoseBase.AnimStateTransitionNode_3.Transition
  nodes [Result(bCanEnterTransition=true)]
 }
 SprintingPoseAdditives {
  entry=SprintingPoseAdditivesState
  path=LocomotionLayer_FullBody.AnimGraphNode_StateMachine_16.SprintingPoseAdditives
  ids=id
  nodes [AnimStateNode_0 = SprintingPoseAdditivesState() {is=title AnimStateNode}]
 }
 SprintingPoseAdditivesState {
  path=LocomotionLayer_FullBody.AnimGraphNode_StateMachine_16.SprintingPoseAdditives.AnimStateNode_0.SprintingPoseAdditivesState
  nodes [
   Layered blend per bone(BlendPoses_0=Use cached pose 'SprintingPose_PreMask'()) {bMeshSpaceRotationBlend=true, layerSetup [{branchFilters [{blendDepth=3, boneName=Spine_05}, {boneName=ik_hand_root}, {boneName=weapon_r}]}]}
   n1 = Get FramerateDependentLeanAngle() -> (FramerateDependentLeanAngle:float)
   n2 = Get SpeedAdjustedPlayrate()
   n3 = Blendspace Player 'CrouchSprint_Lean'!(X=@n1, PlayRate:float=@n2) -> (Pose) {
    blendSpace=/Game/Animation/Game/MainPlayer/Locomotion/Medium/Male/Crouch/Sprint/CrouchSprint_Lean
    groupName=Locomotion
    groupRole=AlwaysFollower
    method=SyncGroup
   }
   n4 = Blendspace Player 'SprintLeanAdditive'!(X=@n1, PlayRate:float=@n2) -> (Pose) {
    blendSpace=/Game/Animation/Game/MainPlayer/Locomotion/Small/Sprint/SprintLeanAdditive
    groupName=Locomotion
    groupRole=AlwaysFollower
    method=SyncGroup
   }
   n5 = Layered blend per bone(
    BasePose=Use cached pose 'SprintingPose_PreMask'()
    BlendPoses_0=Use cached pose 'SprintingPose_PreMask'()
    BlendWeights_0=Get LeftArmMaskWeight() -> (LeftArmMaskWeight:float)
   ) {layerSetup [{branchFilters [{blendDepth=3, boneName=clavicle_l}, {boneName=ik_hand_root}, {boneName=weapon_l}]}]}
   n6 = Apply Additive(
    Base=@n5
    Additive=Blend Poses by bool(BlendPose_0=@n3, BlendPose_1=@n4, bActiveValue=Get bIsCrouchSprinting()) {blendType=Linear}
    Alpha=Get LeanAdditiveAlpha() -> (LeanAdditiveAlpha:float)
   ) {lODThreshold=1, alphaBoolBlend {blendOption=Linear}}
   n7 = Sequence Player!(Sequence=Get ConsumableOffsetPose() -> (ConsumableOffsetPose:Anim Sequence Object Reference)) -> (Pose)
   n8 = Break Fort Anim Input Weapon Additive Anim Asset!(FortAnimInput_WeaponAdditiveAnimAsset:"Fort Anim Input Weapon Additive Anim Asset Structure (by ref)"=Get CurrentWeaponAdditiveAnimSet() -> (CurrentWeaponAdditiveAnimSet:Fort Anim Input Weapon Additive Anim Asset Structure)) -> (
    AdditivePoseOffset:Anim Sequence Object Reference
    TargetingAdditivePoseOffset:Anim Sequence Object Reference
    NonTargetedAdditivePoseOffset:Anim Sequence Object Reference
    RelaxedAdditivePoseOffset:Anim Sequence Object Reference
    RelaxedWhileSprintingAdditiveAnim:Anim Sequence Object Reference
    RelaxedWhileCrouchSprintingAdditiveAnim:Anim Sequence Object Reference
   ) {is=title K2Node_BreakStruct}
   n9 = Sequence Player(Sequence=@n8.RelaxedWhileCrouchSprintingAdditiveAnim) {groupName=Locomotion, groupRole=AlwaysFollower, method=SyncGroup}
   n10 = Sequence Player(Sequence=@n8.RelaxedWhileSprintingAdditiveAnim) {groupName=Locomotion, method=SyncGroup}
   n11 = Get bIsCrouchSprinting()
   Output Animation Pose(Result=Apply Additive!(
     Base=Apply Additive(Base=@n6, Additive=@n7, AlphaCurveName=DisableConsumableOffsetPose) {alphaInputType=Curve, alphaScaleBiasClamp {bClampResult=true, bias=1, scale=-1}}
     Additive=Blend Poses by bool(BlendPose_0=@n9, BlendPose_1=@n10, bActiveValue=@n11) {blendType=Linear}
     AlphaCurveName:Name=DisableWeaponAdditivePose
    ) -> (Pose) {alphaInputType=Curve, alphaScaleBiasClamp {bClampResult=true, bias=1, scale=-1}})
  ]
 }
 LocomotionLayer_FullLocomotion {nodes [
   Input Pose() {name=InPose_FullLocomotion}
   Output Pose() {layerGroup=LocomotionOverride}
  ]}
 LocomotionLayer_GroundLandAdditive {nodes [
   Input Pose()
   Output Pose() {layerGroup=LocomotionOverride}
  ]}
 LocomotionLayer_InAir {nodes [
   Input Pose() {name=InPose_InAir}
   n1 = Get LocalVelocityYawAngle()
   AnimGraphNode_StateMachine_1 = GroundJumpStateMachine() {is=title AnimGraphNode_StateMachine, bAllowConduitEntryStates=true}
   n2 = Layered blend per bone(
    BasePose=@AnimGraphNode_StateMachine_1
    BlendPoses_0=Additive Identity Pose()
    BlendWeights_0=Get JumpUpperBodyBlendWeight() -> (JumpUpperBodyBlendWeight:float)
   ) {lODThreshold=2, layerSetup [{branchFilters [{blendDepth=2, boneName=Spine_01}, {boneName=Ik_hand_root}]}]}
   n3 = Layered blend per bone(
    BasePose=@n2
    BlendPoses_0=Additive Identity Pose()
    BlendWeights_0=Get DisableUpperBodyAdditiveMeshSpaceCurveAlpha() -> (DisableUpperBodyAdditiveMeshSpaceCurveAlpha:float)
   ) {bMeshSpaceRotationBlend=true, lODThreshold=2, layerSetup [{branchFilters [{blendDepth=4, boneName=spine_01}, {boneName=ik_hand_root}]}]}
   AnimGraphNode_StateMachine_2 = GroundJumpLowerBody() {is=title AnimGraphNode_StateMachine, bAllowConduitEntryStates=true}
   n4 = Layered blend per bone(
    BasePose=Apply Additive!(Base=Use cached pose 'SourcePose'(), Additive=@n3, Alpha:float) -> (Pose) {alphaBoolBlend {blendOption=Linear}}
    BlendPoses_0=Two Way Blend(A=@AnimGraphNode_StateMachine_2, B=Local Space Ref Pose(), Alpha=Get InAirLowerBodyDamper() -> (InAirLowerBodyDamper:double))
   ) {
    bMeshSpaceRotationBlend=true
    bMeshSpaceScaleBlend=true
    curveBlendOption=UseMaxValue
    layerSetup [{branchFilters [{boneName=Root}, {blendDepth=-1, boneName=Ik_hand_root}, {blendDepth=-4, boneName=spine_01}, {blendDepth=-1, boneName=VB pelvis_ik_hand_l}]}]
   }
   n5 = Modify Curve(
    SourcePose=Apply Additive(
     Base=@n4
     Additive=Blendspace Player 'M_Jump_ARDS_Lean1D'!(X=@n1) -> (Pose) {blendSpace=/Game/Animation/Game/MainPlayer/Locomotion/Medium/Male/Jump/AssaultRifle/M_Jump_ARDS_Lean1D}
     Alpha=Get JumpAdditiveLeanAlpha() -> (JumpAdditiveLeanAlpha:float)
    ) {alphaBoolBlend {blendOption=Linear}}
    CurveValues_0=0
    Alpha=1
   ) {curveNames [WeaponAiming], curveValues [0]}
   n6 = Use cached pose 'SourcePose'()
   n7 = Blend Poses by bool(
    BlendPose_0=Use cached pose 'SourcePose'()
    BlendPose_1=Blend Poses by bool(
     BlendPose_0=Use cached pose 'SourcePose'()
     BlendPose_1=Blend Poses by bool(BlendPose_0=@n6, BlendPose_1=@n5, BlendTime_1=0.3) {blendType=Linear}
     bActiveValue=Get bIsInGoop() -> (bIsInGoop:bool)
    ) {blendType=Linear}
    BlendTime_1=0.5
    bActiveValue=Get bIsUsingJetpack() -> (bIsUsingJetpack:bool)
   ) {blendType=Linear, transitionType=Inertialization}
   n8 = Sequence Player(Sequence=Break Fort Legacy Weapon Anim Set(FortLegacyWeaponAnimSet=Get LegacyWeaponAnimSet()).FullBodyJumpAdditive) {bLoopAnimation=false}
   Output Pose(Result=Inertialization(Source=Apply Additive(Base=@n7, Additive=@n8) {alphaBoolBlend {blendOption=Linear}})) {layerGroup=LocomotionOverride}
   Save cached pose 'SourcePose'(Pose=Input Pose() {name=InPose_InAirSourcePose}) {cacheName=SourcePose}
   Get bDisableInAirAnimations() -> (bDisableInAirAnimations:bool)
  ]}
 GroundJumpStateMachine {
  entry=Conduit
  path=LocomotionLayer_InAir.AnimGraphNode_StateMachine_1.GroundJumpStateMachine
  ids=id
  nodes [
   AnimStateConduitNode_0 = Conduit()
   AnimStateNode_1 = JumpStart() {is=title AnimStateNode}
   AnimStateNode_10 = FallLoop() {is=title AnimStateNode}
   AnimStateNode_7 = LandPredict() {is=title AnimStateNode}
   AnimStateTransitionNode_0 = AnimStateTransitionNode() {crossfadeDuration=0.1, priorityOrder=100}
   AnimStateTransitionNode_1 = AnimStateTransitionNode() {crossfadeDuration=0.4, logicType=TLT_Inertialization, priorityOrder=3}
   AnimStateTransitionNode_14 = AnimStateTransitionNode() {logicType=TLT_Inertialization}
   AnimStateTransitionNode_16 = AnimStateTransitionNode() {priorityOrder=5}
   AnimStateTransitionNode_2 = AnimStateTransitionNode() {logicType=TLT_Inertialization}
   AnimStateTransitionNode_20 = AnimStateTransitionNode() {bAutomaticRuleBasedOnSequencePlayerInState=true, blendMode=Linear}
   AnimStateTransitionNode_28 = AnimStateTransitionNode() {crossfadeDuration=0.069, logicType=TLT_Inertialization}
   AnimStateTransitionNode_3 = AnimStateTransitionNode() {crossfadeDuration=0.069, logicType=TLT_Inertialization, priorityOrder=6}
   AnimStateTransitionNode_4 = AnimStateTransitionNode() {crossfadeDuration=0.1, logicType=TLT_Inertialization, priorityOrder=2}
   AnimStateTransitionNode_5 = AnimStateTransitionNode() {logicType=TLT_Inertialization, priorityOrder=2}
   AnimStateTransitionNode_7 = AnimStateTransitionNode() {priorityOrder=4}
  ]
  links [
   {
    dst=FallLoop
    rule=LocomotionLayer_InAir.AnimGraphNode_StateMachine_1.GroundJumpStateMachine.AnimStateTransitionNode_2.Transition
    src=Conduit
    via=AnimStateTransitionNode_2
   }
   {
    dst=FallLoop
    rule=LocomotionLayer_InAir.AnimGraphNode_StateMachine_1.GroundJumpStateMachine.AnimStateTransitionNode_5.Transition
    src=Conduit
    via=AnimStateTransitionNode_5
   }
   {
    dst=JumpStart
    rule=LocomotionLayer_InAir.AnimGraphNode_StateMachine_1.GroundJumpStateMachine.AnimStateTransitionNode_0.AnimationTransitionGraph_0
    src=Conduit
    via=AnimStateTransitionNode_0
   }
   {
    dst=LandPredict
    rule=LocomotionLayer_InAir.AnimGraphNode_StateMachine_1.GroundJumpStateMachine.AnimStateTransitionNode_1.AnimationTransitionGraph_0
    src=Conduit
    via=AnimStateTransitionNode_1
   }
   {
    dst=JumpStart
    rule=LocomotionLayer_InAir.AnimGraphNode_StateMachine_1.GroundJumpStateMachine.AnimStateTransitionNode_3.AnimationTransitionGraph_1
    src=FallLoop
    via=AnimStateTransitionNode_3
   }
   {
    dst=LandPredict
    rule=LocomotionLayer_InAir.AnimGraphNode_StateMachine_1.GroundJumpStateMachine.AnimStateTransitionNode_16.AnimationTransitionGraph_1
    src=FallLoop
    via=AnimStateTransitionNode_16
   }
   {
    dst=LandPredict
    rule=LocomotionLayer_InAir.AnimGraphNode_StateMachine_1.GroundJumpStateMachine.AnimStateTransitionNode_7.AnimationTransitionGraph_1
    src=FallLoop
    via=AnimStateTransitionNode_7
   }
   {
    dst=FallLoop
    rule=LocomotionLayer_InAir.AnimGraphNode_StateMachine_1.GroundJumpStateMachine.AnimStateTransitionNode_20.AnimationTransitionGraph_1
    src=JumpStart
    via=AnimStateTransitionNode_20
   }
   {
    dst=LandPredict
    rule=LocomotionLayer_InAir.AnimGraphNode_StateMachine_1.GroundJumpStateMachine.AnimStateTransitionNode_14.AnimationTransitionGraph_1
    src=JumpStart
    via=AnimStateTransitionNode_14
   }
   {
    dst=JumpStart
    rule=LocomotionLayer_InAir.AnimGraphNode_StateMachine_1.GroundJumpStateMachine.AnimStateTransitionNode_28.AnimationTransitionGraph_1
    src=LandPredict
    via=AnimStateTransitionNode_28
   }
   {
    dst=JumpStart
    rule=LocomotionLayer_InAir.AnimGraphNode_StateMachine_1.GroundJumpStateMachine.AnimStateTransitionNode_4.Transition
    src=LandPredict
    via=AnimStateTransitionNode_4
   }
  ]
 }
 Conduit {
  path=LocomotionLayer_InAir.AnimGraphNode_StateMachine_1.GroundJumpStateMachine.AnimStateConduitNode_0.Conduit
  nodes [Result(bCanEnterTransition=true)]
 }
 JumpStart {
  path=LocomotionLayer_InAir.AnimGraphNode_StateMachine_1.GroundJumpStateMachine.AnimStateNode_1.JumpStart
  nodes [
   n1 = Layered blend per bone(
    BasePose=JumpUpAdditive!() -> (Pose:Pose Link Structure) {is=title AnimGraphNode_LinkedAnimLayer, layer=JumpUpAdditive}
    BlendPoses_0="Sequence Player (additive) 'JackalVehicle_JumpCombat_Start'"() {bLoopAnimation=false, sequence=/Game/Animation/Game/MainPlayer/Vehicles/JackalVehicle/JackalVehicle_JumpCombat_Start}
    BlendWeights_0=Break Fort Anim Input Jackal Vehicle!(FortAnimInput_JackalVehicle:"Fort Anim Input Jackal Vehicle Structure (by ref)"=Get JackalVehicleInput() -> (JackalVehicleInput:Fort Anim Input Jackal Vehicle Structure)) -> (
     bIsUsingJackalVehicle:bool=~
     bIsSteeringLeft:bool=~
     bIsSteeringRight:bool=~
     SteerAngleDeadZoneDegrees:float=~
     SteerAngle:float=~
     LeanYaw:float=~
     QuantizedSteerAngle:float=~
     SteerAlpha:float=~
     RunForwardAlpha:float=~
     SlopePitchDegreeAngle:float=~
     SlopeRollDegreeAngle:float=~
     DistanceFromGround:float=~
     PivotPlayRate:float=~
     bIsReversing:bool=~
     bIsBraking:bool=~
     bIsMoving:bool=~
     bIsMovingForward:bool=~
     bIsSprinting:bool=~
     bInAir:bool=~
     bIsFalling:bool=~
     bIsJumping:bool=~
     bIsRelaxed:bool=~
     bIsBoosting:bool=~
     bHasReachedJumpChargeStartThreshold:bool=~
     bHasReachedJumpChargeMidThreshold:bool=~
     bHasReachedJumpChargeFullThreshold:bool=~
     bAdjustRootForFemaleRider:bool=~
     bIsPlayingEmoteOnHoverboard:bool=~
     bShouldApplyAdditive:bool=~
     bPlayPivotOnGroundAndNotBoosting:bool=~
     bIsOnLowerHill:bool=~
     bLowerBodyIdleToLoopTransition:bool=~
     bInterruptHoverboardFullybody:bool=~
     bEnterFullBodyHoverboardState:bool=~
     bDefaultToJumpStartTransition:bool=~
     bLocomotionPoseToJumpTransition:bool=~
     bBoostingToBoostingJumpTransition:bool=~
     bJumpToLocomotionPoseTransition:bool=~
     bJumpApexToJumpFallTransition:bool=~
     bIdleToMovementStartTransition:bool=~
     bIdleToMovementLoopTransition:bool=~
     bMovementLoopToMovementStopTransition:bool=~
     bMovementLoopToPivotTransition:bool=~
     bMovementLoopToIdleTransition:bool=~
     bIdleAdditiveToCollisionNTransition:bool=~
     bSplitBodyToHoverboardBRTransition:bool=~
     bHoverboardBRToSplitBodyTransition:bool=~
     bHoverboardBRMovementToJumpChargeTransition:bool=~
     bIdlesToJackalVehicleTransition:bool=~
     bPlayAdditiveLeans:bool=~
     bPlayBalloonLeans:bool=~
     bPlayJumpTrickVertical:bool=~
     bPlayJumpTrick:bool=~
     bPlayMovingFast:bool=~
     bPlayHipAdjustmentAdditive:bool=~
     bPlayDriveSouth:bool=~
     bPlayHeadAimOffset:bool=~
     bPlaySlopeAimOffset:bool=~
     JumpCombatAdditiveWeight:float=~
     MeleeTwistIdle:float=~
     MeleeTwistLocomotionLoop:float=~
     VelocityImpact:Fort Anim Input Velocity Impact Structure
     ImpactDisplacement:Vector=~
     SteerYaw:float=~
     EmoteHoverboardPosition:Vector=~
     EmoteHoverboardRotation:Rotator=~
     bShouldAttachFeetToHoverboard:bool=~
     FootLeftLocationOffset:Vector=~
     FootLeftRotationOffset:Rotator=~
     FootRightLocationOffset:Vector=~
     FootRightRotationOffset:Rotator=~
     StoppedThreshold:float=~
     MovingForwardThreshold:float=~
     MovingFowardFastThreshold:float=~
     DefaultToJumpStartTransitionThreshold:float=~
     JumpTrickAngularVelocityThreshold:float=~
     JumpDistanceFromGroundThreshold:float=~
     VelocityStartThreshold:float=~
     MovingFastThreshold:float=~
     RelaxedSpeedThreshold:float=~
     JumpChargeStartThreshold:float=~
     JumpChargeMidThreshold:float=~
     JumpChargeFullThreshold:float=~
     RotatingAngularVelocityThreshold:float=~
     IdleToLoopTransitionSpeedThreshold:float=~
     IdleToMovementLoopTransitionThreshold:float=~
     MeleeTwistIdleMultiplier:float=~
     MeleeTwistLocomotionLoopMultiplier:float=~
     LeanYawForMaxPivotPlayRate:float=~
     SteerAlphaForMaxPivotPlayRate:float=~
     PrePivotCardinalDirection:EFortCardinalDirection Enum
    ) {is=title K2Node_BreakStruct}.JumpCombatAdditiveWeight
   ) {
    bMeshSpaceRotationBlend=true
    lODThreshold=2
    layerSetup [{branchFilters [
       {boneName=root}
       {blendDepth=-1, boneName=pelvis}
       {boneName=thigh_r}
       {boneName=thigh_l}
       {blendDepth=-1, boneName=spine_01}
       {blendDepth=-1, boneName=Ik_hand_root}
       {boneName=ik_foot_root}
      ]}]
   }
   Output Animation Pose(Result=Two Way Blend!(A=@n1, B=Additive Identity Pose(), AlphaCurveName:Name=RemoveJumpUpAdditive) -> (Pose) {blendNode {alphaInputType=Curve, alphaBoolBlend {blendOption=Linear}}})
  ]
 }
 FallLoop {
  path=LocomotionLayer_InAir.AnimGraphNode_StateMachine_1.GroundJumpStateMachine.AnimStateNode_10.FallLoop
  nodes [
   n1 = Layered blend per bone(
    BasePose=JumpLoopAdditive!() -> (Pose:Pose Link Structure) {is=title AnimGraphNode_LinkedAnimLayer, layer=JumpLoopAdditive}
    BlendPoses_0="Sequence Player (additive) 'JackalVehicle_JumpCombat_Loop'"() {sequence=/Game/Animation/Game/MainPlayer/Vehicles/JackalVehicle/JackalVehicle_JumpCombat_Loop}
    BlendWeights_0=Break Fort Anim Input Jackal Vehicle(FortAnimInput_JackalVehicle=Get JackalVehicleInput()) -> (
     bIsUsingJackalVehicle=~
     bIsSteeringLeft=~
     bIsSteeringRight=~
     SteerAngleDeadZoneDegrees=~
     SteerAngle=~
     LeanYaw=~
     QuantizedSteerAngle=~
     SteerAlpha=~
     RunForwardAlpha=~
     SlopePitchDegreeAngle=~
     SlopeRollDegreeAngle=~
     DistanceFromGround=~
     PivotPlayRate=~
     bIsReversing=~
     bIsBraking=~
     bIsMoving=~
     bIsMovingForward=~
     bIsSprinting=~
     bInAir=~
     bIsFalling=~
     bIsJumping=~
     bIsRelaxed=~
     bIsBoosting=~
     bHasReachedJumpChargeStartThreshold=~
     bHasReachedJumpChargeMidThreshold=~
     bHasReachedJumpChargeFullThreshold=~
     bAdjustRootForFemaleRider=~
     bIsPlayingEmoteOnHoverboard=~
     bShouldApplyAdditive=~
     bPlayPivotOnGroundAndNotBoosting=~
     bIsOnLowerHill=~
     bLowerBodyIdleToLoopTransition=~
     bInterruptHoverboardFullybody=~
     bEnterFullBodyHoverboardState=~
     bDefaultToJumpStartTransition=~
     bLocomotionPoseToJumpTransition=~
     bBoostingToBoostingJumpTransition=~
     bJumpToLocomotionPoseTransition=~
     bJumpApexToJumpFallTransition=~
     bIdleToMovementStartTransition=~
     bIdleToMovementLoopTransition=~
     bMovementLoopToMovementStopTransition=~
     bMovementLoopToPivotTransition=~
     bMovementLoopToIdleTransition=~
     bIdleAdditiveToCollisionNTransition=~
     bSplitBodyToHoverboardBRTransition=~
     bHoverboardBRToSplitBodyTransition=~
     bHoverboardBRMovementToJumpChargeTransition=~
     bIdlesToJackalVehicleTransition=~
     bPlayAdditiveLeans=~
     bPlayBalloonLeans=~
     bPlayJumpTrickVertical=~
     bPlayJumpTrick=~
     bPlayMovingFast=~
     bPlayHipAdjustmentAdditive=~
     bPlayDriveSouth=~
     bPlayHeadAimOffset=~
     bPlaySlopeAimOffset=~
     JumpCombatAdditiveWeight=~
     MeleeTwistIdle=~
     MeleeTwistLocomotionLoop=~
     ImpactDisplacement=~
     SteerYaw=~
     EmoteHoverboardPosition=~
     EmoteHoverboardRotation=~
     bShouldAttachFeetToHoverboard=~
     FootLeftLocationOffset=~
     FootLeftRotationOffset=~
     FootRightLocationOffset=~
     FootRightRotationOffset=~
     StoppedThreshold=~
     MovingForwardThreshold=~
     MovingFowardFastThreshold=~
     DefaultToJumpStartTransitionThreshold=~
     JumpTrickAngularVelocityThreshold=~
     JumpDistanceFromGroundThreshold=~
     VelocityStartThreshold=~
     MovingFastThreshold=~
     RelaxedSpeedThreshold=~
     JumpChargeStartThreshold=~
     JumpChargeMidThreshold=~
     JumpChargeFullThreshold=~
     RotatingAngularVelocityThreshold=~
     IdleToLoopTransitionSpeedThreshold=~
     IdleToMovementLoopTransitionThreshold=~
     MeleeTwistIdleMultiplier=~
     MeleeTwistLocomotionLoopMultiplier=~
     LeanYawForMaxPivotPlayRate=~
     SteerAlphaForMaxPivotPlayRate=~
    ).JumpCombatAdditiveWeight
   ) {
    bMeshSpaceRotationBlend=true
    lODThreshold=2
    layerSetup [{branchFilters [
       {boneName=root}
       {blendDepth=-1, boneName=pelvis}
       {boneName=thigh_r}
       {boneName=thigh_l}
       {blendDepth=-1, boneName=spine_01}
       {blendDepth=-1, boneName=Ik_hand_root}
       {boneName=ik_foot_root}
      ]}]
   }
   Output Animation Pose(Result=Two Way Blend(A=@n1, B=Additive Identity Pose(), AlphaCurveName=RemoveJumpLoopAdditive) {blendNode {alphaInputType=Curve, alphaBoolBlend {blendOption=Linear}}})
  ]
 }
 LandPredict {
  path=LocomotionLayer_InAir.AnimGraphNode_StateMachine_1.GroundJumpStateMachine.AnimStateNode_7.LandPredict
  nodes [
   n1 = Layered blend per bone(
    BasePose=FallAdditive!() -> (Pose:Pose Link Structure) {is=title AnimGraphNode_LinkedAnimLayer, layer=FallAdditive}
    BlendPoses_0="Sequence Player (additive) 'JackalVehicle_JumpCombat_Fall'"() {bLoopAnimation=false, sequence=/Game/Animation/Game/MainPlayer/Vehicles/JackalVehicle/JackalVehicle_JumpCombat_Fall}
    BlendWeights_0=Break Fort Anim Input Jackal Vehicle(FortAnimInput_JackalVehicle=Get JackalVehicleInput()) -> (
     bIsUsingJackalVehicle=~
     bIsSteeringLeft=~
     bIsSteeringRight=~
     SteerAngleDeadZoneDegrees=~
     SteerAngle=~
     LeanYaw=~
     QuantizedSteerAngle=~
     SteerAlpha=~
     RunForwardAlpha=~
     SlopePitchDegreeAngle=~
     SlopeRollDegreeAngle=~
     DistanceFromGround=~
     PivotPlayRate=~
     bIsReversing=~
     bIsBraking=~
     bIsMoving=~
     bIsMovingForward=~
     bIsSprinting=~
     bInAir=~
     bIsFalling=~
     bIsJumping=~
     bIsRelaxed=~
     bIsBoosting=~
     bHasReachedJumpChargeStartThreshold=~
     bHasReachedJumpChargeMidThreshold=~
     bHasReachedJumpChargeFullThreshold=~
     bAdjustRootForFemaleRider=~
     bIsPlayingEmoteOnHoverboard=~
     bShouldApplyAdditive=~
     bPlayPivotOnGroundAndNotBoosting=~
     bIsOnLowerHill=~
     bLowerBodyIdleToLoopTransition=~
     bInterruptHoverboardFullybody=~
     bEnterFullBodyHoverboardState=~
     bDefaultToJumpStartTransition=~
     bLocomotionPoseToJumpTransition=~
     bBoostingToBoostingJumpTransition=~
     bJumpToLocomotionPoseTransition=~
     bJumpApexToJumpFallTransition=~
     bIdleToMovementStartTransition=~
     bIdleToMovementLoopTransition=~
     bMovementLoopToMovementStopTransition=~
     bMovementLoopToPivotTransition=~
     bMovementLoopToIdleTransition=~
     bIdleAdditiveToCollisionNTransition=~
     bSplitBodyToHoverboardBRTransition=~
     bHoverboardBRToSplitBodyTransition=~
     bHoverboardBRMovementToJumpChargeTransition=~
     bIdlesToJackalVehicleTransition=~
     bPlayAdditiveLeans=~
     bPlayBalloonLeans=~
     bPlayJumpTrickVertical=~
     bPlayJumpTrick=~
     bPlayMovingFast=~
     bPlayHipAdjustmentAdditive=~
     bPlayDriveSouth=~
     bPlayHeadAimOffset=~
     bPlaySlopeAimOffset=~
     JumpCombatAdditiveWeight=~
     MeleeTwistIdle=~
     MeleeTwistLocomotionLoop=~
     ImpactDisplacement=~
     SteerYaw=~
     EmoteHoverboardPosition=~
     EmoteHoverboardRotation=~
     bShouldAttachFeetToHoverboard=~
     FootLeftLocationOffset=~
     FootLeftRotationOffset=~
     FootRightLocationOffset=~
     FootRightRotationOffset=~
     StoppedThreshold=~
     MovingForwardThreshold=~
     MovingFowardFastThreshold=~
     DefaultToJumpStartTransitionThreshold=~
     JumpTrickAngularVelocityThreshold=~
     JumpDistanceFromGroundThreshold=~
     VelocityStartThreshold=~
     MovingFastThreshold=~
     RelaxedSpeedThreshold=~
     JumpChargeStartThreshold=~
     JumpChargeMidThreshold=~
     JumpChargeFullThreshold=~
     RotatingAngularVelocityThreshold=~
     IdleToLoopTransitionSpeedThreshold=~
     IdleToMovementLoopTransitionThreshold=~
     MeleeTwistIdleMultiplier=~
     MeleeTwistLocomotionLoopMultiplier=~
     LeanYawForMaxPivotPlayRate=~
     SteerAlphaForMaxPivotPlayRate=~
    ).JumpCombatAdditiveWeight
   ) {
    bMeshSpaceRotationBlend=true
    lODThreshold=2
    layerSetup [{branchFilters [
       {boneName=root}
       {blendDepth=-1, boneName=pelvis}
       {boneName=thigh_r}
       {boneName=thigh_l}
       {blendDepth=-1, boneName=spine_01}
       {blendDepth=-1, boneName=Ik_hand_root}
       {boneName=ik_foot_root}
      ]}]
   }
   Output Animation Pose(Result=Two Way Blend(A=@n1, B=Additive Identity Pose(), AlphaCurveName=RemoveJumpFallAdditive) {blendNode {alphaInputType=Curve, alphaBoolBlend {blendOption=Linear}}})
  ]
 }
 AnimationTransitionGraph_0 {
  path=LocomotionLayer_InAir.AnimGraphNode_StateMachine_1.GroundJumpStateMachine.AnimStateTransitionNode_0.AnimationTransitionGraph_0
  nodes [Result(bCanEnterTransition=true)]
 }
 AnimationTransitionGraph_0 {
  path=LocomotionLayer_InAir.AnimGraphNode_StateMachine_1.GroundJumpStateMachine.AnimStateTransitionNode_1.AnimationTransitionGraph_0
  nodes [Result(bCanEnterTransition=Get bLandingPredicted() -> (bLandingPredicted:bool))]
 }
 AnimationTransitionGraph_1 {
  path=LocomotionLayer_InAir.AnimGraphNode_StateMachine_1.GroundJumpStateMachine.AnimStateTransitionNode_14.AnimationTransitionGraph_1
  nodes [Result(bCanEnterTransition=Get bLandingPredicted())]
 }
 AnimationTransitionGraph_1 {
  path=LocomotionLayer_InAir.AnimGraphNode_StateMachine_1.GroundJumpStateMachine.AnimStateTransitionNode_16.AnimationTransitionGraph_1
  nodes [Result(bCanEnterTransition=Break Fort Anim Input Common Vehicle!(FortAnimInput_CommonVehicle:"Fort Anim Input Common Vehicle Structure (by ref)"=Get CommonVehicleInput() -> (CommonVehicleInput:Fort Anim Input Common Vehicle Structure)) -> (
     bIsUsingVehicle:bool=~
     bIsJumpingVehicle:bool=~
     bCanChargeJump:bool=~
     bIsChargingJump:bool=~
     bIsJumping:bool=~
     bIsOnGround:bool=~
     bCanDriverAimWeapon:bool=~
    ) {is=title K2Node_BreakStruct}.bIsOnGround)]
 }
 Transition {
  path=LocomotionLayer_InAir.AnimGraphNode_StateMachine_1.GroundJumpStateMachine.AnimStateTransitionNode_2.Transition
  nodes [Result(bCanEnterTransition=Get bSkipJumpStart() -> (bSkipJumpStart:bool))]
 }
 AnimationTransitionGraph_1 {
  path=LocomotionLayer_InAir.AnimGraphNode_StateMachine_1.GroundJumpStateMachine.AnimStateTransitionNode_20.AnimationTransitionGraph_1
  nodes [Result()]
 }
 AnimationTransitionGraph_1 {
  path=LocomotionLayer_InAir.AnimGraphNode_StateMachine_1.GroundJumpStateMachine.AnimStateTransitionNode_28.AnimationTransitionGraph_1
  nodes [Result(bCanEnterTransition=Break Fort Anim Input Transition Properties!(FortAnimInput_TransitionProperties:"Fort Anim Input Transition Properties Structure (by ref)"=Get TransitionPropertiesInput() -> (TransitionPropertiesInput:Fort Anim Input Transition Properties Structure)) -> (bTransition_DoubleJump_Fast:bool=~, bTransition_DoubleJump:bool=~) {is=title K2Node_BreakStruct}.bTransition_DoubleJump)]
 }
 AnimationTransitionGraph_1 {
  path=LocomotionLayer_InAir.AnimGraphNode_StateMachine_1.GroundJumpStateMachine.AnimStateTransitionNode_3.AnimationTransitionGraph_1
  nodes [Result(bCanEnterTransition=Break Fort Anim Input Transition Properties(FortAnimInput_TransitionProperties=Get TransitionPropertiesInput()) -> (bTransition_DoubleJump_Fast=~, bTransition_DoubleJump=~).bTransition_DoubleJump)]
 }
 Transition {
  path=LocomotionLayer_InAir.AnimGraphNode_StateMachine_1.GroundJumpStateMachine.AnimStateTransitionNode_4.Transition
  nodes [Result(bCanEnterTransition=Get bIsJumping())]
 }
 Transition {
  path=LocomotionLayer_InAir.AnimGraphNode_StateMachine_1.GroundJumpStateMachine.AnimStateTransitionNode_5.Transition
  nodes [Result(bCanEnterTransition=Get bIsFalling())]
 }
 AnimationTransitionGraph_1 {
  path=LocomotionLayer_InAir.AnimGraphNode_StateMachine_1.GroundJumpStateMachine.AnimStateTransitionNode_7.AnimationTransitionGraph_1
  nodes [Result(bCanEnterTransition=Get bTransition_NewJumpLoopAdditive_to_NewFallAdditive() -> (bTransition_NewJumpLoopAdditive_to_NewFallAdditive:bool))]
 }
 GroundJumpLowerBody {
  entry=Conduit_1
  path=LocomotionLayer_InAir.AnimGraphNode_StateMachine_2.GroundJumpLowerBody
  ids=id
  nodes [
   AnimStateConduitNode_0 = Conduit_1() {is=title AnimStateConduitNode}
   AnimStateNode_1 = JumpStart()
   AnimStateNode_10 = FallLoop()
   AnimStateNode_7 = LandPredict()
   AnimStateTransitionNode_0 = AnimStateTransitionNode() {crossfadeDuration=0.1, priorityOrder=100}
   AnimStateTransitionNode_1 = AnimStateTransitionNode() {crossfadeDuration=0.4, logicType=TLT_Inertialization}
   AnimStateTransitionNode_14 = AnimStateTransitionNode()
   AnimStateTransitionNode_16 = AnimStateTransitionNode() {priorityOrder=5}
   AnimStateTransitionNode_2 = AnimStateTransitionNode() {logicType=TLT_Inertialization}
   AnimStateTransitionNode_20 = AnimStateTransitionNode() {bAutomaticRuleBasedOnSequencePlayerInState=true, blendMode=Linear}
   AnimStateTransitionNode_28 = AnimStateTransitionNode() {crossfadeDuration=0.069, logicType=TLT_Inertialization}
   AnimStateTransitionNode_3 = AnimStateTransitionNode() {crossfadeDuration=0.069, logicType=TLT_Inertialization, priorityOrder=6}
   AnimStateTransitionNode_4 = AnimStateTransitionNode() {crossfadeDuration=0.1, logicType=TLT_Inertialization, priorityOrder=2}
   AnimStateTransitionNode_5 = AnimStateTransitionNode() {logicType=TLT_Inertialization, priorityOrder=2}
   AnimStateTransitionNode_7 = AnimStateTransitionNode() {priorityOrder=4}
  ]
  links [
   {
    dst=FallLoop
    rule=LocomotionLayer_InAir.AnimGraphNode_StateMachine_2.GroundJumpLowerBody.AnimStateTransitionNode_2.AnimationTransitionGraph_0
    src=Conduit_1
    via=AnimStateTransitionNode_2
   }
   {
    dst=FallLoop
    rule=LocomotionLayer_InAir.AnimGraphNode_StateMachine_2.GroundJumpLowerBody.AnimStateTransitionNode_5.Transition
    src=Conduit_1
    via=AnimStateTransitionNode_5
   }
   {
    dst=JumpStart
    rule=LocomotionLayer_InAir.AnimGraphNode_StateMachine_2.GroundJumpLowerBody.AnimStateTransitionNode_0.AnimationTransitionGraph_1
    src=Conduit_1
    via=AnimStateTransitionNode_0
   }
   {
    dst=LandPredict
    rule=LocomotionLayer_InAir.AnimGraphNode_StateMachine_2.GroundJumpLowerBody.AnimStateTransitionNode_1.AnimationTransitionGraph_1
    src=Conduit_1
    via=AnimStateTransitionNode_1
   }
   {
    dst=JumpStart
    rule=LocomotionLayer_InAir.AnimGraphNode_StateMachine_2.GroundJumpLowerBody.AnimStateTransitionNode_3.AnimationTransitionGraph_0
    src=FallLoop
    via=AnimStateTransitionNode_3
   }
   {
    dst=LandPredict
    rule=LocomotionLayer_InAir.AnimGraphNode_StateMachine_2.GroundJumpLowerBody.AnimStateTransitionNode_16.AnimationTransitionGraph_0
    src=FallLoop
    via=AnimStateTransitionNode_16
   }
   {
    dst=LandPredict
    rule=LocomotionLayer_InAir.AnimGraphNode_StateMachine_2.GroundJumpLowerBody.AnimStateTransitionNode_7.AnimationTransitionGraph_0
    src=FallLoop
    via=AnimStateTransitionNode_7
   }
   {
    dst=FallLoop
    rule=LocomotionLayer_InAir.AnimGraphNode_StateMachine_2.GroundJumpLowerBody.AnimStateTransitionNode_20.AnimationTransitionGraph_0
    src=JumpStart
    via=AnimStateTransitionNode_20
   }
   {
    dst=LandPredict
    rule=LocomotionLayer_InAir.AnimGraphNode_StateMachine_2.GroundJumpLowerBody.AnimStateTransitionNode_14.AnimationTransitionGraph_0
    src=JumpStart
    via=AnimStateTransitionNode_14
   }
   {
    dst=JumpStart
    rule=LocomotionLayer_InAir.AnimGraphNode_StateMachine_2.GroundJumpLowerBody.AnimStateTransitionNode_28.AnimationTransitionGraph_0
    src=LandPredict
    via=AnimStateTransitionNode_28
   }
   {
    dst=JumpStart
    rule=LocomotionLayer_InAir.AnimGraphNode_StateMachine_2.GroundJumpLowerBody.AnimStateTransitionNode_4.AnimationTransitionGraph_0
    src=LandPredict
    via=AnimStateTransitionNode_4
   }
  ]
 }
 Conduit_1 {
  path=LocomotionLayer_InAir.AnimGraphNode_StateMachine_2.GroundJumpLowerBody.AnimStateConduitNode_0.Conduit_1
  nodes [Result(bCanEnterTransition=true)]
 }
 JumpStart {
  path=LocomotionLayer_InAir.AnimGraphNode_StateMachine_2.GroundJumpLowerBody.AnimStateNode_1.JumpStart
  nodes [Output Animation Pose(Result=Sequence Player 'M_JumpUp_ARDS_Base'() {sequence=/Game/Animation/Game/MainPlayer/Locomotion/Medium/Male/Jump/AssaultRifle/M_JumpUp_ARDS_Base})]
 }
 FallLoop {
  path=LocomotionLayer_InAir.AnimGraphNode_StateMachine_2.GroundJumpLowerBody.AnimStateNode_10.FallLoop
  nodes [Output Animation Pose(Result=Sequence Player 'M_JumpApex_ARDS_Loop'() {sequence=/Game/Animation/Game/MainPlayer/Locomotion/Medium/Male/Jump/AssaultRifle/M_JumpApex_ARDS_Loop})]
 }
 LandPredict {
  path=LocomotionLayer_InAir.AnimGraphNode_StateMachine_2.GroundJumpLowerBody.AnimStateNode_7.LandPredict
  nodes [Output Animation Pose(Result=Sequence Player 'M_JumpFall_ARDS_Base'() {bLoopAnimation=false, sequence=/Game/Animation/Game/MainPlayer/Locomotion/Medium/Male/Jump/AssaultRifle/M_JumpFall_ARDS_Base})]
 }
 AnimationTransitionGraph_1 {
  path=LocomotionLayer_InAir.AnimGraphNode_StateMachine_2.GroundJumpLowerBody.AnimStateTransitionNode_0.AnimationTransitionGraph_1
  nodes [Result(bCanEnterTransition=true)]
 }
 AnimationTransitionGraph_1 {
  path=LocomotionLayer_InAir.AnimGraphNode_StateMachine_2.GroundJumpLowerBody.AnimStateTransitionNode_1.AnimationTransitionGraph_1
  nodes [Result(bCanEnterTransition=Get bLandingPredicted())]
 }
 AnimationTransitionGraph_0 {
  path=LocomotionLayer_InAir.AnimGraphNode_StateMachine_2.GroundJumpLowerBody.AnimStateTransitionNode_14.AnimationTransitionGraph_0
  nodes [Result(bCanEnterTransition=Get bLandingPredicted())]
 }
 AnimationTransitionGraph_0 {
  path=LocomotionLayer_InAir.AnimGraphNode_StateMachine_2.GroundJumpLowerBody.AnimStateTransitionNode_16.AnimationTransitionGraph_0
  nodes [Result(bCanEnterTransition=Break Fort Anim Input Common Vehicle(FortAnimInput_CommonVehicle=Get CommonVehicleInput()) -> (bIsUsingVehicle=~, bIsJumpingVehicle=~, bCanChargeJump=~, bIsChargingJump=~, bIsJumping=~, bIsOnGround=~, bCanDriverAimWeapon=~).bIsOnGround)]
 }
 AnimationTransitionGraph_0 {
  path=LocomotionLayer_InAir.AnimGraphNode_StateMachine_2.GroundJumpLowerBody.AnimStateTransitionNode_2.AnimationTransitionGraph_0
  nodes [Result(bCanEnterTransition=Get bSkipJumpStart())]
 }
 AnimationTransitionGraph_0 {
  path=LocomotionLayer_InAir.AnimGraphNode_StateMachine_2.GroundJumpLowerBody.AnimStateTransitionNode_20.AnimationTransitionGraph_0
  nodes [Result()]
 }
 AnimationTransitionGraph_0 {
  path=LocomotionLayer_InAir.AnimGraphNode_StateMachine_2.GroundJumpLowerBody.AnimStateTransitionNode_28.AnimationTransitionGraph_0
  nodes [Result(bCanEnterTransition=Break Fort Anim Input Transition Properties(FortAnimInput_TransitionProperties=Get TransitionPropertiesInput()) -> (bTransition_DoubleJump_Fast=~, bTransition_DoubleJump=~).bTransition_DoubleJump)]
 }
 AnimationTransitionGraph_0 {
  path=LocomotionLayer_InAir.AnimGraphNode_StateMachine_2.GroundJumpLowerBody.AnimStateTransitionNode_3.AnimationTransitionGraph_0
  nodes [Result(bCanEnterTransition=Break Fort Anim Input Transition Properties(FortAnimInput_TransitionProperties=Get TransitionPropertiesInput()) -> (bTransition_DoubleJump_Fast=~, bTransition_DoubleJump=~).bTransition_DoubleJump)]
 }
 AnimationTransitionGraph_0 {
  path=LocomotionLayer_InAir.AnimGraphNode_StateMachine_2.GroundJumpLowerBody.AnimStateTransitionNode_4.AnimationTransitionGraph_0
  nodes [Result(bCanEnterTransition=Get bIsJumping())]
 }
 Transition {
  path=LocomotionLayer_InAir.AnimGraphNode_StateMachine_2.GroundJumpLowerBody.AnimStateTransitionNode_5.Transition
  nodes [Result(bCanEnterTransition=Get bIsFalling())]
 }
 AnimationTransitionGraph_0 {
  path=LocomotionLayer_InAir.AnimGraphNode_StateMachine_2.GroundJumpLowerBody.AnimStateTransitionNode_7.AnimationTransitionGraph_0
  nodes [Result(bCanEnterTransition=Get bTransition_NewJumpLoopAdditive_to_NewFallAdditive())]
 }
 LocomotionLayer_LowerBody {nodes [
   AnimGraphNode_StateMachine_5 = LocomotionLowerBody() {is=title AnimGraphNode_StateMachine, bAllowConduitEntryStates=true}
   Output Pose(Result=@AnimGraphNode_StateMachine_5) {layerGroup=LocomotionOverride}
   Save cached pose 'UpperBodyInputPose'(Pose=Input Pose() {name=InPose_LowerBody}) {cacheName=UpperBodyInputPose}
  ]}
 LocomotionLowerBody {
  entry=Idles
  path=LocomotionLayer_LowerBody.AnimGraphNode_StateMachine_5.LocomotionLowerBody
  ids=id
  nodes [
   AnimStateNode_0 = Turns() {
    is=title AnimStateNode
    stateEntered {guid=#2}
    stateFullyBlended {guid=#3}
    stateLeft {guid=#4}
   }
   AnimStateNode_1 = ShuffleInPlace() {is=title AnimStateNode}
   AnimStateNode_14 = Movement() {is=title AnimStateNode}
   AnimStateNode_2 = Idles() {
    is=title AnimStateNode
    stateEntered {guid=#5}
    stateFullyBlended {guid=#6}
    stateLeft {guid=#7}
   }
   AnimStateNode_3 = JogStop() {is=title AnimStateNode}
   AnimStateNode_4 = JogPostPivot() {is=title AnimStateNode, bAlwaysResetOnEntry=true}
   AnimStateNode_5 = JogPrePivot() {is=title AnimStateNode, bAlwaysResetOnEntry=true}
   AnimStateNode_6 = JogStart() {is=title AnimStateNode}
   AnimStateTransitionNode_0 = AnimStateTransitionNode() {
    crossfadeDuration=0
    transitionEnd {guid=#8}
    transitionInterrupt {guid=#9}
    transitionStart {guid=#10}
   }
   AnimStateTransitionNode_1 = AnimStateTransitionNode() {
    crossfadeDuration=0
    priorityOrder=2
    transitionEnd {guid=#11}
    transitionInterrupt {guid=#12}
    transitionStart {guid=#13}
   }
   AnimStateTransitionNode_10 = AnimStateTransitionNode() {
    crossfadeDuration=0.25
    blendProfileWrapper {blendProfile=&Fortnite_M_Avg_Player_Skeleton:FastFeet}
   }
   AnimStateTransitionNode_11 = AnimStateTransitionNode() {crossfadeDuration=0.25, priorityOrder=2}
   AnimStateTransitionNode_12 = AnimStateTransitionNode()
   AnimStateTransitionNode_14 = AnimStateTransitionNode() {
    crossfadeDuration=0.25
    blendProfileWrapper {blendProfile=&Fortnite_M_Avg_Player_Skeleton:FastFeet}
   }
   AnimStateTransitionNode_15 = AnimStateTransitionNode() {crossfadeDuration=0.0667}
   AnimStateTransitionNode_16 = AnimStateTransitionNode() {bAutomaticRuleBasedOnSequencePlayerInState=true, crossfadeDuration=0.4}
   AnimStateTransitionNode_17 = AnimStateTransitionNode()
   AnimStateTransitionNode_18 = AnimStateTransitionNode() {crossfadeDuration=0.25, priorityOrder=3}
   AnimStateTransitionNode_19 = AnimStateTransitionNode() {
    crossfadeDuration=0.4
    priorityOrder=2
    blendProfileWrapper {blendProfile=&Fortnite_M_Avg_Player_Skeleton:FastFeet}
   }
   AnimStateTransitionNode_2 = AnimStateTransitionNode() {
    bAutomaticRuleBasedOnSequencePlayerInState=true
    crossfadeDuration=0
    transitionEnd {guid=#14}
    transitionInterrupt {guid=#15}
    transitionStart {guid=#16}
   }
   AnimStateTransitionNode_20 = AnimStateTransitionNode() {blendProfileWrapper {blendProfile=&Fortnite_M_Avg_Player_Skeleton:FastFeet}}
   AnimStateTransitionNode_21 = AnimStateTransitionNode() {bAutomaticRuleBasedOnSequencePlayerInState=true, crossfadeDuration=0.5}
   AnimStateTransitionNode_22 = AnimStateTransitionNode() {blendProfileWrapper {blendProfile=&Fortnite_M_Avg_Player_Skeleton:FastFeet}}
   AnimStateTransitionNode_23 = AnimStateTransitionNode() {bAutomaticRuleBasedOnSequencePlayerInState=true, crossfadeDuration=0.5, priorityOrder=2}
   AnimStateTransitionNode_24 = AnimStateTransitionNode() {
    crossfadeDuration=0.15
    blendProfileWrapper {blendProfile=&Fortnite_M_Avg_Player_Skeleton:FastFeet}
   }
   AnimStateTransitionNode_29 = AnimStateTransitionNode() {
    crossfadeDuration=0
    transitionEnd {guid=#17}
    transitionInterrupt {guid=#18}
    transitionStart {guid=#19}
   }
   AnimStateTransitionNode_3 = AnimStateTransitionNode() {crossfadeDuration=0.3, priorityOrder=3}
   AnimStateTransitionNode_4 = AnimStateTransitionNode() {crossfadeDuration=0.35}
   AnimStateTransitionNode_5 = AnimStateTransitionNode() {
    crossfadeDuration=0
    transitionEnd {guid=#20}
    transitionInterrupt {guid=#21}
    transitionStart {guid=#22}
   }
   AnimStateTransitionNode_7 = AnimStateTransitionNode() {
    crossfadeDuration=0.25
    priorityOrder=2
    blendProfileWrapper {blendProfile=&Fortnite_M_Avg_Player_Skeleton:FastFeet}
   }
   AnimStateTransitionNode_8 = AnimStateTransitionNode() {
    bAutomaticRuleBasedOnSequencePlayerInState=true
    crossfadeDuration=0.75
    blendProfileWrapper {blendProfile=&Fortnite_M_Avg_Player_Skeleton:FastFeet}
   }
   AnimStateTransitionNode_9 = AnimStateTransitionNode() {blendProfileWrapper {blendProfile=&Fortnite_M_Avg_Player_Skeleton:FastFeet}}
  ]
  links [
   {
    dst=JogPrePivot
    rule=LocomotionLayer_LowerBody.AnimGraphNode_StateMachine_5.LocomotionLowerBody.AnimStateTransitionNode_14.AnimationTransitionGraph_1
    src=Idles
    via=AnimStateTransitionNode_14
   }
   {
    dst=JogStart
    rule=LocomotionLayer_LowerBody.AnimGraphNode_StateMachine_5.LocomotionLowerBody.AnimStateTransitionNode_20.AnimationTransitionGraph_1
    src=Idles
    via=AnimStateTransitionNode_20
   }
   {
    dst=JogStop
    rule=LocomotionLayer_LowerBody.AnimGraphNode_StateMachine_5.LocomotionLowerBody.AnimStateTransitionNode_10.AnimationTransitionGraph_1
    src=Idles
    via=AnimStateTransitionNode_10
   }
   {
    dst=Movement
    rule=LocomotionLayer_LowerBody.AnimGraphNode_StateMachine_5.LocomotionLowerBody.AnimStateTransitionNode_19.AnimationTransitionGraph_1
    src=Idles
    via=AnimStateTransitionNode_19
   }
   {
    dst=ShuffleInPlace
    rule=LocomotionLayer_LowerBody.AnimGraphNode_StateMachine_5.LocomotionLowerBody.AnimStateTransitionNode_3.AnimationTransitionGraph_1
    src=Idles
    via=AnimStateTransitionNode_3
   }
   {
    dst=Turns
    rule=LocomotionLayer_LowerBody.AnimGraphNode_StateMachine_5.LocomotionLowerBody.AnimStateTransitionNode_0.AnimationTransitionGraph_1
    src=Idles
    via=AnimStateTransitionNode_0
   }
   {
    dst=Turns
    rule=LocomotionLayer_LowerBody.AnimGraphNode_StateMachine_5.LocomotionLowerBody.AnimStateTransitionNode_29.AnimationTransitionGraph_1
    src=Idles
    via=AnimStateTransitionNode_29
   }
   {
    dst=JogPrePivot
    rule=LocomotionLayer_LowerBody.AnimGraphNode_StateMachine_5.LocomotionLowerBody.AnimStateTransitionNode_24.AnimationTransitionGraph_1
    src=JogPostPivot
    via=AnimStateTransitionNode_24
   }
   {
    dst=Movement
    rule=LocomotionLayer_LowerBody.AnimGraphNode_StateMachine_5.LocomotionLowerBody.AnimStateTransitionNode_16.AnimationTransitionGraph_1
    src=JogPostPivot
    via=AnimStateTransitionNode_16
   }
   {
    dst=Movement
    rule=LocomotionLayer_LowerBody.AnimGraphNode_StateMachine_5.LocomotionLowerBody.AnimStateTransitionNode_17.AnimationTransitionGraph_1
    src=JogPostPivot
    via=AnimStateTransitionNode_17
   }
   {
    dst=JogPostPivot
    rule=LocomotionLayer_LowerBody.AnimGraphNode_StateMachine_5.LocomotionLowerBody.AnimStateTransitionNode_15.AnimationTransitionGraph_1
    src=JogPrePivot
    via=AnimStateTransitionNode_15
   }
   {
    dst=Movement
    rule=LocomotionLayer_LowerBody.AnimGraphNode_StateMachine_5.LocomotionLowerBody.AnimStateTransitionNode_12.AnimationTransitionGraph_1
    src=JogPrePivot
    via=AnimStateTransitionNode_12
   }
   {
    dst=Movement
    rule=LocomotionLayer_LowerBody.AnimGraphNode_StateMachine_5.LocomotionLowerBody.AnimStateTransitionNode_23.AnimationTransitionGraph_1
    src=JogPrePivot
    via=AnimStateTransitionNode_23
   }
   {
    dst=Movement
    rule=LocomotionLayer_LowerBody.AnimGraphNode_StateMachine_5.LocomotionLowerBody.AnimStateTransitionNode_21.AnimationTransitionGraph_1
    src=JogStart
    via=AnimStateTransitionNode_21
   }
   {
    dst=Movement
    rule=LocomotionLayer_LowerBody.AnimGraphNode_StateMachine_5.LocomotionLowerBody.AnimStateTransitionNode_22.AnimationTransitionGraph_1
    src=JogStart
    via=AnimStateTransitionNode_22
   }
   {
    dst=Idles
    rule=LocomotionLayer_LowerBody.AnimGraphNode_StateMachine_5.LocomotionLowerBody.AnimStateTransitionNode_11.AnimationTransitionGraph_1
    src=JogStop
    via=AnimStateTransitionNode_11
   }
   {
    dst=Idles
    rule=LocomotionLayer_LowerBody.AnimGraphNode_StateMachine_5.LocomotionLowerBody.AnimStateTransitionNode_8.AnimationTransitionGraph_1
    src=JogStop
    via=AnimStateTransitionNode_8
   }
   {
    dst=Idles
    rule=LocomotionLayer_LowerBody.AnimGraphNode_StateMachine_5.LocomotionLowerBody.AnimStateTransitionNode_18.AnimationTransitionGraph_1
    src=Movement
    via=AnimStateTransitionNode_18
   }
   {
    dst=JogPrePivot
    rule=LocomotionLayer_LowerBody.AnimGraphNode_StateMachine_5.LocomotionLowerBody.AnimStateTransitionNode_9.AnimationTransitionGraph_1
    src=Movement
    via=AnimStateTransitionNode_9
   }
   {
    dst=JogStop
    rule=LocomotionLayer_LowerBody.AnimGraphNode_StateMachine_5.LocomotionLowerBody.AnimStateTransitionNode_7.AnimationTransitionGraph_1
    src=Movement
    via=AnimStateTransitionNode_7
   }
   {
    dst=Idles
    rule=LocomotionLayer_LowerBody.AnimGraphNode_StateMachine_5.LocomotionLowerBody.AnimStateTransitionNode_4.AnimationTransitionGraph_1
    src=ShuffleInPlace
    via=AnimStateTransitionNode_4
   }
   {
    dst=Idles
    rule=LocomotionLayer_LowerBody.AnimGraphNode_StateMachine_5.LocomotionLowerBody.AnimStateTransitionNode_1.AnimationTransitionGraph_1
    src=Turns
    via=AnimStateTransitionNode_1
   }
   {
    dst=Idles
    rule=LocomotionLayer_LowerBody.AnimGraphNode_StateMachine_5.LocomotionLowerBody.AnimStateTransitionNode_2.AnimationTransitionGraph_1
    src=Turns
    via=AnimStateTransitionNode_2
   }
   {
    dst=Idles
    rule=LocomotionLayer_LowerBody.AnimGraphNode_StateMachine_5.LocomotionLowerBody.AnimStateTransitionNode_5.AnimationTransitionGraph_1
    src=Turns
    via=AnimStateTransitionNode_5
   }
  ]
 }
 Turns {
  path=LocomotionLayer_LowerBody.AnimGraphNode_StateMachine_5.LocomotionLowerBody.AnimStateNode_0.Turns
  nodes [
   n1 = "Sequence Player (additive) 'Idle_Turn_Right_Additive'"() {bLoopAnimation=false, sequence=/Game/Animation/Game/MainPlayer/Locomotion/Small/idle/Idle_Turn_Right_Additive}
   n2 = "Sequence Player (additive) 'Idle_Turn_Left_Additive'"() {bLoopAnimation=false, sequence=/Game/Animation/Game/MainPlayer/Locomotion/Small/idle/Idle_Turn_Left_Additive}
   n3 = Break Fort Anim Input Turn in Place!(FortAnimInput_TurnInPlace:"Fort Anim Input Turn In Place Structure (by ref)"=Get TurnInPlaceInput() -> (TurnInPlaceInput:Fort Anim Input Turn In Place Structure)) -> (
    RootYawOffset:float=~
    TurnAnimPlayRate:float=~
    bWantsToTurn:bool=~
    bExitTurn:bool=~
    bUseTurnA:bool=~
    TurnACurveAlpha:float=~
    TurnBCurveAlpha:float=~
    bWantsToTurnInVehicle:bool=~
    bWantsToTurnAgain:bool=~
    bTurningLeft:bool=~
    ShufflePlayRate:float=~
    CrouchShufflePlayRate:float=~
    CrouchShuffleStartPosition:float=~
    CrouchShuffleSpeedScaling:float=~
    bTransition_LowerBody_Shuffle:bool=~
    bTransition_LowerBody_ShuffleADS:bool=~
    bPlayOutro:bool=~
   ) {is=title K2Node_BreakStruct}
   n4 = Local To Component(LocalPose=Apply Additive(
     Base=Use cached pose 'UpperBodyInputPose'()
     Additive=Blend Poses by bool(BlendPose_0=@n2, BlendPose_1=@n1, bActiveValue=@n3.bTurningLeft) {blendType=Linear}
    ) {alphaBoolBlend {blendOption=Linear}})
   n5 = Get IdlePelvisOffset() -> (IdlePelvisOffset:Vector)
   n6 = Get IdlePelvisOffsetAlpha() -> (IdlePelvisOffsetAlpha:float)
   Output Animation Pose(Result=Component To Local(ComponentPose="Transform (Modify) Bone - Bone: pelvis"(ComponentPose=@n4, Alpha=@n6, Translation=@n5, Rotation="0,0,0", Scale="1,1,1") {lODThreshold=2, translationMode=BMM_Additive, alphaBoolBlend {blendOption=Linear}, boneToModify {boneName=pelvis}}))
  ]
 }
 ShuffleInPlace {
  path=LocomotionLayer_LowerBody.AnimGraphNode_StateMachine_5.LocomotionLowerBody.AnimStateNode_1.ShuffleInPlace
  nodes [
   n1 = Sequence Player 'Crouch_ARNonTargeted_E_Core'!(PlayRate:float=Get YawDeltaSmoothed() -> (YawDeltaSmoothed:float)) -> (Pose) {
    groupName=TurnInPlaceSync
    method=SyncGroup
    sequence=/Game/Animation/Game/MainPlayer/Locomotion/Medium/Male/Crouch/AssaultRifle_NonTargeted/Crouch_ARNonTargeted_E_Core
    playRateScaleBiasClampConstants {bClampResult=true, clampMax=1.2, clampMin=-1.2, scale=0.15}
   }
   n2 = Speed Warping!(
    ComponentPose:Component Space Pose Link Structure=Local To Component(LocalPose=@n1)
    Alpha:float=1
    SpeedScaling:float=Get AbsYawDeltaSmoothed() -> (AbsYawDeltaSmoothed:float)
    ManualSpeedWarpingDir:Vector="0,1, 0"
   ) -> (Pose:Component Space Pose Link Structure) {
    is=title FortAnimGraphNode_SpeedWarping
    lODThreshold=3
    alphaBoolBlend {blendOption=Linear}
    feetDefinitions [
     {numBonesInLimb=2, fKFootBone {boneName=foot_l}, iKFootBone {boneName=ik_foot_l}}
     {numBonesInLimb=2, fKFootBone {boneName=foot_r}, iKFootBone {boneName=ik_foot_r}}
    ]
    iKFootRootBone {boneName=ik_foot_root}
    pelvisBone {boneName=pelvis}
    speedScalingScaleBiasClamp {bClampResult=true, scale=0.08}
   }
   n3 = Component To Local(ComponentPose=@n2)
   n4 = Get YawDeltaSmoothed()
   n5 = Sequence Player 'M_Jog_Melee_Twohanded_E_Core'!(PlayRate:float=@n4) -> (Pose) {
    groupName=TurnInPlaceSync
    method=SyncGroup
    sequence=/Game/Animation/Game/MainPlayer/Locomotion/Medium/Male/Jog/Melee_TH/M_Jog_Melee_Twohanded_E_Core
    playRateScaleBiasClampConstants {bClampResult=true, clampMax=1.2, clampMin=-1.2, scale=0.9}
   }
   n6 = Sequence Player 'M_Jog_Melee_TwoHanded_E_ScaledForTurnInPlace'!(PlayRate:float=@n4) -> (Pose) {
    groupName=TurnInPlaceSync
    method=SyncGroup
    sequence=/Game/Animation/Game/MainPlayer/Locomotion/Medium/Male/Jog/Melee_TH/M_Jog_Melee_TwoHanded_E_ScaledForTurnInPlace
    playRateScaleBiasClampConstants {bClampResult=true, clampMax=1.2, clampMin=-1.2, scale=0.18}
   }
   n7 = Two Way Blend!(A=@n6, B=@n5, Alpha:float=Get AbsYawDeltaSmoothed()) -> (Pose) {blendNode {alphaBoolBlend {blendOption=Linear}, alphaScaleBias {scale=0.0625}}}
   n8 = Get bIsCrouching()
   n9 = Local To Component(LocalPose=Rotate Root Bone!(
     BasePose=Blend Poses by bool(BlendPose_0=@n3, BlendPose_1=@n7, bActiveValue=@n8) {blendType=Linear}
     Pitch:float=0
     Yaw=Get MeleeTwistCurveValue() -> (MeleeTwistCurveValue:float)
     MeshToComponent:Rotator="0,-90, 0"
    ) -> (Pose) {yawScaleBiasClamp {bClampResult=true, clampMax=60, clampMin=-60, scale=0.75}})
   n10 = Get IdlePelvisOffset()
   n11 = Get IdlePelvisOffsetAlpha()
   Output Animation Pose(Result=Component To Local(ComponentPose="Transform (Modify) Bone - Bone: pelvis"(ComponentPose=@n9, Alpha=@n11, Translation=@n10, Rotation="0,0,0", Scale="1,1,1") {lODThreshold=2, translationMode=BMM_Additive, alphaBoolBlend {blendOption=Linear}, boneToModify {boneName=pelvis}}))
  ]
 }
 Movement {
  path=LocomotionLayer_LowerBody.AnimGraphNode_StateMachine_5.LocomotionLowerBody.AnimStateNode_14.Movement
  nodes [
   n1 = Break Fort Locomotion Anim Set(FortLocomotionAnimSet=Get LocomotionAnimSet()) -> (EnableAdditiveLeans=~, UseMovementAdditiveForOneHandedTransitions=~, PlayMovementTransitions=~)
   n2 = Get SpeedAdjustedPlayrate()
   n3 = Get FramerateDependentLeanAngle()
   n4 = Blendspace Player!(X=@n3, PlayRate=@n2, BlendSpace=@n1.AdditiveLean) -> (Pose) {groupName=Locomotion, groupRole=AlwaysFollower, method=SyncGroup}
   n5 = Blendspace Player(X=@n3, PlayRate=@n2, BlendSpace=@n1.CrouchAdditiveLean) {groupName=Locomotion, groupRole=AlwaysFollower, method=SyncGroup}
   n6 = Get bIsCrouching()
   n7 = "Blend Poses (EFortWeaponCoreAnimation)"(
    BlendPose_0=Movement_Standing!() -> (Pose:Pose Link Structure) {is=title AnimGraphNode_LinkedAnimLayer, layer=Movement_Standing}
    BlendPose_1=Movement_HeavyWeapon!() -> (Pose:Pose Link Structure) {is=title AnimGraphNode_LinkedAnimLayer, layer=Movement_HeavyWeapon}
    BlendTime_0=0
    BlendTime_1=0
    ActiveEnumValue=Get WeaponCoreAnim()
   ) {blendPose [{linkId=-1, sourceLinkId=-1}, {linkId=-1, sourceLinkId=-1}], blendTime [0.1,0.1]}
   n8 = Movement_Crouching!() -> (Pose:Pose Link Structure) {is=title AnimGraphNode_LinkedAnimLayer, layer=Movement_Crouching}
   n9 = Get bIsCrouching()
   Output Animation Pose(Result=Apply Additive(
     Base=Blend Poses by bool(BlendPose_0=@n8, BlendPose_1=@n7, BlendTime_0=.25, BlendTime_1=0.35, bActiveValue=@n9)
     Additive=Blend Poses by bool(BlendPose_0=@n5, BlendPose_1=@n4, bActiveValue=@n6) {blendType=Linear}
     Alpha=Get LeanAdditiveAlpha()
    ) {lODThreshold=1, alphaBoolBlend {blendOption=Linear}})
  ]
 }
 Idles {
  path=LocomotionLayer_LowerBody.AnimGraphNode_StateMachine_5.LocomotionLowerBody.AnimStateNode_2.Idles
  nodes [
   n1 = Local To Component(LocalPose=Use cached pose 'UpperBodyInputPose'())
   n2 = Get IdlePelvisOffset()
   n3 = Get IdlePelvisOffsetAlpha()
   Output Animation Pose(Result=Component To Local(ComponentPose="Transform (Modify) Bone - Bone: pelvis"(ComponentPose=@n1, Alpha=@n3, Translation=@n2, Rotation="0,0,0", Scale="1,1,1") {lODThreshold=3, translationMode=BMM_Additive, alphaBoolBlend {blendOption=Linear}, boneToModify {boneName=pelvis}}))
  ]
 }
 JogStop {
  path=LocomotionLayer_LowerBody.AnimGraphNode_StateMachine_5.LocomotionLowerBody.AnimStateNode_3.JogStop
  nodes [
   n1 = Break Fort Locomotion Anim Set(FortLocomotionAnimSet=Get LocomotionAnimSet()) -> (EnableAdditiveLeans=~, UseMovementAdditiveForOneHandedTransitions=~, PlayMovementTransitions=~)
   n2 = "Blend Poses (EFortCardinalDirection)"(
    BlendPose_0=Sequence Player(Sequence=@n1.Stop_N) {bLoopAnimation=false, groupName=StopSync, method=SyncGroup}
    BlendPose_1=Sequence Player(Sequence=@n1.Stop_E) {bLoopAnimation=false, groupName=StopSync, method=SyncGroup}
    BlendPose_2=Sequence Player(Sequence=@n1.Stop_S) {bLoopAnimation=false, groupName=StopSync, method=SyncGroup}
    BlendPose_3=Sequence Player(Sequence=@n1.Stop_W) {bLoopAnimation=false, groupName=StopSync, method=SyncGroup}
    BlendTime_0=0
    BlendTime_1=0
    BlendTime_2=0
    BlendTime_3=0
    ActiveEnumValue=Get StopCardinalDirection() -> (StopCardinalDirection=North)
   ) {
    blendType=Linear
    blendPose [{linkId=-1, sourceLinkId=-1}, {linkId=-1, sourceLinkId=-1}, {linkId=-1, sourceLinkId=-1}, {linkId=-1, sourceLinkId=-1}]
    blendTime [0.1,0.1,0.1,0.1]
   }
   n3 = "Blend Poses (EFortCardinalDirection)"(
    BlendPose_0=Sequence Player(Sequence=@n1.Stop_N) {bLoopAnimation=false, groupName=StopSync, method=SyncGroup}
    BlendPose_1=Sequence Player(Sequence=@n1.Stop_E) {bLoopAnimation=false, groupName=StopSync, method=SyncGroup}
    BlendPose_2=Sequence Player(Sequence=@n1.Stop_S) {bLoopAnimation=false, groupName=StopSync, method=SyncGroup}
    BlendPose_3=Sequence Player(Sequence=@n1.Stop_W) {bLoopAnimation=false, groupName=StopSync, method=SyncGroup}
    BlendTime_0=0
    BlendTime_1=0
    BlendTime_2=0
    BlendTime_3=0
    ActiveEnumValue=Get StopCardinalDirection() -> (StopCardinalDirection=North)
   ) {
    blendType=Linear
    blendPose [{linkId=-1, sourceLinkId=-1}, {linkId=-1, sourceLinkId=-1}, {linkId=-1, sourceLinkId=-1}, {linkId=-1, sourceLinkId=-1}]
    blendTime [0.1,0.1,0.1,0.1]
   }
   n4 = Layered blend per bone(
    BasePose=Rotate Root Bone!(BasePose=@n3, Yaw=Get StopAnimDeltaAngle() -> (StopAnimDeltaAngle:float)) -> (Pose) {meshToComponent {yaw=-90}}
    BlendPoses_0=@n2
   ) {bMeshSpaceRotationBlend=true, layerSetup [{branchFilters [{boneName=ik_foot_root}, {boneName=thigh_r}, {boneName=thigh_l}]}]}
   n5 = Local To Component(LocalPose=@n4)
   n6 = Get IdlePelvisOffset()
   n7 = Get IdlePelvisOffsetAlpha()
   Output Animation Pose(Result=Component To Local(ComponentPose="Transform (Modify) Bone - Bone: pelvis"(ComponentPose=@n5, Alpha=@n7, Translation=@n6, Rotation="0,0,0", Scale="1,1,1") {lODThreshold=2, translationMode=BMM_Additive, alphaBoolBlend {blendOption=Linear}, boneToModify {boneName=pelvis}}))
  ]
 }
 JogPostPivot {
  path=LocomotionLayer_LowerBody.AnimGraphNode_StateMachine_5.LocomotionLowerBody.AnimStateNode_4.JogPostPivot
  nodes [
   n1 = Break Fort Locomotion Anim Set(FortLocomotionAnimSet=Get LocomotionAnimSet()) -> (EnableAdditiveLeans=~, UseMovementAdditiveForOneHandedTransitions=~, PlayMovementTransitions=~)
   n2 = Get PrePivotTimer()
   n3 = Get TransitionPlayRate()
   n4 = Get PivotAnimBlendTime()
   n5 = "Blend Poses (EFortCardinalDirection)"(
    BlendPose_0=Sequence Player!(Sequence=@n1.Pivot_S, PlayRate:float=@n3, StartPosition:float=@n2) -> (Pose) {bLoopAnimation=false, groupName=Locomotion, groupRole=AlwaysLeader, method=SyncGroup}
    BlendPose_1=Sequence Player(Sequence=@n1.Pivot_W, PlayRate=@n3, StartPosition=@n2) {bLoopAnimation=false, groupName=Locomotion, groupRole=AlwaysLeader, method=SyncGroup}
    BlendPose_2=Sequence Player(Sequence=@n1.Pivot_N, PlayRate=@n3, StartPosition=@n2) {bLoopAnimation=false, groupName=Locomotion, groupRole=AlwaysLeader, method=SyncGroup}
    BlendPose_3=Sequence Player(Sequence=@n1.Pivot_E, PlayRate=@n3, StartPosition=@n2) {bLoopAnimation=false, groupName=Locomotion, groupRole=AlwaysLeader, method=SyncGroup}
    BlendTime_0=@n4
    BlendTime_1=@n4
    BlendTime_2=@n4
    BlendTime_3=@n4
    ActiveEnumValue=Get PostPivotCardinalDirection() -> (PostPivotCardinalDirection=North)
   ) {
    blendType=Linear
    blendPose [{linkId=-1, sourceLinkId=-1}, {linkId=-1, sourceLinkId=-1}, {linkId=-1, sourceLinkId=-1}, {linkId=-1, sourceLinkId=-1}]
    blendTime [0.1,0.1,0.1,0.1]
   }
   Output Animation Pose(Result=Rotate Root Bone(BasePose=@n5, Yaw=Get PostPivotAnimDeltaAngle() -> (PostPivotAnimDeltaAngle:float)) {meshToComponent {yaw=-90}})
  ]
 }
 JogPrePivot {
  path=LocomotionLayer_LowerBody.AnimGraphNode_StateMachine_5.LocomotionLowerBody.AnimStateNode_5.JogPrePivot
  nodes [
   n1 = Break Fort Locomotion Anim Set(FortLocomotionAnimSet=Get LocomotionAnimSet()) -> (EnableAdditiveLeans=~, UseMovementAdditiveForOneHandedTransitions=~, PlayMovementTransitions=~)
   n2 = Get TransitionPlayRate()
   n3 = Get PivotAnimBlendTime()
   n4 = "Blend Poses (EFortCardinalDirection)"(
    BlendPose_0=Sequence Player!(Sequence=@n1.Pivot_S, PlayRate=@n2) -> (Pose) {bLoopAnimation=false, groupName=PrePivotHack, groupRole=AlwaysLeader, method=SyncGroup}
    BlendPose_1=Sequence Player(Sequence=@n1.Pivot_W, PlayRate=@n2) {bLoopAnimation=false, groupName=PrePivotHack, groupRole=AlwaysLeader, method=SyncGroup}
    BlendPose_2=Sequence Player(Sequence=@n1.Pivot_N, PlayRate=@n2) {bLoopAnimation=false, groupName=PrePivotHack, groupRole=AlwaysLeader, method=SyncGroup}
    BlendPose_3=Sequence Player(Sequence=@n1.Pivot_E, PlayRate=@n2) {bLoopAnimation=false, groupName=PrePivotHack, groupRole=AlwaysLeader, method=SyncGroup}
    BlendTime_0=@n3
    BlendTime_1=@n3
    BlendTime_2=@n3
    BlendTime_3=@n3
    ActiveEnumValue=Get PrePivotCardinalDirection() -> (PrePivotCardinalDirection=North)
   ) {
    blendType=Linear
    childUpateMode=ResetChildOnActivate
    blendPose [{linkId=-1, sourceLinkId=-1}, {linkId=-1, sourceLinkId=-1}, {linkId=-1, sourceLinkId=-1}, {linkId=-1, sourceLinkId=-1}]
    blendTime [0.1,0.1,0.1,0.1]
   }
   Output Animation Pose(Result=Rotate Root Bone(BasePose=@n4, Yaw=Get PrePivotAnimDeltaAngle() -> (PrePivotAnimDeltaAngle:float)) {meshToComponent {yaw=-90}})
  ]
 }
 JogStart {
  path=LocomotionLayer_LowerBody.AnimGraphNode_StateMachine_5.LocomotionLowerBody.AnimStateNode_6.JogStart
  nodes [
   n1 = Break Fort Locomotion Anim Set(FortLocomotionAnimSet=Get LocomotionAnimSet()) -> (EnableAdditiveLeans=~, UseMovementAdditiveForOneHandedTransitions=~, PlayMovementTransitions=~)
   n2 = Break Fort Anim Input Speed Warping(FortAnimInput_SpeedWarping=Get StartSpeedWarping()) -> (SpeedWarpingAmount=~, PlayRate=~)
   n3 = Get StartAnimBlendTime()
   n4 = "Blend Poses (EFortCardinalDirection)"(
    BlendPose_0=Rotate Root Bone(
     BasePose=Sequence Player(Sequence=@n1.Start_N, PlayRate=@n2.PlayRate) {bLoopAnimation=false, groupName=Locomotion, groupRole=TransitionLeader, method=SyncGroup}
     Yaw=Get StartAnimDeltaAngleNorth() -> (StartAnimDeltaAngleNorth:float)
    ) {meshToComponent {yaw=-90}}
    BlendPose_1=Rotate Root Bone(
     BasePose=Sequence Player(Sequence=@n1.Start_E, PlayRate=@n2.PlayRate) {groupName=Locomotion, groupRole=TransitionLeader, method=SyncGroup}
     Yaw=Get StartAnimDeltaAngleEast() -> (StartAnimDeltaAngleEast:float)
    ) {meshToComponent {yaw=-90}}
    BlendPose_2=Rotate Root Bone(
     BasePose=Sequence Player(Sequence=@n1.Start_S, PlayRate=@n2.PlayRate) {bLoopAnimation=false, groupName=Locomotion, groupRole=TransitionLeader, method=SyncGroup}
     Yaw=Get StartAnimDeltaAngleSouth() -> (StartAnimDeltaAngleSouth:float)
    ) {meshToComponent {yaw=-90}}
    BlendPose_3=Rotate Root Bone(
     BasePose=Sequence Player(Sequence=@n1.Start_W, PlayRate=@n2.PlayRate) {bLoopAnimation=false, groupName=Locomotion, groupRole=TransitionLeader, method=SyncGroup}
     Yaw=Get StartAnimDeltaAngleWest() -> (StartAnimDeltaAngleWest:float)
    ) {meshToComponent {yaw=-90}}
    BlendTime_0=@n3
    BlendTime_1=@n3
    BlendTime_2=@n3
    BlendTime_3=@n3
    ActiveEnumValue=Get LocomotionCardinalDirection() -> (LocomotionCardinalDirection=North)
   ) {
    blendPose [{linkId=-1, sourceLinkId=-1}, {linkId=-1, sourceLinkId=-1}, {linkId=-1, sourceLinkId=-1}, {linkId=-1, sourceLinkId=-1}]
    blendTime [0.1,0.1,0.1,0.1]
   }
   n5 = Speed Warping(
    ComponentPose=Local To Component(LocalPose=@n4)
    Alpha=Get StartSpeedWarpingAlpha() -> (StartSpeedWarpingAlpha:float)
    SpeedScaling=Break Fort Anim Input Speed Warping(FortAnimInput_SpeedWarping=Get StartSpeedWarping()) -> (SpeedWarpingAmount=~, PlayRate=~)
    ManualSpeedWarpingDir=Get ActorSpaceVelocityDirPreTurn() -> (ActorSpaceVelocityDirPreTurn:Vector)
   ) {
    lODThreshold=3
    alphaBoolBlend {blendOption=Linear}
    feetDefinitions [
     {numBonesInLimb=2, fKFootBone {boneName=foot_l}, iKFootBone {boneName=ik_foot_l}}
     {numBonesInLimb=2, fKFootBone {boneName=foot_r}, iKFootBone {boneName=ik_foot_r}}
    ]
    iKFootRootBone {boneName=ik_foot_root}
    pelvisBone {boneName=pelvis}
   }
   Output Animation Pose(Result=Component To Local(ComponentPose=@n5))
  ]
 }
 AnimationTransitionGraph_1 {
  path=LocomotionLayer_LowerBody.AnimGraphNode_StateMachine_5.LocomotionLowerBody.AnimStateTransitionNode_0.AnimationTransitionGraph_1
  nodes [Result(bCanEnterTransition=Get bLocomotion_Idles_to_Turns() -> (bLocomotion_Idles_to_Turns:bool))]
 }
 AnimationTransitionGraph_1 {
  path=LocomotionLayer_LowerBody.AnimGraphNode_StateMachine_5.LocomotionLowerBody.AnimStateTransitionNode_1.AnimationTransitionGraph_1
  nodes [Result(bCanEnterTransition=Get bEarlyOut_Lowerbody_Turns() -> (bEarlyOut_Lowerbody_Turns:bool))]
 }
 AnimationTransitionGraph_1 {
  path=LocomotionLayer_LowerBody.AnimGraphNode_StateMachine_5.LocomotionLowerBody.AnimStateTransitionNode_10.AnimationTransitionGraph_1
  nodes [Result(bCanEnterTransition=Get bShouldPlayStopTransition())]
 }
 AnimationTransitionGraph_1 {
  path=LocomotionLayer_LowerBody.AnimGraphNode_StateMachine_5.LocomotionLowerBody.AnimStateTransitionNode_11.AnimationTransitionGraph_1
  nodes [Result(bCanEnterTransition=Get bShouldEarlyOutStopState())]
 }
 AnimationTransitionGraph_1 {
  path=LocomotionLayer_LowerBody.AnimGraphNode_StateMachine_5.LocomotionLowerBody.AnimStateTransitionNode_12.AnimationTransitionGraph_1
  nodes [Result(bCanEnterTransition=Get bShouldEarlyOutPivotState())]
 }
 AnimationTransitionGraph_1 {
  path=LocomotionLayer_LowerBody.AnimGraphNode_StateMachine_5.LocomotionLowerBody.AnimStateTransitionNode_14.AnimationTransitionGraph_1
  nodes [Result(bCanEnterTransition=Get bShouldPlayPivotTransition())]
 }
 AnimationTransitionGraph_1 {
  path=LocomotionLayer_LowerBody.AnimGraphNode_StateMachine_5.LocomotionLowerBody.AnimStateTransitionNode_15.AnimationTransitionGraph_1
  nodes [Result(bCanEnterTransition=Get bShouldPlayPostPivotTransition())]
 }
 AnimationTransitionGraph_1 {
  path=LocomotionLayer_LowerBody.AnimGraphNode_StateMachine_5.LocomotionLowerBody.AnimStateTransitionNode_16.AnimationTransitionGraph_1
  nodes [Result()]
 }
 AnimationTransitionGraph_1 {
  path=LocomotionLayer_LowerBody.AnimGraphNode_StateMachine_5.LocomotionLowerBody.AnimStateTransitionNode_17.AnimationTransitionGraph_1
  nodes [Result(bCanEnterTransition=Get bShouldEarlyOutPivotState())]
 }
 AnimationTransitionGraph_1 {
  path=LocomotionLayer_LowerBody.AnimGraphNode_StateMachine_5.LocomotionLowerBody.AnimStateTransitionNode_18.AnimationTransitionGraph_1
  nodes [Result(bCanEnterTransition=NOT Boolean(A=Get bTransition_LowerBody_Movement()))]
 }
 AnimationTransitionGraph_1 {
  path=LocomotionLayer_LowerBody.AnimGraphNode_StateMachine_5.LocomotionLowerBody.AnimStateTransitionNode_19.AnimationTransitionGraph_1
  nodes [Result(bCanEnterTransition=Get bTransition_LowerBody_Movement())]
 }
 AnimationTransitionGraph_1 {
  path=LocomotionLayer_LowerBody.AnimGraphNode_StateMachine_5.LocomotionLowerBody.AnimStateTransitionNode_2.AnimationTransitionGraph_1
  nodes [Result()]
 }
 AnimationTransitionGraph_1 {
  path=LocomotionLayer_LowerBody.AnimGraphNode_StateMachine_5.LocomotionLowerBody.AnimStateTransitionNode_20.AnimationTransitionGraph_1
  nodes [Result(bCanEnterTransition=Get bShouldPlayStartTransition())]
 }
 AnimationTransitionGraph_1 {
  path=LocomotionLayer_LowerBody.AnimGraphNode_StateMachine_5.LocomotionLowerBody.AnimStateTransitionNode_21.AnimationTransitionGraph_1
  nodes [Result()]
 }
 AnimationTransitionGraph_1 {
  path=LocomotionLayer_LowerBody.AnimGraphNode_StateMachine_5.LocomotionLowerBody.AnimStateTransitionNode_22.AnimationTransitionGraph_1
  nodes [Result(bCanEnterTransition=Get bShouldEarlyOutStartState())]
 }
 AnimationTransitionGraph_1 {
  path=LocomotionLayer_LowerBody.AnimGraphNode_StateMachine_5.LocomotionLowerBody.AnimStateTransitionNode_23.AnimationTransitionGraph_1
  nodes [Result()]
 }
 AnimationTransitionGraph_1 {
  path=LocomotionLayer_LowerBody.AnimGraphNode_StateMachine_5.LocomotionLowerBody.AnimStateTransitionNode_24.AnimationTransitionGraph_1
  nodes [Result(bCanEnterTransition=Get bShouldPlayPivotTransition())]
 }
 AnimationTransitionGraph_1 {
  path=LocomotionLayer_LowerBody.AnimGraphNode_StateMachine_5.LocomotionLowerBody.AnimStateTransitionNode_29.AnimationTransitionGraph_1
  nodes [Result(bCanEnterTransition=Break Fort Anim Input Turn in Place(FortAnimInput_TurnInPlace=Get TurnInPlaceInput()) -> (
     RootYawOffset=~
     TurnAnimPlayRate=~
     bWantsToTurn=~
     bExitTurn=~
     bUseTurnA=~
     TurnACurveAlpha=~
     TurnBCurveAlpha=~
     bWantsToTurnInVehicle=~
     bWantsToTurnAgain=~
     bTurningLeft=~
     ShufflePlayRate=~
     CrouchShufflePlayRate=~
     CrouchShuffleStartPosition=~
     CrouchShuffleSpeedScaling=~
     bTransition_LowerBody_Shuffle=~
     bTransition_LowerBody_ShuffleADS=~
     bPlayOutro=~
    ).bWantsToTurnInVehicle)]
 }
 AnimationTransitionGraph_1 {
  path=LocomotionLayer_LowerBody.AnimGraphNode_StateMachine_5.LocomotionLowerBody.AnimStateTransitionNode_3.AnimationTransitionGraph_1
  nodes [Result(bCanEnterTransition=Break Fort Anim Input Turn in Place(FortAnimInput_TurnInPlace=Get TurnInPlaceInput()) -> (
     RootYawOffset=~
     TurnAnimPlayRate=~
     bWantsToTurn=~
     bExitTurn=~
     bUseTurnA=~
     TurnACurveAlpha=~
     TurnBCurveAlpha=~
     bWantsToTurnInVehicle=~
     bWantsToTurnAgain=~
     bTurningLeft=~
     ShufflePlayRate=~
     CrouchShufflePlayRate=~
     CrouchShuffleStartPosition=~
     CrouchShuffleSpeedScaling=~
     bTransition_LowerBody_Shuffle=~
     bTransition_LowerBody_ShuffleADS=~
     bPlayOutro=~
    ).bTransition_LowerBody_Shuffle)]
 }
 AnimationTransitionGraph_1 {
  path=LocomotionLayer_LowerBody.AnimGraphNode_StateMachine_5.LocomotionLowerBody.AnimStateTransitionNode_4.AnimationTransitionGraph_1
  nodes [
   n1 = Get TurnInPlaceInput()
   Result(bCanEnterTransition=NOT Boolean(A=Break Fort Anim Input Turn in Place(FortAnimInput_TurnInPlace=@n1) -> (
      RootYawOffset=~
      TurnAnimPlayRate=~
      bWantsToTurn=~
      bExitTurn=~
      bUseTurnA=~
      TurnACurveAlpha=~
      TurnBCurveAlpha=~
      bWantsToTurnInVehicle=~
      bWantsToTurnAgain=~
      bTurningLeft=~
      ShufflePlayRate=~
      CrouchShufflePlayRate=~
      CrouchShuffleStartPosition=~
      CrouchShuffleSpeedScaling=~
      bTransition_LowerBody_Shuffle=~
      bTransition_LowerBody_ShuffleADS=~
      bPlayOutro=~
     ).bTransition_LowerBody_Shuffle))
  ]
 }
 AnimationTransitionGraph_1 {
  path=LocomotionLayer_LowerBody.AnimGraphNode_StateMachine_5.LocomotionLowerBody.AnimStateTransitionNode_5.AnimationTransitionGraph_1
  nodes [Result(bCanEnterTransition=Break Fort Anim Input Turn in Place(FortAnimInput_TurnInPlace=Get TurnInPlaceInput()) -> (
     RootYawOffset=~
     TurnAnimPlayRate=~
     bWantsToTurn=~
     bExitTurn=~
     bUseTurnA=~
     TurnACurveAlpha=~
     TurnBCurveAlpha=~
     bWantsToTurnInVehicle=~
     bWantsToTurnAgain=~
     bTurningLeft=~
     ShufflePlayRate=~
     CrouchShufflePlayRate=~
     CrouchShuffleStartPosition=~
     CrouchShuffleSpeedScaling=~
     bTransition_LowerBody_Shuffle=~
     bTransition_LowerBody_ShuffleADS=~
     bPlayOutro=~
    ).bWantsToTurnAgain)]
 }
 AnimationTransitionGraph_1 {
  path=LocomotionLayer_LowerBody.AnimGraphNode_StateMachine_5.LocomotionLowerBody.AnimStateTransitionNode_7.AnimationTransitionGraph_1
  nodes [Result(bCanEnterTransition=Get bShouldPlayStopTransition())]
 }
 AnimationTransitionGraph_1 {
  path=LocomotionLayer_LowerBody.AnimGraphNode_StateMachine_5.LocomotionLowerBody.AnimStateTransitionNode_8.AnimationTransitionGraph_1
  nodes [Result()]
 }
 AnimationTransitionGraph_1 {
  path=LocomotionLayer_LowerBody.AnimGraphNode_StateMachine_5.LocomotionLowerBody.AnimStateTransitionNode_9.AnimationTransitionGraph_1
  nodes [Result(bCanEnterTransition=Get bShouldPlayPivotTransition())]
 }
 LocomotionLayer_UpperBody {nodes [
   Input Pose() {name=InPose_UpperBody}
   Output Pose() {layerGroup=LocomotionOverride}
  ]}
 Movement_Crouching {nodes [
   n1 = Break Fort Legacy Weapon Anim Set(FortLegacyWeaponAnimSet=Get LegacyWeaponAnimSet())
   n2 = Get CrouchPlayRate()
   n3 = Get Speed2D()
   n4 = Get LocalVelocityYawAngle()
   n5 = Blendspace Player!(X=Get LocalVelocityYawAngle(), Y:float=Get Speed2D(), PlayRate=Get CrouchPlayRate(), BlendSpace=@n1.CrouchCoreNonTargetingBlendSpace) -> (Pose) {groupName=Locomotion, method=SyncGroup}
   n6 = Get MeleeTwistCurveValue()
   Output Pose(Result=Blend Poses by bool(
     BlendPose_0=Blendspace Player(X=@n4, Y=@n3, PlayRate=@n2, BlendSpace=@n1.CrouchCoreTargetingBlendSpace) {groupName=Locomotion, method=SyncGroup}
     BlendPose_1=Rotate Root Bone!(BasePose=@n5, Pitch:float=0, Yaw=@n6, MeshToComponent:Rotator="0,-90, 0") -> (Pose) {yawScaleBiasClamp {scale=0.35}}
     BlendTime_1=.2
     bActiveValue=Get bIsTargeting()
    ) {blendType=Linear}) {layerGroup=Movement}
  ]}
 Movement_HeavyWeapon {nodes [
   n1 = Get LocomotionAnimSet()
   n2 = Get SpeedWarping()
   Output Pose(Result=Blendspace Player!(
     X=Get LocalVelocityYawAngleMinusMeleeTwist() -> (LocalVelocityYawAngleMinusMeleeTwist:float)
     Y=Get Speed2D()
     PlayRate=Break Fort Anim Input Speed Warping(FortAnimInput_SpeedWarping=@n2) -> (SpeedWarpingAmount=~, PlayRate=~).PlayRate
     StartPosition:float=Get JogStartPosition() -> (JogStartPosition:double)
     BlendSpace=Break Fort Locomotion Anim Set(FortLocomotionAnimSet=@n1) -> (EnableAdditiveLeans=~, UseMovementAdditiveForOneHandedTransitions=~, PlayMovementTransitions=~).MovementBlendSpace
    ) -> (Pose) {groupName=Locomotion, method=SyncGroup}) {layerGroup=Movement}
  ]}
 Movement_Standing {nodes [
   n1 = Break Fort Locomotion Anim Set(FortLocomotionAnimSet=Get LocomotionAnimSet()) -> (EnableAdditiveLeans=~, UseMovementAdditiveForOneHandedTransitions=~, PlayMovementTransitions=~)
   n2 = Get Speed2D()
   n3 = Get SpeedAdjustedPlayrate()
   n4 = Get LocalVelocityYawAngle()
   n5 = Get JogStartPosition()
   n6 = Blendspace Player(X=@n4, Y=@n2, PlayRate=@n3, StartPosition=@n5, BlendSpace=@n1.MovementBlendSpaceTargetingRightFootForward) {groupName=Locomotion, method=SyncGroup}
   n7 = Blendspace Player(X=@n4, Y=@n2, PlayRate=@n3, StartPosition=@n5, BlendSpace=@n1.MovementBlendSpaceTargeting) {groupName=Locomotion, method=SyncGroup, startPosition=0.5}
   n8 = Get bShouldWalkRightFootFwd() -> (bShouldWalkRightFootFwd:bool)
   n9 = Break Fort Anim Input Speed Warping(FortAnimInput_SpeedWarping=Get SpeedWarping()) -> (SpeedWarpingAmount=~, PlayRate=~)
   n10 = Get JogStartPosition()
   n11 = Get Speed2D()
   n12 = Get LocalVelocityYawAngleMinusMeleeTwist()
   n13 = Blend Poses by bool(
    BlendPose_0=Blendspace Player(X=@n12, Y=@n11, PlayRate=@n9.PlayRate, StartPosition=@n10, BlendSpace=@n1.MovementBlendSpaceNo45s) {groupName=Locomotion, method=SyncGroup}
    BlendPose_1=Blendspace Player(X=@n12, Y=@n11, PlayRate=@n9.PlayRate, StartPosition=@n10, BlendSpace=@n1.MovementBlendSpace) {groupName=Locomotion, method=SyncGroup}
    BlendTime_0=.0667
    BlendTime_1=.0667
    bActiveValue=Get bIsPlayingMeleeAnim()
   ) {blendType=Linear}
   n14 = Get MeleeTwistCurveValue()
   n15 = Speed Warping(
    ComponentPose=Local To Component(LocalPose=Rotate Root Bone!(BasePose=@n13, Yaw=@n14, MeshToComponent="0,-90, 0") -> (Pose))
    Alpha=1
    SpeedScaling=Break Fort Anim Input Speed Warping(FortAnimInput_SpeedWarping=Get SpeedWarping()) -> (SpeedWarpingAmount=~, PlayRate=~)
    ManualSpeedWarpingDir=Get ActorSpaceVelocityDirPreTurn()
   ) {
    lODThreshold=3
    alphaBoolBlend {blendOption=Linear}
    feetDefinitions [
     {numBonesInLimb=2, fKFootBone {boneName=foot_l}, iKFootBone {boneName=ik_foot_l}}
     {numBonesInLimb=2, fKFootBone {boneName=foot_r}, iKFootBone {boneName=ik_foot_r}}
    ]
    iKFootRootBone {boneName=ik_foot_root}
    pelvisBone {boneName=pelvis}
   }
   Output Pose(Result=Blend Poses by bool(
     BlendPose_0=Blend Poses by bool(BlendPose_0=@n6, BlendPose_1=@n7, BlendTime_0=.3, BlendTime_1=.3, bActiveValue=@n8) {blendType=Linear}
     BlendPose_1=Component To Local(ComponentPose=@n15)
     BlendTime_1=.2
     bActiveValue=Get bIsTargeting()
    ) {blendType=Linear}) {layerGroup=Movement}
  ]}
 ResetBlendListNode {nodes [
   n1 = ResetBlendListNode!() -> (then, Context:"Anim Update Context Structure (by ref)", Node:"Anim Node Reference Structure (by ref)") {is=title K2Node_FunctionEntry}
   > n2 = ConvertToBlendListBase!(execute, Node:"Anim Node Reference Structure (by ref)"=@n1.Node) -> (Succeeded:Exec, Failed:Exec, ReturnValue:Blend List Base Reference Structure)
   > ResetNode!(execute, BlendListBase:"Blend List Base Reference Structure (by ref)"=@n2) -> (then)
   > Return Node()
  ]}
]
variables [
 CrouchPlayRate {type=number, default=1}
 InAirLowerBodyDamper {type=number}
 JogStartPosition {type=number, default=0.625}
 PivotAnimBlendTime {type=number, default=0.1}
 StartAnimBlendTime {type=number, default=0.33}
]
