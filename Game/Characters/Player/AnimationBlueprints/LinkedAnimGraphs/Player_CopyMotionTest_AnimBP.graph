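# Player linked AnimBP test rig for the Copy Motion node, built in the VehicleFinalPoseOverride layer on reference poses, ignoring that layer's input.
# EventGraph drives a sine value, applies a procedural motion data asset and holds rotator/transform math experiments.
# Sine-animated ik_hand_r motion, relative to spine_05, is copied to ik_hand_l in pelvis space, with adjustable root and spine rotations.
# Input root rotation is then negated, and Leg IK solves both hands to their IK bones.
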AnimBlueprint /Game/Characters/Player/AnimationBlueprints/LinkedAnimGraphs/Player_CopyMotionTest_AnimBP
target_skeleton=&Fortnite_M_Avg_Player_Skeleton
graphs [
 AnimGraph {nodes [Output Pose()]}
 EventGraph {nodes [
   n1 = Event BlueprintUpdateAnimation()
   > n2 = Sequence!(execute) -> (then_0, then_1, then_2:Exec)
   n3 = Get RotatorA() -> (RotatorA:Rotator)
   n4 = "ToQuaternion (Rotator)"!(InRot:Rotator=@n3) -> (ReturnValue:Quat Structure)
   n5 = Get RotatorB() -> (RotatorB:Rotator)
   n6 = "ToQuaternion (Rotator)"(InRot=@n5)
   > n7 = Set RotatorC(RotatorC:Rotator="ToRotator (Quat)"!(Q:"Quat Structure (by ref)"=Quat * Quat(A:"Quat Structure (by ref)"=@n6, B:"Quat Structure (by ref)"=@n4) -> (ReturnValue:Quat Structure) {is=title K2Node_PromotableOperator}) -> (ReturnValue:Rotator)) -> (Output_Get:Rotator)
   > Set VectorB(VectorB:Vector=Rotate Vector!(A:Vector=Get VectorA() -> (VectorA:Vector), B:Rotator=@n7) -> (ReturnValue:Vector)) -> (Output_Get:Vector)
   > n8 = Set VectorC(VectorC:Vector=Rotate Vector(A=Get VectorA(), B=@n3)) -> (Output_Get:Vector)
   > Set VectorD(VectorD:Vector=Rotate Vector(A=@n8, B=@n5)) -> (Output_Get:Vector)
   n9 = Set Time(execute=@n2.then_2, Time:double=float + float(A=@n1.DeltaTimeX, B=Get Time() -> (Time:double))) -> (Output_Get:double)
   > Set Sinusoidal(Sinusoidal:double=float * float(A="Sin (Radians)"(A=@n9), B=Get Factor() -> (Factor:double)) {is=title K2Node_PromotableOperator}) -> (Output_Get:double)
   > Set ProceduralMotionDataSet(ProceduralMotionDataSet:Fort Procedural Motion Data Set Structure=ApplyTo!(self:Fort Procedural Motion Data Asset Object Reference=Get ProceduralMotionDataAsset() -> (ProceduralMotionDataAsset:Fort Procedural Motion Data Asset Object Reference)) -> (DataAsset:Fort Procedural Motion Data Set Structure)) -> (Output_Get:Fort Procedural Motion Data Set Structure)
   n10 = Get VectorA()
   Set VectorCombined(VectorCombined:Vector=Rotate Vector(A=@n10, B=Get RotatorCombined() -> (RotatorCombined:Rotator))) -> (Output_Get:Vector)
   > n11 = Set VectorB(VectorB=Rotate Vector(A=@n10, B=Get RotatorA()))
   > n12 = Set VectorB(VectorB=Rotate Vector(A=@n11, B=Get RotatorB()))
   > Set VectorB(VectorB=Rotate Vector(A=@n12, B=Get RotatorC() -> (RotatorC:Rotator)))
   Set TransformC(TransformC:Transform=ComposeTransforms!(A:"Transform (by ref)"=Get TransformA() -> (TransformA:Transform), B:"Transform (by ref)"=Get TransformB() -> (TransformB:Transform)) -> (ReturnValue:Transform)) -> (Output_Get:Transform)
   > Set TransformD(TransformD:Transform=ComposeTransforms(A=Get TransformB(), B=Get TransformA())) -> (Output_Get:Transform)
   TryGetPawnOwner()
   InvertTransform!(T:"Transform (by ref)") -> (ReturnValue:Transform)
   InvertTransform()
  ]}
 VehicleFinalPoseOverride {nodes [
   Input Pose() {name=InFinalPose}
   n1 = Local Space Ref Pose()
   n2 = Get Sinusoidal() -> (Sinusoidal:double)
   n3 = Copy Motion!(
    ComponentPose:Component Space Pose Link Structure=Local To Component(LocalPose=Local Space Ref Pose())
    Alpha:float=1
    BasePose:Component Space Pose Link Structure="Transform (Modify) Bone - Bone: ik_hand_r"!(
     ComponentPose=Local To Component(LocalPose=@n1)
     Alpha=1
     Translation_X:double=@n2
     Translation_Y:double=@n2
     Translation_Z:double=@n2
     Rotation="0,0,0"
     Scale="1,1,1"
    ) -> (Pose) {translationMode=BMM_Additive, boneToModify {boneName=ik_hand_r}}
    BasePoseReference:Component Space Pose Link Structure=Local To Component(LocalPose=Local Space Ref Pose())
    Delay:float=0
    TranslationOffset:Rotator="0,0,0"
    RotationOffset:Rotator="0,0,0"
    RotationPivot:Vector="0,0,0"
    TargetCurveScale:float=1
    bUseLegacyBehavior:bool=False
    TranslationScale:Vector="1,1,1"
    TranslationRemapCurve:Curve Vector Object Reference
    RotationScale:float=0
    RotationRemapCurve:Curve Float Object Reference
   ) -> (Pose:Component Space Pose Link Structure) {
    is=title AnimGraphNode_CopyMotion
    bDeltaSmoothing=true
    bUseBasePose=true
    poseHistoryTag=PoseHistory
    rotationPitch_CurveName=None_rotation_Y
    rotationRoll_CurveName=None_rotation_X
    rotationScale=0
    rotationYaw_CurveName=None_rotation_Z
    targetCurveName=RightElbowTwist
    translationX_CurveName=None_translation_X
    translationY_CurveName=None_translation_Y
    translationZ_CurveName=None_translation_Z
    alphaBoolBlend {blendOption=Linear}
    applySpace {boneName=spine_05}
    boneToModify {boneName=ik_hand_l}
    copySpace {boneName=Spine_01}
    sourceBone {boneName=ik_hand_r}
   }
   Local To Component(LocalPose=Two Way Blend(B=Component To Local(ComponentPose=@n3), Alpha=0.01))
   "Transform (Modify) Bone - Bone: ik_hand_l"!(
    ComponentPose=Local To Component(LocalPose=Local Space Ref Pose())
    Alpha=1
    Translation_X:double=0
    Translation_Y:double=0
    Translation_Z:double=0
    Rotation="0,0,90"
    Scale="1,1,1"
   ) -> (Pose) {rotationMode=BMM_Additive, translationMode=BMM_Additive, boneToModify {boneName=ik_hand_l}, rotation {roll=90}}
   n4 = Local Space Ref Pose()
   n5 = Get RefRootPitchRotation() -> (RefRootPitchRotation:double)
   n6 = Get RefRootYawRotation() -> (RefRootYawRotation:double)
   n7 = Local To Component(LocalPose=Local Space Ref Pose())
   n8 = Get Sinusoidal()
   n9 = Get Sinusoidal()
   n10 = Component To Local(ComponentPose="Transform (Modify) Bone - Bone: spine_05"!(
     ComponentPose="Transform (Modify) Bone - Bone: ik_hand_r"(ComponentPose=@n7, Alpha=1, Translation_X=@n8, Translation_Y=@n8, Translation_Z=@n8, Rotation="0,0,90", Scale="1,1,1") {rotationMode=BMM_Additive, translationMode=BMM_Additive, boneToModify {boneName=ik_hand_r}, rotation {roll=90}}
     Alpha=1
     Translation_X:double=@n9
     Translation_Y:double=@n9
     Translation_Z:double=0
     Rotation="0,0,0"
     Scale="1,1,1"
    ) -> (Pose) {translationMode=BMM_Additive, boneToModify {boneName=spine_05}})
   n11 = Get BaseRootPitchRotation() -> (BaseRootPitchRotation:double)
   n12 = Get BaseRootYawRotation() -> (BaseRootYawRotation:double)
   n13 = Rotate Root Bone!(
    BasePose:Pose Link Structure=Local Space Ref Pose()
    Pitch:float=Get InputRootPitchRotation() -> (InputRootPitchRotation:double)
    Yaw:float=Get InputRootYawRotation() -> (InputRootYawRotation:double)
    MeshToComponent:Rotator="0,0,0"
   ) -> (Pose:Pose Link Structure) {is=title AnimGraphNode_RotateRootBone}
   n14 = Copy Motion!(
    ComponentPose="Transform (Modify) Bone - Bone: Spine_01"!(
     ComponentPose=Local To Component(LocalPose=@n13)
     Alpha=1
     Translation="0,0,0"
     Rotation_Roll:float=Get SpineRoll() -> (SpineRoll:double)
     Rotation_Pitch:float=Get SpinePitch() -> (SpinePitch:double)
     Rotation_Yaw:float=Get SpineYaw() -> (SpineYaw:double)
     Scale="1,1,1"
    ) -> (Pose) {rotationMode=BMM_Additive, translationMode=BMM_Additive, boneToModify {boneName=Spine_01}, rotation {roll=70}}
    Alpha=1
    BasePose=Local To Component(LocalPose=Rotate Root Bone(BasePose=@n10, Pitch=@n11, Yaw=@n12, MeshToComponent="0,0,0"))
    BasePoseReference=Local To Component(LocalPose=Rotate Root Bone(BasePose=@n4, Pitch=@n5, Yaw=@n6, MeshToComponent="0,0,0"))
    Delay=0
    bUseRootSpace:bool=Get RootSpace() -> (RootSpace:bool)
    TranslationOffset="0,0,0"
    RotationOffset="0,0,0"
    RotationPivot="0,0,0"
    TargetCurveScale=1
    bUseLegacyBehavior=False
    TranslationScale="1,1,1"
    TranslationRemapCurve
    RotationScale=1
    RotationRemapCurve
   ) -> (Pose) {
    bDeltaSmoothing=true
    bUseBasePose=true
    poseHistoryTag=PoseHistory
    rotationPitch_CurveName=None_rotation_Y
    rotationRoll_CurveName=None_rotation_X
    rotationYaw_CurveName=None_rotation_Z
    targetCurveName=RightElbowTwist
    translationX_CurveName=None_translation_X
    translationY_CurveName=None_translation_Y
    translationZ_CurveName=None_translation_Z
    alphaBoolBlend {blendOption=Linear}
    applySpace {boneName=pelvis}
    boneToModify {boneName=ik_hand_l}
    copySpace {boneName=spine_05}
    sourceBone {boneName=ik_hand_r}
   }
   n15 = Component To Local(ComponentPose=@n14)
   n16 = float * float(A=Get InputRootYawRotation(), B=-1)
   n17 = float * float(A=Get InputRootPitchRotation(), B=-1)
   n18 = Leg IK!(
    ComponentPose:Component Space Pose Link Structure=Local To Component(LocalPose=Rotate Root Bone(BasePose=@n15, Pitch=@n17, Yaw=@n16, MeshToComponent="0,0,0"))
    Alpha:float=1
   ) -> (Pose:Component Space Pose Link Structure) {
    is=title AnimGraphNode_LegIK
    alphaBoolBlend {blendOption=Linear}
    legsDefinition [{
      bEnableKneeTwistCorrection=false
      bEnableRotationLimit=false
      footBoneForwardAxis=Y
      hingeRotationAxis=Z
      minRotationAngle=15
      numBonesInLimb=2
      twistOffsetCurveName=None
      fKFootBone {boneName=hand_l}
      iKFootBone {boneName=ik_hand_l}
     }]
   }
   n19 = Leg IK(ComponentPose=@n18, Alpha=1) {
    alphaBoolBlend {blendOption=Linear}
    legsDefinition [{
      bEnableKneeTwistCorrection=false
      bEnableRotationLimit=false
      footBoneForwardAxis=Y
      hingeRotationAxis=Z
      minRotationAngle=15
      numBonesInLimb=2
      twistOffsetCurveName=None
      fKFootBone {boneName=hand_r}
      iKFootBone {boneName=ik_hand_r}
     }]
   }
   Output Pose(Result=Component To Local(ComponentPose=@n19)) {layerGroup=VehicleOverride}
  ]}
 VehicleFullBodyOverride {nodes [
   Input Pose() {name=InPoseFullBody}
   Output Pose() {layerGroup=VehicleOverride}
  ]}
 VehicleLowerBodyOverride {nodes [
   Input Pose() {name=InPoseLowerBodyDefault}
   Input Pose() {name=InPoseUpperBody}
   Output Pose() {layerGroup=VehicleOverride}
  ]}
 VehicleSplitBodyOverride {nodes [
   Input Pose() {name=InPoseSplitBody}
   Input Pose() {name=InPoseUpperAndLowerBody}
   Input Pose() {name=InPoseUpperAndLowerBodyRemoveRoot}
   Output Pose() {layerGroup=Vehicle Override}
  ]}
]
variables [
 ApplySpaceRotation {type=boolean}
 BaseRootPitchRotation {type=number}
 BaseRootYawRotation {type=number}
 Factor {type=number, default=25}
 InputRootPitchRotation {type=number}
 InputRootYawRotation {type=number}
 ProceduralMotionDataAsset {type=FortProceduralMotionDataAsset, default=/Game/Animation/Game/MainPlayerFN/DataAssets/_Default/AR/AR_Procedural_Combat_DA}
 ProceduralMotionDataSet {
  type=FortProceduralMotionDataSet
  default {
   bApplyIKHandGunOffsetsWhenIdle=false
   bEnableWhileSliding=false
   bEnabled=false
   bUseAlternateCopyMotionBoneSetup=false
   bUseCopyMotionLegacyBehavior=false
   delay=0
   iKHandGunInHandSpaceAlpha=0
   iKHandGunInHandSpaceRotationWeight=0
   leftElbowTwistOffset=0
   leftElbowTwistScale=0
   proceduralAlpha=0
   rightElbowTwistOffset=0
   rightElbowTwistScale=0
   rotationRemapCurve=null
   rotationScale=0
   rotationSmoothingTime=0
   translationRemapCurve=null
   translationSmoothingTime=0
   iKHandGunInHandSpaceTranslationWeight=<0 0 0>
   iKHandGunRotationOffset=<rot 0 0 0>
   iKHandGunTranslationOffset=<0 0 0>
   rotationOffset=<rot 0 0 0>
   rotationPivot=<0 0 0>
   translationRotationalOffset=<rot 0 0 0>
   translationScale=<0 0 0>
  }
 }
 RefRootPitchRotation {type=number}
 RefRootYawRotation {type=number}
 RootSpace {type=boolean}
 RotatorA {type=Rotator}
 RotatorB {type=Rotator}
 RotatorC {type=Rotator}
 RotatorCombined {type=Rotator}
 Sinusoidal {type=number}
 SpinePitch {type=number}
 SpineRoll {type=number}
 SpineYaw {type=number}
 Time {type=number}
 TransformA {type=Transform, default {location=<0 0 0>, rotation=<rot 0 0 0>, scale=<1 1 1>}}
 TransformB {type=Transform, default {location=<0 0 0>, rotation=<rot 0 0 0>, scale=<1 1 1>}}
 TransformC {type=Transform, default {location=<0 0 0>, rotation=<rot 0 0 0>, scale=<1 1 1>}}
 TransformD {type=Transform, default {location=<0 0 0>, rotation=<rot 0 0 0>, scale=<1 1 1>}}
 VectorA {type=Vector}
 VectorB {type=Vector}
 VectorC {type=Vector}
 VectorCombined {type=Vector}
 VectorD {type=Vector}
]
