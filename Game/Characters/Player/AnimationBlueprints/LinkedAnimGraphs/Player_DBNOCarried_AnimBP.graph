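# Linked player anim graph for a downed (DBNO) character being fireman-carried or interrogated, driven by the carrier's DBNOCarriedInput.
# A Pickup/Carried/Drop state machine time-matches carrier montages; Carried nests locomotion states with yaw and pitch additives.
# Locomotion follows the carrier through explicit-time evaluators or sync-group players.
# After slots, the root snaps to a world-space attach transform; outside interrogation, head trail and right-hand IK (fading on drop) finish the pose.
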AnimBlueprint /Game/Characters/Player/AnimationBlueprints/LinkedAnimGraphs/Player_DBNOCarried_AnimBP
target_skeleton=&Fortnite_M_Avg_Player_Skeleton
graphs [
 AnimGraph {nodes [
   n1 = Trail controller - Bone: head!(ComponentPose=Local To Component(LocalPose=Use cached pose 'PreDBNOHandAdjustment'()), AlphaCurveName:Name=Disable_Custom) -> (Pose) {
    alphaInputType=Curve
    chainLength=3
    maxDeltaTime=0.0333
    relaxationSpeedScale=2
    stretchLimit=0.1
    alphaScaleBiasClamp {bClampResult=true, bias=0.15, clampMax=0.15, scale=-1}
    rotationLimits [{limitMax=<180 180 180>}, {limitMax=<180 180 180>}, {limitMax=<180 180 180>}]
    rotationOffsets [<0 0 0>, <0 0 0>, <0 0 0>]
    trailBone {boneName=head}
   }
   n2 = Break Fort Anim Input DBNOCarried!(FortAnimInput_DBNOCarried:"Fort Anim Input DBNOCarried Structure (by ref)"=Get DBNOCarriedInput() -> (DBNOCarriedInput:Fort Anim Input DBNOCarried Structure)) -> (
    DropStateData:Cached Anim State Data Structure
    CarrierYawSpringState:Float Spring State Structure
    CarrierSpinePitchSpringState:Float Spring State Structure
    CarrierDropMontage:Anim Montage Object Reference
    CarrierPickupMontage:Anim Montage Object Reference
    CarrierPawn:Fort Player Pawn Object Reference
    CarrierAnimBP:Fort Player Anim Instance Object Reference
    CarrierInterrogationMontage:Anim Montage Object Reference
    LHandIKLocation:Vector=~
    LHandIKRotation:Rotator=~
    TempAttachLocation:Vector=~
    TempAttachRotation:Rotator=~
    SubAnimPhysicsWeight:float=~
    DropMontagePosition:float=~
    PickupMontagePosition:float=~
    InterrogationMontagePosition:float=~
    PickupToIdleTransitionPosition:float=~
    CarrierPawnVelocityZ:float=~
    CarrierYawDeltaSmoothed:float=~
    CarriedJogNAnimPosition:float=~
    CarriedJogSAnimPosition:float=~
    CarriedCrouchNAnimPosition:float=~
    CarriedCrouchSAnimPosition:float=~
    CarriedSprintAnimPosition:float=~
    CarriedCrouchSprintAnimPosition:float=~
    CarrierSpinePitch:float=~
    DropStateWeight:float=~
    bBeingCarried:bool=~
    bIsBeingPickedUp:bool=~
    bIsBeingDropped:bool=~
    bCarrierIsCrouching:bool=~
    bCarrierIsMovingBackward:bool=~
    bCarrierHasSyncMarkers:bool=~
    bTransitionFromPickupToIdle:bool=~
    bTransitionFromIdleToJogging:bool=~
    bTransitionFromJoggingToSprinting:bool=~
    bTransitionFromInAirToLanding:bool=~
    bJackalPlayJumpTrickVertical:bool=~
   ) {is=title K2Node_BreakStruct}
   n3 = "Transform (Modify) Bone - Bone: ik_hand_r"!(ComponentPose=@n1, Alpha=Get RightHandIKAlpha() -> (RightHandIKAlpha:float), Translation=@n2.LHandIKLocation, Rotation=@n2.LHandIKRotation) -> (Pose) {
    rotationMode=BMM_Replace
    rotationSpace=BCS_WorldSpace
    translationMode=BMM_Replace
    translationSpace=BCS_WorldSpace
    alphaBoolBlend {blendOption=Linear}
    boneToModify {boneName=ik_hand_r}
   }
   n4 = Two Bone IK - Bone: hand_r!(
    ComponentPose:Component Space Pose Link Structure=@n3
    Alpha:float=Break Fort Anim Input DBNOCarried(FortAnimInput_DBNOCarried=Get DBNOCarriedInput()) -> (
     LHandIKLocation=~
     LHandIKRotation=~
     TempAttachLocation=~
     TempAttachRotation=~
     SubAnimPhysicsWeight=~
     DropMontagePosition=~
     PickupMontagePosition=~
     InterrogationMontagePosition=~
     PickupToIdleTransitionPosition=~
     CarrierPawnVelocityZ=~
     CarrierYawDeltaSmoothed=~
     CarriedJogNAnimPosition=~
     CarriedJogSAnimPosition=~
     CarriedCrouchNAnimPosition=~
     CarriedCrouchSAnimPosition=~
     CarriedSprintAnimPosition=~
     CarriedCrouchSprintAnimPosition=~
     CarrierSpinePitch=~
     DropStateWeight=~
     bBeingCarried=~
     bIsBeingPickedUp=~
     bIsBeingDropped=~
     bCarrierIsCrouching=~
     bCarrierIsMovingBackward=~
     bCarrierHasSyncMarkers=~
     bTransitionFromPickupToIdle=~
     bTransitionFromIdleToJogging=~
     bTransitionFromJoggingToSprinting=~
     bTransitionFromInAirToLanding=~
     bJackalPlayJumpTrickVertical=~
    ).DropStateWeight
    JointTargetLocation:Vector="25,25,5"
   ) -> (Pose:Component Space Pose Link Structure) {
    is=title AnimGraphNode_TwoBoneIK
    effectorLocationSpace=BCS_BoneSpace
    jointTargetLocationSpace=BCS_BoneSpace
    alphaBoolBlend {blendOption=Linear}
    alphaScaleBias {bias=1, scale=-1}
    effectorTarget {boneReference {boneName=ik_hand_r}}
    iKBone {boneName=hand_r}
    jointTarget {boneReference {boneName=lowerarm_r}}
    jointTargetLocation=<25 25 5>
   }
   Output Pose(Result=Blend Poses by bool(
     BlendPose_0=Use cached pose 'PreDBNOHandAdjustment'()
     BlendPose_1=Component To Local(ComponentPose=@n4)
     bActiveValue=Get bIsBeingInterrogated() -> (bIsBeingInterrogated:bool)
    ) {blendType=Linear}) {layerGroup=None}
   n5 = "Sequence Player (additive) 'FireCarry_Idle_DBNO_CMF'"() {sequence=/Game/Animation/Game/MainPlayer/Poses/Medium/Female/FireCarry/FireCarry_Idle_DBNO_CMF}
   AnimGraphNode_StateMachine_0 = DBNOBehavior() {is=title AnimGraphNode_StateMachine}
   n6 = Get bIsFemale() -> (bIsFemale:bool)
   n7 = Slot 'FullBodyDBNO'(Source=Blend Poses by bool(
     BlendPose_0=Sequence Player 'Interrogation_SwipeMap_InterrogateeCMF'() {sequence=/Game/Animation/Game/MainPlayer/Combat/Gadgets/Medium/Male/Interrogation/Interrogation_SwipeMap_InterrogateeCMF}
     BlendPose_1=Apply Additive!(Base=@AnimGraphNode_StateMachine_0, Additive=@n5, bAlphaBoolEnabled:bool=@n6) -> (Pose) {alphaInputType=Bool, alphaBoolBlend {blendOption=Linear}}
     BlendTime_0=0
     BlendTime_1=0
     bActiveValue=Get bIsBeingInterrogated()
    ) {blendType=Linear}) {slotName=FullBodyDBNO}
   n8 = Break Fort Anim Input DBNOCarried(FortAnimInput_DBNOCarried=Get DBNOCarriedInput()) -> (
    LHandIKLocation=~
    LHandIKRotation=~
    TempAttachLocation=~
    TempAttachRotation=~
    SubAnimPhysicsWeight=~
    DropMontagePosition=~
    PickupMontagePosition=~
    InterrogationMontagePosition=~
    PickupToIdleTransitionPosition=~
    CarrierPawnVelocityZ=~
    CarrierYawDeltaSmoothed=~
    CarriedJogNAnimPosition=~
    CarriedJogSAnimPosition=~
    CarriedCrouchNAnimPosition=~
    CarriedCrouchSAnimPosition=~
    CarriedSprintAnimPosition=~
    CarriedCrouchSprintAnimPosition=~
    CarrierSpinePitch=~
    DropStateWeight=~
    bBeingCarried=~
    bIsBeingPickedUp=~
    bIsBeingDropped=~
    bCarrierIsCrouching=~
    bCarrierIsMovingBackward=~
    bCarrierHasSyncMarkers=~
    bTransitionFromPickupToIdle=~
    bTransitionFromIdleToJogging=~
    bTransitionFromJoggingToSprinting=~
    bTransitionFromInAirToLanding=~
    bJackalPlayJumpTrickVertical=~
   )
   n9 = "Transform (Modify) Bone - Bone: Root"!(
    ComponentPose=Local To Component(LocalPose=Slot 'AdditiveCorrectiveDBNO'(Source=@n7) {slotName=AdditiveCorrectiveDBNO})
    Alpha=1
    Translation=@n8.TempAttachLocation
    Rotation=@n8.TempAttachRotation
   ) -> (Pose) {
    rotationMode=BMM_Replace
    rotationSpace=BCS_WorldSpace
    translationMode=BMM_Replace
    translationSpace=BCS_WorldSpace
    alphaBoolBlend {blendOption=Linear}
    boneToModify {boneName=Root}
   }
   Save cached pose 'PreDBNOHandAdjustment'(Pose=Component To Local(ComponentPose=@n9)) {cacheName=PreDBNOHandAdjustment}
  ]}
 DBNOBehavior {
  entry=Pickup
  path=AnimGraph.AnimGraphNode_StateMachine_0.DBNOBehavior
  ids=id
  nodes [
   AnimStateNode_0 = Pickup() {is=title AnimStateNode}
   AnimStateNode_1 = Carried() {is=title AnimStateNode}
   AnimStateNode_2 = Drop() {is=title AnimStateNode}
   AnimStateTransitionNode_0 = AnimStateTransitionNode() {crossfadeDuration=0.4}
   AnimStateTransitionNode_1 = AnimStateTransitionNode() {crossfadeDuration=0.1}
  ]
  links [
   {
    dst=Drop
    rule=AnimGraph.AnimGraphNode_StateMachine_0.DBNOBehavior.AnimStateTransitionNode_1.Transition
    src=Carried
    via=AnimStateTransitionNode_1
   }
   {
    dst=Carried
    rule=AnimGraph.AnimGraphNode_StateMachine_0.DBNOBehavior.AnimStateTransitionNode_0.Transition
    src=Pickup
    via=AnimStateTransitionNode_0
   }
  ]
 }
 Pickup {
  path=AnimGraph.AnimGraphNode_StateMachine_0.DBNOBehavior.AnimStateNode_0.Pickup
  nodes [
   n1 = Get DBNOCarriedInput()
   Output Animation Pose(Result="Sequence Evaluator (by time) 'FiremansCarry_Pickup_DBNO_CMM'"!(ExplicitTime:float=Break Fort Anim Input DBNOCarried(FortAnimInput_DBNOCarried=@n1) -> (
      LHandIKLocation=~
      LHandIKRotation=~
      TempAttachLocation=~
      TempAttachRotation=~
      SubAnimPhysicsWeight=~
      DropMontagePosition=~
      PickupMontagePosition=~
      InterrogationMontagePosition=~
      PickupToIdleTransitionPosition=~
      CarrierPawnVelocityZ=~
      CarrierYawDeltaSmoothed=~
      CarriedJogNAnimPosition=~
      CarriedJogSAnimPosition=~
      CarriedCrouchNAnimPosition=~
      CarriedCrouchSAnimPosition=~
      CarriedSprintAnimPosition=~
      CarriedCrouchSprintAnimPosition=~
      CarrierSpinePitch=~
      DropStateWeight=~
      bBeingCarried=~
      bIsBeingPickedUp=~
      bIsBeingDropped=~
      bCarrierIsCrouching=~
      bCarrierIsMovingBackward=~
      bCarrierHasSyncMarkers=~
      bTransitionFromPickupToIdle=~
      bTransitionFromIdleToJogging=~
      bTransitionFromJoggingToSprinting=~
      bTransitionFromInAirToLanding=~
      bJackalPlayJumpTrickVertical=~
     ).PickupMontagePosition) -> (Pose:Pose Link Structure) {is=title AnimGraphNode_SequenceEvaluator, sequence=/Game/Animation/Game/MainPlayer/Locomotion/Medium/Male/DBNO/FiremansCarry_Pickup_DBNO_CMM})
  ]
 }
 Carried {
  path=AnimGraph.AnimGraphNode_StateMachine_0.DBNOBehavior.AnimStateNode_1.Carried
  nodes [
   n1 = Break Fort Anim Input DBNOCarried(FortAnimInput_DBNOCarried=Get DBNOCarriedInput()) -> (
    LHandIKLocation=~
    LHandIKRotation=~
    TempAttachLocation=~
    TempAttachRotation=~
    SubAnimPhysicsWeight=~
    DropMontagePosition=~
    PickupMontagePosition=~
    InterrogationMontagePosition=~
    PickupToIdleTransitionPosition=~
    CarrierPawnVelocityZ=~
    CarrierYawDeltaSmoothed=~
    CarriedJogNAnimPosition=~
    CarriedJogSAnimPosition=~
    CarriedCrouchNAnimPosition=~
    CarriedCrouchSAnimPosition=~
    CarriedSprintAnimPosition=~
    CarriedCrouchSprintAnimPosition=~
    CarrierSpinePitch=~
    DropStateWeight=~
    bBeingCarried=~
    bIsBeingPickedUp=~
    bIsBeingDropped=~
    bCarrierIsCrouching=~
    bCarrierIsMovingBackward=~
    bCarrierHasSyncMarkers=~
    bTransitionFromPickupToIdle=~
    bTransitionFromIdleToJogging=~
    bTransitionFromJoggingToSprinting=~
    bTransitionFromInAirToLanding=~
    bJackalPlayJumpTrickVertical=~
   )
   n2 = Blendspace Player 'FireCarry_DBNO_Spin_BS'!(X=@n1.CarrierYawDeltaSmoothed) -> (Pose) {blendSpace=/Game/Animation/Game/MainPlayer/Poses/Medium/Male/FireCarry/FireCarry_DBNO_Spin_BS}
   AnimGraphNode_StateMachine_0 = CarriedLocomotion() {is=title AnimGraphNode_StateMachine}
   Output Animation Pose(Result=Apply Additive!(
     Base=Apply Additive(Base=@AnimGraphNode_StateMachine_0, Additive=@n2) {alphaBoolBlend {blendOption=Linear}}
     Additive=Blendspace Player 'FireCarry_DBNO_Pitch_BS'!(X=@n1.CarrierSpinePitch) -> (Pose) {blendSpace=/Game/Animation/Game/MainPlayer/Poses/Medium/Male/FireCarry/FireCarry_DBNO_Pitch_BS}
     Alpha:float
    ) -> (Pose) {alphaBoolBlend {blendOption=Linear}})
  ]
 }
 CarriedLocomotion {
  entry=Idle
  path=AnimGraph.AnimGraphNode_StateMachine_0.DBNOBehavior.AnimStateNode_1.Carried.AnimGraphNode_StateMachine_0.CarriedLocomotion
  ids=id
  nodes [
   AnimStateNode_0 = Idle()
   AnimStateNode_1 = Jogging() {is=title AnimStateNode}
   AnimStateNode_2 = Sprinting() {is=title AnimStateNode}
   AnimStateNode_3 = InAir() {is=title AnimStateNode}
   AnimStateNode_4 = Landing() {is=title AnimStateNode}
   AnimStateTransitionNode_0 = AnimStateTransitionNode() {crossfadeDuration=0.4}
   AnimStateTransitionNode_1 = AnimStateTransitionNode()
   AnimStateTransitionNode_2 = AnimStateTransitionNode() {crossfadeDuration=0.5}
   AnimStateTransitionNode_3 = AnimStateTransitionNode() {crossfadeDuration=0.5}
   AnimStateTransitionNode_4 = AnimStateTransitionNode()
   AnimStateTransitionNode_6 = AnimStateTransitionNode()
   AnimStateTransitionNode_7 = AnimStateTransitionNode() {bAutomaticRuleBasedOnSequencePlayerInState=true, crossfadeDuration=0.4}
  ]
  links [
   {
    dst=InAir
    rule=AnimGraph.AnimGraphNode_StateMachine_0.DBNOBehavior.AnimStateNode_1.Carried.AnimGraphNode_StateMachine_0.CarriedLocomotion.AnimStateTransitionNode_4.Transition
    src=Idle
    via=AnimStateTransitionNode_4
   }
   {
    dst=Jogging
    rule=AnimGraph.AnimGraphNode_StateMachine_0.DBNOBehavior.AnimStateNode_1.Carried.AnimGraphNode_StateMachine_0.CarriedLocomotion.AnimStateTransitionNode_0.Transition
    src=Idle
    via=AnimStateTransitionNode_0
   }
   {
    dst=Landing
    rule=AnimGraph.AnimGraphNode_StateMachine_0.DBNOBehavior.AnimStateNode_1.Carried.AnimGraphNode_StateMachine_0.CarriedLocomotion.AnimStateTransitionNode_6.Transition
    src=InAir
    via=AnimStateTransitionNode_6
   }
   {
    dst=Idle
    rule=AnimGraph.AnimGraphNode_StateMachine_0.DBNOBehavior.AnimStateNode_1.Carried.AnimGraphNode_StateMachine_0.CarriedLocomotion.AnimStateTransitionNode_1.Transition
    src=Jogging
    via=AnimStateTransitionNode_1
   }
   {
    dst=Sprinting
    rule=AnimGraph.AnimGraphNode_StateMachine_0.DBNOBehavior.AnimStateNode_1.Carried.AnimGraphNode_StateMachine_0.CarriedLocomotion.AnimStateTransitionNode_2.Transition
    src=Jogging
    via=AnimStateTransitionNode_2
   }
   {
    dst=Idle
    rule=AnimGraph.AnimGraphNode_StateMachine_0.DBNOBehavior.AnimStateNode_1.Carried.AnimGraphNode_StateMachine_0.CarriedLocomotion.AnimStateTransitionNode_7.Transition
    src=Landing
    via=AnimStateTransitionNode_7
   }
   {
    dst=Jogging
    rule=AnimGraph.AnimGraphNode_StateMachine_0.DBNOBehavior.AnimStateNode_1.Carried.AnimGraphNode_StateMachine_0.CarriedLocomotion.AnimStateTransitionNode_3.Transition
    src=Sprinting
    via=AnimStateTransitionNode_3
   }
  ]
 }
 Idle {
  path=AnimGraph.AnimGraphNode_StateMachine_0.DBNOBehavior.AnimStateNode_1.Carried.AnimGraphNode_StateMachine_0.CarriedLocomotion.AnimStateNode_0.Idle
  nodes [Output Animation Pose(Result=Apply Additive(
     Base=Sequence Player 'FireCarry_Idle_DBNO_CMM'() {playRateBasis=0, sequence=/Game/Animation/Game/MainPlayer/Poses/Medium/Male/FireCarry/FireCarry_Idle_DBNO_CMM}
     Additive="Sequence Player (additive) 'Torch_Swim_Idle_Noise_NonTargeted'"!(PlayRate:float=0.5) -> (Pose) {playRate=0.1, sequence=/Game/Animation/Game/MainPlayer/Swimming/Torch/Torch_Swim_Idle_Noise_NonTargeted}
     Alpha=0.05
    ) {alphaBoolBlend {blendOption=Linear}})]
 }
 Jogging {
  path=AnimGraph.AnimGraphNode_StateMachine_0.DBNOBehavior.AnimStateNode_1.Carried.AnimGraphNode_StateMachine_0.CarriedLocomotion.AnimStateNode_1.Jogging
  nodes [
   n1 = Break Fort Anim Input DBNOCarried(FortAnimInput_DBNOCarried=Get DBNOCarriedInput()) -> (
    LHandIKLocation=~
    LHandIKRotation=~
    TempAttachLocation=~
    TempAttachRotation=~
    SubAnimPhysicsWeight=~
    DropMontagePosition=~
    PickupMontagePosition=~
    InterrogationMontagePosition=~
    PickupToIdleTransitionPosition=~
    CarrierPawnVelocityZ=~
    CarrierYawDeltaSmoothed=~
    CarriedJogNAnimPosition=~
    CarriedJogSAnimPosition=~
    CarriedCrouchNAnimPosition=~
    CarriedCrouchSAnimPosition=~
    CarriedSprintAnimPosition=~
    CarriedCrouchSprintAnimPosition=~
    CarrierSpinePitch=~
    DropStateWeight=~
    bBeingCarried=~
    bIsBeingPickedUp=~
    bIsBeingDropped=~
    bCarrierIsCrouching=~
    bCarrierIsMovingBackward=~
    bCarrierHasSyncMarkers=~
    bTransitionFromPickupToIdle=~
    bTransitionFromIdleToJogging=~
    bTransitionFromJoggingToSprinting=~
    bTransitionFromInAirToLanding=~
    bJackalPlayJumpTrickVertical=~
   )
   n2 = "Sequence Evaluator (by time) 'Crouch_FireCarry_DBNO_S'"!(ExplicitTime:float=@n1.CarriedCrouchSAnimPosition) -> (Pose:Pose Link Structure) {
    is=title AnimGraphNode_SequenceEvaluator
    groupName=DBNOCarried
    groupRole=AlwaysLeader
    method=SyncGroup
    sequence=/Game/Animation/Game/MainPlayer/Locomotion/Medium/Male/Crouch/FireCarry/Crouch_FireCarry_DBNO_S
   }
   n3 = Sequence Player 'Crouch_FireCarry_DBNO_S'() {
    groupName=DBNOCarried
    groupRole=AlwaysFollower
    method=SyncGroup
    sequence=&Crouch_FireCarry_DBNO_S
   }
   n4 = Apply Additive(
    Base=Blend Poses by bool(BlendPose_0=@n2, BlendPose_1=@n3, bActiveValue=@n1.bCarrierHasSyncMarkers) {blendType=Linear}
    Additive="Sequence Player (additive) 'Crouch_FiremanCarry_DBNO_InMotion_S'"() {bLoopAnimation=false, sequence=/Game/Animation/Game/MainPlayerFN/Poses/FiremanCarry/Crouch_FiremanCarry_DBNO_InMotion_S}
    Alpha=Get MotionMatchingAdditivePoseAlpha() -> (MotionMatchingAdditivePoseAlpha:float)
   ) {alphaBoolBlend {blendOption=Linear}}
   n5 = "Sequence Evaluator (by time) 'FireCarry_Jog_S_DBNO_CMM'"!(ExplicitTime:float=@n1.CarriedJogSAnimPosition) -> (Pose:Pose Link Structure) {
    is=title AnimGraphNode_SequenceEvaluator
    groupName=DBNOCarried
    groupRole=AlwaysLeader
    method=SyncGroup
    sequence=/Game/Animation/Game/MainPlayer/Locomotion/Medium/Male/Jog/FireCarry/FireCarry_Jog_S_DBNO_CMM
   }
   n6 = Sequence Player 'FireCarry_Jog_S_DBNO_CMM'() {
    groupName=DBNOCarried
    groupRole=AlwaysFollower
    method=SyncGroup
    sequence=&FireCarry_Jog_S_DBNO_CMM
   }
   n7 = Blend Poses by bool(BlendPose_0=@n5, BlendPose_1=@n6, bActiveValue=@n1.bCarrierHasSyncMarkers) {blendType=Linear}
   n8 = "Sequence Evaluator (by time) 'Crouch_FireCarry_DBNO_N'"!(ExplicitTime:float=@n1.CarriedCrouchNAnimPosition) -> (Pose:Pose Link Structure) {
    is=title AnimGraphNode_SequenceEvaluator
    groupName=DBNOCarried
    groupRole=AlwaysLeader
    method=SyncGroup
    sequence=/Game/Animation/Game/MainPlayer/Locomotion/Medium/Male/Crouch/FireCarry/Crouch_FireCarry_DBNO_N
   }
   n9 = Sequence Player 'Crouch_FireCarry_DBNO_N'() {
    groupName=DBNOCarried
    groupRole=AlwaysFollower
    method=SyncGroup
    sequence=&Crouch_FireCarry_DBNO_N
   }
   n10 = Apply Additive(
    Base=Blend Poses by bool(BlendPose_0=@n8, BlendPose_1=@n9, bActiveValue=@n1.bCarrierHasSyncMarkers) {blendType=Linear}
    Additive="Sequence Player (additive) 'Crouch_FiremanCarry_DBNO_InMotion'"() {bLoopAnimation=false, sequence=/Game/Animation/Game/MainPlayerFN/Poses/FiremanCarry/Crouch_FiremanCarry_DBNO_InMotion}
    Alpha=Get MotionMatchingAdditivePoseAlpha()
   ) {alphaBoolBlend {blendOption=Linear}}
   n11 = "Sequence Evaluator (by time) 'FireCarry_Jog_N_DBNO_CMM'"!(ExplicitTime:float=@n1.CarriedJogNAnimPosition) -> (Pose:Pose Link Structure) {
    is=title AnimGraphNode_SequenceEvaluator
    groupName=DBNOCarried
    groupRole=AlwaysLeader
    method=SyncGroup
    sequence=/Game/Animation/Game/MainPlayer/Locomotion/Medium/Male/Jog/FireCarry/FireCarry_Jog_N_DBNO_CMM
   }
   n12 = Sequence Player 'FireCarry_Jog_N_DBNO_CMM'() {
    groupName=DBNOCarried
    groupRole=AlwaysFollower
    method=SyncGroup
    sequence=&FireCarry_Jog_N_DBNO_CMM
   }
   n13 = Blend Poses by bool(BlendPose_0=@n11, BlendPose_1=@n12, bActiveValue=@n1.bCarrierHasSyncMarkers) {blendType=Linear}
   Output Animation Pose(Result=Blend Poses by bool(
     BlendPose_0=Blend Poses by bool(BlendPose_0=@n10, BlendPose_1=@n13, BlendTime_0=0.3, BlendTime_1=0.3, bActiveValue=@n1.bCarrierIsCrouching) {blendType=Linear}
     BlendPose_1=Blend Poses by bool(BlendPose_0=@n4, BlendPose_1=@n7, BlendTime_0=0.3, BlendTime_1=0.3, bActiveValue=@n1.bCarrierIsCrouching) {blendType=Linear}
     BlendTime_0=0.25
     BlendTime_1=0.25
     bActiveValue=@n1.bCarrierIsMovingBackward
    ) {blendType=Linear})
  ]
 }
 Sprinting {
  path=AnimGraph.AnimGraphNode_StateMachine_0.DBNOBehavior.AnimStateNode_1.Carried.AnimGraphNode_StateMachine_0.CarriedLocomotion.AnimStateNode_2.Sprinting
  nodes [
   n1 = Break Fort Anim Input DBNOCarried(FortAnimInput_DBNOCarried=Get DBNOCarriedInput()) -> (
    LHandIKLocation=~
    LHandIKRotation=~
    TempAttachLocation=~
    TempAttachRotation=~
    SubAnimPhysicsWeight=~
    DropMontagePosition=~
    PickupMontagePosition=~
    InterrogationMontagePosition=~
    PickupToIdleTransitionPosition=~
    CarrierPawnVelocityZ=~
    CarrierYawDeltaSmoothed=~
    CarriedJogNAnimPosition=~
    CarriedJogSAnimPosition=~
    CarriedCrouchNAnimPosition=~
    CarriedCrouchSAnimPosition=~
    CarriedSprintAnimPosition=~
    CarriedCrouchSprintAnimPosition=~
    CarrierSpinePitch=~
    DropStateWeight=~
    bBeingCarried=~
    bIsBeingPickedUp=~
    bIsBeingDropped=~
    bCarrierIsCrouching=~
    bCarrierIsMovingBackward=~
    bCarrierHasSyncMarkers=~
    bTransitionFromPickupToIdle=~
    bTransitionFromIdleToJogging=~
    bTransitionFromJoggingToSprinting=~
    bTransitionFromInAirToLanding=~
    bJackalPlayJumpTrickVertical=~
   )
   n2 = "Sequence Evaluator (by time) 'FireCarry_CrouchSprint_DBNO_CMM'"!(ExplicitTime:float=@n1.CarriedCrouchSprintAnimPosition) -> (Pose:Pose Link Structure) {
    is=title AnimGraphNode_SequenceEvaluator
    groupName=DBNOCarried
    groupRole=AlwaysLeader
    method=SyncGroup
    sequence=/Game/Animation/Game/MainPlayer/Locomotion/Medium/Male/Crouch/Sprint/FireCarry_CrouchSprint_DBNO_CMM
   }
   n3 = Sequence Player 'FireCarry_CrouchSprint_DBNO_CMM'() {
    groupName=DBNOCarried
    groupRole=AlwaysFollower
    method=SyncGroup
    sequence=&FireCarry_CrouchSprint_DBNO_CMM
   }
   n4 = "Sequence Evaluator (by time) 'Sprint_FireCarry_DBNO_CMM'"!(ExplicitTime:float=@n1.CarriedSprintAnimPosition) -> (Pose:Pose Link Structure) {
    is=title AnimGraphNode_SequenceEvaluator
    groupName=DBNOCarried
    groupRole=AlwaysLeader
    method=SyncGroup
    sequence=/Game/Animation/Game/MainPlayer/Locomotion/Medium/Male/Sprint/Sprint_FireCarry_DBNO_CMM
   }
   n5 = Sequence Player 'Sprint_FireCarry_DBNO_CMM'() {
    groupName=DBNOCarried
    groupRole=AlwaysFollower
    method=SyncGroup
    sequence=&Sprint_FireCarry_DBNO_CMM
   }
   Output Animation Pose(Result=Blend Poses by bool(
     BlendPose_0=Blend Poses by bool(BlendPose_0=@n2, BlendPose_1=@n3, bActiveValue=@n1.bCarrierHasSyncMarkers) {blendType=Linear}
     BlendPose_1=Blend Poses by bool(BlendPose_0=@n4, BlendPose_1=@n5, bActiveValue=@n1.bCarrierHasSyncMarkers) {blendType=Linear}
     BlendTime_0=0.2
     BlendTime_1=0.2
     bActiveValue=@n1.bCarrierIsCrouching
    ) {blendType=Linear})
  ]
 }
 InAir {
  path=AnimGraph.AnimGraphNode_StateMachine_0.DBNOBehavior.AnimStateNode_1.Carried.AnimGraphNode_StateMachine_0.CarriedLocomotion.AnimStateNode_3.InAir
  nodes [
   n1 = Apply Additive(
    Base=Sequence Player 'FireCarry_Idle_DBNO_CMM'() {playRateBasis=0, sequence=&FireCarry_Idle_DBNO_CMM}
    Additive="Sequence Player (additive) 'Torch_Swim_Idle_Noise_NonTargeted'"(PlayRate=-0.75) {playRate=0.1, sequence=&Torch_Swim_Idle_Noise_NonTargeted}
    Alpha=0.4
   ) {alphaBoolBlend {blendOption=Linear}}
   n2 = Blendspace Player 'FireCarry_DBNO_ZTrans_BS'!(X=Break Fort Anim Input DBNOCarried(FortAnimInput_DBNOCarried=Get DBNOCarriedInput()) -> (
     LHandIKLocation=~
     LHandIKRotation=~
     TempAttachLocation=~
     TempAttachRotation=~
     SubAnimPhysicsWeight=~
     DropMontagePosition=~
     PickupMontagePosition=~
     InterrogationMontagePosition=~
     PickupToIdleTransitionPosition=~
     CarrierPawnVelocityZ=~
     CarrierYawDeltaSmoothed=~
     CarriedJogNAnimPosition=~
     CarriedJogSAnimPosition=~
     CarriedCrouchNAnimPosition=~
     CarriedCrouchSAnimPosition=~
     CarriedSprintAnimPosition=~
     CarriedCrouchSprintAnimPosition=~
     CarrierSpinePitch=~
     DropStateWeight=~
     bBeingCarried=~
     bIsBeingPickedUp=~
     bIsBeingDropped=~
     bCarrierIsCrouching=~
     bCarrierIsMovingBackward=~
     bCarrierHasSyncMarkers=~
     bTransitionFromPickupToIdle=~
     bTransitionFromIdleToJogging=~
     bTransitionFromJoggingToSprinting=~
     bTransitionFromInAirToLanding=~
     bJackalPlayJumpTrickVertical=~
    ).CarrierPawnVelocityZ) -> (Pose) {blendSpace=/Game/Animation/Game/MainPlayer/Poses/Medium/Male/FireCarry/FireCarry_DBNO_ZTrans_BS}
   n3 = Get DBNOCarriedInput()
   Output Animation Pose(Result=Blend Poses by bool(
     BlendPose_0=Sequence Player 'DBNO_JackalVehicle_Passenger_Hunker_Pose'() {sequence=/Game/Animation/Game/MainPlayer/Vehicles/JackalVehicle/DBNO/DBNO_JackalVehicle_Passenger_Hunker_Pose}
     BlendPose_1=Apply Additive(Base=@n1, Additive=@n2, Alpha=0.75) {alphaBoolBlend {blendOption=Linear}}
     BlendTime_0=0.25
     BlendTime_1=0.25
     bActiveValue=Break Fort Anim Input DBNOCarried(FortAnimInput_DBNOCarried=@n3) -> (
      LHandIKLocation=~
      LHandIKRotation=~
      TempAttachLocation=~
      TempAttachRotation=~
      SubAnimPhysicsWeight=~
      DropMontagePosition=~
      PickupMontagePosition=~
      InterrogationMontagePosition=~
      PickupToIdleTransitionPosition=~
      CarrierPawnVelocityZ=~
      CarrierYawDeltaSmoothed=~
      CarriedJogNAnimPosition=~
      CarriedJogSAnimPosition=~
      CarriedCrouchNAnimPosition=~
      CarriedCrouchSAnimPosition=~
      CarriedSprintAnimPosition=~
      CarriedCrouchSprintAnimPosition=~
      CarrierSpinePitch=~
      DropStateWeight=~
      bBeingCarried=~
      bIsBeingPickedUp=~
      bIsBeingDropped=~
      bCarrierIsCrouching=~
      bCarrierIsMovingBackward=~
      bCarrierHasSyncMarkers=~
      bTransitionFromPickupToIdle=~
      bTransitionFromIdleToJogging=~
      bTransitionFromJoggingToSprinting=~
      bTransitionFromInAirToLanding=~
      bJackalPlayJumpTrickVertical=~
     ).bJackalPlayJumpTrickVertical
    ) {blendType=Linear})
  ]
 }
 Landing {
  path=AnimGraph.AnimGraphNode_StateMachine_0.DBNOBehavior.AnimStateNode_1.Carried.AnimGraphNode_StateMachine_0.CarriedLocomotion.AnimStateNode_4.Landing
  nodes [
   n1 = Sequence Player 'FireCarry_Idle_DBNO_CMM'() {bIgnoreForRelevancyTest=true, playRateBasis=0, sequence=&FireCarry_Idle_DBNO_CMM}
   Output Animation Pose(Result=Apply Additive(
     Base=@n1
     Additive="Sequence Player (additive) 'FiremansCarry_DBNO_Land_Additive'"() {bLoopAnimation=false, sequence=/Game/Animation/Game/MainPlayer/Locomotion/Medium/Male/DBNO/FiremansCarry_DBNO_Land_Additive}
     Alpha=0.6
    ) {alphaBoolBlend {blendOption=Linear}})
  ]
 }
 Transition {
  path=AnimGraph.AnimGraphNode_StateMachine_0.DBNOBehavior.AnimStateNode_1.Carried.AnimGraphNode_StateMachine_0.CarriedLocomotion.AnimStateTransitionNode_0.Transition
  nodes [Result(bCanEnterTransition=Break Fort Anim Input DBNOCarried(FortAnimInput_DBNOCarried=Get DBNOCarriedInput()) -> (
     LHandIKLocation=~
     LHandIKRotation=~
     TempAttachLocation=~
     TempAttachRotation=~
     SubAnimPhysicsWeight=~
     DropMontagePosition=~
     PickupMontagePosition=~
     InterrogationMontagePosition=~
     PickupToIdleTransitionPosition=~
     CarrierPawnVelocityZ=~
     CarrierYawDeltaSmoothed=~
     CarriedJogNAnimPosition=~
     CarriedJogSAnimPosition=~
     CarriedCrouchNAnimPosition=~
     CarriedCrouchSAnimPosition=~
     CarriedSprintAnimPosition=~
     CarriedCrouchSprintAnimPosition=~
     CarrierSpinePitch=~
     DropStateWeight=~
     bBeingCarried=~
     bIsBeingPickedUp=~
     bIsBeingDropped=~
     bCarrierIsCrouching=~
     bCarrierIsMovingBackward=~
     bCarrierHasSyncMarkers=~
     bTransitionFromPickupToIdle=~
     bTransitionFromIdleToJogging=~
     bTransitionFromJoggingToSprinting=~
     bTransitionFromInAirToLanding=~
     bJackalPlayJumpTrickVertical=~
    ).bTransitionFromIdleToJogging)]
 }
 Transition {
  path=AnimGraph.AnimGraphNode_StateMachine_0.DBNOBehavior.AnimStateNode_1.Carried.AnimGraphNode_StateMachine_0.CarriedLocomotion.AnimStateTransitionNode_1.Transition
  nodes [
   n1 = Get DBNOCarriedInput()
   Result(bCanEnterTransition=NOT Boolean(A=Break Fort Anim Input DBNOCarried(FortAnimInput_DBNOCarried=@n1) -> (
      LHandIKLocation=~
      LHandIKRotation=~
      TempAttachLocation=~
      TempAttachRotation=~
      SubAnimPhysicsWeight=~
      DropMontagePosition=~
      PickupMontagePosition=~
      InterrogationMontagePosition=~
      PickupToIdleTransitionPosition=~
      CarrierPawnVelocityZ=~
      CarrierYawDeltaSmoothed=~
      CarriedJogNAnimPosition=~
      CarriedJogSAnimPosition=~
      CarriedCrouchNAnimPosition=~
      CarriedCrouchSAnimPosition=~
      CarriedSprintAnimPosition=~
      CarriedCrouchSprintAnimPosition=~
      CarrierSpinePitch=~
      DropStateWeight=~
      bBeingCarried=~
      bIsBeingPickedUp=~
      bIsBeingDropped=~
      bCarrierIsCrouching=~
      bCarrierIsMovingBackward=~
      bCarrierHasSyncMarkers=~
      bTransitionFromPickupToIdle=~
      bTransitionFromIdleToJogging=~
      bTransitionFromJoggingToSprinting=~
      bTransitionFromInAirToLanding=~
      bJackalPlayJumpTrickVertical=~
     ).bTransitionFromIdleToJogging))
  ]
 }
 Transition {
  path=AnimGraph.AnimGraphNode_StateMachine_0.DBNOBehavior.AnimStateNode_1.Carried.AnimGraphNode_StateMachine_0.CarriedLocomotion.AnimStateTransitionNode_2.Transition
  nodes [Result(bCanEnterTransition=Break Fort Anim Input DBNOCarried(FortAnimInput_DBNOCarried=Get DBNOCarriedInput()) -> (
     LHandIKLocation=~
     LHandIKRotation=~
     TempAttachLocation=~
     TempAttachRotation=~
     SubAnimPhysicsWeight=~
     DropMontagePosition=~
     PickupMontagePosition=~
     InterrogationMontagePosition=~
     PickupToIdleTransitionPosition=~
     CarrierPawnVelocityZ=~
     CarrierYawDeltaSmoothed=~
     CarriedJogNAnimPosition=~
     CarriedJogSAnimPosition=~
     CarriedCrouchNAnimPosition=~
     CarriedCrouchSAnimPosition=~
     CarriedSprintAnimPosition=~
     CarriedCrouchSprintAnimPosition=~
     CarrierSpinePitch=~
     DropStateWeight=~
     bBeingCarried=~
     bIsBeingPickedUp=~
     bIsBeingDropped=~
     bCarrierIsCrouching=~
     bCarrierIsMovingBackward=~
     bCarrierHasSyncMarkers=~
     bTransitionFromPickupToIdle=~
     bTransitionFromIdleToJogging=~
     bTransitionFromJoggingToSprinting=~
     bTransitionFromInAirToLanding=~
     bJackalPlayJumpTrickVertical=~
    ).bTransitionFromJoggingToSprinting)]
 }
 Transition {
  path=AnimGraph.AnimGraphNode_StateMachine_0.DBNOBehavior.AnimStateNode_1.Carried.AnimGraphNode_StateMachine_0.CarriedLocomotion.AnimStateTransitionNode_3.Transition
  nodes [
   n1 = Get DBNOCarriedInput()
   Result(bCanEnterTransition=NOT Boolean(A=Break Fort Anim Input DBNOCarried(FortAnimInput_DBNOCarried=@n1) -> (
      LHandIKLocation=~
      LHandIKRotation=~
      TempAttachLocation=~
      TempAttachRotation=~
      SubAnimPhysicsWeight=~
      DropMontagePosition=~
      PickupMontagePosition=~
      InterrogationMontagePosition=~
      PickupToIdleTransitionPosition=~
      CarrierPawnVelocityZ=~
      CarrierYawDeltaSmoothed=~
      CarriedJogNAnimPosition=~
      CarriedJogSAnimPosition=~
      CarriedCrouchNAnimPosition=~
      CarriedCrouchSAnimPosition=~
      CarriedSprintAnimPosition=~
      CarriedCrouchSprintAnimPosition=~
      CarrierSpinePitch=~
      DropStateWeight=~
      bBeingCarried=~
      bIsBeingPickedUp=~
      bIsBeingDropped=~
      bCarrierIsCrouching=~
      bCarrierIsMovingBackward=~
      bCarrierHasSyncMarkers=~
      bTransitionFromPickupToIdle=~
      bTransitionFromIdleToJogging=~
      bTransitionFromJoggingToSprinting=~
      bTransitionFromInAirToLanding=~
      bJackalPlayJumpTrickVertical=~
     ).bTransitionFromJoggingToSprinting))
  ]
 }
 Transition {
  path=AnimGraph.AnimGraphNode_StateMachine_0.DBNOBehavior.AnimStateNode_1.Carried.AnimGraphNode_StateMachine_0.CarriedLocomotion.AnimStateTransitionNode_4.Transition
  nodes [
   n1 = Get DBNOCarriedInput()
   Result(bCanEnterTransition=NOT Boolean(A=Break Fort Anim Input DBNOCarried(FortAnimInput_DBNOCarried=@n1) -> (
      LHandIKLocation=~
      LHandIKRotation=~
      TempAttachLocation=~
      TempAttachRotation=~
      SubAnimPhysicsWeight=~
      DropMontagePosition=~
      PickupMontagePosition=~
      InterrogationMontagePosition=~
      PickupToIdleTransitionPosition=~
      CarrierPawnVelocityZ=~
      CarrierYawDeltaSmoothed=~
      CarriedJogNAnimPosition=~
      CarriedJogSAnimPosition=~
      CarriedCrouchNAnimPosition=~
      CarriedCrouchSAnimPosition=~
      CarriedSprintAnimPosition=~
      CarriedCrouchSprintAnimPosition=~
      CarrierSpinePitch=~
      DropStateWeight=~
      bBeingCarried=~
      bIsBeingPickedUp=~
      bIsBeingDropped=~
      bCarrierIsCrouching=~
      bCarrierIsMovingBackward=~
      bCarrierHasSyncMarkers=~
      bTransitionFromPickupToIdle=~
      bTransitionFromIdleToJogging=~
      bTransitionFromJoggingToSprinting=~
      bTransitionFromInAirToLanding=~
      bJackalPlayJumpTrickVertical=~
     ).bTransitionFromInAirToLanding))
  ]
 }
 Transition {
  path=AnimGraph.AnimGraphNode_StateMachine_0.DBNOBehavior.AnimStateNode_1.Carried.AnimGraphNode_StateMachine_0.CarriedLocomotion.AnimStateTransitionNode_6.Transition
  nodes [Result(bCanEnterTransition=Break Fort Anim Input DBNOCarried(FortAnimInput_DBNOCarried=Get DBNOCarriedInput()) -> (
     LHandIKLocation=~
     LHandIKRotation=~
     TempAttachLocation=~
     TempAttachRotation=~
     SubAnimPhysicsWeight=~
     DropMontagePosition=~
     PickupMontagePosition=~
     InterrogationMontagePosition=~
     PickupToIdleTransitionPosition=~
     CarrierPawnVelocityZ=~
     CarrierYawDeltaSmoothed=~
     CarriedJogNAnimPosition=~
     CarriedJogSAnimPosition=~
     CarriedCrouchNAnimPosition=~
     CarriedCrouchSAnimPosition=~
     CarriedSprintAnimPosition=~
     CarriedCrouchSprintAnimPosition=~
     CarrierSpinePitch=~
     DropStateWeight=~
     bBeingCarried=~
     bIsBeingPickedUp=~
     bIsBeingDropped=~
     bCarrierIsCrouching=~
     bCarrierIsMovingBackward=~
     bCarrierHasSyncMarkers=~
     bTransitionFromPickupToIdle=~
     bTransitionFromIdleToJogging=~
     bTransitionFromJoggingToSprinting=~
     bTransitionFromInAirToLanding=~
     bJackalPlayJumpTrickVertical=~
    ).bTransitionFromInAirToLanding)]
 }
 Transition {
  path=AnimGraph.AnimGraphNode_StateMachine_0.DBNOBehavior.AnimStateNode_1.Carried.AnimGraphNode_StateMachine_0.CarriedLocomotion.AnimStateTransitionNode_7.Transition
  nodes [Result()]
 }
 Drop {
  path=AnimGraph.AnimGraphNode_StateMachine_0.DBNOBehavior.AnimStateNode_2.Drop
  nodes [
   n1 = Get DBNOCarriedInput()
   Output Animation Pose(Result="Sequence Evaluator (by time) 'FiremansCarry_Drop_DBNO_CMM'"!(ExplicitTime:float=Break Fort Anim Input DBNOCarried(FortAnimInput_DBNOCarried=@n1) -> (
      LHandIKLocation=~
      LHandIKRotation=~
      TempAttachLocation=~
      TempAttachRotation=~
      SubAnimPhysicsWeight=~
      DropMontagePosition=~
      PickupMontagePosition=~
      InterrogationMontagePosition=~
      PickupToIdleTransitionPosition=~
      CarrierPawnVelocityZ=~
      CarrierYawDeltaSmoothed=~
      CarriedJogNAnimPosition=~
      CarriedJogSAnimPosition=~
      CarriedCrouchNAnimPosition=~
      CarriedCrouchSAnimPosition=~
      CarriedSprintAnimPosition=~
      CarriedCrouchSprintAnimPosition=~
      CarrierSpinePitch=~
      DropStateWeight=~
      bBeingCarried=~
      bIsBeingPickedUp=~
      bIsBeingDropped=~
      bCarrierIsCrouching=~
      bCarrierIsMovingBackward=~
      bCarrierHasSyncMarkers=~
      bTransitionFromPickupToIdle=~
      bTransitionFromIdleToJogging=~
      bTransitionFromJoggingToSprinting=~
      bTransitionFromInAirToLanding=~
      bJackalPlayJumpTrickVertical=~
     ).DropMontagePosition) -> (Pose:Pose Link Structure) {is=title AnimGraphNode_SequenceEvaluator, sequence=/Game/Animation/Game/MainPlayer/Locomotion/Medium/Male/DBNO/FiremansCarry_Drop_DBNO_CMM})
  ]
 }
 Transition {
  path=AnimGraph.AnimGraphNode_StateMachine_0.DBNOBehavior.AnimStateTransitionNode_0.Transition
  nodes [Result(bCanEnterTransition=Break Fort Anim Input DBNOCarried(FortAnimInput_DBNOCarried=Get DBNOCarriedInput()) -> (
     LHandIKLocation=~
     LHandIKRotation=~
     TempAttachLocation=~
     TempAttachRotation=~
     SubAnimPhysicsWeight=~
     DropMontagePosition=~
     PickupMontagePosition=~
     InterrogationMontagePosition=~
     PickupToIdleTransitionPosition=~
     CarrierPawnVelocityZ=~
     CarrierYawDeltaSmoothed=~
     CarriedJogNAnimPosition=~
     CarriedJogSAnimPosition=~
     CarriedCrouchNAnimPosition=~
     CarriedCrouchSAnimPosition=~
     CarriedSprintAnimPosition=~
     CarriedCrouchSprintAnimPosition=~
     CarrierSpinePitch=~
     DropStateWeight=~
     bBeingCarried=~
     bIsBeingPickedUp=~
     bIsBeingDropped=~
     bCarrierIsCrouching=~
     bCarrierIsMovingBackward=~
     bCarrierHasSyncMarkers=~
     bTransitionFromPickupToIdle=~
     bTransitionFromIdleToJogging=~
     bTransitionFromJoggingToSprinting=~
     bTransitionFromInAirToLanding=~
     bJackalPlayJumpTrickVertical=~
    ).bTransitionFromPickupToIdle)]
 }
 Transition {
  path=AnimGraph.AnimGraphNode_StateMachine_0.DBNOBehavior.AnimStateTransitionNode_1.Transition
  nodes [Result(bCanEnterTransition=Break Fort Anim Input DBNOCarried(FortAnimInput_DBNOCarried=Get DBNOCarriedInput()) -> (
     LHandIKLocation=~
     LHandIKRotation=~
     TempAttachLocation=~
     TempAttachRotation=~
     SubAnimPhysicsWeight=~
     DropMontagePosition=~
     PickupMontagePosition=~
     InterrogationMontagePosition=~
     PickupToIdleTransitionPosition=~
     CarrierPawnVelocityZ=~
     CarrierYawDeltaSmoothed=~
     CarriedJogNAnimPosition=~
     CarriedJogSAnimPosition=~
     CarriedCrouchNAnimPosition=~
     CarriedCrouchSAnimPosition=~
     CarriedSprintAnimPosition=~
     CarriedCrouchSprintAnimPosition=~
     CarrierSpinePitch=~
     DropStateWeight=~
     bBeingCarried=~
     bIsBeingPickedUp=~
     bIsBeingDropped=~
     bCarrierIsCrouching=~
     bCarrierIsMovingBackward=~
     bCarrierHasSyncMarkers=~
     bTransitionFromPickupToIdle=~
     bTransitionFromIdleToJogging=~
     bTransitionFromJoggingToSprinting=~
     bTransitionFromInAirToLanding=~
     bJackalPlayJumpTrickVertical=~
    ).bIsBeingDropped)]
 }
 EventGraph {links [], nodes []}
]
variables []
