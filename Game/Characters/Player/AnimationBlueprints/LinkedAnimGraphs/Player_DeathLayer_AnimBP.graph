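# Player death linked anim layer that replaces the SecondMovementMode final pose with non-looping elimination animations, ignoring the input pose.
# A zero-time blend on isDBNO picks spin-to-float for downed (DBNO) deaths, otherwise standing stomach shot.
# The event graph caches that flag from the pawn and binds OnDeathEffects to store damage and hit location, which the layer never reads.
# BeginPlay spawns a test death-pop actor at the owner's location.
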AnimBlueprint /Game/Characters/Player/AnimationBlueprints/LinkedAnimGraphs/Player_DeathLayer_AnimBP
target_skeleton=&Fortnite_M_Avg_Player_Skeleton
graphs [
 AnimGraph {nodes [Output Pose()]}
 EventGraph {nodes [
   n1 = CustomEvent!() -> (
    OutputDelegate
    then
    DeadActor:Actor Object Reference
    Damage:float=~
    InstigatedBy:Fort Pawn Object Reference
    HitLocation:Vector=~
    DamageCauser:Actor Object Reference
   ) {is=title K2Node_CustomEvent}
   > Set Hit Location(Hit Location:Vector=@n1.HitLocation) -> (Output_Get:Vector)
   > Set Damage(Damage:float=@n1.Damage) -> (Output_Get:float)
   Event BlueprintBeginPlay() {is=title K2Node_Event}
   > SpawnActor Death Pop BPTest!(
    execute:Exec
    Class:Actor Class Reference=/Game/Animation/Game/MainPlayer/Reacts/Deaths/customElims/ElimProto/deathPop_BPTest.deathPop_BPTest_C
    SpawnTransform_Location:Vector=Get Actor Location(self=GetOwningActor())
    SpawnTransform_Rotation:Rotator="0, 0, 0"
    SpawnTransform_Scale:Vector="1,1,1"
    CollisionHandlingOverride:ESpawnActorCollisionHandlingMethod Enum=Undefined
    TransformScaleMethod:ESpawnActorScaleMethod Enum=MultiplyWithRoot
    Owner:Actor Object Reference
    Instigator:Pawn Object Reference
   ) -> (then:Exec, ReturnValue:Death Pop BPTest Object Reference) {is=title K2Node_SpawnActorFromClass}
   Event BlueprintUpdateAnimation()
   n2 = Get FortPlayerPawn() -> (FortPlayerPawn:Fort Player Pawn Object Reference)
   > Bind Event to On Death Effects(self:Fort Pawn Object Reference=@n2, Delegate=@n1) {is=title K2Node_AddDelegate}
   > Set isDBNO(isDBNO:bool=WasDBNOOnDeath!(self:Fort Pawn Object Reference=@n2) -> (ReturnValue:bool)) -> (Output_Get:bool)
  ]}
 SecondMovementMode_FinalPoseOverride {nodes [
   Input Pose() {name=InFinalPose}
   Local Space Ref Pose()
   Output Pose(Result=Blend Poses by bool(
     BlendPose_0=Sequence Player 'elimination_dbno_spinToFloat'() {bLoopAnimation=false, sequence=/Game/Animation/Game/MainPlayer/Reacts/Deaths/customElims/elimination_dbno_spinToFloat}
     BlendPose_1=Sequence Player 'elimination_standing_stomachShot'() {bLoopAnimation=false, sequence=/Game/Animation/Game/MainPlayer/Reacts/Deaths/customElims/elimination_standing_stomachShot}
     BlendTime_0=0
     BlendTime_1=0
     bActiveValue=Get isDBNO() -> (isDBNO:bool)
    )) {layerGroup=SecondMovementMode}
  ]}
]
variables [
 Damage {type=number}
 Hit Location {type=Vector}
 isDBNO {type=boolean}
]
