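# Player linked anim graph that layers weapon-specific first-person additive motion onto an incoming full-body pose.
# FirstPerson (locomotion, tac sprint, slide, crouch), landing, sprint-outro and ADS state machines stack additively, with FirstPerson_OLD selected by bIs_FPAD_Null.
# The cached result passes through one of two blend-profile masks, chosen by bEnableCameraBoneSpace1stPersonAnims, before alpha-scaled application.
# States pull clips and blendspaces from generic numbered slots in WeaponSpecificAnimData, often damped by ADS alphas.
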AnimBlueprint /Game/Characters/Player/AnimationBlueprints/LinkedAnimGraphs/Player_FirstPersonAdditives_AnimBP
target_skeleton=&Fortnite_M_Avg_Player_Skeleton
graphs [
 AnimGraph {nodes [
   n1 = Profile Blend!(BasePose:Pose Link Structure, BlendPose:Pose Link Structure=Use cached pose 'AdditivePose'(), BlendWeight:float=1) -> (Pose:Pose Link Structure) {
    is=title AnimGraphNode_BlendProfileLayeredBlend
    bCustomCurveBlending=true
    blendProfileAsset=/Game/Characters/Player/BlendMasks/FirstPersonAdditives_BlendMask_V2
    blendPose {linkId=16842751}
   }
   n2 = Profile Blend(BlendPose=Use cached pose 'AdditivePose'(), BlendWeight=1) {bCustomCurveBlending=true, blendProfileAsset=/Game/Characters/Player/BlendMasks/FirstPersonAdditives_BlendMask, blendPose {linkId=16842751}}
   n3 = Get bEnableCameraBoneSpace1stPersonAnims() -> (bEnableCameraBoneSpace1stPersonAnims:bool)
   Output Pose(Result=Apply Additive(
     Base=Use cached pose 'InPoseFullBody'()
     Additive=Blend Poses by bool(BlendPose_0=@n1, BlendPose_1=@n2, bActiveValue=@n3) {blendType=Linear}
     Alpha=Get DisableFirstPersonAddiitivesAlpha() -> (DisableFirstPersonAddiitivesAlpha:double)
    ) {alphaBoolBlend {blendOption=Linear}}) {layerGroup=None}
   Save cached pose 'InPoseFullBody'(Pose=Input Pose() {name=InPose_FullBody}) {cacheName=InPoseFullBody}
   AnimGraphNode_StateMachine_0 = FirstPerson() {is=title AnimGraphNode_StateMachine}
   AnimGraphNode_StateMachine_4 = AdditiveLandingSM() {is=title AnimGraphNode_StateMachine}
   AnimGraphNode_StateMachine_5 = TacSprintOutroSM() {is=title AnimGraphNode_StateMachine}
   n4 = Apply Additive(
    Base=Apply Additive(Base=@AnimGraphNode_StateMachine_0, Additive=@AnimGraphNode_StateMachine_4) {alphaBoolBlend {blendOption=Linear}}
    Additive=@AnimGraphNode_StateMachine_5
   ) {alphaBoolBlend {blendOption=Linear}}
   AnimGraphNode_StateMachine_3 = ADS() {is=title AnimGraphNode_StateMachine}
   AnimGraphNode_StateMachine_2 = FirstPerson_OLD() {is=title AnimGraphNode_StateMachine}
   Save cached pose 'AdditivePose'(Pose=Blend Poses by bool(
     BlendPose_0=@AnimGraphNode_StateMachine_2
     BlendPose_1=Apply Additive(Base=@n4, Additive=@AnimGraphNode_StateMachine_3) {alphaBoolBlend {blendOption=Linear}}
     BlendTime_0=0.25
     BlendTime_1=0.25
     bActiveValue=Get bIs_FPAD_Null() -> (bIs_FPAD_Null:bool)
    ) {blendType=Linear}) {cacheName=AdditivePose}
  ]}
 FirstPerson {
  entry=OnGround
  path=AnimGraph.AnimGraphNode_StateMachine_0.FirstPerson
  ids=id
  nodes [
   AnimStateNode_2 = OnGround() {is=title AnimStateNode}
   AnimStateNode_3 = InAir() {is=title AnimStateNode}
   AnimStateNode_4 = LandOnGround() {is=title AnimStateNode}
   AnimStateTransitionNode_0 = AnimStateTransitionNode() {bDisabled=true, crossfadeDuration=0.1}
   AnimStateTransitionNode_2 = AnimStateTransitionNode() {crossfadeDuration=0.15}
   AnimStateTransitionNode_3 = AnimStateTransitionNode() {priorityOrder=2}
   AnimStateTransitionNode_4 = AnimStateTransitionNode() {bAutomaticRuleBasedOnSequencePlayerInState=true, bDisabled=true}
   AnimStateTransitionNode_7 = AnimStateTransitionNode() {bDisabled=true, crossfadeDuration=0.66}
  ]
  links [
   {
    dst=OnGround
    rule=AnimGraph.AnimGraphNode_StateMachine_0.FirstPerson.AnimStateTransitionNode_3.AnimationTransitionGraph_1
    src=InAir
    via=AnimStateTransitionNode_3
   }
   {
    dst=OnGround
    rule=AnimGraph.AnimGraphNode_StateMachine_0.FirstPerson.AnimStateTransitionNode_0.AnimationTransitionGraph_0
    src=LandOnGround
    via=AnimStateTransitionNode_0
   }
   {
    dst=OnGround
    rule=AnimGraph.AnimGraphNode_StateMachine_0.FirstPerson.AnimStateTransitionNode_4.AnimationTransitionGraph_0
    src=LandOnGround
    via=AnimStateTransitionNode_4
   }
   {
    dst=OnGround
    rule=AnimGraph.AnimGraphNode_StateMachine_0.FirstPerson.AnimStateTransitionNode_7.AnimationTransitionGraph_0
    src=LandOnGround
    via=AnimStateTransitionNode_7
   }
   {
    dst=InAir
    rule=AnimGraph.AnimGraphNode_StateMachine_0.FirstPerson.AnimStateTransitionNode_2.AnimationTransitionGraph_0
    src=OnGround
    via=AnimStateTransitionNode_2
   }
  ]
 }
 OnGround {
  path=AnimGraph.AnimGraphNode_StateMachine_0.FirstPerson.AnimStateNode_2.OnGround
  nodes [
   AnimGraphNode_StateMachine_3 = ADS()
   Apply Additive(Additive=@AnimGraphNode_StateMachine_3) {alphaBoolBlend {blendOption=Linear}}
   Additive Identity Pose()
   AnimGraphNode_StateMachine_0 = OnGroundSM() {is=title AnimGraphNode_StateMachine}
   AnimGraphNode_StateMachine_1 = Crouch Into Exit() {is=title AnimGraphNode_StateMachine}
   AnimGraphNode_StateMachine_4 = StartStopDrift() {is=title AnimGraphNode_StateMachine, bSkipFirstUpdateTransition=false, maxTransitionsPerFrame=4}
   Output Animation Pose(Result=Apply Additive(
     Base=Apply Additive(Base=@AnimGraphNode_StateMachine_0, Additive=@AnimGraphNode_StateMachine_1) {alphaBoolBlend {blendOption=Linear}}
     Additive=@AnimGraphNode_StateMachine_4
    ) {alphaBoolBlend {blendOption=Linear}})
  ]
 }
 OnGroundSM {
  entry=Idle
  path=AnimGraph.AnimGraphNode_StateMachine_0.FirstPerson.AnimStateNode_2.OnGround.AnimGraphNode_StateMachine_0.OnGroundSM
  ids=id
  nodes [
   AnimStateNode_0 = TacSprintStart() {is=title AnimStateNode}
   AnimStateNode_1 = TacSprintJumpOut() {is=title AnimStateNode}
   AnimStateNode_2 = Sliding() {is=title AnimStateNode}
   AnimStateNode_3 = StartLocomotion() {is=title AnimStateNode}
   AnimStateNode_4 = StopLocomotion() {is=title AnimStateNode}
   AnimStateNode_5 = TacSprintLoop() {is=title AnimStateNode}
   AnimStateNode_6 = Idle() {bAlwaysResetOnEntry=true}
   AnimStateNode_7 = Locomotion() {is=title AnimStateNode}
   AnimStateTransitionNode_0 = AnimStateTransitionNode() {bAutomaticRuleBasedOnSequencePlayerInState=true, crossfadeDuration=0}
   AnimStateTransitionNode_1 = AnimStateTransitionNode()
   AnimStateTransitionNode_10 = AnimStateTransitionNode()
   AnimStateTransitionNode_11 = AnimStateTransitionNode() {crossfadeDuration=0.25}
   AnimStateTransitionNode_12 = AnimStateTransitionNode()
   AnimStateTransitionNode_15 = AnimStateTransitionNode()
   AnimStateTransitionNode_16 = AnimStateTransitionNode() {crossfadeDuration=0.3}
   AnimStateTransitionNode_18 = AnimStateTransitionNode()
   AnimStateTransitionNode_2 = AnimStateTransitionNode()
   AnimStateTransitionNode_3 = AnimStateTransitionNode()
   AnimStateTransitionNode_4 = AnimStateTransitionNode() {crossfadeDuration=0.3}
   AnimStateTransitionNode_5 = AnimStateTransitionNode() {bAutomaticRuleBasedOnSequencePlayerInState=true, crossfadeDuration=0.15}
   AnimStateTransitionNode_6 = AnimStateTransitionNode() {crossfadeDuration=0.1}
   AnimStateTransitionNode_7 = AnimStateTransitionNode() {bAutomaticRuleBasedOnSequencePlayerInState=true, blendMode=ExpOut, crossfadeDuration=0.75}
   AnimStateTransitionNode_8 = AnimStateTransitionNode() {crossfadeDuration=0.3}
   AnimStateTransitionNode_9 = AnimStateTransitionNode() {bAutomaticRuleBasedOnSequencePlayerInState=true}
  ]
  links [
   {
    dst=StartLocomotion
    rule=AnimGraph.AnimGraphNode_StateMachine_0.FirstPerson.AnimStateNode_2.OnGround.AnimGraphNode_StateMachine_0.OnGroundSM.AnimStateTransitionNode_6.Transition
    src=Idle
    via=AnimStateTransitionNode_6
   }
   {
    dst=TacSprintJumpOut
    rule=AnimGraph.AnimGraphNode_StateMachine_0.FirstPerson.AnimStateNode_2.OnGround.AnimGraphNode_StateMachine_0.OnGroundSM.AnimStateTransitionNode_12.Transition
    src=Idle
    via=AnimStateTransitionNode_12
   }
   {
    dst=Sliding
    rule=AnimGraph.AnimGraphNode_StateMachine_0.FirstPerson.AnimStateNode_2.OnGround.AnimGraphNode_StateMachine_0.OnGroundSM.AnimStateTransitionNode_2.Transition
    src=Locomotion
    via=AnimStateTransitionNode_2
   }
   {
    dst=StopLocomotion
    rule=AnimGraph.AnimGraphNode_StateMachine_0.FirstPerson.AnimStateNode_2.OnGround.AnimGraphNode_StateMachine_0.OnGroundSM.AnimStateTransitionNode_8.Transition
    src=Locomotion
    via=AnimStateTransitionNode_8
   }
   {
    dst=TacSprintStart
    rule=AnimGraph.AnimGraphNode_StateMachine_0.FirstPerson.AnimStateNode_2.OnGround.AnimGraphNode_StateMachine_0.OnGroundSM.AnimStateTransitionNode_18.Transition
    src=Locomotion
    via=AnimStateTransitionNode_18
   }
   {
    dst=Locomotion
    rule=AnimGraph.AnimGraphNode_StateMachine_0.FirstPerson.AnimStateNode_2.OnGround.AnimGraphNode_StateMachine_0.OnGroundSM.AnimStateTransitionNode_3.Transition
    src=Sliding
    via=AnimStateTransitionNode_3
   }
   {
    dst=Locomotion
    rule=AnimGraph.AnimGraphNode_StateMachine_0.FirstPerson.AnimStateNode_2.OnGround.AnimGraphNode_StateMachine_0.OnGroundSM.AnimStateTransitionNode_1.Transition
    src=StartLocomotion
    via=AnimStateTransitionNode_1
   }
   {
    dst=Locomotion
    rule=AnimGraph.AnimGraphNode_StateMachine_0.FirstPerson.AnimStateNode_2.OnGround.AnimGraphNode_StateMachine_0.OnGroundSM.AnimStateTransitionNode_7.Transition
    src=StartLocomotion
    via=AnimStateTransitionNode_7
   }
   {
    dst=StopLocomotion
    rule=AnimGraph.AnimGraphNode_StateMachine_0.FirstPerson.AnimStateNode_2.OnGround.AnimGraphNode_StateMachine_0.OnGroundSM.AnimStateTransitionNode_10.Transition
    src=StartLocomotion
    via=AnimStateTransitionNode_10
   }
   {
    dst=Idle
    rule=AnimGraph.AnimGraphNode_StateMachine_0.FirstPerson.AnimStateNode_2.OnGround.AnimGraphNode_StateMachine_0.OnGroundSM.AnimStateTransitionNode_9.Transition
    src=StopLocomotion
    via=AnimStateTransitionNode_9
   }
   {
    dst=StartLocomotion
    rule=AnimGraph.AnimGraphNode_StateMachine_0.FirstPerson.AnimStateNode_2.OnGround.AnimGraphNode_StateMachine_0.OnGroundSM.AnimStateTransitionNode_11.Transition
    src=StopLocomotion
    via=AnimStateTransitionNode_11
   }
   {
    dst=TacSprintLoop
    rule=AnimGraph.AnimGraphNode_StateMachine_0.FirstPerson.AnimStateNode_2.OnGround.AnimGraphNode_StateMachine_0.OnGroundSM.AnimStateTransitionNode_5.Transition
    src=TacSprintJumpOut
    via=AnimStateTransitionNode_5
   }
   {
    dst=Locomotion
    rule=AnimGraph.AnimGraphNode_StateMachine_0.FirstPerson.AnimStateNode_2.OnGround.AnimGraphNode_StateMachine_0.OnGroundSM.AnimStateTransitionNode_16.AnimationTransitionGraph_1
    src=TacSprintLoop
    via=AnimStateTransitionNode_16
   }
   {
    dst=Sliding
    rule=AnimGraph.AnimGraphNode_StateMachine_0.FirstPerson.AnimStateNode_2.OnGround.AnimGraphNode_StateMachine_0.OnGroundSM.AnimStateTransitionNode_4.Transition
    src=TacSprintLoop
    via=AnimStateTransitionNode_4
   }
   {
    dst=Locomotion
    rule=AnimGraph.AnimGraphNode_StateMachine_0.FirstPerson.AnimStateNode_2.OnGround.AnimGraphNode_StateMachine_0.OnGroundSM.AnimStateTransitionNode_15.AnimationTransitionGraph_1
    src=TacSprintStart
    via=AnimStateTransitionNode_15
   }
   {
    dst=TacSprintLoop
    rule=AnimGraph.AnimGraphNode_StateMachine_0.FirstPerson.AnimStateNode_2.OnGround.AnimGraphNode_StateMachine_0.OnGroundSM.AnimStateTransitionNode_0.AnimationTransitionGraph_1
    src=TacSprintStart
    via=AnimStateTransitionNode_0
   }
  ]
 }
 TacSprintStart {
  path=AnimGraph.AnimGraphNode_StateMachine_0.FirstPerson.AnimStateNode_2.OnGround.AnimGraphNode_StateMachine_0.OnGroundSM.AnimStateNode_0.TacSprintStart
  nodes [
   n1 = Get WeaponSpecificAnimData() -> (WeaponSpecificAnimData:Fort Weapon First Person Animation Data Structure)
   Output Animation Pose(Result=Sequence Player!(Sequence:Anim Sequence Base Object Reference=Break Fort Weapon First Person Animation Data!(FortWeaponFirstPersonAnimationData:"Fort Weapon First Person Animation Data Structure (by ref)"=@n1) -> (
      DesiredAnimationMode:EFirstPersonAnimationMode Enum
      ReloadWeapon:Anim Montage Object Reference
      FireCharacter:Anim Montage Object Reference
      FireDownsightsCharacter:Anim Montage Object Reference
      TacSprintExitEquipBlendTime:float=~
      CharacterAnimSequence1:Anim Sequence Object Reference
      CharacterAnimSequence2:Anim Sequence Object Reference
      CharacterAnimSequence3:Anim Sequence Object Reference
      CharacterAnimSequence4:Anim Sequence Object Reference
      CharacterAnimSequence5:Anim Sequence Object Reference
      CharacterAnimSequence6:Anim Sequence Object Reference
      CharacterAnimSequence7:Anim Sequence Object Reference
      CharacterAnimSequence8:Anim Sequence Object Reference
      CharacterAnimSequence9:Anim Sequence Object Reference
      CharacterAnimSequence10:Anim Sequence Object Reference
      CharacterAnimSequence11:Anim Sequence Object Reference
      CharacterAnimSequence12:Anim Sequence Object Reference
      CharacterAnimSequence13:Anim Sequence Object Reference
      CharacterAnimSequence14:Anim Sequence Object Reference
      CharacterAnimSequence15:Anim Sequence Object Reference
      CharacterAnimSequence16:Anim Sequence Object Reference
      CharacterAnimSequence17:Anim Sequence Object Reference
      CharacterAnimSequence18:Anim Sequence Object Reference
      CharacterAnimSequence19:Anim Sequence Object Reference
      CharacterAnimSequence20:Anim Sequence Object Reference
      CharacterAnimSequence21:Anim Sequence Object Reference
      CharacterAnimSequence22:Anim Sequence Object Reference
      CharacterAnimSequence23:Anim Sequence Object Reference
      CharacterAnimSequence24:Anim Sequence Object Reference
      CharacterAnimSequence25:Anim Sequence Object Reference
      CharacterBlendSpace1:Blend Space Object Reference
      CharacterBlendSpace2:Blend Space Object Reference
      CharacterBlendSpace3:Blend Space Object Reference
      CharacterBlendSpace4:Blend Space Object Reference
      CharacterBlendSpace5:Blend Space Object Reference
      CharacterBlendSpace6:Blend Space Object Reference
      CharacterBlendSpace7:Blend Space Object Reference
      CharacterBlendSpace8:Blend Space Object Reference
      CharacterBlendSpace9:Blend Space Object Reference
      CharacterBlendSpace10:Blend Space Object Reference
      CharacterBlendSpace11:Blend Space Object Reference
      CharacterBlendSpace12:Blend Space Object Reference
      CharacterBlendSpace13:Blend Space Object Reference
      CharacterBlendSpace14:Blend Space Object Reference
      CharacterBlendSpace15:Blend Space Object Reference
      CharacterBlendSpace16:Blend Space Object Reference
      CharacterBlendSpace17:Blend Space Object Reference
      CharacterBlendSpace18:Blend Space Object Reference
      CharacterBlendSpace19:Blend Space Object Reference
      CharacterBlendSpace20:Blend Space Object Reference
      CharacterMontage1:Anim Montage Object Reference
      CharacterMontage2:Anim Montage Object Reference
      CharacterMontage3:Anim Montage Object Reference
      CharacterMontage4:Anim Montage Object Reference
      CharacterMontage5:Anim Montage Object Reference
      CharacterMontage6:Anim Montage Object Reference
      CharacterMontage7:Anim Montage Object Reference
      CharacterMontage8:Anim Montage Object Reference
      CharacterMontage9:Anim Montage Object Reference
      CharacterMontage10:Anim Montage Object Reference
      Bool1:bool=~
      Bool2:bool=~
      Bool3:bool=~
      Bool4:bool=~
      Bool5:bool=~
      Bool6:bool=~
      Bool7:bool=~
      Bool8:bool=~
      Bool9:bool=~
      Bool10:bool=~
      Float1:float=~
      Float2:float=~
      Float3:float=~
      Float4:float=~
      Float5:float=~
      Float6:float=~
      Float7:float=~
      Float8:float=~
      Float9:float=~
      Float10:float=~
     ) {is=title K2Node_BreakStruct}.CharacterAnimSequence18) -> (Pose) {bLoopAnimation=false})
  ]
 }
 TacSprintJumpOut {
  path=AnimGraph.AnimGraphNode_StateMachine_0.FirstPerson.AnimStateNode_2.OnGround.AnimGraphNode_StateMachine_0.OnGroundSM.AnimStateNode_1.TacSprintJumpOut
  nodes [
   n1 = Get WeaponSpecificAnimData()
   Output Animation Pose(Result=Sequence Player(Sequence=Break Fort Weapon First Person Animation Data(FortWeaponFirstPersonAnimationData=@n1) -> (
      TacSprintExitEquipBlendTime=~
      Bool1=~
      Bool2=~
      Bool3=~
      Bool4=~
      Bool5=~
      Bool6=~
      Bool7=~
      Bool8=~
      Bool9=~
      Bool10=~
      Float1=~
      Float2=~
      Float3=~
      Float4=~
      Float5=~
      Float6=~
      Float7=~
      Float8=~
      Float9=~
      Float10=~
     ).CharacterAnimSequence22) {bLoopAnimation=false})
  ]
 }
 Sliding {
  path=AnimGraph.AnimGraphNode_StateMachine_0.FirstPerson.AnimStateNode_2.OnGround.AnimGraphNode_StateMachine_0.OnGroundSM.AnimStateNode_2.Sliding
  nodes [
   AnimGraphNode_StateMachine_0 = SlidingSM() {is=title AnimGraphNode_StateMachine, bAllowConduitEntryStates=true}
   Output Animation Pose(Result=@AnimGraphNode_StateMachine_0)
   Break Fort Weapon First Person Animation Data(FortWeaponFirstPersonAnimationData=Get WeaponSpecificAnimData()) -> (
    TacSprintExitEquipBlendTime=~
    Bool1=~
    Bool2=~
    Bool3=~
    Bool4=~
    Bool5=~
    Bool6=~
    Bool7=~
    Bool8=~
    Bool9=~
    Bool10=~
    Float1=~
    Float2=~
    Float3=~
    Float4=~
    Float5=~
    Float6=~
    Float7=~
    Float8=~
    Float9=~
    Float10=~
   )
  ]
 }
 SlidingSM {
  entry=Conduit
  path=AnimGraph.AnimGraphNode_StateMachine_0.FirstPerson.AnimStateNode_2.OnGround.AnimGraphNode_StateMachine_0.OnGroundSM.AnimStateNode_2.Sliding.AnimGraphNode_StateMachine_0.SlidingSM
  ids=id
  nodes [
   AnimStateConduitNode_0 = Conduit()
   AnimStateNode_0 = SlideIntroFromTacSprint() {is=title AnimStateNode}
   AnimStateNode_1 = Sliding()
   AnimStateTransitionNode_0 = AnimStateTransitionNode()
   AnimStateTransitionNode_1 = AnimStateTransitionNode() {bAutomaticRuleBasedOnSequencePlayerInState=true}
   AnimStateTransitionNode_2 = AnimStateTransitionNode()
  ]
  links [
   {
    dst=SlideIntroFromTacSprint
    rule=AnimGraph.AnimGraphNode_StateMachine_0.FirstPerson.AnimStateNode_2.OnGround.AnimGraphNode_StateMachine_0.OnGroundSM.AnimStateNode_2.Sliding.AnimGraphNode_StateMachine_0.SlidingSM.AnimStateTransitionNode_0.Transition
    src=Conduit
    via=AnimStateTransitionNode_0
   }
   {
    dst=Sliding
    rule=AnimGraph.AnimGraphNode_StateMachine_0.FirstPerson.AnimStateNode_2.OnGround.AnimGraphNode_StateMachine_0.OnGroundSM.AnimStateNode_2.Sliding.AnimGraphNode_StateMachine_0.SlidingSM.AnimStateTransitionNode_2.Transition
    src=Conduit
    via=AnimStateTransitionNode_2
   }
   {
    dst=Sliding
    rule=AnimGraph.AnimGraphNode_StateMachine_0.FirstPerson.AnimStateNode_2.OnGround.AnimGraphNode_StateMachine_0.OnGroundSM.AnimStateNode_2.Sliding.AnimGraphNode_StateMachine_0.SlidingSM.AnimStateTransitionNode_1.Transition
    src=SlideIntroFromTacSprint
    via=AnimStateTransitionNode_1
   }
  ]
 }
 Conduit {
  path=AnimGraph.AnimGraphNode_StateMachine_0.FirstPerson.AnimStateNode_2.OnGround.AnimGraphNode_StateMachine_0.OnGroundSM.AnimStateNode_2.Sliding.AnimGraphNode_StateMachine_0.SlidingSM.AnimStateConduitNode_0.Conduit
  nodes [Result(bCanEnterTransition=true)]
 }
 SlideIntroFromTacSprint {
  path=AnimGraph.AnimGraphNode_StateMachine_0.FirstPerson.AnimStateNode_2.OnGround.AnimGraphNode_StateMachine_0.OnGroundSM.AnimStateNode_2.Sliding.AnimGraphNode_StateMachine_0.SlidingSM.AnimStateNode_0.SlideIntroFromTacSprint
  nodes [
   n1 = Get WeaponSpecificAnimData()
   Output Animation Pose(Result=Sequence Player(Sequence=Break Fort Weapon First Person Animation Data(FortWeaponFirstPersonAnimationData=@n1) -> (
      TacSprintExitEquipBlendTime=~
      Bool1=~
      Bool2=~
      Bool3=~
      Bool4=~
      Bool5=~
      Bool6=~
      Bool7=~
      Bool8=~
      Bool9=~
      Bool10=~
      Float1=~
      Float2=~
      Float3=~
      Float4=~
      Float5=~
      Float6=~
      Float7=~
      Float8=~
      Float9=~
      Float10=~
     ).CharacterAnimSequence20) {bLoopAnimation=false})
  ]
 }
 Sliding {
  path=AnimGraph.AnimGraphNode_StateMachine_0.FirstPerson.AnimStateNode_2.OnGround.AnimGraphNode_StateMachine_0.OnGroundSM.AnimStateNode_2.Sliding.AnimGraphNode_StateMachine_0.SlidingSM.AnimStateNode_1.Sliding
  nodes [
   n1 = Break Fort Weapon First Person Animation Data(FortWeaponFirstPersonAnimationData=Get WeaponSpecificAnimData()) -> (
    TacSprintExitEquipBlendTime=~
    Bool1=~
    Bool2=~
    Bool3=~
    Bool4=~
    Bool5=~
    Bool6=~
    Bool7=~
    Bool8=~
    Bool9=~
    Bool10=~
    Float1=~
    Float2=~
    Float3=~
    Float4=~
    Float5=~
    Float6=~
    Float7=~
    Float8=~
    Float9=~
    Float10=~
   )
   n2 = Sequence Player(Sequence=@n1.CharacterAnimSequence14) {bLoopAnimation=false}
   n3 = Sequence Player(Sequence=@n1.CharacterAnimSequence3)
   n4 = Additive Identity Pose()
   n5 = Sequence Player(Sequence=@n1.CharacterAnimSequence3)
   Output Animation Pose(Result=Blend Poses by bool(
     BlendPose_0=Apply Additive(Base=@n4, Additive=@n5, Alpha=0.25) {alphaBoolBlend {blendOption=Linear}}
     BlendPose_1=Apply Additive(Base=@n3, Additive=@n2) {alphaBoolBlend {blendOption=Linear}}
     BlendTime_0=0.2
     BlendTime_1=0.2
     bActiveValue=Property Access!() -> (Value:bool=~) {is=title K2Node_PropertyAccess}
    ) {blendType=Linear})
  ]
 }
 Transition {
  path=AnimGraph.AnimGraphNode_StateMachine_0.FirstPerson.AnimStateNode_2.OnGround.AnimGraphNode_StateMachine_0.OnGroundSM.AnimStateNode_2.Sliding.AnimGraphNode_StateMachine_0.SlidingSM.AnimStateTransitionNode_0.Transition
  nodes [Result(bCanEnterTransition=Get bIsTacSprinting() -> (bIsTacSprinting:bool))]
 }
 Transition {
  path=AnimGraph.AnimGraphNode_StateMachine_0.FirstPerson.AnimStateNode_2.OnGround.AnimGraphNode_StateMachine_0.OnGroundSM.AnimStateNode_2.Sliding.AnimGraphNode_StateMachine_0.SlidingSM.AnimStateTransitionNode_1.Transition
  nodes [Result()]
 }
 Transition {
  path=AnimGraph.AnimGraphNode_StateMachine_0.FirstPerson.AnimStateNode_2.OnGround.AnimGraphNode_StateMachine_0.OnGroundSM.AnimStateNode_2.Sliding.AnimGraphNode_StateMachine_0.SlidingSM.AnimStateTransitionNode_2.Transition
  nodes [Result(bCanEnterTransition=NOT Boolean(A=Get bIsTacSprinting()))]
 }
 StartLocomotion {
  path=AnimGraph.AnimGraphNode_StateMachine_0.FirstPerson.AnimStateNode_2.OnGround.AnimGraphNode_StateMachine_0.OnGroundSM.AnimStateNode_3.StartLocomotion
  nodes [
   n1 = Break Fort Weapon First Person Animation Data(FortWeaponFirstPersonAnimationData=Get WeaponSpecificAnimData()) -> (
    TacSprintExitEquipBlendTime=~
    Bool1=~
    Bool2=~
    Bool3=~
    Bool4=~
    Bool5=~
    Bool6=~
    Bool7=~
    Bool8=~
    Bool9=~
    Bool10=~
    Float1=~
    Float2=~
    Float3=~
    Float4=~
    Float5=~
    Float6=~
    Float7=~
    Float8=~
    Float9=~
    Float10=~
   )
   Output Animation Pose(Result=Apply Additive(
     Base=Additive Identity Pose()
     Additive=Sequence Player(Sequence=@n1.CharacterAnimSequence4) {bLoopAnimation=false}
     Alpha=Get StartDriftAlpha() -> (StartDriftAlpha:double)
    ) {alphaBoolBlend {blendOption=Linear}})
   Get ADS_25Alpha() -> (ADS_25Alpha:double)
  ]
 }
 StopLocomotion {
  path=AnimGraph.AnimGraphNode_StateMachine_0.FirstPerson.AnimStateNode_2.OnGround.AnimGraphNode_StateMachine_0.OnGroundSM.AnimStateNode_4.StopLocomotion
  nodes [
   n1 = Break Fort Weapon First Person Animation Data(FortWeaponFirstPersonAnimationData=Get WeaponSpecificAnimData()) -> (
    TacSprintExitEquipBlendTime=~
    Bool1=~
    Bool2=~
    Bool3=~
    Bool4=~
    Bool5=~
    Bool6=~
    Bool7=~
    Bool8=~
    Bool9=~
    Bool10=~
    Float1=~
    Float2=~
    Float3=~
    Float4=~
    Float5=~
    Float6=~
    Float7=~
    Float8=~
    Float9=~
    Float10=~
   )
   Output Animation Pose(Result=Apply Additive(
     Base=Additive Identity Pose()
     Additive=Sequence Player(Sequence=@n1.CharacterAnimSequence8) {bLoopAnimation=false}
     Alpha=Get StopDriftAlpha() -> (StopDriftAlpha:double)
    ) {alphaBoolBlend {blendOption=Linear}})
   Get ADS_25Alpha()
  ]
 }
 TacSprintLoop {
  path=AnimGraph.AnimGraphNode_StateMachine_0.FirstPerson.AnimStateNode_2.OnGround.AnimGraphNode_StateMachine_0.OnGroundSM.AnimStateNode_5.TacSprintLoop
  nodes [
   n1 = Get WeaponSpecificAnimData()
   Output Animation Pose(Result=Sequence Player(Sequence=Break Fort Weapon First Person Animation Data(FortWeaponFirstPersonAnimationData=@n1) -> (
      TacSprintExitEquipBlendTime=~
      Bool1=~
      Bool2=~
      Bool3=~
      Bool4=~
      Bool5=~
      Bool6=~
      Bool7=~
      Bool8=~
      Bool9=~
      Bool10=~
      Float1=~
      Float2=~
      Float3=~
      Float4=~
      Float5=~
      Float6=~
      Float7=~
      Float8=~
      Float9=~
      Float10=~
     ).CharacterAnimSequence19))
  ]
 }
 Idle {
  path=AnimGraph.AnimGraphNode_StateMachine_0.FirstPerson.AnimStateNode_2.OnGround.AnimGraphNode_StateMachine_0.OnGroundSM.AnimStateNode_6.Idle
  nodes [
   "Sequence Player (additive) 'Paprika_Idle_FP'"() {sequence=/Game/Animation/Game/MainPlayerFN/Animation/FP/Paprika/Paprika_Idle_FP}
   n1 = Break Fort Weapon First Person Animation Data(FortWeaponFirstPersonAnimationData=Get WeaponSpecificAnimData()) -> (
    TacSprintExitEquipBlendTime=~
    Bool1=~
    Bool2=~
    Bool3=~
    Bool4=~
    Bool5=~
    Bool6=~
    Bool7=~
    Bool8=~
    Bool9=~
    Bool10=~
    Float1=~
    Float2=~
    Float3=~
    Float4=~
    Float5=~
    Float6=~
    Float7=~
    Float8=~
    Float9=~
    Float10=~
   )
   n2 = Blend Poses by bool(
    BlendPose_0=Sequence Player(Sequence=@n1.CharacterAnimSequence13)
    BlendPose_1=Sequence Player(Sequence=@n1.CharacterAnimSequence12)
    BlendTime_0=0.2
    BlendTime_1=0.2
    bActiveValue=Property Access() -> (Value=~)
   ) {blendType=Linear}
   n3 = Additive Identity Pose()
   n4 = Sequence Player(Sequence=@n1.CharacterAnimSequence12)
   n5 = Sequence Player(Sequence=@n1.CharacterAnimSequence13)
   n6 = Property Access() -> (Value=~)
   Output Animation Pose(Result=Blend Poses by bool(
     BlendPose_0=Apply Additive(Base=@n3, Additive=@n2, Alpha=@n1.Float7) {alphaBoolBlend {blendOption=Linear}}
     BlendPose_1=Blend Poses by bool(BlendPose_0=@n5, BlendPose_1=@n4, BlendTime_0=0.2, BlendTime_1=0.2, bActiveValue=@n6) {blendType=Linear}
     BlendTime_0=0.2
     BlendTime_1=0.2
     bActiveValue=Property Access() -> (Value=~)
    ) {blendType=Linear})
  ]
 }
 Locomotion {
  path=AnimGraph.AnimGraphNode_StateMachine_0.FirstPerson.AnimStateNode_2.OnGround.AnimGraphNode_StateMachine_0.OnGroundSM.AnimStateNode_7.Locomotion
  nodes [
   AnimGraphNode_StateMachine_0 = TacSprint() {is=title AnimGraphNode_StateMachine}
   Blend Poses by bool(BlendPose_0=@AnimGraphNode_StateMachine_0, bActiveValue=Get bIsTacSprinting()) {blendType=Linear}
   n1 = Break Fort Weapon First Person Animation Data(FortWeaponFirstPersonAnimationData=Get WeaponSpecificAnimData()) -> (
    TacSprintExitEquipBlendTime=~
    Bool1=~
    Bool2=~
    Bool3=~
    Bool4=~
    Bool5=~
    Bool6=~
    Bool7=~
    Bool8=~
    Bool9=~
    Bool10=~
    Float1=~
    Float2=~
    Float3=~
    Float4=~
    Float5=~
    Float6=~
    Float7=~
    Float8=~
    Float9=~
    Float10=~
   )
   n2 = Apply Additive(
    Base=Additive Identity Pose()
    Additive=Blendspace Player!(X=Property Access() -> (Value:float=~), Y=Property Access() -> (Value=~), BlendSpace:Blend Space Object Reference=@n1.CharacterBlendSpace6) -> (Pose)
    Alpha=Get ADS_25Alpha()
   ) {alphaBoolBlend {blendOption=Linear}}
   n3 = Property Access() -> (Value=~)
   n4 = Property Access() -> (Value=~)
   n5 = Property Access() -> (Value=~)
   n6 = Property Access() -> (Value=~)
   n7 = Property Access() -> (Value=~)
   n8 = Property Access() -> (Value=~)
   n9 = Blend Poses by bool(
    BlendPose_0=Blend Poses by bool(
     BlendPose_0=Blendspace Player(X=@n6, Y=@n5, BlendSpace=@n1.CharacterBlendSpace1)
     BlendPose_1=Blendspace Player(X=@n8, Y=@n7, BlendSpace=@n1.CharacterBlendSpace2)
     BlendTime_0=0.2
     BlendTime_1=0.2
     bActiveValue=Get bJogToRunThreshold() -> (bJogToRunThreshold:bool)
    ) {blendType=Linear}
    BlendPose_1=Apply Additive(Base=Additive Identity Pose(), Additive=Blendspace Player(X=@n4, Y=@n3, BlendSpace=@n1.CharacterBlendSpace3), Alpha=Get ADS_25Alpha()) {alphaBoolBlend {blendOption=Linear}}
    BlendTime_0=0.2
    BlendTime_1=0.2
    bActiveValue=Get bWalkToJogThreshold() -> (bWalkToJogThreshold:bool)
   ) {blendType=Linear}
   n10 = Property Access() -> (Value:bool=~)
   Output Animation Pose(Result=Apply Additive(
     Base=Additive Identity Pose()
     Additive=Blend Poses by bool(BlendPose_0=@n2, BlendPose_1=@n9, BlendTime_0=0.2, BlendTime_1=0.2, bActiveValue=@n10) {blendType=Linear}
     Alpha=Get RunFwdAnimIntensity() -> (RunFwdAnimIntensity:double)
    ) {alphaBoolBlend {blendOption=Linear}})
   Property Access() -> (Value=~)
  ]
 }
 TacSprint {
  entry=TacSprintStart
  path=AnimGraph.AnimGraphNode_StateMachine_0.FirstPerson.AnimStateNode_2.OnGround.AnimGraphNode_StateMachine_0.OnGroundSM.AnimStateNode_7.Locomotion.AnimGraphNode_StateMachine_0.TacSprint
  ids=id
  nodes [
   AnimStateNode_0 = TacSprintStart()
   AnimStateNode_1 = TacSprintLoop()
   AnimStateTransitionNode_0 = AnimStateTransitionNode() {bAutomaticRuleBasedOnSequencePlayerInState=true, crossfadeDuration=0}
  ]
  links [{
    dst=TacSprintLoop
    rule=AnimGraph.AnimGraphNode_StateMachine_0.FirstPerson.AnimStateNode_2.OnGround.AnimGraphNode_StateMachine_0.OnGroundSM.AnimStateNode_7.Locomotion.AnimGraphNode_StateMachine_0.TacSprint.AnimStateTransitionNode_0.AnimationTransitionGraph_0
    src=TacSprintStart
    via=AnimStateTransitionNode_0
   }]
 }
 TacSprintStart {
  path=AnimGraph.AnimGraphNode_StateMachine_0.FirstPerson.AnimStateNode_2.OnGround.AnimGraphNode_StateMachine_0.OnGroundSM.AnimStateNode_7.Locomotion.AnimGraphNode_StateMachine_0.TacSprint.AnimStateNode_0.TacSprintStart
  nodes [
   n1 = Get WeaponSpecificAnimData()
   Output Animation Pose(Result=Sequence Player(Sequence=Break Fort Weapon First Person Animation Data(FortWeaponFirstPersonAnimationData=@n1) -> (
      TacSprintExitEquipBlendTime=~
      Bool1=~
      Bool2=~
      Bool3=~
      Bool4=~
      Bool5=~
      Bool6=~
      Bool7=~
      Bool8=~
      Bool9=~
      Bool10=~
      Float1=~
      Float2=~
      Float3=~
      Float4=~
      Float5=~
      Float6=~
      Float7=~
      Float8=~
      Float9=~
      Float10=~
     ).CharacterAnimSequence18) {bLoopAnimation=false})
  ]
 }
 TacSprintLoop {
  path=AnimGraph.AnimGraphNode_StateMachine_0.FirstPerson.AnimStateNode_2.OnGround.AnimGraphNode_StateMachine_0.OnGroundSM.AnimStateNode_7.Locomotion.AnimGraphNode_StateMachine_0.TacSprint.AnimStateNode_1.TacSprintLoop
  nodes [
   n1 = Get WeaponSpecificAnimData()
   Output Animation Pose(Result=Sequence Player(Sequence=Break Fort Weapon First Person Animation Data(FortWeaponFirstPersonAnimationData=@n1) -> (
      TacSprintExitEquipBlendTime=~
      Bool1=~
      Bool2=~
      Bool3=~
      Bool4=~
      Bool5=~
      Bool6=~
      Bool7=~
      Bool8=~
      Bool9=~
      Bool10=~
      Float1=~
      Float2=~
      Float3=~
      Float4=~
      Float5=~
      Float6=~
      Float7=~
      Float8=~
      Float9=~
      Float10=~
     ).CharacterAnimSequence19))
  ]
 }
 AnimationTransitionGraph_0 {
  path=AnimGraph.AnimGraphNode_StateMachine_0.FirstPerson.AnimStateNode_2.OnGround.AnimGraphNode_StateMachine_0.OnGroundSM.AnimStateNode_7.Locomotion.AnimGraphNode_StateMachine_0.TacSprint.AnimStateTransitionNode_0.AnimationTransitionGraph_0
  nodes [Result()]
 }
 AnimationTransitionGraph_1 {
  path=AnimGraph.AnimGraphNode_StateMachine_0.FirstPerson.AnimStateNode_2.OnGround.AnimGraphNode_StateMachine_0.OnGroundSM.AnimStateTransitionNode_0.AnimationTransitionGraph_1
  nodes [Result()]
 }
 Transition {
  path=AnimGraph.AnimGraphNode_StateMachine_0.FirstPerson.AnimStateNode_2.OnGround.AnimGraphNode_StateMachine_0.OnGroundSM.AnimStateTransitionNode_1.Transition
  nodes [Result(bCanEnterTransition=Get bIsTacSprinting())]
 }
 Transition {
  path=AnimGraph.AnimGraphNode_StateMachine_0.FirstPerson.AnimStateNode_2.OnGround.AnimGraphNode_StateMachine_0.OnGroundSM.AnimStateTransitionNode_10.Transition
  nodes [
   Result(bCanEnterTransition=NOT Boolean(A=Property Access() -> (Value=~)))
   NOT Boolean(A=Property Access() -> (Value=~))
  ]
 }
 Transition {
  path=AnimGraph.AnimGraphNode_StateMachine_0.FirstPerson.AnimStateNode_2.OnGround.AnimGraphNode_StateMachine_0.OnGroundSM.AnimStateTransitionNode_11.Transition
  nodes [Result(bCanEnterTransition=Property Access() -> (Value=~))]
 }
 Transition {
  path=AnimGraph.AnimGraphNode_StateMachine_0.FirstPerson.AnimStateNode_2.OnGround.AnimGraphNode_StateMachine_0.OnGroundSM.AnimStateTransitionNode_12.Transition
  nodes [Result(bCanEnterTransition=Get bIsTacSprinting())]
 }
 AnimationTransitionGraph_1 {
  path=AnimGraph.AnimGraphNode_StateMachine_0.FirstPerson.AnimStateNode_2.OnGround.AnimGraphNode_StateMachine_0.OnGroundSM.AnimStateTransitionNode_15.AnimationTransitionGraph_1
  nodes [Result(bCanEnterTransition=NOT Boolean(A=Get bIsTacSprinting()))]
 }
 AnimationTransitionGraph_1 {
  path=AnimGraph.AnimGraphNode_StateMachine_0.FirstPerson.AnimStateNode_2.OnGround.AnimGraphNode_StateMachine_0.OnGroundSM.AnimStateTransitionNode_16.AnimationTransitionGraph_1
  nodes [Result(bCanEnterTransition=NOT Boolean(A=Get bIsTacSprinting()))]
 }
 Transition {
  path=AnimGraph.AnimGraphNode_StateMachine_0.FirstPerson.AnimStateNode_2.OnGround.AnimGraphNode_StateMachine_0.OnGroundSM.AnimStateTransitionNode_18.Transition
  nodes [Result(bCanEnterTransition=Get bIsTacSprinting())]
 }
 Transition {
  path=AnimGraph.AnimGraphNode_StateMachine_0.FirstPerson.AnimStateNode_2.OnGround.AnimGraphNode_StateMachine_0.OnGroundSM.AnimStateTransitionNode_2.Transition
  nodes [Result(bCanEnterTransition=Property Access() -> (Value=~))]
 }
 Transition {
  path=AnimGraph.AnimGraphNode_StateMachine_0.FirstPerson.AnimStateNode_2.OnGround.AnimGraphNode_StateMachine_0.OnGroundSM.AnimStateTransitionNode_3.Transition
  nodes [Result(bCanEnterTransition=NOT Boolean(A=Property Access() -> (Value=~)))]
 }
 Transition {
  path=AnimGraph.AnimGraphNode_StateMachine_0.FirstPerson.AnimStateNode_2.OnGround.AnimGraphNode_StateMachine_0.OnGroundSM.AnimStateTransitionNode_4.Transition
  nodes [Result(bCanEnterTransition=Property Access() -> (Value=~))]
 }
 Transition {
  path=AnimGraph.AnimGraphNode_StateMachine_0.FirstPerson.AnimStateNode_2.OnGround.AnimGraphNode_StateMachine_0.OnGroundSM.AnimStateTransitionNode_5.Transition
  nodes [Result()]
 }
 Transition {
  path=AnimGraph.AnimGraphNode_StateMachine_0.FirstPerson.AnimStateNode_2.OnGround.AnimGraphNode_StateMachine_0.OnGroundSM.AnimStateTransitionNode_6.Transition
  nodes [
   Result(bCanEnterTransition=Property Access() -> (Value=~))
   NOT Boolean(A=Property Access() -> (Value=~))
  ]
 }
 Transition {
  path=AnimGraph.AnimGraphNode_StateMachine_0.FirstPerson.AnimStateNode_2.OnGround.AnimGraphNode_StateMachine_0.OnGroundSM.AnimStateTransitionNode_7.Transition
  nodes [Result()]
 }
 Transition {
  path=AnimGraph.AnimGraphNode_StateMachine_0.FirstPerson.AnimStateNode_2.OnGround.AnimGraphNode_StateMachine_0.OnGroundSM.AnimStateTransitionNode_8.Transition
  nodes [Result(bCanEnterTransition=NOT Boolean(A=Property Access() -> (Value=~)))]
 }
 Transition {
  path=AnimGraph.AnimGraphNode_StateMachine_0.FirstPerson.AnimStateNode_2.OnGround.AnimGraphNode_StateMachine_0.OnGroundSM.AnimStateTransitionNode_9.Transition
  nodes [Result()]
 }
 Crouch Into Exit {
  entry=Identity
  path=AnimGraph.AnimGraphNode_StateMachine_0.FirstPerson.AnimStateNode_2.OnGround.AnimGraphNode_StateMachine_1.Crouch Into Exit
  ids=id
  nodes [
   AnimStateNode_0 = Identity() {is=title AnimStateNode}
   AnimStateNode_1 = Crouch Into() {is=title AnimStateNode}
   AnimStateNode_2 = Crouch Exit() {is=title AnimStateNode}
   AnimStateTransitionNode_0 = AnimStateTransitionNode()
   AnimStateTransitionNode_2 = AnimStateTransitionNode()
   AnimStateTransitionNode_3 = AnimStateTransitionNode()
   AnimStateTransitionNode_4 = AnimStateTransitionNode() {bAutomaticRuleBasedOnSequencePlayerInState=true}
  ]
  links [
   {
    dst=Crouch Into
    rule=AnimGraph.AnimGraphNode_StateMachine_0.FirstPerson.AnimStateNode_2.OnGround.AnimGraphNode_StateMachine_1.Crouch Into Exit.AnimStateTransitionNode_3.Transition
    src=Crouch Exit
    via=AnimStateTransitionNode_3
   }
   {
    dst=Identity
    rule=AnimGraph.AnimGraphNode_StateMachine_0.FirstPerson.AnimStateNode_2.OnGround.AnimGraphNode_StateMachine_1.Crouch Into Exit.AnimStateTransitionNode_4.Transition
    src=Crouch Exit
    via=AnimStateTransitionNode_4
   }
   {
    dst=Crouch Exit
    rule=AnimGraph.AnimGraphNode_StateMachine_0.FirstPerson.AnimStateNode_2.OnGround.AnimGraphNode_StateMachine_1.Crouch Into Exit.AnimStateTransitionNode_2.Transition
    src=Crouch Into
    via=AnimStateTransitionNode_2
   }
   {
    dst=Crouch Into
    rule=AnimGraph.AnimGraphNode_StateMachine_0.FirstPerson.AnimStateNode_2.OnGround.AnimGraphNode_StateMachine_1.Crouch Into Exit.AnimStateTransitionNode_0.Transition
    src=Identity
    via=AnimStateTransitionNode_0
   }
  ]
 }
 Identity {
  path=AnimGraph.AnimGraphNode_StateMachine_0.FirstPerson.AnimStateNode_2.OnGround.AnimGraphNode_StateMachine_1.Crouch Into Exit.AnimStateNode_0.Identity
  nodes [Output Animation Pose(Result=Additive Identity Pose())]
 }
 Crouch Into {
  path=AnimGraph.AnimGraphNode_StateMachine_0.FirstPerson.AnimStateNode_2.OnGround.AnimGraphNode_StateMachine_1.Crouch Into Exit.AnimStateNode_1.Crouch Into
  nodes [
   n1 = Additive Identity Pose()
   n2 = Break Fort Weapon First Person Animation Data(FortWeaponFirstPersonAnimationData=Get WeaponSpecificAnimData()) -> (
    TacSprintExitEquipBlendTime=~
    Bool1=~
    Bool2=~
    Bool3=~
    Bool4=~
    Bool5=~
    Bool6=~
    Bool7=~
    Bool8=~
    Bool9=~
    Bool10=~
    Float1=~
    Float2=~
    Float3=~
    Float4=~
    Float5=~
    Float6=~
    Float7=~
    Float8=~
    Float9=~
    Float10=~
   )
   n3 = Sequence Player(Sequence=@n2.CharacterAnimSequence5) {bLoopAnimation=false}
   Output Animation Pose(Result=Apply Additive(
     Base=Additive Identity Pose()
     Additive=Apply Additive(Base=@n1, Additive=@n3, Alpha=@n2.Float5) {alphaBoolBlend {blendOption=Linear}}
     Alpha=Get ADS_10Alpha() -> (ADS_10Alpha:double)
    ) {alphaBoolBlend {blendOption=Linear}})
   Get ADS_25Alpha()
  ]
 }
 Crouch Exit {
  path=AnimGraph.AnimGraphNode_StateMachine_0.FirstPerson.AnimStateNode_2.OnGround.AnimGraphNode_StateMachine_1.Crouch Into Exit.AnimStateNode_2.Crouch Exit
  nodes [
   n1 = Additive Identity Pose()
   n2 = Break Fort Weapon First Person Animation Data(FortWeaponFirstPersonAnimationData=Get WeaponSpecificAnimData()) -> (
    TacSprintExitEquipBlendTime=~
    Bool1=~
    Bool2=~
    Bool3=~
    Bool4=~
    Bool5=~
    Bool6=~
    Bool7=~
    Bool8=~
    Bool9=~
    Bool10=~
    Float1=~
    Float2=~
    Float3=~
    Float4=~
    Float5=~
    Float6=~
    Float7=~
    Float8=~
    Float9=~
    Float10=~
   )
   n3 = Sequence Player(Sequence=@n2.CharacterAnimSequence6) {bLoopAnimation=false}
   Output Animation Pose(Result=Apply Additive(
     Base=Additive Identity Pose()
     Additive=Apply Additive(Base=@n1, Additive=@n3, Alpha=@n2.Float6) {alphaBoolBlend {blendOption=Linear}}
     Alpha=Get ADS_10Alpha()
    ) {alphaBoolBlend {blendOption=Linear}})
   Get ADS_25Alpha()
  ]
 }
 Transition {
  path=AnimGraph.AnimGraphNode_StateMachine_0.FirstPerson.AnimStateNode_2.OnGround.AnimGraphNode_StateMachine_1.Crouch Into Exit.AnimStateTransitionNode_0.Transition
  nodes [Result(bCanEnterTransition=Property Access() -> (Value=~))]
 }
 Transition {
  path=AnimGraph.AnimGraphNode_StateMachine_0.FirstPerson.AnimStateNode_2.OnGround.AnimGraphNode_StateMachine_1.Crouch Into Exit.AnimStateTransitionNode_2.Transition
  nodes [Result(bCanEnterTransition=NOT Boolean(A=Property Access() -> (Value=~)))]
 }
 Transition {
  path=AnimGraph.AnimGraphNode_StateMachine_0.FirstPerson.AnimStateNode_2.OnGround.AnimGraphNode_StateMachine_1.Crouch Into Exit.AnimStateTransitionNode_3.Transition
  nodes [Result(bCanEnterTransition=Property Access() -> (Value=~))]
 }
 Transition {
  path=AnimGraph.AnimGraphNode_StateMachine_0.FirstPerson.AnimStateNode_2.OnGround.AnimGraphNode_StateMachine_1.Crouch Into Exit.AnimStateTransitionNode_4.Transition
  nodes [Result()]
 }
 ADS {
  entry=Identity
  path=AnimGraph.AnimGraphNode_StateMachine_0.FirstPerson.AnimStateNode_2.OnGround.AnimGraphNode_StateMachine_3.ADS
  ids=id
  nodes [
   AnimStateNode_5 = Identity()
   AnimStateNode_6 = ADS_Outro() {is=title AnimStateNode}
   AnimStateNode_7 = ADS_Intro() {is=title AnimStateNode}
   AnimStateTransitionNode_0 = AnimStateTransitionNode()
   AnimStateTransitionNode_5 = AnimStateTransitionNode() {crossfadeDuration=0.1}
   AnimStateTransitionNode_7 = AnimStateTransitionNode() {crossfadeDuration=0.1}
   AnimStateTransitionNode_8 = AnimStateTransitionNode() {bAutomaticRuleBasedOnSequencePlayerInState=true, crossfadeDuration=0.1}
  ]
  links [
   {
    dst=ADS_Outro
    rule=AnimGraph.AnimGraphNode_StateMachine_0.FirstPerson.AnimStateNode_2.OnGround.AnimGraphNode_StateMachine_3.ADS.AnimStateTransitionNode_7.AnimationTransitionGraph_0
    src=ADS_Intro
    via=AnimStateTransitionNode_7
   }
   {
    dst=Identity
    rule=AnimGraph.AnimGraphNode_StateMachine_0.FirstPerson.AnimStateNode_2.OnGround.AnimGraphNode_StateMachine_3.ADS.AnimStateTransitionNode_0.AnimationTransitionGraph_0
    src=ADS_Outro
    via=AnimStateTransitionNode_0
   }
   {
    dst=Identity
    rule=AnimGraph.AnimGraphNode_StateMachine_0.FirstPerson.AnimStateNode_2.OnGround.AnimGraphNode_StateMachine_3.ADS.AnimStateTransitionNode_8.AnimationTransitionGraph_0
    src=ADS_Outro
    via=AnimStateTransitionNode_8
   }
   {
    dst=ADS_Intro
    rule=AnimGraph.AnimGraphNode_StateMachine_0.FirstPerson.AnimStateNode_2.OnGround.AnimGraphNode_StateMachine_3.ADS.AnimStateTransitionNode_5.AnimationTransitionGraph_0
    src=Identity
    via=AnimStateTransitionNode_5
   }
  ]
 }
 Identity {
  path=AnimGraph.AnimGraphNode_StateMachine_0.FirstPerson.AnimStateNode_2.OnGround.AnimGraphNode_StateMachine_3.ADS.AnimStateNode_5.Identity
  nodes [Output Animation Pose(Result=Additive Identity Pose())]
 }
 ADS_Outro {
  path=AnimGraph.AnimGraphNode_StateMachine_0.FirstPerson.AnimStateNode_2.OnGround.AnimGraphNode_StateMachine_3.ADS.AnimStateNode_6.ADS_Outro
  nodes [
   n1 = Get WeaponSpecificAnimData()
   Output Animation Pose(Result=Sequence Player(Sequence=Break Fort Weapon First Person Animation Data(FortWeaponFirstPersonAnimationData=@n1) -> (
      TacSprintExitEquipBlendTime=~
      Bool1=~
      Bool2=~
      Bool3=~
      Bool4=~
      Bool5=~
      Bool6=~
      Bool7=~
      Bool8=~
      Bool9=~
      Bool10=~
      Float1=~
      Float2=~
      Float3=~
      Float4=~
      Float5=~
      Float6=~
      Float7=~
      Float8=~
      Float9=~
      Float10=~
     ).CharacterAnimSequence2) {bLoopAnimation=false})
  ]
 }
 ADS_Intro {
  path=AnimGraph.AnimGraphNode_StateMachine_0.FirstPerson.AnimStateNode_2.OnGround.AnimGraphNode_StateMachine_3.ADS.AnimStateNode_7.ADS_Intro
  nodes [
   n1 = Get WeaponSpecificAnimData()
   Output Animation Pose(Result=Sequence Player(Sequence=Break Fort Weapon First Person Animation Data(FortWeaponFirstPersonAnimationData=@n1) -> (
      TacSprintExitEquipBlendTime=~
      Bool1=~
      Bool2=~
      Bool3=~
      Bool4=~
      Bool5=~
      Bool6=~
      Bool7=~
      Bool8=~
      Bool9=~
      Bool10=~
      Float1=~
      Float2=~
      Float3=~
      Float4=~
      Float5=~
      Float6=~
      Float7=~
      Float8=~
      Float9=~
      Float10=~
     ).CharacterAnimSequence1) {bLoopAnimation=false})
  ]
 }
 AnimationTransitionGraph_0 {
  path=AnimGraph.AnimGraphNode_StateMachine_0.FirstPerson.AnimStateNode_2.OnGround.AnimGraphNode_StateMachine_3.ADS.AnimStateTransitionNode_0.AnimationTransitionGraph_0
  nodes [Result(bCanEnterTransition=Property Access() -> (Value=~))]
 }
 AnimationTransitionGraph_0 {
  path=AnimGraph.AnimGraphNode_StateMachine_0.FirstPerson.AnimStateNode_2.OnGround.AnimGraphNode_StateMachine_3.ADS.AnimStateTransitionNode_5.AnimationTransitionGraph_0
  nodes [Result(bCanEnterTransition=Property Access() -> (Value=~))]
 }
 AnimationTransitionGraph_0 {
  path=AnimGraph.AnimGraphNode_StateMachine_0.FirstPerson.AnimStateNode_2.OnGround.AnimGraphNode_StateMachine_3.ADS.AnimStateTransitionNode_7.AnimationTransitionGraph_0
  nodes [Result(bCanEnterTransition=NOT Boolean(A=Property Access() -> (Value=~)))]
 }
 AnimationTransitionGraph_0 {
  path=AnimGraph.AnimGraphNode_StateMachine_0.FirstPerson.AnimStateNode_2.OnGround.AnimGraphNode_StateMachine_3.ADS.AnimStateTransitionNode_8.AnimationTransitionGraph_0
  nodes [Result()]
 }
 StartStopDrift {
  entry=Identity
  path=AnimGraph.AnimGraphNode_StateMachine_0.FirstPerson.AnimStateNode_2.OnGround.AnimGraphNode_StateMachine_4.StartStopDrift
  ids=id
  nodes [
   AnimStateNode_10 = Start_A() {is=title AnimStateNode}
   AnimStateNode_11 = Stop_A() {is=title AnimStateNode}
   AnimStateNode_3 = Start_B() {is=title AnimStateNode}
   AnimStateNode_4 = Stop_B() {is=title AnimStateNode}
   AnimStateNode_5 = Identity()
   AnimStateTransitionNode_10 = AnimStateTransitionNode() {crossfadeDuration=0.4}
   AnimStateTransitionNode_11 = AnimStateTransitionNode() {bAutomaticRuleBasedOnSequencePlayerInState=true}
   AnimStateTransitionNode_2 = AnimStateTransitionNode() {bAutomaticRuleBasedOnSequencePlayerInState=true}
   AnimStateTransitionNode_5 = AnimStateTransitionNode() {crossfadeDuration=0.25}
   AnimStateTransitionNode_7 = AnimStateTransitionNode() {crossfadeDuration=0.4}
   AnimStateTransitionNode_8 = AnimStateTransitionNode() {crossfadeDuration=0.4}
   AnimStateTransitionNode_9 = AnimStateTransitionNode() {crossfadeDuration=0.25}
  ]
  links [
   {
    dst=Start_A
    rule=AnimGraph.AnimGraphNode_StateMachine_0.FirstPerson.AnimStateNode_2.OnGround.AnimGraphNode_StateMachine_4.StartStopDrift.AnimStateTransitionNode_7.Transition
    src=Identity
    via=AnimStateTransitionNode_7
   }
   {
    dst=Stop_A
    rule=AnimGraph.AnimGraphNode_StateMachine_0.FirstPerson.AnimStateNode_2.OnGround.AnimGraphNode_StateMachine_4.StartStopDrift.AnimStateTransitionNode_5.AnimationTransitionGraph_0
    src=Start_A
    via=AnimStateTransitionNode_5
   }
   {
    dst=Stop_B
    rule=AnimGraph.AnimGraphNode_StateMachine_0.FirstPerson.AnimStateNode_2.OnGround.AnimGraphNode_StateMachine_4.StartStopDrift.AnimStateTransitionNode_9.Transition
    src=Start_B
    via=AnimStateTransitionNode_9
   }
   {
    dst=Identity
    rule=AnimGraph.AnimGraphNode_StateMachine_0.FirstPerson.AnimStateNode_2.OnGround.AnimGraphNode_StateMachine_4.StartStopDrift.AnimStateTransitionNode_2.AnimationTransitionGraph_0
    src=Stop_A
    via=AnimStateTransitionNode_2
   }
   {
    dst=Start_B
    rule=AnimGraph.AnimGraphNode_StateMachine_0.FirstPerson.AnimStateNode_2.OnGround.AnimGraphNode_StateMachine_4.StartStopDrift.AnimStateTransitionNode_8.Transition
    src=Stop_A
    via=AnimStateTransitionNode_8
   }
   {
    dst=Identity
    rule=AnimGraph.AnimGraphNode_StateMachine_0.FirstPerson.AnimStateNode_2.OnGround.AnimGraphNode_StateMachine_4.StartStopDrift.AnimStateTransitionNode_11.Transition
    src=Stop_B
    via=AnimStateTransitionNode_11
   }
   {
    dst=Start_A
    rule=AnimGraph.AnimGraphNode_StateMachine_0.FirstPerson.AnimStateNode_2.OnGround.AnimGraphNode_StateMachine_4.StartStopDrift.AnimStateTransitionNode_10.Transition
    src=Stop_B
    via=AnimStateTransitionNode_10
   }
  ]
 }
 Start_A {
  path=AnimGraph.AnimGraphNode_StateMachine_0.FirstPerson.AnimStateNode_2.OnGround.AnimGraphNode_StateMachine_4.StartStopDrift.AnimStateNode_10.Start_A
  nodes [
   n1 = Break Fort Weapon First Person Animation Data(FortWeaponFirstPersonAnimationData=Get WeaponSpecificAnimData()) -> (
    TacSprintExitEquipBlendTime=~
    Bool1=~
    Bool2=~
    Bool3=~
    Bool4=~
    Bool5=~
    Bool6=~
    Bool7=~
    Bool8=~
    Bool9=~
    Bool10=~
    Float1=~
    Float2=~
    Float3=~
    Float4=~
    Float5=~
    Float6=~
    Float7=~
    Float8=~
    Float9=~
    Float10=~
   )
   Sequence Player(Sequence=@n1.CharacterAnimSequence15) {bLoopAnimation=false}
   n2 = Get LocalVelocityYawAngle() -> (LocalVelocityYawAngle:float)
   Output Animation Pose(Result=Apply Additive(
     Base=Additive Identity Pose()
     Additive=Blendspace Player(X=@n2, Y=0, BlendSpace=@n1.CharacterBlendSpace4) {bLoop=false, bResetPlayTimeWhenBlendSpaceChanges=false}
     Alpha=Get StartDriftAlpha()
    ) {alphaBoolBlend {blendOption=Linear}}) {
    node {stateEntryFunction {functionName=OnStateEntry}}
    stateEntryFunction {bSelfContext=true, memberGuid=#1, memberName=StartStopDrift_StartA_OnStateEntry}
   }
  ]
 }
 Stop_A {
  path=AnimGraph.AnimGraphNode_StateMachine_0.FirstPerson.AnimStateNode_2.OnGround.AnimGraphNode_StateMachine_4.StartStopDrift.AnimStateNode_11.Stop_A
  nodes [
   Additive Identity Pose()
   n1 = Break Fort Weapon First Person Animation Data(FortWeaponFirstPersonAnimationData=Get WeaponSpecificAnimData()) -> (
    TacSprintExitEquipBlendTime=~
    Bool1=~
    Bool2=~
    Bool3=~
    Bool4=~
    Bool5=~
    Bool6=~
    Bool7=~
    Bool8=~
    Bool9=~
    Bool10=~
    Float1=~
    Float2=~
    Float3=~
    Float4=~
    Float5=~
    Float6=~
    Float7=~
    Float8=~
    Float9=~
    Float10=~
   )
   Sequence Player(Sequence=@n1.CharacterAnimSequence16) {bLoopAnimation=false}
   n2 = Get Local Velocity Yaw Angle On Entry() -> (Local Velocity Yaw Angle On Entry:float)
   Output Animation Pose(Result=Apply Additive(
     Base=Additive Identity Pose()
     Additive=Blendspace Player(X=@n2, Y=0, BlendSpace=@n1.CharacterBlendSpace5) {bLoop=false, bResetPlayTimeWhenBlendSpaceChanges=false}
     Alpha=Get Stop Drift Alpha On Entry() -> (Stop Drift Alpha On Entry:double)
    ) {alphaBoolBlend {blendOption=Linear}}) {
    node {stateEntryFunction {functionName=OnStateEntry}}
    stateEntryFunction {bSelfContext=true, memberGuid=#2, memberName=StartStopDrift_StopA_OnStateEntry}
   }
   Get StopDriftAlpha()
   Get LocalVelocityYawAngle()
  ]
 }
 Start_B {
  path=AnimGraph.AnimGraphNode_StateMachine_0.FirstPerson.AnimStateNode_2.OnGround.AnimGraphNode_StateMachine_4.StartStopDrift.AnimStateNode_3.Start_B
  nodes [
   n1 = Break Fort Weapon First Person Animation Data(FortWeaponFirstPersonAnimationData=Get WeaponSpecificAnimData()) -> (
    TacSprintExitEquipBlendTime=~
    Bool1=~
    Bool2=~
    Bool3=~
    Bool4=~
    Bool5=~
    Bool6=~
    Bool7=~
    Bool8=~
    Bool9=~
    Bool10=~
    Float1=~
    Float2=~
    Float3=~
    Float4=~
    Float5=~
    Float6=~
    Float7=~
    Float8=~
    Float9=~
    Float10=~
   )
   Sequence Player(Sequence=@n1.CharacterAnimSequence15) {bLoopAnimation=false}
   n2 = Get LocalVelocityYawAngle()
   Output Animation Pose(Result=Apply Additive(
     Base=Additive Identity Pose()
     Additive=Blendspace Player(X=@n2, Y=0, BlendSpace=@n1.CharacterBlendSpace4) {bLoop=false, bResetPlayTimeWhenBlendSpaceChanges=false}
     Alpha=Get StartDriftAlpha()
    ) {alphaBoolBlend {blendOption=Linear}}) {
    node {stateEntryFunction {functionName=OnStateEntry}}
    stateEntryFunction {bSelfContext=true, memberGuid=#3, memberName=StartStopDrift_StartB_OnStateEntry}
   }
  ]
 }
 Stop_B {
  path=AnimGraph.AnimGraphNode_StateMachine_0.FirstPerson.AnimStateNode_2.OnGround.AnimGraphNode_StateMachine_4.StartStopDrift.AnimStateNode_4.Stop_B
  nodes [
   Additive Identity Pose()
   n1 = Break Fort Weapon First Person Animation Data(FortWeaponFirstPersonAnimationData=Get WeaponSpecificAnimData()) -> (
    TacSprintExitEquipBlendTime=~
    Bool1=~
    Bool2=~
    Bool3=~
    Bool4=~
    Bool5=~
    Bool6=~
    Bool7=~
    Bool8=~
    Bool9=~
    Bool10=~
    Float1=~
    Float2=~
    Float3=~
    Float4=~
    Float5=~
    Float6=~
    Float7=~
    Float8=~
    Float9=~
    Float10=~
   )
   Sequence Player(Sequence=@n1.CharacterAnimSequence16) {bLoopAnimation=false}
   n2 = Get Local Velocity Yaw Angle On Entry()
   Output Animation Pose(Result=Apply Additive(
     Base=Additive Identity Pose()
     Additive=Blendspace Player(X=@n2, Y=0, BlendSpace=@n1.CharacterBlendSpace5) {bLoop=false, bResetPlayTimeWhenBlendSpaceChanges=false}
     Alpha=Get Stop Drift Alpha On Entry()
    ) {alphaBoolBlend {blendOption=Linear}}) {
    node {stateEntryFunction {functionName=OnStateEntry}}
    stateEntryFunction {bSelfContext=true, memberGuid=#4, memberName=StartStopDrift_StopB_OnStateEntry}
   }
   Get LocalVelocityYawAngle()
   Get StopDriftAlpha()
  ]
 }
 Identity {
  path=AnimGraph.AnimGraphNode_StateMachine_0.FirstPerson.AnimStateNode_2.OnGround.AnimGraphNode_StateMachine_4.StartStopDrift.AnimStateNode_5.Identity
  nodes [Output Animation Pose(Result=Additive Identity Pose())]
 }
 Transition {
  path=AnimGraph.AnimGraphNode_StateMachine_0.FirstPerson.AnimStateNode_2.OnGround.AnimGraphNode_StateMachine_4.StartStopDrift.AnimStateTransitionNode_10.Transition
  nodes [Result(bCanEnterTransition=Property Access() -> (Value=~))]
 }
 Transition {
  path=AnimGraph.AnimGraphNode_StateMachine_0.FirstPerson.AnimStateNode_2.OnGround.AnimGraphNode_StateMachine_4.StartStopDrift.AnimStateTransitionNode_11.Transition
  nodes [Result()]
 }
 AnimationTransitionGraph_0 {
  path=AnimGraph.AnimGraphNode_StateMachine_0.FirstPerson.AnimStateNode_2.OnGround.AnimGraphNode_StateMachine_4.StartStopDrift.AnimStateTransitionNode_2.AnimationTransitionGraph_0
  nodes [Result()]
 }
 AnimationTransitionGraph_0 {
  path=AnimGraph.AnimGraphNode_StateMachine_0.FirstPerson.AnimStateNode_2.OnGround.AnimGraphNode_StateMachine_4.StartStopDrift.AnimStateTransitionNode_5.AnimationTransitionGraph_0
  nodes [Result(bCanEnterTransition=NOT Boolean(A=Property Access() -> (Value=~)))]
 }
 Transition {
  path=AnimGraph.AnimGraphNode_StateMachine_0.FirstPerson.AnimStateNode_2.OnGround.AnimGraphNode_StateMachine_4.StartStopDrift.AnimStateTransitionNode_7.Transition
  nodes [Result(bCanEnterTransition=Property Access() -> (Value=~))]
 }
 Transition {
  path=AnimGraph.AnimGraphNode_StateMachine_0.FirstPerson.AnimStateNode_2.OnGround.AnimGraphNode_StateMachine_4.StartStopDrift.AnimStateTransitionNode_8.Transition
  nodes [Result(bCanEnterTransition=Property Access() -> (Value=~))]
 }
 Transition {
  path=AnimGraph.AnimGraphNode_StateMachine_0.FirstPerson.AnimStateNode_2.OnGround.AnimGraphNode_StateMachine_4.StartStopDrift.AnimStateTransitionNode_9.Transition
  nodes [Result(bCanEnterTransition=NOT Boolean(A=Property Access() -> (Value=~)))]
 }
 InAir {
  path=AnimGraph.AnimGraphNode_StateMachine_0.FirstPerson.AnimStateNode_3.InAir
  nodes [
   n1 = Get WeaponSpecificAnimData()
   Apply Additive(
    Additive=Sequence Player(Sequence=Break Fort Weapon First Person Animation Data(FortWeaponFirstPersonAnimationData=@n1) -> (
      TacSprintExitEquipBlendTime=~
      Bool1=~
      Bool2=~
      Bool3=~
      Bool4=~
      Bool5=~
      Bool6=~
      Bool7=~
      Bool8=~
      Bool9=~
      Bool10=~
      Float1=~
      Float2=~
      Float3=~
      Float4=~
      Float5=~
      Float6=~
      Float7=~
      Float8=~
      Float9=~
      Float10=~
     ).CharacterAnimSequence17)
    Alpha=Get TacSprintJumpPoseAlpha() -> (TacSprintJumpPoseAlpha:double)
   ) {alphaBoolBlend {blendOption=Linear}}
   AnimGraphNode_StateMachine_0 = InAirSM() {is=title AnimGraphNode_StateMachine, bAllowConduitEntryStates=true}
   Output Animation Pose(Result=@AnimGraphNode_StateMachine_0)
  ]
 }
 InAirSM {
  entry=Identity
  path=AnimGraph.AnimGraphNode_StateMachine_0.FirstPerson.AnimStateNode_3.InAir.AnimGraphNode_StateMachine_0.InAirSM
  ids=id
  nodes [
   AnimStateConduitNode_0 = IsTacSprintJump?() {is=title AnimStateConduitNode}
   AnimStateNode_0 = TacSprintJump() {is=title AnimStateNode}
   AnimStateNode_1 = Falling / InAir() {is=title AnimStateNode}
   AnimStateNode_2 = Identity()
   AnimStateNode_3 = Jump() {is=title AnimStateNode}
   AnimStateTransitionNode_0 = AnimStateTransitionNode()
   AnimStateTransitionNode_1 = AnimStateTransitionNode()
   AnimStateTransitionNode_2 = AnimStateTransitionNode() {bAutomaticRuleBasedOnSequencePlayerInState=true, crossfadeDuration=0.15}
   AnimStateTransitionNode_5 = AnimStateTransitionNode() {bAutomaticRuleBasedOnSequencePlayerInState=true}
   AnimStateTransitionNode_6 = AnimStateTransitionNode() {priorityOrder=2}
   AnimStateTransitionNode_7 = AnimStateTransitionNode()
  ]
  links [
   {
    dst=Falling / InAir
    rule=AnimGraph.AnimGraphNode_StateMachine_0.FirstPerson.AnimStateNode_3.InAir.AnimGraphNode_StateMachine_0.InAirSM.AnimStateTransitionNode_6.Transition
    src=Identity
    via=AnimStateTransitionNode_6
   }
   {
    dst=IsTacSprintJump?
    rule=AnimGraph.AnimGraphNode_StateMachine_0.FirstPerson.AnimStateNode_3.InAir.AnimGraphNode_StateMachine_0.InAirSM.AnimStateTransitionNode_7.Transition
    src=Identity
    via=AnimStateTransitionNode_7
   }
   {
    dst=Jump
    rule=AnimGraph.AnimGraphNode_StateMachine_0.FirstPerson.AnimStateNode_3.InAir.AnimGraphNode_StateMachine_0.InAirSM.AnimStateTransitionNode_0.Transition
    src=IsTacSprintJump?
    via=AnimStateTransitionNode_0
   }
   {
    dst=TacSprintJump
    rule=AnimGraph.AnimGraphNode_StateMachine_0.FirstPerson.AnimStateNode_3.InAir.AnimGraphNode_StateMachine_0.InAirSM.AnimStateTransitionNode_1.Transition
    src=IsTacSprintJump?
    via=AnimStateTransitionNode_1
   }
   {
    dst=Falling / InAir
    rule=AnimGraph.AnimGraphNode_StateMachine_0.FirstPerson.AnimStateNode_3.InAir.AnimGraphNode_StateMachine_0.InAirSM.AnimStateTransitionNode_5.Transition
    src=Jump
    via=AnimStateTransitionNode_5
   }
   {
    dst=Falling / InAir
    rule=AnimGraph.AnimGraphNode_StateMachine_0.FirstPerson.AnimStateNode_3.InAir.AnimGraphNode_StateMachine_0.InAirSM.AnimStateTransitionNode_2.Transition
    src=TacSprintJump
    via=AnimStateTransitionNode_2
   }
  ]
 }
 IsTacSprintJump? {
  path=AnimGraph.AnimGraphNode_StateMachine_0.FirstPerson.AnimStateNode_3.InAir.AnimGraphNode_StateMachine_0.InAirSM.AnimStateConduitNode_0.IsTacSprintJump?
  nodes [Result(bCanEnterTransition=true)]
 }
 TacSprintJump {
  path=AnimGraph.AnimGraphNode_StateMachine_0.FirstPerson.AnimStateNode_3.InAir.AnimGraphNode_StateMachine_0.InAirSM.AnimStateNode_0.TacSprintJump
  nodes [
   n1 = Get WeaponSpecificAnimData()
   Output Animation Pose(Result=Sequence Player(Sequence=Break Fort Weapon First Person Animation Data(FortWeaponFirstPersonAnimationData=@n1) -> (
      TacSprintExitEquipBlendTime=~
      Bool1=~
      Bool2=~
      Bool3=~
      Bool4=~
      Bool5=~
      Bool6=~
      Bool7=~
      Bool8=~
      Bool9=~
      Bool10=~
      Float1=~
      Float2=~
      Float3=~
      Float4=~
      Float5=~
      Float6=~
      Float7=~
      Float8=~
      Float9=~
      Float10=~
     ).CharacterAnimSequence21) {bLoopAnimation=false})
  ]
 }
 Falling / InAir {
  path=AnimGraph.AnimGraphNode_StateMachine_0.FirstPerson.AnimStateNode_3.InAir.AnimGraphNode_StateMachine_0.InAirSM.AnimStateNode_1.Falling / InAir
  nodes [
   n1 = Break Fort Weapon First Person Animation Data(FortWeaponFirstPersonAnimationData=Get WeaponSpecificAnimData()) -> (
    TacSprintExitEquipBlendTime=~
    Bool1=~
    Bool2=~
    Bool3=~
    Bool4=~
    Bool5=~
    Bool6=~
    Bool7=~
    Bool8=~
    Bool9=~
    Bool10=~
    Float1=~
    Float2=~
    Float3=~
    Float4=~
    Float5=~
    Float6=~
    Float7=~
    Float8=~
    Float9=~
    Float10=~
   )
   n2 = Sequence Player(Sequence=@n1.CharacterAnimSequence17)
   n3 = Sequence Player(Sequence=@n1.CharacterAnimSequence10)
   n4 = Get TacSprintJumpPoseAlpha()
   Output Animation Pose(Result=Blend Poses by bool(
     BlendPose_0=Sequence Player!(Sequence=@n1.CharacterAnimSequence24, bLoopAnimation:bool=True) -> (Pose)
     BlendPose_1=Two Way Blend(A=@n3, B=@n2, Alpha=@n4)
     bActiveValue=Get bIsTargeting()
    ))
  ]
 }
 Identity {
  path=AnimGraph.AnimGraphNode_StateMachine_0.FirstPerson.AnimStateNode_3.InAir.AnimGraphNode_StateMachine_0.InAirSM.AnimStateNode_2.Identity
  nodes [Output Animation Pose(Result=Additive Identity Pose())]
 }
 Jump {
  path=AnimGraph.AnimGraphNode_StateMachine_0.FirstPerson.AnimStateNode_3.InAir.AnimGraphNode_StateMachine_0.InAirSM.AnimStateNode_3.Jump
  nodes [
   n1 = Break Fort Weapon First Person Animation Data(FortWeaponFirstPersonAnimationData=Get WeaponSpecificAnimData()) -> (
    TacSprintExitEquipBlendTime=~
    Bool1=~
    Bool2=~
    Bool3=~
    Bool4=~
    Bool5=~
    Bool6=~
    Bool7=~
    Bool8=~
    Bool9=~
    Bool10=~
    Float1=~
    Float2=~
    Float3=~
    Float4=~
    Float5=~
    Float6=~
    Float7=~
    Float8=~
    Float9=~
    Float10=~
   )
   Output Animation Pose(Result=Blend Poses by bool(
     BlendPose_0=Sequence Player(Sequence=@n1.CharacterAnimSequence23, bLoopAnimation=False) {bLoopAnimation=false}
     BlendPose_1=Sequence Player!(Sequence=@n1.CharacterAnimSequence9) -> (Pose) {bLoopAnimation=false}
     bActiveValue=Get bIsTargeting()
    ))
  ]
 }
 Transition {
  path=AnimGraph.AnimGraphNode_StateMachine_0.FirstPerson.AnimStateNode_3.InAir.AnimGraphNode_StateMachine_0.InAirSM.AnimStateTransitionNode_0.Transition
  nodes [Result(bCanEnterTransition=NOT Boolean(A=Get bIsTacSprinting()))]
 }
 Transition {
  path=AnimGraph.AnimGraphNode_StateMachine_0.FirstPerson.AnimStateNode_3.InAir.AnimGraphNode_StateMachine_0.InAirSM.AnimStateTransitionNode_1.Transition
  nodes [Result(bCanEnterTransition=Get bIsTacSprinting())]
 }
 Transition {
  path=AnimGraph.AnimGraphNode_StateMachine_0.FirstPerson.AnimStateNode_3.InAir.AnimGraphNode_StateMachine_0.InAirSM.AnimStateTransitionNode_2.Transition
  nodes [Result()]
 }
 Transition {
  path=AnimGraph.AnimGraphNode_StateMachine_0.FirstPerson.AnimStateNode_3.InAir.AnimGraphNode_StateMachine_0.InAirSM.AnimStateTransitionNode_5.Transition
  nodes [Result()]
 }
 Transition {
  path=AnimGraph.AnimGraphNode_StateMachine_0.FirstPerson.AnimStateNode_3.InAir.AnimGraphNode_StateMachine_0.InAirSM.AnimStateTransitionNode_6.Transition
  nodes [
   Result(bCanEnterTransition=Property Access() -> (Value=~))
   NOT Boolean(A=Property Access() -> (Value=~))
  ]
 }
 Transition {
  path=AnimGraph.AnimGraphNode_StateMachine_0.FirstPerson.AnimStateNode_3.InAir.AnimGraphNode_StateMachine_0.InAirSM.AnimStateTransitionNode_7.Transition
  nodes [Result(bCanEnterTransition=Property Access() -> (Value=~))]
 }
 LandOnGround {
  path=AnimGraph.AnimGraphNode_StateMachine_0.FirstPerson.AnimStateNode_4.LandOnGround
  nodes [
   n1 = Break Fort Weapon First Person Animation Data(FortWeaponFirstPersonAnimationData=Get WeaponSpecificAnimData()) -> (
    TacSprintExitEquipBlendTime=~
    Bool1=~
    Bool2=~
    Bool3=~
    Bool4=~
    Bool5=~
    Bool6=~
    Bool7=~
    Bool8=~
    Bool9=~
    Bool10=~
    Float1=~
    Float2=~
    Float3=~
    Float4=~
    Float5=~
    Float6=~
    Float7=~
    Float8=~
    Float9=~
    Float10=~
   )
   n2 = Sequence Player(Sequence=@n1.CharacterAnimSequence17)
   n3 = Sequence Player(Sequence=@n1.CharacterAnimSequence11) {bLoopAnimation=false}
   n4 = Get TacSprintJumpPoseAlpha()
   Output Animation Pose(Result=Apply Additive(
     Base=Additive Identity Pose()
     Additive=Apply Additive(Base=@n3, Additive=@n2, Alpha=@n4) {alphaBoolBlend {blendOption=Linear}}
     Alpha=Get JumpLandAlpha() -> (JumpLandAlpha:double)
    ) {alphaBoolBlend {blendOption=Linear}})
  ]
 }
 AnimationTransitionGraph_0 {
  path=AnimGraph.AnimGraphNode_StateMachine_0.FirstPerson.AnimStateTransitionNode_0.AnimationTransitionGraph_0
  nodes [Result(bCanEnterTransition=Property Access() -> (Value=~))]
 }
 AnimationTransitionGraph_0 {
  path=AnimGraph.AnimGraphNode_StateMachine_0.FirstPerson.AnimStateTransitionNode_2.AnimationTransitionGraph_0
  nodes [
   Result(bCanEnterTransition=Property Access() -> (Value=~))
   NOT Boolean(A=Property Access() -> (Value=~))
  ]
 }
 AnimationTransitionGraph_1 {
  path=AnimGraph.AnimGraphNode_StateMachine_0.FirstPerson.AnimStateTransitionNode_3.AnimationTransitionGraph_1
  nodes [Result(bCanEnterTransition=NOT Boolean(A=Property Access() -> (Value=~)))]
 }
 AnimationTransitionGraph_0 {path=AnimGraph.AnimGraphNode_StateMachine_0.FirstPerson.AnimStateTransitionNode_4.AnimationTransitionGraph_0, nodes [Result()]}
 AnimationTransitionGraph_0 {
  path=AnimGraph.AnimGraphNode_StateMachine_0.FirstPerson.AnimStateTransitionNode_7.AnimationTransitionGraph_0
  nodes [Result(bCanEnterTransition=Property Access() -> (Value=~))]
 }
 FirstPerson_OLD {
  entry=Idle
  path=AnimGraph.AnimGraphNode_StateMachine_2.FirstPerson_OLD
  ids=id
  nodes [
   AnimStateNode_0 = Idle()
   AnimStateNode_1 = Locomotion()
   AnimStateNode_2 = NotOnGround() {is=title AnimStateNode}
   AnimStateTransitionNode_0 = AnimStateTransitionNode()
   AnimStateTransitionNode_1 = AnimStateTransitionNode()
   AnimStateTransitionNode_3 = AnimStateTransitionNode()
   AnimStateTransitionNode_4 = AnimStateTransitionNode()
  ]
  links [
   {
    dst=Locomotion
    rule=AnimGraph.AnimGraphNode_StateMachine_2.FirstPerson_OLD.AnimStateTransitionNode_0.Transition
    src=Idle
    via=AnimStateTransitionNode_0
   }
   {
    dst=Idle
    rule=AnimGraph.AnimGraphNode_StateMachine_2.FirstPerson_OLD.AnimStateTransitionNode_1.Transition
    src=Locomotion
    via=AnimStateTransitionNode_1
   }
   {
    dst=NotOnGround
    rule=AnimGraph.AnimGraphNode_StateMachine_2.FirstPerson_OLD.AnimStateTransitionNode_3.Transition
    src=Locomotion
    via=AnimStateTransitionNode_3
   }
   {
    dst=Locomotion
    rule=AnimGraph.AnimGraphNode_StateMachine_2.FirstPerson_OLD.AnimStateTransitionNode_4.Transition
    src=NotOnGround
    via=AnimStateTransitionNode_4
   }
  ]
 }
 Idle {
  path=AnimGraph.AnimGraphNode_StateMachine_2.FirstPerson_OLD.AnimStateNode_0.Idle
  nodes [Output Animation Pose(Result="Sequence Player (additive) 'Paprika_Idle_FP'"() {sequence=&Paprika_Idle_FP})]
 }
 Locomotion {
  path=AnimGraph.AnimGraphNode_StateMachine_2.FirstPerson_OLD.AnimStateNode_1.Locomotion
  nodes [
   n1 = Get LocalVelocityYawAngle()
   n2 = Get ClampedSpeed2D() -> (ClampedSpeed2D:float)
   Output Animation Pose(Result=Blend Poses by bool(
     BlendPose_0=Blendspace Player 'Paprika_Walk_BS_FP'(X=@n1, Y=@n2) {blendSpace=/Game/Animation/Game/MainPlayerFN/Animation/FP/Paprika/Paprika_Walk_BS_FP}
     BlendPose_1="Sequence Player (additive) 'Paprika_Noise_FP'"() {bLoopAnimation=false, sequence=/Game/Animation/Game/MainPlayerFN/Animation/FP/Paprika/Paprika_Noise_FP}
     bActiveValue=Get bUseProceduralLayering() -> (bUseProceduralLayering:bool)
    ) {blendType=Linear})
  ]
 }
 NotOnGround {path=AnimGraph.AnimGraphNode_StateMachine_2.FirstPerson_OLD.AnimStateNode_2.NotOnGround, nodes [Output Animation Pose()]}
 Transition {
  path=AnimGraph.AnimGraphNode_StateMachine_2.FirstPerson_OLD.AnimStateTransitionNode_0.Transition
  nodes [Result(bCanEnterTransition=Get bIsMovingAndNotInAction() -> (bIsMovingAndNotInAction:bool))]
 }
 Transition {
  path=AnimGraph.AnimGraphNode_StateMachine_2.FirstPerson_OLD.AnimStateTransitionNode_1.Transition
  nodes [Result(bCanEnterTransition=NOT Boolean(A=Get bIsMovingAndNotInAction()))]
 }
 Transition {
  path=AnimGraph.AnimGraphNode_StateMachine_2.FirstPerson_OLD.AnimStateTransitionNode_3.Transition
  nodes [Result(bCanEnterTransition=NOT Boolean(A=Property Access() -> (Value=~)))]
 }
 Transition {
  path=AnimGraph.AnimGraphNode_StateMachine_2.FirstPerson_OLD.AnimStateTransitionNode_4.Transition
  nodes [Result(bCanEnterTransition=Property Access() -> (Value=~))]
 }
 ADS {
  entry=Identity
  path=AnimGraph.AnimGraphNode_StateMachine_3.ADS
  ids=id
  nodes [
   AnimStateNode_5 = Identity()
   AnimStateNode_6 = ADS_Outro()
   AnimStateNode_7 = ADS_Intro()
   AnimStateTransitionNode_0 = AnimStateTransitionNode()
   AnimStateTransitionNode_5 = AnimStateTransitionNode() {crossfadeDuration=0.1}
   AnimStateTransitionNode_7 = AnimStateTransitionNode() {crossfadeDuration=0.1}
   AnimStateTransitionNode_8 = AnimStateTransitionNode() {bAutomaticRuleBasedOnSequencePlayerInState=true, crossfadeDuration=0.1}
  ]
  links [
   {
    dst=ADS_Outro
    rule=AnimGraph.AnimGraphNode_StateMachine_3.ADS.AnimStateTransitionNode_7.AnimationTransitionGraph_0
    src=ADS_Intro
    via=AnimStateTransitionNode_7
   }
   {
    dst=Identity
    rule=AnimGraph.AnimGraphNode_StateMachine_3.ADS.AnimStateTransitionNode_0.AnimationTransitionGraph_0
    src=ADS_Outro
    via=AnimStateTransitionNode_0
   }
   {
    dst=Identity
    rule=AnimGraph.AnimGraphNode_StateMachine_3.ADS.AnimStateTransitionNode_8.AnimationTransitionGraph_0
    src=ADS_Outro
    via=AnimStateTransitionNode_8
   }
   {
    dst=ADS_Intro
    rule=AnimGraph.AnimGraphNode_StateMachine_3.ADS.AnimStateTransitionNode_5.AnimationTransitionGraph_0
    src=Identity
    via=AnimStateTransitionNode_5
   }
  ]
 }
 Identity {path=AnimGraph.AnimGraphNode_StateMachine_3.ADS.AnimStateNode_5.Identity, nodes [Output Animation Pose(Result=Additive Identity Pose())]}
 ADS_Outro {
  path=AnimGraph.AnimGraphNode_StateMachine_3.ADS.AnimStateNode_6.ADS_Outro
  nodes [
   n1 = Get WeaponSpecificAnimData()
   Output Animation Pose(Result=Sequence Player(Sequence=Break Fort Weapon First Person Animation Data(FortWeaponFirstPersonAnimationData=@n1) -> (
      TacSprintExitEquipBlendTime=~
      Bool1=~
      Bool2=~
      Bool3=~
      Bool4=~
      Bool5=~
      Bool6=~
      Bool7=~
      Bool8=~
      Bool9=~
      Bool10=~
      Float1=~
      Float2=~
      Float3=~
      Float4=~
      Float5=~
      Float6=~
      Float7=~
      Float8=~
      Float9=~
      Float10=~
     ).CharacterAnimSequence2) {bLoopAnimation=false})
  ]
 }
 ADS_Intro {
  path=AnimGraph.AnimGraphNode_StateMachine_3.ADS.AnimStateNode_7.ADS_Intro
  nodes [
   n1 = Get WeaponSpecificAnimData()
   Output Animation Pose(Result=Sequence Player(Sequence=Break Fort Weapon First Person Animation Data(FortWeaponFirstPersonAnimationData=@n1) -> (
      TacSprintExitEquipBlendTime=~
      Bool1=~
      Bool2=~
      Bool3=~
      Bool4=~
      Bool5=~
      Bool6=~
      Bool7=~
      Bool8=~
      Bool9=~
      Bool10=~
      Float1=~
      Float2=~
      Float3=~
      Float4=~
      Float5=~
      Float6=~
      Float7=~
      Float8=~
      Float9=~
      Float10=~
     ).CharacterAnimSequence1) {bLoopAnimation=false})
  ]
 }
 AnimationTransitionGraph_0 {
  path=AnimGraph.AnimGraphNode_StateMachine_3.ADS.AnimStateTransitionNode_0.AnimationTransitionGraph_0
  nodes [Result(bCanEnterTransition=Property Access() -> (Value=~))]
 }
 AnimationTransitionGraph_0 {
  path=AnimGraph.AnimGraphNode_StateMachine_3.ADS.AnimStateTransitionNode_5.AnimationTransitionGraph_0
  nodes [Result(bCanEnterTransition=Property Access() -> (Value=~))]
 }
 AnimationTransitionGraph_0 {
  path=AnimGraph.AnimGraphNode_StateMachine_3.ADS.AnimStateTransitionNode_7.AnimationTransitionGraph_0
  nodes [Result(bCanEnterTransition=NOT Boolean(A=Property Access() -> (Value=~)))]
 }
 AnimationTransitionGraph_0 {path=AnimGraph.AnimGraphNode_StateMachine_3.ADS.AnimStateTransitionNode_8.AnimationTransitionGraph_0, nodes [Result()]}
 AdditiveLandingSM {
  entry=Passthrough
  path=AnimGraph.AnimGraphNode_StateMachine_4.AdditiveLandingSM
  ids=id
  nodes [
   AnimStateNode_0 = InAir Passthrough() {is=title AnimStateNode}
   AnimStateNode_1 = Passthrough() {
    is=title AnimStateNode
    stateEntered {guid=#5}
    stateFullyBlended {guid=#6}
    stateLeft {guid=#7}
   }
   AnimStateNode_4 = LandOnGround()
   AnimStateTransitionNode_10 = AnimStateTransitionNode() {bAutomaticRuleBasedOnSequencePlayerInState=true, crossfadeDuration=0.1}
   AnimStateTransitionNode_4 = AnimStateTransitionNode() {crossfadeDuration=0}
   AnimStateTransitionNode_5 = AnimStateTransitionNode() {crossfadeDuration=0.1}
  ]
  links [
   {
    dst=LandOnGround
    rule=AnimGraph.AnimGraphNode_StateMachine_4.AdditiveLandingSM.AnimStateTransitionNode_5.Transition
    src=InAir Passthrough
    via=AnimStateTransitionNode_5
   }
   {
    dst=Passthrough
    rule=AnimGraph.AnimGraphNode_StateMachine_4.AdditiveLandingSM.AnimStateTransitionNode_10.Transition
    src=LandOnGround
    via=AnimStateTransitionNode_10
   }
   {
    dst=InAir Passthrough
    rule=AnimGraph.AnimGraphNode_StateMachine_4.AdditiveLandingSM.AnimStateTransitionNode_4.Transition
    src=Passthrough
    via=AnimStateTransitionNode_4
   }
  ]
 }
 InAir Passthrough {
  path=AnimGraph.AnimGraphNode_StateMachine_4.AdditiveLandingSM.AnimStateNode_0.InAir Passthrough
  nodes [Output Animation Pose(Result=Additive Identity Pose())]
 }
 Passthrough {
  path=AnimGraph.AnimGraphNode_StateMachine_4.AdditiveLandingSM.AnimStateNode_1.Passthrough
  nodes [Output Animation Pose(Result=Additive Identity Pose())]
 }
 LandOnGround {
  path=AnimGraph.AnimGraphNode_StateMachine_4.AdditiveLandingSM.AnimStateNode_4.LandOnGround
  nodes [
   Apply Additive(Base=Additive Identity Pose(), Alpha=Get JumpLandAlpha()) {alphaBoolBlend {blendOption=Linear}}
   n1 = Break Fort Weapon First Person Animation Data(FortWeaponFirstPersonAnimationData=Get WeaponSpecificAnimData()) -> (
    TacSprintExitEquipBlendTime=~
    Bool1=~
    Bool2=~
    Bool3=~
    Bool4=~
    Bool5=~
    Bool6=~
    Bool7=~
    Bool8=~
    Bool9=~
    Bool10=~
    Float1=~
    Float2=~
    Float3=~
    Float4=~
    Float5=~
    Float6=~
    Float7=~
    Float8=~
    Float9=~
    Float10=~
   )
   Output Animation Pose(Result=Blend Poses by bool(
     BlendPose_0=Sequence Player!(Sequence=@n1.CharacterAnimSequence25, bLoopAnimation:bool=False) -> (Pose) {bLoopAnimation=false}
     BlendPose_1=Sequence Player!(Sequence=@n1.CharacterAnimSequence11) -> (Pose) {bLoopAnimation=false}
     bActiveValue=Get bIsTargeting()
    ))
  ]
 }
 Transition {path=AnimGraph.AnimGraphNode_StateMachine_4.AdditiveLandingSM.AnimStateTransitionNode_10.Transition, nodes [Result()]}
 Transition {
  path=AnimGraph.AnimGraphNode_StateMachine_4.AdditiveLandingSM.AnimStateTransitionNode_4.Transition
  nodes [Result(bCanEnterTransition=Property Access() -> (Value=~))]
 }
 Transition {
  path=AnimGraph.AnimGraphNode_StateMachine_4.AdditiveLandingSM.AnimStateTransitionNode_5.Transition
  nodes [Result(bCanEnterTransition=NOT Boolean(A=Property Access() -> (Value=~)))]
 }
 TacSprintOutroSM {
  entry=Not TacSprinting
  path=AnimGraph.AnimGraphNode_StateMachine_5.TacSprintOutroSM
  ids=id
  nodes [
   AnimStateNode_0 = TacSprintOutro() {is=title AnimStateNode}
   AnimStateNode_1 = Not TacSprinting() {is=title AnimStateNode}
   AnimStateNode_2 = TacSprinting() {is=title AnimStateNode}
   AnimStateNode_4 = Sliding()
   AnimStateTransitionNode_0 = AnimStateTransitionNode()
   AnimStateTransitionNode_1 = AnimStateTransitionNode()
   AnimStateTransitionNode_12 = AnimStateTransitionNode()
   AnimStateTransitionNode_3 = AnimStateTransitionNode()
   AnimStateTransitionNode_4 = AnimStateTransitionNode() {bAutomaticRuleBasedOnSequencePlayerInState=true}
   AnimStateTransitionNode_8 = AnimStateTransitionNode()
  ]
  links [
   {
    dst=Sliding
    rule=AnimGraph.AnimGraphNode_StateMachine_5.TacSprintOutroSM.AnimStateTransitionNode_1.Transition
    src=Not TacSprinting
    via=AnimStateTransitionNode_1
   }
   {
    dst=TacSprinting
    rule=AnimGraph.AnimGraphNode_StateMachine_5.TacSprintOutroSM.AnimStateTransitionNode_12.Transition
    src=Not TacSprinting
    via=AnimStateTransitionNode_12
   }
   {
    dst=Not TacSprinting
    rule=AnimGraph.AnimGraphNode_StateMachine_5.TacSprintOutroSM.AnimStateTransitionNode_3.Transition
    src=Sliding
    via=AnimStateTransitionNode_3
   }
   {
    dst=Not TacSprinting
    rule=AnimGraph.AnimGraphNode_StateMachine_5.TacSprintOutroSM.AnimStateTransitionNode_4.Transition
    src=TacSprintOutro
    via=AnimStateTransitionNode_4
   }
   {
    dst=TacSprintOutro
    rule=AnimGraph.AnimGraphNode_StateMachine_5.TacSprintOutroSM.AnimStateTransitionNode_0.Transition
    src=TacSprinting
    via=AnimStateTransitionNode_0
   }
   {
    dst=TacSprintOutro
    rule=AnimGraph.AnimGraphNode_StateMachine_5.TacSprintOutroSM.AnimStateTransitionNode_8.Transition
    src=TacSprinting
    via=AnimStateTransitionNode_8
   }
  ]
 }
 TacSprintOutro {
  path=AnimGraph.AnimGraphNode_StateMachine_5.TacSprintOutroSM.AnimStateNode_0.TacSprintOutro
  nodes [
   n1 = Get WeaponSpecificAnimData()
   Output Animation Pose(Result=Sequence Player(Sequence=Break Fort Weapon First Person Animation Data(FortWeaponFirstPersonAnimationData=@n1) -> (
      TacSprintExitEquipBlendTime=~
      Bool1=~
      Bool2=~
      Bool3=~
      Bool4=~
      Bool5=~
      Bool6=~
      Bool7=~
      Bool8=~
      Bool9=~
      Bool10=~
      Float1=~
      Float2=~
      Float3=~
      Float4=~
      Float5=~
      Float6=~
      Float7=~
      Float8=~
      Float9=~
      Float10=~
     ).CharacterAnimSequence16) {bLoopAnimation=false})
  ]
 }
 Not TacSprinting {
  path=AnimGraph.AnimGraphNode_StateMachine_5.TacSprintOutroSM.AnimStateNode_1.Not TacSprinting
  nodes [Output Animation Pose(Result=Additive Identity Pose())]
 }
 TacSprinting {
  path=AnimGraph.AnimGraphNode_StateMachine_5.TacSprintOutroSM.AnimStateNode_2.TacSprinting
  nodes [Output Animation Pose(Result=Additive Identity Pose())]
 }
 Sliding {
  path=AnimGraph.AnimGraphNode_StateMachine_5.TacSprintOutroSM.AnimStateNode_4.Sliding
  nodes [Output Animation Pose(Result=Additive Identity Pose())]
 }
 Transition {
  path=AnimGraph.AnimGraphNode_StateMachine_5.TacSprintOutroSM.AnimStateTransitionNode_0.Transition
  nodes [Result(bCanEnterTransition=Property Access() -> (Value=~))]
 }
 Transition {
  path=AnimGraph.AnimGraphNode_StateMachine_5.TacSprintOutroSM.AnimStateTransitionNode_1.Transition
  nodes [Result(bCanEnterTransition=Property Access() -> (Value=~))]
 }
 Transition {
  path=AnimGraph.AnimGraphNode_StateMachine_5.TacSprintOutroSM.AnimStateTransitionNode_12.Transition
  nodes [Result(bCanEnterTransition=Get bIsTacSprinting())]
 }
 Transition {
  path=AnimGraph.AnimGraphNode_StateMachine_5.TacSprintOutroSM.AnimStateTransitionNode_3.Transition
  nodes [Result(bCanEnterTransition=NOT Boolean(A=Property Access() -> (Value=~)))]
 }
 Transition {path=AnimGraph.AnimGraphNode_StateMachine_5.TacSprintOutroSM.AnimStateTransitionNode_4.Transition, nodes [Result()]}
 Transition {
  path=AnimGraph.AnimGraphNode_StateMachine_5.TacSprintOutroSM.AnimStateTransitionNode_8.Transition
  nodes [Result(bCanEnterTransition=NOT Boolean(A=Get bIsTacSprinting()))]
 }
 EventGraph {nodes [
   n1 = Event BlueprintUpdateAnimation()
   n2 = VectorSpringInterp!(
    execute
    Current:Vector=Get Vector Spring() -> (Vector Spring:Vector)
    Target:Vector=Property Access() -> (Value:Vector=~)
    SpringState:"Vector Spring State Structure (by ref)"=Make Vector Spring State!() -> (VectorSpringState:Vector Spring State Structure) {is=title K2Node_MakeStruct}
    Stiffness:float=75
    CriticalDampingFactor:float=0.6
    DeltaTime:float=@n1.DeltaTimeX
    Mass:float=1
    TargetVelocityAmount:float=0.75
    bClamp:bool=false
    MinValue:Vector="-1,-1,-1"
    MaxValue:Vector="1,1,1"
    bInitializeFromTarget:bool=false
   ) -> (then, ReturnValue:Vector)
   > n3 = Set Vector Spring(Vector Spring:Vector=@n2) -> (Output_Get:Vector)
   n4 = InverseTransformLocation!(T:"Transform (by ref)"=Property Access() -> (Value:"Transform (by ref)"), Location:Vector=@n3) -> (ReturnValue:Vector)
   > Set Vector Spring Local(Vector Spring Local:Vector=NegateVector!(A:Vector=@n4) -> (ReturnValue:Vector)) -> (Output_Get:Vector)
   > DrawDebugSphere!(
    execute
    Center:Vector=Get Vector Spring()
    Radius:float=10
    Segments:int=12
    LineColor:Linear Color Structure=(R=1,G=1,B=1,A=1)
    Duration:float=0
    Thickness:float=1
    DepthPriority:EDrawDebugSceneDepthPriorityGroup Enum=World
   ) -> (then)
   > PrintString!(execute, InString="To String (Vector)"!(InVec:Vector=@n4) -> (ReturnValue:String), bPrintToScreen, bPrintToLog, TextColor, Duration, Key) -> (then)
   Set DisableFirstPersonAddiitivesAlpha(execute=@n1.then, DisableFirstPersonAddiitivesAlpha:double=float - float(A=1, B=GetCurveValue(CurveName=DisableFirstPersonAdditives))) -> (Output_Get:double)
  ]}
 StartStopDrift_StartA_OnStateEntry {nodes [
   StartStopDrift_StartA_OnStateEntry!() -> (then, Context:"Anim Update Context Structure (by ref)", Node:"Anim Node Reference Structure (by ref)") {is=title K2Node_FunctionEntry}
   > Return Node()
  ]}
 StartStopDrift_StartB_OnStateEntry {nodes [
   StartStopDrift_StartB_OnStateEntry!() -> (then, Context:"Anim Update Context Structure (by ref)", Node:"Anim Node Reference Structure (by ref)") {is=title K2Node_FunctionEntry}
   > Return Node()
  ]}
 StartStopDrift_StopA_OnStateEntry {nodes [
   StartStopDrift_StopA_OnStateEntry!() -> (then, Context:"Anim Update Context Structure (by ref)", Node:"Anim Node Reference Structure (by ref)") {is=title K2Node_FunctionEntry}
   > Set Local Velocity Yaw Angle On Entry(Local Velocity Yaw Angle On Entry:float=Get LocalVelocityYawAngle()) -> (Output_Get:float)
   > Set Stop Drift Alpha On Entry(Stop Drift Alpha On Entry:double=Get StopDriftAlpha()) -> (Output_Get:double)
   > Return Node()
  ]}
 StartStopDrift_StopB_OnStateEntry {nodes [
   StartStopDrift_StopB_OnStateEntry!() -> (then, Context:"Anim Update Context Structure (by ref)", Node:"Anim Node Reference Structure (by ref)") {is=title K2Node_FunctionEntry}
   > Set Local Velocity Yaw Angle On Entry(Local Velocity Yaw Angle On Entry=Get LocalVelocityYawAngle())
   > Set Stop Drift Alpha On Entry(Stop Drift Alpha On Entry=Get StopDriftAlpha())
   > Return Node()
  ]}
]
variables [
 Accel Spring {type=number}
 DisableFirstPersonAddiitivesAlpha {type=number}
 Float Spring State {type=FloatSpringState, default {}}
 Local Velocity Yaw Angle On Entry {type=number}
 StartStopSecondary {type=boolean}
 Stop Drift Alpha On Entry {type=number}
 Vector Spring {type=Vector}
 Vector Spring Local {type=Vector}
 Vector Spring State {type=VectorSpringState, default {}}
]
