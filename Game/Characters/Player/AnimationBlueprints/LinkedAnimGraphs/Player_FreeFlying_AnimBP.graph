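# Linked animation Blueprint for player free flying in creative-tool fly mode, producing additive body poses and velocity-driven leans.
# A state machine plays a non-looping Start, then automatically inertializes into Loop when the sequence ends.
# Each state blends by weapon core animation between the current anim set's fly-mode additive and a jetpack assault-rifle additive, whose targeting branches share one sequence.
# A lean blendspace driven by local X/Y velocity is layered on additively.
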AnimBlueprint /Game/Characters/Player/AnimationBlueprints/LinkedAnimGraphs/Player_FreeFlying_AnimBP
target_skeleton=&Fortnite_M_Avg_Player_Skeleton
graphs [
 AnimGraph {nodes [
   AnimGraphNode_StateMachine_0 = Flying() {is=title AnimGraphNode_StateMachine}
   Output Pose(Result=@AnimGraphNode_StateMachine_0) {layerGroup=None}
  ]}
 Flying {
  entry=CreateToolFlyModeStart
  path=AnimGraph.AnimGraphNode_StateMachine_0.Flying
  ids=id
  nodes [
   AnimStateNode_11 = CreateToolFlyModeStart() {is=title AnimStateNode, stateEntered {guid=#1, notifyName=CreateToolFlyModeStart_Entered}}
   AnimStateNode_13 = CreateToolFlyModeLoop() {is=title AnimStateNode, stateEntered {guid=#2, notifyName=CreateToolFlyModeLoop_Entered}}
   AnimStateTransitionNode_44 = AnimStateTransitionNode() {bAutomaticRuleBasedOnSequencePlayerInState=true, logicType=TLT_Inertialization}
  ]
  links [{
    dst=CreateToolFlyModeLoop
    rule=AnimGraph.AnimGraphNode_StateMachine_0.Flying.AnimStateTransitionNode_44.AnimationTransitionGraph_0
    src=CreateToolFlyModeStart
    via=AnimStateTransitionNode_44
   }]
 }
 CreateToolFlyModeStart {
  path=AnimGraph.AnimGraphNode_StateMachine_0.Flying.AnimStateNode_11.CreateToolFlyModeStart
  nodes [
   n1 = Get CurrentAnimSet() -> (CurrentAnimSet:Fort Anim Input Player Anim Asset Structure)
   n2 = "Blend Poses (EFortWeaponCoreAnimation)"!(
    BlendPose_0:Pose Link Structure=Sequence Player!(Sequence:Anim Sequence Base Object Reference=Break Fort Anim Input Player Anim Asset!(FortAnimInput_PlayerAnimAsset:"Fort Anim Input Player Anim Asset Structure (by ref)"=@n1) -> (
      FlyModeStartAdditive:Anim Sequence Object Reference
      FullBodyJumpAdditive:Anim Sequence Object Reference
      LandingAdditiveOverride:Anim Sequence Object Reference
      TurnInPlaceLeftAdditiveOverride:Anim Sequence Object Reference
      TurnInPlaceRightAdditiveOverride:Anim Sequence Object Reference
      bForceKneeSliding:bool=~
     ) {is=title K2Node_BreakStruct}) -> (Pose) {bLoopAnimation=false}
    BlendPose_1:Pose Link Structure=Blend Poses by bool(
     BlendPose_0="Sequence Player (additive) 'JetpackUp_ARDS_Additive'"() {bLoopAnimation=false, sequence=/Game/Animation/Game/MainPlayer/Locomotion/Medium/Male/Jump/AssaultRifle/JetpackUp_ARDS_Additive}
     BlendPose_1="Sequence Player (additive) 'JetpackUp_ARDS_Additive'"() {bLoopAnimation=false, sequence=&JetpackUp_ARDS_Additive}
     BlendTime_1=Get ADSToNonADSBlendTime() -> (ADSToNonADSBlendTime:float)
     bActiveValue=Get bIsTargeting()
    ) {blendType=Linear}
    BlendTime_0:float=0
    BlendTime_1:float=0
    ActiveEnumValue:Byte=Get WeaponCoreAnim!() -> (WeaponCoreAnim)
   ) -> (Pose:Pose Link Structure) {
    is=title AnimGraphNode_BlendListByEnum
    blendType=Linear
    blendPose [{linkId=-1, sourceLinkId=-1}, {linkId=-1, sourceLinkId=-1}]
    blendTime [0.1,0.1]
   }
   n3 = BreakVector(InVec=Get LocalVelocity() -> (LocalVelocity:Vector))
   n4 = Blendspace Player 'CreativeTool_Leans_CMM_BS'(X=@n3.Y, Y=@n3) {bIgnoreForRelevancyTest=true, blendSpace=/Game/Animation/Game/MainPlayer/Locomotion/Medium/Male/Jump/CreativeTool/CreativeTool_Leans_CMM_BS}
   Output Animation Pose(Result=Apply Additive(Base=@n2, Additive=@n4) {lODThreshold=1, alphaBoolBlend {blendOption=Linear}})
  ]
 }
 CreateToolFlyModeLoop {
  path=AnimGraph.AnimGraphNode_StateMachine_0.Flying.AnimStateNode_13.CreateToolFlyModeLoop
  nodes [
   n1 = Get CurrentAnimSet()
   n2 = "Blend Poses (EFortWeaponCoreAnimation)"(
    BlendPose_0=Sequence Player(Sequence=Break Fort Anim Input Player Anim Asset!(FortAnimInput_PlayerAnimAsset=@n1) -> (
      FlyModeLoopAdditive:Anim Sequence Object Reference
      FullBodyJumpAdditive
      LandingAdditiveOverride
      TurnInPlaceLeftAdditiveOverride
      TurnInPlaceRightAdditiveOverride
      bForceKneeSliding=~
     ))
    BlendPose_1=Blend Poses by bool(
     BlendPose_0="Sequence Player (additive) 'JetpackApex_ARDS_Loop_Additive'"() {sequence=/Game/Animation/Game/MainPlayer/Locomotion/Medium/Male/Jump/AssaultRifle/JetpackApex_ARDS_Loop_Additive}
     BlendPose_1="Sequence Player (additive) 'JetpackApex_ARDS_Loop_Additive'"() {sequence=&JetpackApex_ARDS_Loop_Additive}
     BlendTime_1=Get ADSToNonADSBlendTime()
     bActiveValue=Get bIsTargeting()
    ) {blendType=Linear}
    BlendTime_0=0
    BlendTime_1=0
    ActiveEnumValue=Get WeaponCoreAnim()
   ) {blendType=Linear, blendPose [{linkId=-1, sourceLinkId=-1}, {linkId=-1, sourceLinkId=-1}], blendTime [0.1,0.1]}
   n3 = BreakVector(InVec=Get LocalVelocity())
   Output Animation Pose(Result=Apply Additive(
     Base=@n2
     Additive=Blendspace Player 'CreativeTool_Leans_CMM_BS'(X=@n3.Y, Y=@n3) {blendSpace=&CreativeTool_Leans_CMM_BS}
    ) {lODThreshold=1, alphaBoolBlend {blendOption=Linear}})
  ]
 }
 AnimationTransitionGraph_0 {path=AnimGraph.AnimGraphNode_StateMachine_0.Flying.AnimStateTransitionNode_44.AnimationTransitionGraph_0, nodes [Result()]}
 EventGraph {links [], nodes []}
]
variables []
