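# Linked player anim Blueprint for goop movement, built on jetpack animations with a start-and-loop state machine that re-enters start when a goop jump executes.
# Both states apply velocity-driven lean onto item layer poses; the loop switches between goop lean with a half-weight slide additive and jetpack lean via bIsInGoop.
# The two jetpack layers blend by weapon core animation between current anim set sequences and assault rifle additives; other interface layers hold no logic.
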AnimBlueprint /Game/Characters/Player/AnimationBlueprints/LinkedAnimGraphs/Player_Goop_AnimBP
target_skeleton=&Fortnite_M_Avg_Player_Skeleton
graphs [
 AnimGraph {nodes [
   AnimGraphNode_StateMachine_0 = New State Machine() {is=title AnimGraphNode_StateMachine}
   Output Pose(Result=@AnimGraphNode_StateMachine_0) {layerGroup=None}
   n1 = ItemAnimLayerInterface - ItemJetPackJumpAdditive!() -> (Pose:Pose Link Structure) {
    is=title AnimGraphNode_LinkedAnimLayer
    interface=/Game/Characters/Player/Male/Male_Avg_Base/AnimLayerInterface/ItemAnimLayerInterface.ItemAnimLayerInterface_C
    layer=ItemJetPackJumpAdditive
   }
   Save cached pose 'JetpackJump_Pose'(Pose=@n1) {cacheName=JetpackJump_Pose}
  ]}
 New State Machine {
  entry=JetPackStart
  path=AnimGraph.AnimGraphNode_StateMachine_0.New State Machine
  ids=id
  nodes [
   AnimStateNode_0 = JetPackLoop() {is=title AnimStateNode}
   AnimStateNode_3 = JetPackStart() {
    is=title AnimStateNode
    stateEntered {guid=#1}
    stateFullyBlended {guid=#2}
    stateLeft {guid=#3}
   }
   AnimStateTransitionNode_0 = AnimStateTransitionNode() {bAutomaticRuleBasedOnSequencePlayerInState=true, crossfadeDuration=0.5}
   AnimStateTransitionNode_1 = AnimStateTransitionNode() {priorityOrder=2}
  ]
  links [
   {
    dst=JetPackStart
    rule=AnimGraph.AnimGraphNode_StateMachine_0.New State Machine.AnimStateTransitionNode_1.Transition
    src=JetPackLoop
    via=AnimStateTransitionNode_1
   }
   {
    dst=JetPackLoop
    rule=AnimGraph.AnimGraphNode_StateMachine_0.New State Machine.AnimStateTransitionNode_0.Transition
    src=JetPackStart
    via=AnimStateTransitionNode_0
   }
  ]
 }
 JetPackLoop {
  path=AnimGraph.AnimGraphNode_StateMachine_0.New State Machine.AnimStateNode_0.JetPackLoop
  nodes [
   n1 = BreakVector(InVec=Get LocalVelocity() -> (LocalVelocity:Vector))
   n2 = Apply Additive(
    Base=Blendspace Player 'Goop_ARDS_Lean'(X=@n1.Y, Y=@n1) {blendSpace=/Game/Animation/Game/MainPlayer/Locomotion/Medium/Male/Goop/AssaultRifle/Goop_ARDS_Lean}
    Additive="Sequence Player (additive) 'Slide_Center_CMM_Additive'"() {playRate=0.8, sequence=/Game/Animation/Game/MainPlayer/Locomotion/Medium/Male/Sliding/Poses/Slide_Center_CMM_Additive}
    Alpha=0.5
   ) {alphaBoolBlend {blendOption=Linear}}
   n3 = Blendspace Player 'Jetpack_ARDS_Lean'(X=@n1.Y, Y=@n1) {blendSpace=/Game/Animation/Game/MainPlayer/Locomotion/Medium/Male/Jump/AssaultRifle/Jetpack_ARDS_Lean}
   n4 = Get bIsInGoop() -> (bIsInGoop:bool)
   Output Animation Pose(Result=Apply Additive(
     Base=Use cached pose 'JetpackJump_Pose'()
     Additive=Blend Poses by bool(BlendPose_0=@n2, BlendPose_1=@n3, bActiveValue=@n4) {blendType=Linear}
    ) {lODThreshold=1, alphaBoolBlend {blendOption=Linear}})
  ]
 }
 JetPackStart {
  path=AnimGraph.AnimGraphNode_StateMachine_0.New State Machine.AnimStateNode_3.JetPackStart
  nodes [
   n1 = BreakVector(InVec=Get LocalVelocity())
   n2 = ItemAnimLayerInterface - ItemJetPackStartAdditive!() -> (Pose:Pose Link Structure) {
    is=title AnimGraphNode_LinkedAnimLayer
    interface=&ItemAnimLayerInterface_C
    layer=ItemJetPackStartAdditive
   }
   Output Animation Pose(Result=Apply Additive(
     Base=@n2
     Additive=Blendspace Player 'Jetpack_ARDS_Lean'(X=@n1.Y, Y=@n1) {bIgnoreForRelevancyTest=true, blendSpace=&Jetpack_ARDS_Lean}
    ) {lODThreshold=1, alphaBoolBlend {blendOption=Linear}})
  ]
 }
 Transition {path=AnimGraph.AnimGraphNode_StateMachine_0.New State Machine.AnimStateTransitionNode_0.Transition, nodes [Result()]}
 Transition {
  path=AnimGraph.AnimGraphNode_StateMachine_0.New State Machine.AnimStateTransitionNode_1.Transition
  nodes [Result(bCanEnterTransition=Get bGoopJumpExecuted() -> (bGoopJumpExecuted:bool))]
 }
 EventGraph {links [], nodes []}
 ItemAimOffset {nodes [
   Input Pose!() -> (Pose, AimOffsetAlpha:double, Yaw:double, Pitch:double) {name=InPose_AimOffset}
   Input Pose!() -> (Pose) {name=InPose_PreAimOffsets}
   Output Pose() {layerGroup=Item Override}
  ]}
 ItemFallAdditive {nodes [
   Input Pose() {name=InPoseItemFallAdditive}
   Output Pose() {layerGroup=Item Override}
  ]}
 ItemFallLandAdditive {nodes [
   Input Pose() {name=InPoseItemFallLandAdditive}
   Output Pose() {layerGroup=Item Override}
  ]}
 ItemFallLevel2Override {nodes [
   Input Pose() {name=InPose_FallLevel2}
   Output Pose() {layerGroup=Item Override}
  ]}
 ItemFinalPoseOverride {nodes [
   Input Pose() {name=InFinalPose}
   Output Pose() {layerGroup=Item Override}
  ]}
 ItemFullBodyAimOffsets {nodes [
   Input Pose!() -> (Pose, FullBody_AimOffsetAlpha:double, Aim_Yaw:double, Aim_Pitch:double) {name=InPoseFullBodyAimOffsets}
   Input Pose!() -> (Pose) {name=InPosePreFullBodyAimOffsets}
   Output Pose() {layerGroup=Item Override}
  ]}
 ItemFullBodyOverride {nodes [
   Input Pose() {name=InPoseFullBody}
   Output Pose() {layerGroup=Item Override}
  ]}
 ItemIdleOverride {nodes [
   Input Pose() {name=InPose_Idle}
   Output Pose() {layerGroup=Item Override}
  ]}
 ItemJetPackJumpAdditive {nodes [
   n1 = Get CurrentAnimSet() -> (CurrentAnimSet:Fort Anim Input Player Anim Asset Structure)
   n2 = "Blend Poses (EFortWeaponCoreAnimation)"!(
    BlendPose_0:Pose Link Structure=Sequence Player!(Sequence:Anim Sequence Base Object Reference=Break Fort Anim Input Player Anim Asset!(FortAnimInput_PlayerAnimAsset:"Fort Anim Input Player Anim Asset Structure (by ref)"=@n1) -> (
      JetPackJumpAdditve:Anim Sequence Object Reference
      FullBodyJumpAdditive:Anim Sequence Object Reference
      LandingAdditiveOverride:Anim Sequence Object Reference
      TurnInPlaceLeftAdditiveOverride:Anim Sequence Object Reference
      TurnInPlaceRightAdditiveOverride:Anim Sequence Object Reference
      bForceKneeSliding:bool=~
     ) {is=title K2Node_BreakStruct}) -> (Pose) {bLoopAnimation=false}
    BlendPose_1:Pose Link Structure=Blend Poses by bool(
     BlendPose_0="Sequence Player (additive) 'JetpackApex_ARDS_Loop_Additive'"() {sequence=/Game/Animation/Game/MainPlayer/Locomotion/Medium/Male/Jump/AssaultRifle/JetpackApex_ARDS_Loop_Additive}
     BlendPose_1="Sequence Player (additive) 'JetpackApex_ARDS_Loop_Additive'"() {sequence=&JetpackApex_ARDS_Loop_Additive}
     BlendTime_1=Get ADSToNonADSBlendTime() -> (ADSToNonADSBlendTime:float)
     bActiveValue=Get bIsTargeting()
    ) {blendType=Linear}
    BlendTime_0:float=0
    BlendTime_1:float=0
    ActiveEnumValue:Byte=Get WeaponCoreAnim!() -> (WeaponCoreAnim)
   ) -> (Pose:Pose Link Structure) {
    is=title AnimGraphNode_BlendListByEnum
    blendType=Linear
    blendPose [{linkId=-1, sourceLinkId=-1}, {linkId=-1, sourceLinkId=-1}]
    blendTime [0.1,0.1]
   }
   Output Pose(Result=@n2) {layerGroup=Item Override}
  ]}
 ItemJetPackStartAdditive {nodes [
   n1 = Get CurrentAnimSet()
   n2 = "Blend Poses (EFortWeaponCoreAnimation)"(
    BlendPose_0=Sequence Player(Sequence=Break Fort Anim Input Player Anim Asset!(FortAnimInput_PlayerAnimAsset=@n1) -> (
      JetPackStartAdditve:Anim Sequence Object Reference
      FullBodyJumpAdditive
      LandingAdditiveOverride
      TurnInPlaceLeftAdditiveOverride
      TurnInPlaceRightAdditiveOverride
      bForceKneeSliding=~
     )) {bLoopAnimation=false}
    BlendPose_1=Blend Poses by bool(
     BlendPose_0="Sequence Player (additive) 'JetpackUp_ARDS_Additive'"() {bLoopAnimation=false, sequence=/Game/Animation/Game/MainPlayer/Locomotion/Medium/Male/Jump/AssaultRifle/JetpackUp_ARDS_Additive}
     BlendPose_1="Sequence Player (additive) 'JetpackUp_ARDS_Additive'"() {bLoopAnimation=false, sequence=&JetpackUp_ARDS_Additive}
     BlendTime_1=Get ADSToNonADSBlendTime()
     bActiveValue=Get bIsTargeting()
    ) {blendType=Linear}
    BlendTime_0=0
    BlendTime_1=0
    ActiveEnumValue=Get WeaponCoreAnim()
   ) {blendType=Linear, blendPose [{linkId=-1, sourceLinkId=-1}, {linkId=-1, sourceLinkId=-1}], blendTime [0.1,0.1]}
   Output Pose(Result=@n2) {layerGroup=Item Override}
  ]}
 ItemJumpLoopAdditive {nodes [
   Input Pose() {name=InPoseItemJumpLoopAdditive}
   Output Pose() {layerGroup=Item Override}
  ]}
 ItemJumpOverride {nodes [
   Input Pose() {name=InPoseJump}
   Output Pose() {layerGroup=Item Override}
  ]}
 ItemJumpUpAdditive {nodes [
   Input Pose() {name=InPoseItemJumpUpAdditive}
   Output Pose() {layerGroup=Item Override}
  ]}
 ItemPreAimOffsets {nodes [
   Input Pose() {name=InPosePreAimOffsets}
   Output Pose() {layerGroup=Item Override}
  ]}
 ItemPreIK {nodes [
   Input Pose() {name=InPose_PreIK}
   Output Pose() {layerGroup=Item Override}
  ]}
 ItemRightHandIK {nodes [
   Input Pose() {name=InPosePreRightHandIK}
   Input Pose() {name=InPosePostRightHandIK}
   Output Pose() {layerGroup=Item Override}
  ]}
 ItemSkeletalControl {nodes [
   Input Pose()
   Output Pose() {layerGroup=Item Override}
  ]}
 ItemSwimJumpFallLoop {nodes [Output Pose() {layerGroup=Item Override}]}
 ItemSwimJumpSurfaceEnd {nodes [Output Pose() {layerGroup=Item Override}]}
 ItemSwimJumpSurfaceLoop {nodes [Output Pose() {layerGroup=Item Override}]}
 ItemUpperBody {nodes [
   Input Pose!() -> (Pose, InputParam:Fort Anim Input Adjusted Aim Structure) {name=InPoseUpperBody}
   Output Pose() {layerGroup=Item Override}
  ]}
 ItemZipLineStartAdditive {nodes [Output Pose() {layerGroup=Item Override}]}
]
variables []
