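# Linked anim graph for the player's older in-air animation, layering jump, fall and predicted-landing motion onto an incoming SourcePose.
# Two matching state machines drive item-overridable additives, with Jackal vehicle combat blends, and lower-body base poses merged per bone.
# A velocity-yaw lean follows; jetpack, goop or a disable flag bypasses all this to SourcePose.
# The anim set's full-body jump additive and inertialization finish the pose.
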AnimBlueprint /Game/Characters/Player/AnimationBlueprints/LinkedAnimGraphs/Player_OldInAir_AnimBP
target_skeleton=&Fortnite_M_Avg_Player_Skeleton
graphs [
 AnimGraph {nodes [
   n1 = Get LocalVelocityYawAngle() -> (LocalVelocityYawAngle:float)
   AnimGraphNode_StateMachine_1 = GroundJumpStateMachine() {is=title AnimGraphNode_StateMachine, bAllowConduitEntryStates=true}
   n2 = Layered blend per bone(
    BasePose=@AnimGraphNode_StateMachine_1
    BlendPoses_0=Additive Identity Pose()
    BlendWeights_0=Get JumpUpperBodyBlendWeight() -> (JumpUpperBodyBlendWeight:float)
   ) {lODThreshold=2, layerSetup [{branchFilters [{blendDepth=2, boneName=Spine_01}, {boneName=Ik_hand_root}]}]}
   n3 = Layered blend per bone(
    BasePose=@n2
    BlendPoses_0=Additive Identity Pose()
    BlendWeights_0=Get DisableUpperBodyAdditiveMeshSpaceCurveAlpha() -> (DisableUpperBodyAdditiveMeshSpaceCurveAlpha:float)
   ) {bMeshSpaceRotationBlend=true, lODThreshold=2, layerSetup [{branchFilters [{blendDepth=4, boneName=spine_01}, {boneName=ik_hand_root}]}]}
   AnimGraphNode_StateMachine_2 = GroundJumpLowerBody() {is=title AnimGraphNode_StateMachine, bAllowConduitEntryStates=true}
   n4 = Layered blend per bone(
    BasePose=Apply Additive(Base=Use cached pose 'SourcePose'(), Additive=@n3) {alphaBoolBlend {blendOption=Linear}}
    BlendPoses_0=@AnimGraphNode_StateMachine_2
   ) {
    bMeshSpaceRotationBlend=true
    bMeshSpaceScaleBlend=true
    curveBlendOption=UseMaxValue
    layerSetup [{branchFilters [{boneName=Root}, {blendDepth=-1, boneName=Ik_hand_root}, {blendDepth=-4, boneName=spine_01}, {blendDepth=-1, boneName=VB pelvis_ik_hand_l}]}]
   }
   n5 = Modify Curve(
    SourcePose=Apply Additive(
     Base=@n4
     Additive=Blendspace Player 'M_Jump_ARDS_Lean1D'!(X=@n1) -> (Pose) {blendSpace=/Game/Animation/Game/MainPlayer/Locomotion/Medium/Male/Jump/AssaultRifle/M_Jump_ARDS_Lean1D}
     Alpha=Get JumpAdditiveLeanAlpha() -> (JumpAdditiveLeanAlpha:float)
    ) {alphaBoolBlend {blendOption=Linear}}
    CurveValues_0=0
    Alpha=1
   ) {curveNames [WeaponAiming], curveValues [0]}
   n6 = Use cached pose 'SourcePose'()
   n7 = Get bDisableInAirAnimations() -> (bDisableInAirAnimations:bool)
   n8 = Blend Poses by bool(
    BlendPose_0=Use cached pose 'SourcePose'()
    BlendPose_1=Blend Poses by bool(
     BlendPose_0=Use cached pose 'SourcePose'()
     BlendPose_1=Blend Poses by bool(BlendPose_0=@n6, BlendPose_1=@n5, BlendTime_1=0.3, bActiveValue=@n7) {blendType=Linear}
     bActiveValue=Get bIsInGoop() -> (bIsInGoop:bool)
    ) {blendType=Linear}
    BlendTime_1=0.5
    bActiveValue=Get bIsUsingJetpack() -> (bIsUsingJetpack:bool)
   ) {blendType=Linear, transitionType=Inertialization}
   n9 = Sequence Player!(Sequence:Anim Sequence Base Object Reference=Break Fort Anim Input Player Anim Asset!(FortAnimInput_PlayerAnimAsset:"Fort Anim Input Player Anim Asset Structure (by ref)"=Get CurrentAnimSet() -> (CurrentAnimSet:Fort Anim Input Player Anim Asset Structure)) -> (
     bPlayUpperBodySlotOnFullBodyInAir:bool=~
     bDisableFullBodyAimOffsetDuringMelee:bool=~
     bShouldApplyAimOffsetFullBody:bool=~
     FullBodyAimOffsetAlpha:float=~
     UpperBodyAimOffsetAlpha:float=~
     bOverrideDisableArmsHeadAdditive:bool=~
     DisableArmsHeadAdditiveOverride:float=~
     SprintPlayrateClampRange:Fort Anim Playrate Range Structure
     SprintTargetingPlayrateClampRange:Fort Anim Playrate Range Structure
     CrouchSprintPlayrateClampRange:Fort Anim Playrate Range Structure
     CrouchSprintTargetingPlayrateClampRange:Fort Anim Playrate Range Structure
     TargetingAimOffset:Aim Offset Blend Space Object Reference
     NonTargetingAimOffset:Aim Offset Blend Space Object Reference
     RelaxedAimOffset:Aim Offset Blend Space Object Reference
     SwimTargetingPose:Anim Sequence Object Reference
     SwimNonTargetingPose:Anim Sequence Object Reference
     SwimJogAdditiveBlendSpace:Blend Space Object Reference
     SwimTargetingAimOffset:Aim Offset Blend Space Object Reference
     SwimIdleNoise:Anim Sequence Object Reference
     SwimJumpSurfaceEndAdditve:Anim Sequence Object Reference
     SwimJumpSurfaceLoopAdditve:Anim Sequence Object Reference
     SwimJumpFallLoopAdditve:Anim Sequence Object Reference
     FlyModeLoopAdditive:Anim Sequence Object Reference
     FlyModeStartAdditive:Anim Sequence Object Reference
     ZipLineStartAdditve:Anim Sequence Object Reference
     JetPackJumpAdditve:Anim Sequence Object Reference
     JetPackStartAdditve:Anim Sequence Object Reference
     FallAdditive:Anim Sequence Object Reference
     JumpLoopAdditive:Anim Sequence Object Reference
     JumpUpAdditive:Anim Sequence Object Reference
     FullBodyJumpAdditive:Anim Sequence Object Reference
     WeaponChargeLoop:Anim Sequence Object Reference
     LandingAdditiveOverride:Anim Sequence Object Reference
     TurnInPlaceLeftAdditiveOverride:Anim Sequence Object Reference
     TurnInPlaceRightAdditiveOverride:Anim Sequence Object Reference
     bOverridePitchAndYawOffsets:bool=~
     AnimSetOffsets:Fort Anim Input Adjusted Aim Offset Structure
     BonesToHide:Array of Names
     bOverrideHandIKRetargetingWeight:bool=~
     HandIKRetargetingWeight:float=~
     bForceKneeSliding:bool=~
     TargetingPose:Anim Sequence Object Reference
     NonTargetingPose:Anim Sequence Object Reference
     IdleNoise:Anim Sequence Object Reference
    ) {is=title K2Node_BreakStruct}.FullBodyJumpAdditive) -> (Pose) {bLoopAnimation=false}
   Output Pose(Result=Inertialization!(Source:Pose Link Structure=Apply Additive(Base=@n8, Additive=@n9) {alphaBoolBlend {blendOption=Linear}}) -> (Pose:Pose Link Structure) {is=title AnimGraphNode_Inertialization}) {layerGroup=None}
   Save cached pose 'SourcePose'(Pose=Input Pose() {name=SourcePose}) {cacheName=SourcePose}
  ]}
 GroundJumpStateMachine {
  entry=Conduit
  path=AnimGraph.AnimGraphNode_StateMachine_1.GroundJumpStateMachine
  ids=id
  nodes [
   AnimStateConduitNode_0 = Conduit()
   AnimStateNode_1 = JumpStart() {is=title AnimStateNode}
   AnimStateNode_10 = FallLoop() {is=title AnimStateNode}
   AnimStateNode_7 = LandPredict() {is=title AnimStateNode}
   AnimStateTransitionNode_0 = AnimStateTransitionNode() {crossfadeDuration=0.1, priorityOrder=100}
   AnimStateTransitionNode_1 = AnimStateTransitionNode() {crossfadeDuration=0.4, logicType=TLT_Inertialization, priorityOrder=3}
   AnimStateTransitionNode_14 = AnimStateTransitionNode() {logicType=TLT_Inertialization}
   AnimStateTransitionNode_16 = AnimStateTransitionNode() {priorityOrder=5}
   AnimStateTransitionNode_2 = AnimStateTransitionNode() {logicType=TLT_Inertialization}
   AnimStateTransitionNode_20 = AnimStateTransitionNode() {bAutomaticRuleBasedOnSequencePlayerInState=true, blendMode=Linear}
   AnimStateTransitionNode_28 = AnimStateTransitionNode() {crossfadeDuration=0.069, logicType=TLT_Inertialization}
   AnimStateTransitionNode_3 = AnimStateTransitionNode() {crossfadeDuration=0.069, logicType=TLT_Inertialization, priorityOrder=6}
   AnimStateTransitionNode_4 = AnimStateTransitionNode() {crossfadeDuration=0.1, logicType=TLT_Inertialization, priorityOrder=2}
   AnimStateTransitionNode_5 = AnimStateTransitionNode() {logicType=TLT_Inertialization, priorityOrder=2}
   AnimStateTransitionNode_7 = AnimStateTransitionNode() {priorityOrder=4}
  ]
  links [
   {
    dst=FallLoop
    rule=AnimGraph.AnimGraphNode_StateMachine_1.GroundJumpStateMachine.AnimStateTransitionNode_2.Transition
    src=Conduit
    via=AnimStateTransitionNode_2
   }
   {
    dst=FallLoop
    rule=AnimGraph.AnimGraphNode_StateMachine_1.GroundJumpStateMachine.AnimStateTransitionNode_5.Transition
    src=Conduit
    via=AnimStateTransitionNode_5
   }
   {
    dst=JumpStart
    rule=AnimGraph.AnimGraphNode_StateMachine_1.GroundJumpStateMachine.AnimStateTransitionNode_0.AnimationTransitionGraph_0
    src=Conduit
    via=AnimStateTransitionNode_0
   }
   {
    dst=LandPredict
    rule=AnimGraph.AnimGraphNode_StateMachine_1.GroundJumpStateMachine.AnimStateTransitionNode_1.AnimationTransitionGraph_0
    src=Conduit
    via=AnimStateTransitionNode_1
   }
   {
    dst=JumpStart
    rule=AnimGraph.AnimGraphNode_StateMachine_1.GroundJumpStateMachine.AnimStateTransitionNode_3.AnimationTransitionGraph_1
    src=FallLoop
    via=AnimStateTransitionNode_3
   }
   {
    dst=LandPredict
    rule=AnimGraph.AnimGraphNode_StateMachine_1.GroundJumpStateMachine.AnimStateTransitionNode_16.AnimationTransitionGraph_1
    src=FallLoop
    via=AnimStateTransitionNode_16
   }
   {
    dst=LandPredict
    rule=AnimGraph.AnimGraphNode_StateMachine_1.GroundJumpStateMachine.AnimStateTransitionNode_7.AnimationTransitionGraph_1
    src=FallLoop
    via=AnimStateTransitionNode_7
   }
   {
    dst=FallLoop
    rule=AnimGraph.AnimGraphNode_StateMachine_1.GroundJumpStateMachine.AnimStateTransitionNode_20.AnimationTransitionGraph_1
    src=JumpStart
    via=AnimStateTransitionNode_20
   }
   {
    dst=LandPredict
    rule=AnimGraph.AnimGraphNode_StateMachine_1.GroundJumpStateMachine.AnimStateTransitionNode_14.AnimationTransitionGraph_1
    src=JumpStart
    via=AnimStateTransitionNode_14
   }
   {
    dst=JumpStart
    rule=AnimGraph.AnimGraphNode_StateMachine_1.GroundJumpStateMachine.AnimStateTransitionNode_28.AnimationTransitionGraph_1
    src=LandPredict
    via=AnimStateTransitionNode_28
   }
   {
    dst=JumpStart
    rule=AnimGraph.AnimGraphNode_StateMachine_1.GroundJumpStateMachine.AnimStateTransitionNode_4.Transition
    src=LandPredict
    via=AnimStateTransitionNode_4
   }
  ]
 }
 Conduit {path=AnimGraph.AnimGraphNode_StateMachine_1.GroundJumpStateMachine.AnimStateConduitNode_0.Conduit, nodes [Result(bCanEnterTransition=true)]}
 JumpStart {
  path=AnimGraph.AnimGraphNode_StateMachine_1.GroundJumpStateMachine.AnimStateNode_1.JumpStart
  nodes [
   AnimGraphNode_Layer_0 = ItemAnimLayerInterface - ItemJumpUpAdditive!(InPoseItemJumpUpAdditive:Pose Link Structure=JumpUpAdditive!() -> (Pose:Pose Link Structure) {is=title AnimGraphNode_LinkedAnimLayer, layer=JumpUpAdditive}) -> (Pose:Pose Link Structure) {
    is=title AnimGraphNode_LinkedAnimLayer
    interface=/Game/Characters/Player/Male/Male_Avg_Base/AnimLayerInterface/ItemAnimLayerInterface.ItemAnimLayerInterface_C
    layer=ItemJumpUpAdditive
   }
   n1 = Layered blend per bone(
    BasePose=@AnimGraphNode_Layer_0
    BlendPoses_0="Sequence Player (additive) 'JackalVehicle_JumpCombat_Start'"() {bLoopAnimation=false, sequence=/Game/Animation/Game/MainPlayer/Vehicles/JackalVehicle/JackalVehicle_JumpCombat_Start}
    BlendWeights_0=Break Fort Anim Input Jackal Vehicle!(FortAnimInput_JackalVehicle:"Fort Anim Input Jackal Vehicle Structure (by ref)"=Get JackalVehicleInput() -> (JackalVehicleInput:Fort Anim Input Jackal Vehicle Structure)) -> (
     bIsUsingJackalVehicle:bool=~
     bIsSteeringLeft:bool=~
     bIsSteeringRight:bool=~
     SteerAngleDeadZoneDegrees:float=~
     SteerAngle:float=~
     LeanYaw:float=~
     QuantizedSteerAngle:float=~
     SteerAlpha:float=~
     RunForwardAlpha:float=~
     SlopePitchDegreeAngle:float=~
     SlopeRollDegreeAngle:float=~
     DistanceFromGround:float=~
     PivotPlayRate:float=~
     bIsReversing:bool=~
     bIsBraking:bool=~
     bIsMoving:bool=~
     bIsMovingForward:bool=~
     bIsSprinting:bool=~
     bInAir:bool=~
     bIsFalling:bool=~
     bIsJumping:bool=~
     bIsRelaxed:bool=~
     bIsBoosting:bool=~
     bHasReachedJumpChargeStartThreshold:bool=~
     bHasReachedJumpChargeMidThreshold:bool=~
     bHasReachedJumpChargeFullThreshold:bool=~
     bAdjustRootForFemaleRider:bool=~
     bIsPlayingEmoteOnHoverboard:bool=~
     bShouldApplyAdditive:bool=~
     bPlayPivotOnGroundAndNotBoosting:bool=~
     bIsOnLowerHill:bool=~
     bLowerBodyIdleToLoopTransition:bool=~
     bInterruptHoverboardFullybody:bool=~
     bEnterFullBodyHoverboardState:bool=~
     bDefaultToJumpStartTransition:bool=~
     bLocomotionPoseToJumpTransition:bool=~
     bBoostingToBoostingJumpTransition:bool=~
     bJumpToLocomotionPoseTransition:bool=~
     bJumpApexToJumpFallTransition:bool=~
     bIdleToMovementStartTransition:bool=~
     bIdleToMovementLoopTransition:bool=~
     bMovementLoopToMovementStopTransition:bool=~
     bMovementLoopToPivotTransition:bool=~
     bMovementLoopToIdleTransition:bool=~
     bIdleAdditiveToCollisionNTransition:bool=~
     bSplitBodyToHoverboardBRTransition:bool=~
     bHoverboardBRToSplitBodyTransition:bool=~
     bHoverboardBRMovementToJumpChargeTransition:bool=~
     bIdlesToJackalVehicleTransition:bool=~
     bPlayAdditiveLeans:bool=~
     bPlayBalloonLeans:bool=~
     bPlayJumpTrickVertical:bool=~
     bPlayJumpTrick:bool=~
     bPlayMovingFast:bool=~
     bPlayHipAdjustmentAdditive:bool=~
     bPlayDriveSouth:bool=~
     bPlayHeadAimOffset:bool=~
     bPlaySlopeAimOffset:bool=~
     JumpCombatAdditiveWeight:float=~
     MeleeTwistIdle:float=~
     MeleeTwistLocomotionLoop:float=~
     VelocityImpact:Fort Anim Input Velocity Impact Structure
     ImpactDisplacement:Vector=~
     SteerYaw:float=~
     EmoteHoverboardPosition:Vector=~
     EmoteHoverboardRotation:Rotator=~
     bShouldAttachFeetToHoverboard:bool=~
     FootLeftLocationOffset:Vector=~
     FootLeftRotationOffset:Rotator=~
     FootRightLocationOffset:Vector=~
     FootRightRotationOffset:Rotator=~
     StoppedThreshold:float=~
     MovingForwardThreshold:float=~
     MovingFowardFastThreshold:float=~
     DefaultToJumpStartTransitionThreshold:float=~
     JumpTrickAngularVelocityThreshold:float=~
     JumpDistanceFromGroundThreshold:float=~
     VelocityStartThreshold:float=~
     MovingFastThreshold:float=~
     RelaxedSpeedThreshold:float=~
     JumpChargeStartThreshold:float=~
     JumpChargeMidThreshold:float=~
     JumpChargeFullThreshold:float=~
     RotatingAngularVelocityThreshold:float=~
     IdleToLoopTransitionSpeedThreshold:float=~
     IdleToMovementLoopTransitionThreshold:float=~
     MeleeTwistIdleMultiplier:float=~
     MeleeTwistLocomotionLoopMultiplier:float=~
     LeanYawForMaxPivotPlayRate:float=~
     SteerAlphaForMaxPivotPlayRate:float=~
     PrePivotCardinalDirection:EFortCardinalDirection Enum
    ) {is=title K2Node_BreakStruct}.JumpCombatAdditiveWeight
   ) {
    bMeshSpaceRotationBlend=true
    lODThreshold=2
    layerSetup [{branchFilters [
       {boneName=root}
       {blendDepth=-1, boneName=pelvis}
       {boneName=thigh_r}
       {boneName=thigh_l}
       {blendDepth=-1, boneName=spine_01}
       {blendDepth=-1, boneName=Ik_hand_root}
       {boneName=ik_foot_root}
      ]}]
   }
   Output Animation Pose(Result=Two Way Blend!(A=@n1, B=Additive Identity Pose(), AlphaCurveName:Name=RemoveJumpUpAdditive) -> (Pose) {blendNode {alphaInputType=Curve, alphaBoolBlend {blendOption=Linear}}})
  ]
 }
 FallLoop {
  path=AnimGraph.AnimGraphNode_StateMachine_1.GroundJumpStateMachine.AnimStateNode_10.FallLoop
  nodes [
   AnimGraphNode_Layer_0 = ItemAnimLayerInterface - ItemJumpLoopAdditive!(InPoseItemJumpLoopAdditive:Pose Link Structure=JumpLoopAdditive!() -> (Pose:Pose Link Structure) {is=title AnimGraphNode_LinkedAnimLayer, layer=JumpLoopAdditive}) -> (Pose:Pose Link Structure) {
    is=title AnimGraphNode_LinkedAnimLayer
    interface=&ItemAnimLayerInterface_C
    layer=ItemJumpLoopAdditive
   }
   n1 = Layered blend per bone(
    BasePose=@AnimGraphNode_Layer_0
    BlendPoses_0="Sequence Player (additive) 'JackalVehicle_JumpCombat_Loop'"() {sequence=/Game/Animation/Game/MainPlayer/Vehicles/JackalVehicle/JackalVehicle_JumpCombat_Loop}
    BlendWeights_0=Break Fort Anim Input Jackal Vehicle(FortAnimInput_JackalVehicle=Get JackalVehicleInput()) -> (
     bIsUsingJackalVehicle=~
     bIsSteeringLeft=~
     bIsSteeringRight=~
     SteerAngleDeadZoneDegrees=~
     SteerAngle=~
     LeanYaw=~
     QuantizedSteerAngle=~
     SteerAlpha=~
     RunForwardAlpha=~
     SlopePitchDegreeAngle=~
     SlopeRollDegreeAngle=~
     DistanceFromGround=~
     PivotPlayRate=~
     bIsReversing=~
     bIsBraking=~
     bIsMoving=~
     bIsMovingForward=~
     bIsSprinting=~
     bInAir=~
     bIsFalling=~
     bIsJumping=~
     bIsRelaxed=~
     bIsBoosting=~
     bHasReachedJumpChargeStartThreshold=~
     bHasReachedJumpChargeMidThreshold=~
     bHasReachedJumpChargeFullThreshold=~
     bAdjustRootForFemaleRider=~
     bIsPlayingEmoteOnHoverboard=~
     bShouldApplyAdditive=~
     bPlayPivotOnGroundAndNotBoosting=~
     bIsOnLowerHill=~
     bLowerBodyIdleToLoopTransition=~
     bInterruptHoverboardFullybody=~
     bEnterFullBodyHoverboardState=~
     bDefaultToJumpStartTransition=~
     bLocomotionPoseToJumpTransition=~
     bBoostingToBoostingJumpTransition=~
     bJumpToLocomotionPoseTransition=~
     bJumpApexToJumpFallTransition=~
     bIdleToMovementStartTransition=~
     bIdleToMovementLoopTransition=~
     bMovementLoopToMovementStopTransition=~
     bMovementLoopToPivotTransition=~
     bMovementLoopToIdleTransition=~
     bIdleAdditiveToCollisionNTransition=~
     bSplitBodyToHoverboardBRTransition=~
     bHoverboardBRToSplitBodyTransition=~
     bHoverboardBRMovementToJumpChargeTransition=~
     bIdlesToJackalVehicleTransition=~
     bPlayAdditiveLeans=~
     bPlayBalloonLeans=~
     bPlayJumpTrickVertical=~
     bPlayJumpTrick=~
     bPlayMovingFast=~
     bPlayHipAdjustmentAdditive=~
     bPlayDriveSouth=~
     bPlayHeadAimOffset=~
     bPlaySlopeAimOffset=~
     JumpCombatAdditiveWeight=~
     MeleeTwistIdle=~
     MeleeTwistLocomotionLoop=~
     ImpactDisplacement=~
     SteerYaw=~
     EmoteHoverboardPosition=~
     EmoteHoverboardRotation=~
     bShouldAttachFeetToHoverboard=~
     FootLeftLocationOffset=~
     FootLeftRotationOffset=~
     FootRightLocationOffset=~
     FootRightRotationOffset=~
     StoppedThreshold=~
     MovingForwardThreshold=~
     MovingFowardFastThreshold=~
     DefaultToJumpStartTransitionThreshold=~
     JumpTrickAngularVelocityThreshold=~
     JumpDistanceFromGroundThreshold=~
     VelocityStartThreshold=~
     MovingFastThreshold=~
     RelaxedSpeedThreshold=~
     JumpChargeStartThreshold=~
     JumpChargeMidThreshold=~
     JumpChargeFullThreshold=~
     RotatingAngularVelocityThreshold=~
     IdleToLoopTransitionSpeedThreshold=~
     IdleToMovementLoopTransitionThreshold=~
     MeleeTwistIdleMultiplier=~
     MeleeTwistLocomotionLoopMultiplier=~
     LeanYawForMaxPivotPlayRate=~
     SteerAlphaForMaxPivotPlayRate=~
    ).JumpCombatAdditiveWeight
   ) {
    bMeshSpaceRotationBlend=true
    lODThreshold=2
    layerSetup [{branchFilters [
       {boneName=root}
       {blendDepth=-1, boneName=pelvis}
       {boneName=thigh_r}
       {boneName=thigh_l}
       {blendDepth=-1, boneName=spine_01}
       {blendDepth=-1, boneName=Ik_hand_root}
       {boneName=ik_foot_root}
      ]}]
   }
   Output Animation Pose(Result=Two Way Blend(A=@n1, B=Additive Identity Pose(), AlphaCurveName=RemoveJumpLoopAdditive) {blendNode {alphaInputType=Curve, alphaBoolBlend {blendOption=Linear}}})
  ]
 }
 LandPredict {
  path=AnimGraph.AnimGraphNode_StateMachine_1.GroundJumpStateMachine.AnimStateNode_7.LandPredict
  nodes [
   AnimGraphNode_Layer_0 = ItemAnimLayerInterface - ItemFallAdditive!(InPoseItemFallAdditive:Pose Link Structure=FallAdditive!() -> (Pose:Pose Link Structure) {is=title AnimGraphNode_LinkedAnimLayer, layer=FallAdditive}) -> (Pose:Pose Link Structure) {
    is=title AnimGraphNode_LinkedAnimLayer
    interface=&ItemAnimLayerInterface_C
    layer=ItemFallAdditive
   }
   n1 = Layered blend per bone(
    BasePose=@AnimGraphNode_Layer_0
    BlendPoses_0="Sequence Player (additive) 'JackalVehicle_JumpCombat_Fall'"() {bLoopAnimation=false, sequence=/Game/Animation/Game/MainPlayer/Vehicles/JackalVehicle/JackalVehicle_JumpCombat_Fall}
    BlendWeights_0=Break Fort Anim Input Jackal Vehicle(FortAnimInput_JackalVehicle=Get JackalVehicleInput()) -> (
     bIsUsingJackalVehicle=~
     bIsSteeringLeft=~
     bIsSteeringRight=~
     SteerAngleDeadZoneDegrees=~
     SteerAngle=~
     LeanYaw=~
     QuantizedSteerAngle=~
     SteerAlpha=~
     RunForwardAlpha=~
     SlopePitchDegreeAngle=~
     SlopeRollDegreeAngle=~
     DistanceFromGround=~
     PivotPlayRate=~
     bIsReversing=~
     bIsBraking=~
     bIsMoving=~
     bIsMovingForward=~
     bIsSprinting=~
     bInAir=~
     bIsFalling=~
     bIsJumping=~
     bIsRelaxed=~
     bIsBoosting=~
     bHasReachedJumpChargeStartThreshold=~
     bHasReachedJumpChargeMidThreshold=~
     bHasReachedJumpChargeFullThreshold=~
     bAdjustRootForFemaleRider=~
     bIsPlayingEmoteOnHoverboard=~
     bShouldApplyAdditive=~
     bPlayPivotOnGroundAndNotBoosting=~
     bIsOnLowerHill=~
     bLowerBodyIdleToLoopTransition=~
     bInterruptHoverboardFullybody=~
     bEnterFullBodyHoverboardState=~
     bDefaultToJumpStartTransition=~
     bLocomotionPoseToJumpTransition=~
     bBoostingToBoostingJumpTransition=~
     bJumpToLocomotionPoseTransition=~
     bJumpApexToJumpFallTransition=~
     bIdleToMovementStartTransition=~
     bIdleToMovementLoopTransition=~
     bMovementLoopToMovementStopTransition=~
     bMovementLoopToPivotTransition=~
     bMovementLoopToIdleTransition=~
     bIdleAdditiveToCollisionNTransition=~
     bSplitBodyToHoverboardBRTransition=~
     bHoverboardBRToSplitBodyTransition=~
     bHoverboardBRMovementToJumpChargeTransition=~
     bIdlesToJackalVehicleTransition=~
     bPlayAdditiveLeans=~
     bPlayBalloonLeans=~
     bPlayJumpTrickVertical=~
     bPlayJumpTrick=~
     bPlayMovingFast=~
     bPlayHipAdjustmentAdditive=~
     bPlayDriveSouth=~
     bPlayHeadAimOffset=~
     bPlaySlopeAimOffset=~
     JumpCombatAdditiveWeight=~
     MeleeTwistIdle=~
     MeleeTwistLocomotionLoop=~
     ImpactDisplacement=~
     SteerYaw=~
     EmoteHoverboardPosition=~
     EmoteHoverboardRotation=~
     bShouldAttachFeetToHoverboard=~
     FootLeftLocationOffset=~
     FootLeftRotationOffset=~
     FootRightLocationOffset=~
     FootRightRotationOffset=~
     StoppedThreshold=~
     MovingForwardThreshold=~
     MovingFowardFastThreshold=~
     DefaultToJumpStartTransitionThreshold=~
     JumpTrickAngularVelocityThreshold=~
     JumpDistanceFromGroundThreshold=~
     VelocityStartThreshold=~
     MovingFastThreshold=~
     RelaxedSpeedThreshold=~
     JumpChargeStartThreshold=~
     JumpChargeMidThreshold=~
     JumpChargeFullThreshold=~
     RotatingAngularVelocityThreshold=~
     IdleToLoopTransitionSpeedThreshold=~
     IdleToMovementLoopTransitionThreshold=~
     MeleeTwistIdleMultiplier=~
     MeleeTwistLocomotionLoopMultiplier=~
     LeanYawForMaxPivotPlayRate=~
     SteerAlphaForMaxPivotPlayRate=~
    ).JumpCombatAdditiveWeight
   ) {
    bMeshSpaceRotationBlend=true
    lODThreshold=2
    layerSetup [{branchFilters [
       {boneName=root}
       {blendDepth=-1, boneName=pelvis}
       {boneName=thigh_r}
       {boneName=thigh_l}
       {blendDepth=-1, boneName=spine_01}
       {blendDepth=-1, boneName=Ik_hand_root}
       {boneName=ik_foot_root}
      ]}]
   }
   Output Animation Pose(Result=Two Way Blend(A=@n1, B=Additive Identity Pose(), AlphaCurveName=RemoveJumpFallAdditive) {blendNode {alphaInputType=Curve, alphaBoolBlend {blendOption=Linear}}})
  ]
 }
 AnimationTransitionGraph_0 {
  path=AnimGraph.AnimGraphNode_StateMachine_1.GroundJumpStateMachine.AnimStateTransitionNode_0.AnimationTransitionGraph_0
  nodes [Result(bCanEnterTransition=true)]
 }
 AnimationTransitionGraph_0 {
  path=AnimGraph.AnimGraphNode_StateMachine_1.GroundJumpStateMachine.AnimStateTransitionNode_1.AnimationTransitionGraph_0
  nodes [Result(bCanEnterTransition=Get bLandingPredicted() -> (bLandingPredicted:bool))]
 }
 AnimationTransitionGraph_1 {
  path=AnimGraph.AnimGraphNode_StateMachine_1.GroundJumpStateMachine.AnimStateTransitionNode_14.AnimationTransitionGraph_1
  nodes [Result(bCanEnterTransition=Get bLandingPredicted())]
 }
 AnimationTransitionGraph_1 {
  path=AnimGraph.AnimGraphNode_StateMachine_1.GroundJumpStateMachine.AnimStateTransitionNode_16.AnimationTransitionGraph_1
  nodes [Result(bCanEnterTransition=Break Fort Anim Input Common Vehicle!(FortAnimInput_CommonVehicle:"Fort Anim Input Common Vehicle Structure (by ref)"=Get CommonVehicleInput() -> (CommonVehicleInput:Fort Anim Input Common Vehicle Structure)) -> (
     bIsUsingVehicle:bool=~
     bIsJumpingVehicle:bool=~
     bCanChargeJump:bool=~
     bIsChargingJump:bool=~
     bIsJumping:bool=~
     bIsOnGround:bool=~
     bCanDriverAimWeapon:bool=~
    ) {is=title K2Node_BreakStruct}.bIsOnGround)]
 }
 Transition {
  path=AnimGraph.AnimGraphNode_StateMachine_1.GroundJumpStateMachine.AnimStateTransitionNode_2.Transition
  nodes [Result(bCanEnterTransition=Get bSkipJumpStart() -> (bSkipJumpStart:bool))]
 }
 AnimationTransitionGraph_1 {path=AnimGraph.AnimGraphNode_StateMachine_1.GroundJumpStateMachine.AnimStateTransitionNode_20.AnimationTransitionGraph_1, nodes [Result()]}
 AnimationTransitionGraph_1 {
  path=AnimGraph.AnimGraphNode_StateMachine_1.GroundJumpStateMachine.AnimStateTransitionNode_28.AnimationTransitionGraph_1
  nodes [Result(bCanEnterTransition=Break Fort Anim Input Transition Properties!(FortAnimInput_TransitionProperties:"Fort Anim Input Transition Properties Structure (by ref)"=Get TransitionPropertiesInput() -> (TransitionPropertiesInput:Fort Anim Input Transition Properties Structure)) -> (bTransition_DoubleJump_Fast:bool=~, bTransition_DoubleJump:bool=~) {is=title K2Node_BreakStruct}.bTransition_DoubleJump)]
 }
 AnimationTransitionGraph_1 {
  path=AnimGraph.AnimGraphNode_StateMachine_1.GroundJumpStateMachine.AnimStateTransitionNode_3.AnimationTransitionGraph_1
  nodes [Result(bCanEnterTransition=Break Fort Anim Input Transition Properties(FortAnimInput_TransitionProperties=Get TransitionPropertiesInput()) -> (bTransition_DoubleJump_Fast=~, bTransition_DoubleJump=~).bTransition_DoubleJump)]
 }
 Transition {
  path=AnimGraph.AnimGraphNode_StateMachine_1.GroundJumpStateMachine.AnimStateTransitionNode_4.Transition
  nodes [Result(bCanEnterTransition=Get bIsJumping())]
 }
 Transition {
  path=AnimGraph.AnimGraphNode_StateMachine_1.GroundJumpStateMachine.AnimStateTransitionNode_5.Transition
  nodes [Result(bCanEnterTransition=Get bIsFalling())]
 }
 AnimationTransitionGraph_1 {
  path=AnimGraph.AnimGraphNode_StateMachine_1.GroundJumpStateMachine.AnimStateTransitionNode_7.AnimationTransitionGraph_1
  nodes [Result(bCanEnterTransition=Get bTransition_NewJumpLoopAdditive_to_NewFallAdditive() -> (bTransition_NewJumpLoopAdditive_to_NewFallAdditive:bool))]
 }
 GroundJumpLowerBody {
  entry=Conduit_1
  path=AnimGraph.AnimGraphNode_StateMachine_2.GroundJumpLowerBody
  ids=id
  nodes [
   AnimStateConduitNode_0 = Conduit_1() {is=title AnimStateConduitNode}
   AnimStateNode_1 = JumpStart()
   AnimStateNode_10 = FallLoop()
   AnimStateNode_7 = LandPredict()
   AnimStateTransitionNode_0 = AnimStateTransitionNode() {crossfadeDuration=0.1, priorityOrder=100}
   AnimStateTransitionNode_1 = AnimStateTransitionNode() {crossfadeDuration=0.4, logicType=TLT_Inertialization}
   AnimStateTransitionNode_14 = AnimStateTransitionNode()
   AnimStateTransitionNode_16 = AnimStateTransitionNode() {priorityOrder=5}
   AnimStateTransitionNode_2 = AnimStateTransitionNode() {logicType=TLT_Inertialization}
   AnimStateTransitionNode_20 = AnimStateTransitionNode() {bAutomaticRuleBasedOnSequencePlayerInState=true, blendMode=Linear}
   AnimStateTransitionNode_28 = AnimStateTransitionNode() {crossfadeDuration=0.069, logicType=TLT_Inertialization}
   AnimStateTransitionNode_3 = AnimStateTransitionNode() {crossfadeDuration=0.069, logicType=TLT_Inertialization, priorityOrder=6}
   AnimStateTransitionNode_4 = AnimStateTransitionNode() {crossfadeDuration=0.1, logicType=TLT_Inertialization, priorityOrder=2}
   AnimStateTransitionNode_5 = AnimStateTransitionNode() {logicType=TLT_Inertialization, priorityOrder=2}
   AnimStateTransitionNode_7 = AnimStateTransitionNode() {priorityOrder=4}
  ]
  links [
   {
    dst=FallLoop
    rule=AnimGraph.AnimGraphNode_StateMachine_2.GroundJumpLowerBody.AnimStateTransitionNode_2.AnimationTransitionGraph_0
    src=Conduit_1
    via=AnimStateTransitionNode_2
   }
   {
    dst=FallLoop
    rule=AnimGraph.AnimGraphNode_StateMachine_2.GroundJumpLowerBody.AnimStateTransitionNode_5.Transition
    src=Conduit_1
    via=AnimStateTransitionNode_5
   }
   {
    dst=JumpStart
    rule=AnimGraph.AnimGraphNode_StateMachine_2.GroundJumpLowerBody.AnimStateTransitionNode_0.AnimationTransitionGraph_1
    src=Conduit_1
    via=AnimStateTransitionNode_0
   }
   {
    dst=LandPredict
    rule=AnimGraph.AnimGraphNode_StateMachine_2.GroundJumpLowerBody.AnimStateTransitionNode_1.AnimationTransitionGraph_1
    src=Conduit_1
    via=AnimStateTransitionNode_1
   }
   {
    dst=JumpStart
    rule=AnimGraph.AnimGraphNode_StateMachine_2.GroundJumpLowerBody.AnimStateTransitionNode_3.AnimationTransitionGraph_0
    src=FallLoop
    via=AnimStateTransitionNode_3
   }
   {
    dst=LandPredict
    rule=AnimGraph.AnimGraphNode_StateMachine_2.GroundJumpLowerBody.AnimStateTransitionNode_16.AnimationTransitionGraph_0
    src=FallLoop
    via=AnimStateTransitionNode_16
   }
   {
    dst=LandPredict
    rule=AnimGraph.AnimGraphNode_StateMachine_2.GroundJumpLowerBody.AnimStateTransitionNode_7.AnimationTransitionGraph_0
    src=FallLoop
    via=AnimStateTransitionNode_7
   }
   {
    dst=FallLoop
    rule=AnimGraph.AnimGraphNode_StateMachine_2.GroundJumpLowerBody.AnimStateTransitionNode_20.AnimationTransitionGraph_0
    src=JumpStart
    via=AnimStateTransitionNode_20
   }
   {
    dst=LandPredict
    rule=AnimGraph.AnimGraphNode_StateMachine_2.GroundJumpLowerBody.AnimStateTransitionNode_14.AnimationTransitionGraph_0
    src=JumpStart
    via=AnimStateTransitionNode_14
   }
   {
    dst=JumpStart
    rule=AnimGraph.AnimGraphNode_StateMachine_2.GroundJumpLowerBody.AnimStateTransitionNode_28.AnimationTransitionGraph_0
    src=LandPredict
    via=AnimStateTransitionNode_28
   }
   {
    dst=JumpStart
    rule=AnimGraph.AnimGraphNode_StateMachine_2.GroundJumpLowerBody.AnimStateTransitionNode_4.AnimationTransitionGraph_0
    src=LandPredict
    via=AnimStateTransitionNode_4
   }
  ]
 }
 Conduit_1 {path=AnimGraph.AnimGraphNode_StateMachine_2.GroundJumpLowerBody.AnimStateConduitNode_0.Conduit_1, nodes [Result(bCanEnterTransition=true)]}
 JumpStart {
  path=AnimGraph.AnimGraphNode_StateMachine_2.GroundJumpLowerBody.AnimStateNode_1.JumpStart
  nodes [Output Animation Pose(Result=Sequence Player 'M_JumpUp_ARDS_Base'() {sequence=/Game/Animation/Game/MainPlayer/Locomotion/Medium/Male/Jump/AssaultRifle/M_JumpUp_ARDS_Base})]
 }
 FallLoop {
  path=AnimGraph.AnimGraphNode_StateMachine_2.GroundJumpLowerBody.AnimStateNode_10.FallLoop
  nodes [Output Animation Pose(Result=Sequence Player 'M_JumpApex_ARDS_Loop'() {sequence=/Game/Animation/Game/MainPlayer/Locomotion/Medium/Male/Jump/AssaultRifle/M_JumpApex_ARDS_Loop})]
 }
 LandPredict {
  path=AnimGraph.AnimGraphNode_StateMachine_2.GroundJumpLowerBody.AnimStateNode_7.LandPredict
  nodes [Output Animation Pose(Result=Sequence Player 'M_JumpFall_ARDS_Base'() {bLoopAnimation=false, sequence=/Game/Animation/Game/MainPlayer/Locomotion/Medium/Male/Jump/AssaultRifle/M_JumpFall_ARDS_Base})]
 }
 AnimationTransitionGraph_1 {
  path=AnimGraph.AnimGraphNode_StateMachine_2.GroundJumpLowerBody.AnimStateTransitionNode_0.AnimationTransitionGraph_1
  nodes [Result(bCanEnterTransition=true)]
 }
 AnimationTransitionGraph_1 {
  path=AnimGraph.AnimGraphNode_StateMachine_2.GroundJumpLowerBody.AnimStateTransitionNode_1.AnimationTransitionGraph_1
  nodes [Result(bCanEnterTransition=Get bLandingPredicted())]
 }
 AnimationTransitionGraph_0 {
  path=AnimGraph.AnimGraphNode_StateMachine_2.GroundJumpLowerBody.AnimStateTransitionNode_14.AnimationTransitionGraph_0
  nodes [Result(bCanEnterTransition=Get bLandingPredicted())]
 }
 AnimationTransitionGraph_0 {
  path=AnimGraph.AnimGraphNode_StateMachine_2.GroundJumpLowerBody.AnimStateTransitionNode_16.AnimationTransitionGraph_0
  nodes [Result(bCanEnterTransition=Break Fort Anim Input Common Vehicle(FortAnimInput_CommonVehicle=Get CommonVehicleInput()) -> (bIsUsingVehicle=~, bIsJumpingVehicle=~, bCanChargeJump=~, bIsChargingJump=~, bIsJumping=~, bIsOnGround=~, bCanDriverAimWeapon=~).bIsOnGround)]
 }
 AnimationTransitionGraph_0 {
  path=AnimGraph.AnimGraphNode_StateMachine_2.GroundJumpLowerBody.AnimStateTransitionNode_2.AnimationTransitionGraph_0
  nodes [Result(bCanEnterTransition=Get bSkipJumpStart())]
 }
 AnimationTransitionGraph_0 {path=AnimGraph.AnimGraphNode_StateMachine_2.GroundJumpLowerBody.AnimStateTransitionNode_20.AnimationTransitionGraph_0, nodes [Result()]}
 AnimationTransitionGraph_0 {
  path=AnimGraph.AnimGraphNode_StateMachine_2.GroundJumpLowerBody.AnimStateTransitionNode_28.AnimationTransitionGraph_0
  nodes [Result(bCanEnterTransition=Break Fort Anim Input Transition Properties(FortAnimInput_TransitionProperties=Get TransitionPropertiesInput()) -> (bTransition_DoubleJump_Fast=~, bTransition_DoubleJump=~).bTransition_DoubleJump)]
 }
 AnimationTransitionGraph_0 {
  path=AnimGraph.AnimGraphNode_StateMachine_2.GroundJumpLowerBody.AnimStateTransitionNode_3.AnimationTransitionGraph_0
  nodes [Result(bCanEnterTransition=Break Fort Anim Input Transition Properties(FortAnimInput_TransitionProperties=Get TransitionPropertiesInput()) -> (bTransition_DoubleJump_Fast=~, bTransition_DoubleJump=~).bTransition_DoubleJump)]
 }
 AnimationTransitionGraph_0 {
  path=AnimGraph.AnimGraphNode_StateMachine_2.GroundJumpLowerBody.AnimStateTransitionNode_4.AnimationTransitionGraph_0
  nodes [Result(bCanEnterTransition=Get bIsJumping())]
 }
 Transition {
  path=AnimGraph.AnimGraphNode_StateMachine_2.GroundJumpLowerBody.AnimStateTransitionNode_5.Transition
  nodes [Result(bCanEnterTransition=Get bIsFalling())]
 }
 AnimationTransitionGraph_0 {
  path=AnimGraph.AnimGraphNode_StateMachine_2.GroundJumpLowerBody.AnimStateTransitionNode_7.AnimationTransitionGraph_0
  nodes [Result(bCanEnterTransition=Get bTransition_NewJumpLoopAdditive_to_NewFallAdditive())]
 }
 EventGraph {links [], nodes []}
 FallAdditive {nodes [
   n1 = Break Fort Anim Input Player Anim Asset!(FortAnimInput_PlayerAnimAsset=Get CurrentAnimSet()) -> (
    FallAdditive
    FullBodyJumpAdditive
    LandingAdditiveOverride
    TurnInPlaceLeftAdditiveOverride
    TurnInPlaceRightAdditiveOverride
    bForceKneeSliding=~
   )
   n2 = Get FallPlayRate() -> (FallPlayRate:float)
   n3 = "Sequence Player (additive) 'M_JumpFall_ARDS_Additive'"!(PlayRate:float=@n2) -> (Pose) {bLoopAnimation=false, sequence=/Game/Animation/Game/MainPlayer/Locomotion/Medium/Male/Jump/AssaultRifle/M_JumpFall_ARDS_Additive}
   n4 = "Sequence Player (additive) 'M_JumpFall_ARDS_Additive'"(PlayRate=@n2) {bLoopAnimation=false, sequence=&M_JumpFall_ARDS_Additive}
   n5 = Get bIsTargeting()
   n6 = Get ADSToNonADSBlendTime() -> (ADSToNonADSBlendTime:float)
   n7 = "Sequence Player (additive) 'Balloons_JumpFall_Base_Additive_CMM'"!(PlayRate:float=Get FallPlayRate()) -> (Pose) {bLoopAnimation=false, sequence=/Game/Animation/Game/MainPlayer/Combat/Gadgets/Medium/Male/Balloons/Balloons_JumpFall_Base_Additive_CMM}
   n8 = "Blend Poses (EFortWeaponCoreAnimation)"!(
    BlendPose_0:Pose Link Structure=Blend Poses by bool(
     BlendPose_0=Additive Identity Pose()
     BlendPose_1=Sequence Player!(Sequence=@n1, PlayRate:float=@n2) -> (Pose) {bLoopAnimation=false}
     bActiveValue=Get bIsDBNO()
    ) {blendType=Linear}
    BlendPose_1:Pose Link Structure=Blend Poses by bool(
     BlendPose_0="Sequence Player (additive) 'M_JumpFall_ARDS_Additive'"(PlayRate=@n2) {bLoopAnimation=false, sequence=&M_JumpFall_ARDS_Additive}
     BlendPose_1=Blend Poses by bool(BlendPose_0=@n3, BlendPose_1=@n4, BlendTime_1=@n6, bActiveValue=@n5) {blendType=Linear}
     BlendTime_0=.2
     BlendTime_1=.2
     bActiveValue=Get bIsCrouching()
    ) {blendType=Linear}
    BlendPose_2:Pose Link Structure=@n7
    BlendTime_0:float=0
    BlendTime_1:float=0
    BlendTime_2:float=0
    ActiveEnumValue:Byte=Get WeaponCoreAnim!() -> (WeaponCoreAnim)
   ) -> (Pose:Pose Link Structure) {
    is=title AnimGraphNode_BlendListByEnum
    blendType=Linear
    blendPose [{linkId=-1, sourceLinkId=-1}, {linkId=-1, sourceLinkId=-1}, {linkId=-1, sourceLinkId=-1}]
    blendTime [0.1,0.1,0.1]
   }
   n9 = Get FallPlayRate()
   Output Pose(Result=Blend Poses by bool(
     BlendPose_0="Sequence Player (additive) 'M_JumpFall_ARDS_Additive'"(PlayRate=@n9) {bLoopAnimation=false, sequence=&M_JumpFall_ARDS_Additive}
     BlendPose_1=@n8
     BlendTime_0=0
     BlendTime_1=0
     bActiveValue=Get bIsDBNOCarrying() -> (bIsDBNOCarrying:bool)
    ) {blendType=Linear}) {layerGroup=None}
  ]}
 ItemAimOffset {nodes [
   Input Pose!() -> (Pose, AimOffsetAlpha:double, Yaw:double, Pitch:double) {name=InPose_AimOffset}
   Input Pose!() -> (Pose) {name=InPose_PreAimOffsets}
   Output Pose() {layerGroup=Item Override}
  ]}
 ItemFallAdditive {nodes [Output Pose(Result=Input Pose() {name=InPoseItemFallAdditive}) {layerGroup=Item Override}]}
 ItemFallLandAdditive {nodes [
   Input Pose() {name=InPoseItemFallLandAdditive}
   Output Pose() {layerGroup=Item Override}
  ]}
 ItemFallLevel2Override {nodes [
   Input Pose() {name=InPose_FallLevel2}
   Output Pose() {layerGroup=Item Override}
  ]}
 ItemFinalPoseOverride {nodes [
   Input Pose() {name=InFinalPose}
   Output Pose() {layerGroup=Item Override}
  ]}
 ItemFullBodyAimOffsets {nodes [
   Input Pose!() -> (Pose, FullBody_AimOffsetAlpha:double, Aim_Yaw:double, Aim_Pitch:double) {name=InPoseFullBodyAimOffsets}
   Input Pose!() -> (Pose) {name=InPosePreFullBodyAimOffsets}
   Output Pose() {layerGroup=Item Override}
  ]}
 ItemFullBodyOverride {nodes [
   Input Pose() {name=InPoseFullBody}
   Output Pose() {layerGroup=Item Override}
  ]}
 ItemIdleOverride {nodes [
   Input Pose() {name=InPose_Idle}
   Output Pose() {layerGroup=Item Override}
  ]}
 ItemJetPackJumpAdditive {nodes [Output Pose() {layerGroup=Item Override}]}
 ItemJetPackStartAdditive {nodes [Output Pose() {layerGroup=Item Override}]}
 ItemJumpLoopAdditive {nodes [Output Pose(Result=Input Pose() {name=InPoseItemJumpLoopAdditive}) {layerGroup=Item Override}]}
 ItemJumpOverride {nodes [
   Input Pose() {name=InPoseJump}
   Output Pose() {layerGroup=Item Override}
  ]}
 ItemJumpUpAdditive {nodes [Output Pose(Result=Input Pose() {name=InPoseItemJumpUpAdditive}) {layerGroup=Item Override}]}
 ItemPreAimOffsets {nodes [
   Input Pose() {name=InPosePreAimOffsets}
   Output Pose() {layerGroup=Item Override}
  ]}
 ItemPreIK {nodes [
   Input Pose() {name=InPose_PreIK}
   Output Pose() {layerGroup=Item Override}
  ]}
 ItemRightHandIK {nodes [
   Input Pose() {name=InPosePreRightHandIK}
   Input Pose() {name=InPosePostRightHandIK}
   Output Pose() {layerGroup=Item Override}
  ]}
 ItemSkeletalControl {nodes [
   Input Pose()
   Output Pose() {layerGroup=Item Override}
  ]}
 ItemSwimJumpFallLoop {nodes [Output Pose() {layerGroup=Item Override}]}
 ItemSwimJumpSurfaceEnd {nodes [Output Pose() {layerGroup=Item Override}]}
 ItemSwimJumpSurfaceLoop {nodes [Output Pose() {layerGroup=Item Override}]}
 ItemUpperBody {nodes [
   Input Pose!() -> (Pose, InputParam:Fort Anim Input Adjusted Aim Structure) {name=InPoseUpperBody}
   Output Pose() {layerGroup=Item Override}
  ]}
 ItemZipLineStartAdditive {nodes [Output Pose() {layerGroup=Item Override}]}
 JumpLoopAdditive {nodes [
   n1 = Break Fort Anim Input Player Anim Asset!(FortAnimInput_PlayerAnimAsset=Get CurrentAnimSet()) -> (
    JumpLoopAdditive:Anim Sequence Object Reference
    FullBodyJumpAdditive
    LandingAdditiveOverride
    TurnInPlaceLeftAdditiveOverride
    TurnInPlaceRightAdditiveOverride
    bForceKneeSliding=~
   )
   n2 = "Blend Poses (EFortWeaponCoreAnimation)"(
    BlendPose_0=Blend Poses by bool(BlendPose_0=Additive Identity Pose(), BlendPose_1=Sequence Player!(Sequence=@n1) -> (Pose) {playRate=0.5}, bActiveValue=Get bIsDBNO()) {blendType=Linear}
    BlendPose_1=Blend Poses by bool(
     BlendPose_0="Sequence Player (additive) 'M_JumpApex_ARDS_Loop_Additive'"() {playRate=0.5, sequence=/Game/Animation/Game/MainPlayer/Locomotion/Medium/Male/Jump/AssaultRifle/M_JumpApex_ARDS_Loop_Additive}
     BlendPose_1="Sequence Player (additive) 'M_JumpApex_ARDS_Loop_Additive'"() {playRate=0.5, sequence=&M_JumpApex_ARDS_Loop_Additive}
     BlendTime_1=Get ADSToNonADSBlendTime()
     bActiveValue=Get bIsTargeting()
    ) {blendType=Linear}
    BlendPose_2="Sequence Player (additive) 'Balloons_JumpApex_Base_Additive_CMM'"() {sequence=/Game/Animation/Game/MainPlayer/Combat/Gadgets/Medium/Male/Balloons/Balloons_JumpApex_Base_Additive_CMM}
    BlendTime_0=0
    BlendTime_1=0
    BlendTime_2=0
    ActiveEnumValue=Get WeaponCoreAnim()
   ) {
    blendType=Linear
    blendPose [{linkId=-1, sourceLinkId=-1}, {linkId=-1, sourceLinkId=-1}, {linkId=-1, sourceLinkId=-1}]
    blendTime [0.1,0.1,0.1]
   }
   Output Pose(Result=Blend Poses by bool(
     BlendPose_0="Sequence Player (additive) 'M_JumpApex_ARDS_Loop_Additive'"() {playRate=0.5, sequence=&M_JumpApex_ARDS_Loop_Additive}
     BlendPose_1=@n2
     BlendTime_0=0
     BlendTime_1=0
     bActiveValue=Get bIsDBNOCarrying()
    ) {blendType=Linear}) {layerGroup=None}
  ]}
 JumpUpAdditive {nodes [
   n1 = Break Fort Anim Input Player Anim Asset!(FortAnimInput_PlayerAnimAsset=Get CurrentAnimSet()) -> (
    JumpUpAdditive:Anim Sequence Object Reference
    FullBodyJumpAdditive
    LandingAdditiveOverride
    TurnInPlaceLeftAdditiveOverride
    TurnInPlaceRightAdditiveOverride
    bForceKneeSliding=~
   )
   n2 = "Sequence Player (additive) 'Balloons_JumpUp_Base_Additive_CMM'"() {bLoopAnimation=false, sequence=/Game/Animation/Game/MainPlayer/Combat/Gadgets/Medium/Male/Balloons/Balloons_JumpUp_Base_Additive_CMM}
   n3 = "Blend Poses (EFortWeaponCoreAnimation)"(
    BlendPose_0=Blend Poses by bool(BlendPose_0=Additive Identity Pose(), BlendPose_1=Sequence Player(Sequence=@n1) {bLoopAnimation=false}, bActiveValue=Get bIsDBNO()) {blendType=Linear}
    BlendPose_1=Blend Poses by bool(
     BlendPose_0="Sequence Player (additive) 'M_JumpUp_ARDS_Additive'"() {bLoopAnimation=false, sequence=/Game/Animation/Game/MainPlayer/Locomotion/Medium/Male/Jump/AssaultRifle/M_JumpUp_ARDS_Additive}
     BlendPose_1="Sequence Player (additive) 'M_JumpUp_ARDS_Additive'"() {bLoopAnimation=false, sequence=&M_JumpUp_ARDS_Additive}
     BlendTime_1=Get ADSToNonADSBlendTime()
     bActiveValue=Get bIsTargeting()
    ) {blendType=Linear}
    BlendPose_2=@n2
    BlendTime_0=0
    BlendTime_1=0
    BlendTime_2=0
    ActiveEnumValue=Get WeaponCoreAnim()
   ) {
    blendType=Linear
    blendPose [{linkId=-1, sourceLinkId=-1}, {linkId=-1, sourceLinkId=-1}, {linkId=-1, sourceLinkId=-1}]
    blendTime [0.1,0.1,0.1]
   }
   Output Pose(Result=Blend Poses by bool(
     BlendPose_0="Sequence Player (additive) 'M_JumpUp_ARDS_Additive'"() {bLoopAnimation=false, sequence=&M_JumpUp_ARDS_Additive}
     BlendPose_1=@n3
     BlendTime_0=0
     BlendTime_1=0
     bActiveValue=Get bIsDBNOCarrying()
    ) {blendType=Linear}) {layerGroup=None}
  ]}
]
variables []
