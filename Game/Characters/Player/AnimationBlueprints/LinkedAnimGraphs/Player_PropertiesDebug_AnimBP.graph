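# Player_PropertiesDebug_AnimBP is a linked-layer animation Blueprint on the Fortnite average male player skeleton, named for debugging player properties.
# Its AnimGraph feeds an Additive Identity Pose straight into the output pose, so it adds no visible motion.
# The event graph is empty and it declares no variables.
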AnimBlueprint /Game/Characters/Player/AnimationBlueprints/LinkedAnimGraphs/Player_PropertiesDebug_AnimBP
target_skeleton=&Fortnite_M_Avg_Player_Skeleton
graphs [
 AnimGraph {nodes [Output Pose(Result=Additive Identity Pose()) {layerGroup=None}]}
 EventGraph {links [], nodes []}
]
variables []
