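# Linked anim Blueprint that animates the player riding ziplines, through a state machine with start, loop, pivot and DBNO states.
# Start and loop stack lean blendspaces onto item-layer poses, with DBNO reachable from the other states through an alias.
# Pivots select lower-body turns by cardinal direction and weapon-equipped state, layered with the cached jetpack-jump item pose.
# Its zipline-start and jetpack-jump item layers choose additives by weapon core animation.
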AnimBlueprint /Game/Characters/Player/AnimationBlueprints/LinkedAnimGraphs/Player_ZipLine_AnimBP
target_skeleton=&Fortnite_M_Avg_Player_Skeleton
graphs [
 AnimGraph {nodes [
   AnimGraphNode_StateMachine_0 = ZipLine() {is=title AnimGraphNode_StateMachine, bAllowConduitEntryStates=true}
   Output Pose(Result=@AnimGraphNode_StateMachine_0) {layerGroup=None}
   n1 = ItemAnimLayerInterface - ItemJetPackJumpAdditive!() -> (Pose:Pose Link Structure) {
    is=title AnimGraphNode_LinkedAnimLayer
    interface=/Game/Characters/Player/Male/Male_Avg_Base/AnimLayerInterface/ItemAnimLayerInterface.ItemAnimLayerInterface_C
    layer=ItemJetPackJumpAdditive
   }
   Save cached pose 'ItemJetPackJumpAdditive_SavedPose'(Pose=@n1) {cacheName=ItemJetPackJumpAdditive_SavedPose}
  ]}
 ZipLine {
  entry=ZipLineStart
  path=AnimGraph.AnimGraphNode_StateMachine_0.ZipLine
  ids=id
  nodes [
   to DBNO() {is=title AnimStateAliasNode, also {aliases [ZipLineLoop,ZipLinePivot,ZipLineStart]}}
   AnimStateNode_0 = "DBNO "() {
    is=title AnimStateNode
    stateEntered {guid=#1}
    stateFullyBlended {guid=#2}
    stateLeft {guid=#3}
   }
   AnimStateNode_12 = ZipLinePivot() {is=title AnimStateNode}
   AnimStateNode_5 = ZipLineStart() {is=title AnimStateNode}
   AnimStateNode_6 = ZipLineLoop() {is=title AnimStateNode}
   AnimStateTransitionNode_0 = AnimStateTransitionNode() {bAutomaticRuleBasedOnSequencePlayerInState=true, crossfadeDuration=0.5}
   AnimStateTransitionNode_1 = AnimStateTransitionNode() {bAutomaticRuleBasedOnSequencePlayerInState=true, crossfadeDuration=0.4}
   AnimStateTransitionNode_2 = AnimStateTransitionNode()
   AnimStateTransitionNode_3 = AnimStateTransitionNode()
   AnimStateTransitionNode_6 = AnimStateTransitionNode() {
    bAllowInertializationForSelfTransitions=true
    transitionEnd {guid=#4}
    transitionInterrupt {guid=#5}
    transitionStart {guid=#6}
   }
  ]
  links [
   {
    dst=ZipLinePivot
    rule=AnimGraph.AnimGraphNode_StateMachine_0.ZipLine.AnimStateTransitionNode_3.Transition
    src=ZipLineLoop
    via=AnimStateTransitionNode_3
   }
   {
    dst=ZipLineLoop
    rule=AnimGraph.AnimGraphNode_StateMachine_0.ZipLine.AnimStateTransitionNode_1.Transition
    src=ZipLinePivot
    via=AnimStateTransitionNode_1
   }
   {
    dst=ZipLineLoop
    rule=AnimGraph.AnimGraphNode_StateMachine_0.ZipLine.AnimStateTransitionNode_2.Transition
    src=ZipLinePivot
    via=AnimStateTransitionNode_2
   }
   {
    dst=ZipLineLoop
    rule=AnimGraph.AnimGraphNode_StateMachine_0.ZipLine.AnimStateTransitionNode_0.Transition
    src=ZipLineStart
    via=AnimStateTransitionNode_0
   }
   {
    dst="DBNO "
    rule=AnimGraph.AnimGraphNode_StateMachine_0.ZipLine.AnimStateTransitionNode_6.Transition
    src=to DBNO
    via=AnimStateTransitionNode_6
   }
  ]
 }
 "DBNO " {
  path="AnimGraph.AnimGraphNode_StateMachine_0.ZipLine.AnimStateNode_0.DBNO "
  nodes [
   n1 = Blendspace Player 'ZipLine_LowerBody_Leans_BS'!(X=Break Fort Anim Input Zipline!(FortAnimInput_Zipline:"Fort Anim Input Zipline Structure (by ref)"=Get ZiplineInput() -> (ZiplineInput:Fort Anim Input Zipline Structure)) -> (
     bIsZiplining:bool=~
     bIsOnAscender:bool=~
     bShouldPlayPivotTransition:bool=~
     LeanYaw:float=~
     PivotBlendDelayRemaining:float=~
     PivotBlendDelay:float=~
     PivotCardinalDirection:EFortCardinalDirection Enum
     WorldVelocityLastTick:Vector=~
    ) {is=title K2Node_BreakStruct}.LeanYaw) -> (Pose) {blendSpace=/Game/Animation/Game/MainPlayer/Combat/Gadgets/Medium/Male/ZipLine/ZipLine_LowerBody_Leans_BS}
   n2 = Local Space Ref Pose()
   Output Animation Pose(Result=Modify Curve!(SourcePose=Apply Additive(Base=@n2, Additive=@n1, Alpha=0.3), CurveValues_0=1, CurveValues_1:float=1, Alpha=1) -> (Pose) {curveNames [BothHandFK,BothLegFK], curveValues [0,0]})
  ]
 }
 ZipLinePivot {
  path=AnimGraph.AnimGraphNode_StateMachine_0.ZipLine.AnimStateNode_12.ZipLinePivot
  nodes [
   n1 = "Sequence Player (additive) 'ZipLine_LowerBody_Pivot_NtoS'"() {
    bLoopAnimation=false
    groupName=PivotNtoS
    method=SyncGroup
    sequence=/Game/Animation/Game/MainPlayer/Combat/Gadgets/Medium/Male/ZipLine/ZipLine_LowerBody_Pivot_NtoS
   }
   n2 = "Sequence Player (additive) 'ZipLine_LowerBody_Pivot_StoN'"() {
    bLoopAnimation=false
    groupName=PivotStoN
    method=SyncGroup
    sequence=/Game/Animation/Game/MainPlayer/Combat/Gadgets/Medium/Male/ZipLine/ZipLine_LowerBody_Pivot_StoN
   }
   n3 = "Sequence Player (additive) 'ZipLine_LowerBody_Pivot_EtoW'"() {
    bLoopAnimation=false
    groupName=PivotEtoW
    method=SyncGroup
    sequence=/Game/Animation/Game/MainPlayer/Combat/Gadgets/Medium/Male/ZipLine/ZipLine_LowerBody_Pivot_EtoW
   }
   n4 = "Sequence Player (additive) 'ZipLine_LowerBody_Pivot_WtoE'"() {
    bLoopAnimation=false
    groupName=PivotWtoE
    method=SyncGroup
    sequence=/Game/Animation/Game/MainPlayer/Combat/Gadgets/Medium/Male/ZipLine/ZipLine_LowerBody_Pivot_WtoE
   }
   n5 = Get PivotAnimBlendTime() -> (PivotAnimBlendTime:double)
   n6 = Break Fort Anim Input Zipline(FortAnimInput_Zipline=Get ZiplineInput()) -> (
    bIsZiplining=~
    bIsOnAscender=~
    bShouldPlayPivotTransition=~
    LeanYaw=~
    PivotBlendDelayRemaining=~
    PivotBlendDelay=~
    WorldVelocityLastTick=~
   )
   n7 = "Blend Poses (EFortCardinalDirection)"!(
    BlendPose_0:Pose Link Structure=@n2
    BlendPose_1:Pose Link Structure=@n4
    BlendPose_2:Pose Link Structure=@n1
    BlendPose_3:Pose Link Structure=@n3
    BlendTime_0:float=@n5
    BlendTime_1:float=@n5
    BlendTime_2:float=@n5
    BlendTime_3:float=@n5
    ActiveEnumValue:Byte=@n6.PivotCardinalDirection
   ) -> (Pose:Pose Link Structure) {
    is=title AnimGraphNode_BlendListByEnum
    blendType=Linear
    blendPose [{linkId=-1, sourceLinkId=-1}, {linkId=-1, sourceLinkId=-1}, {linkId=-1, sourceLinkId=-1}, {linkId=-1, sourceLinkId=-1}]
    blendTime [0.1,0.1,0.1,0.1]
   }
   n8 = "Sequence Player (additive) 'ZipLine_LowerBody_Pivot_NtoS_NoWep'"() {
    bLoopAnimation=false
    groupName=PivotNtoS
    method=SyncGroup
    sequence=/Game/Animation/Game/MainPlayer/Combat/Gadgets/Medium/Male/ZipLine/ZipLine_LowerBody_Pivot_NtoS_NoWep
   }
   n9 = "Sequence Player (additive) 'ZipLine_LowerBody_Pivot_EtoW_NoWep'"() {
    bLoopAnimation=false
    groupName=PivotEtoW
    method=SyncGroup
    sequence=/Game/Animation/Game/MainPlayer/Combat/Gadgets/Medium/Male/ZipLine/ZipLine_LowerBody_Pivot_EtoW_NoWep
   }
   n10 = "Sequence Player (additive) 'ZipLine_LowerBody_Pivot_StoN_NoWep'"() {
    bLoopAnimation=false
    groupName=PivotStoN
    method=SyncGroup
    sequence=/Game/Animation/Game/MainPlayer/Combat/Gadgets/Medium/Male/ZipLine/ZipLine_LowerBody_Pivot_StoN_NoWep
   }
   n11 = "Sequence Player (additive) 'ZipLine_LowerBody_Pivot_WtoE_NoWep'"() {
    bLoopAnimation=false
    groupName=PivotWtoE
    method=SyncGroup
    sequence=/Game/Animation/Game/MainPlayer/Combat/Gadgets/Medium/Male/ZipLine/ZipLine_LowerBody_Pivot_WtoE_NoWep
   }
   n12 = Get PivotAnimBlendTime()
   n13 = "Blend Poses (EFortCardinalDirection)"(
    BlendPose_0=@n10
    BlendPose_1=@n11
    BlendPose_2=@n8
    BlendPose_3=@n9
    BlendTime_0=@n12
    BlendTime_1=@n12
    BlendTime_2=@n12
    BlendTime_3=@n12
    ActiveEnumValue=@n6.PivotCardinalDirection
   ) {
    blendType=Linear
    blendPose [{linkId=-1, sourceLinkId=-1}, {linkId=-1, sourceLinkId=-1}, {linkId=-1, sourceLinkId=-1}, {linkId=-1, sourceLinkId=-1}]
    blendTime [0.1,0.1,0.1,0.1]
   }
   n14 = Layered blend per bone(
    BasePose=Blend Poses by bool(BlendPose_0=@n7, BlendPose_1=@n13, bActiveValue=Get bIsWeaponEquipped() -> (bIsWeaponEquipped:bool)) {blendType=Linear}
    BlendPoses_0=Use cached pose 'ItemJetPackJumpAdditive_SavedPose'()
   ) {layerSetup [{branchFilters [{blendDepth=4, boneName=spine_01}, {boneName=weapon_r}, {boneName=ik_hand_root}]}]}
   Output Animation Pose(Result=@n14)
  ]
 }
 ZipLineStart {
  path=AnimGraph.AnimGraphNode_StateMachine_0.ZipLine.AnimStateNode_5.ZipLineStart
  nodes [
   n1 = Blendspace Player 'ZipLine_LowerBody_Leans_BS'(X=Get LocalVelocityYawAngle() -> (LocalVelocityYawAngle:float)) {bIgnoreForRelevancyTest=true, blendSpace=&ZipLine_LowerBody_Leans_BS}
   n2 = ItemAnimLayerInterface - ItemZipLineStartAdditive!() -> (Pose:Pose Link Structure) {
    is=title AnimGraphNode_LinkedAnimLayer
    interface=&ItemAnimLayerInterface_C
    layer=ItemZipLineStartAdditive
   }
   Output Animation Pose(Result=Apply Additive(Base=@n2, Additive=@n1) {lODThreshold=1, alphaBoolBlend {blendOption=Linear}})
  ]
 }
 ZipLineLoop {
  path=AnimGraph.AnimGraphNode_StateMachine_0.ZipLine.AnimStateNode_6.ZipLineLoop
  nodes [
   n1 = Blendspace Player 'ZipLine_LowerBody_Leans_BS'(X=Break Fort Anim Input Zipline(FortAnimInput_Zipline=Get ZiplineInput()) -> (
     bIsZiplining=~
     bIsOnAscender=~
     bShouldPlayPivotTransition=~
     LeanYaw=~
     PivotBlendDelayRemaining=~
     PivotBlendDelay=~
     WorldVelocityLastTick=~
    ).LeanYaw) {blendSpace=&ZipLine_LowerBody_Leans_BS}
   n2 = Rotate Root Bone!(
    BasePose:Pose Link Structure=Use cached pose 'ItemJetPackJumpAdditive_SavedPose'()
    Pitch:float=0
    Yaw:float=Get DampedMeleeTwistCurveValue() -> (DampedMeleeTwistCurveValue:float)
    MeshToComponent:Rotator="0,-90, 0"
   ) -> (Pose:Pose Link Structure) {is=title AnimGraphNode_RotateRootBone, yawScaleBiasClamp {bClampResult=true, clampMax=60, clampMin=-60, scale=0.75}}
   Output Animation Pose(Result=Apply Additive(
     Base=Apply Additive(Base=@n2, Additive=@n1) {alphaBoolBlend {blendOption=Linear}}
     Additive="Sequence Player (additive) 'ZiplineApex_Loop_CMM_Additive'"() {sequence=/Game/Animation/Game/MainPlayer/Locomotion/Medium/Male/Jump/Zipline/ZiplineApex_Loop_CMM_Additive}
    ) {lODThreshold=1, alphaBoolBlend {blendOption=Linear}})
  ]
 }
 Transition {path=AnimGraph.AnimGraphNode_StateMachine_0.ZipLine.AnimStateTransitionNode_0.Transition, nodes [Result()]}
 Transition {path=AnimGraph.AnimGraphNode_StateMachine_0.ZipLine.AnimStateTransitionNode_1.Transition, nodes [Result()]}
 Transition {
  path=AnimGraph.AnimGraphNode_StateMachine_0.ZipLine.AnimStateTransitionNode_2.Transition
  nodes [
   n1 = Get ZiplineInput()
   Result(bCanEnterTransition=NOT Boolean(A=Break Fort Anim Input Zipline(FortAnimInput_Zipline=@n1) -> (
      bIsZiplining=~
      bIsOnAscender=~
      bShouldPlayPivotTransition=~
      LeanYaw=~
      PivotBlendDelayRemaining=~
      PivotBlendDelay=~
      WorldVelocityLastTick=~
     )))
  ]
 }
 Transition {
  path=AnimGraph.AnimGraphNode_StateMachine_0.ZipLine.AnimStateTransitionNode_3.Transition
  nodes [Result(bCanEnterTransition=Break Fort Anim Input Zipline(FortAnimInput_Zipline=Get ZiplineInput()) -> (
     bIsZiplining=~
     bIsOnAscender=~
     bShouldPlayPivotTransition=~
     LeanYaw=~
     PivotBlendDelayRemaining=~
     PivotBlendDelay=~
     WorldVelocityLastTick=~
    ).bShouldPlayPivotTransition)]
 }
 Transition {path=AnimGraph.AnimGraphNode_StateMachine_0.ZipLine.AnimStateTransitionNode_6.Transition, nodes [Result(bCanEnterTransition=Get bIsDBNO())]}
 EventGraph {links [], nodes []}
 ItemAimOffset {nodes [
   Input Pose!() -> (Pose, AimOffsetAlpha:double, Yaw:double, Pitch:double) {name=InPose_AimOffset}
   Input Pose!() -> (Pose) {name=InPose_PreAimOffsets}
   Output Pose() {layerGroup=Item Override}
  ]}
 ItemFallAdditive {nodes [
   Input Pose() {name=InPoseItemFallAdditive}
   Output Pose() {layerGroup=Item Override}
  ]}
 ItemFallLandAdditive {nodes [
   Input Pose() {name=InPoseItemFallLandAdditive}
   Output Pose() {layerGroup=Item Override}
  ]}
 ItemFallLevel2Override {nodes [
   Input Pose() {name=InPose_FallLevel2}
   Output Pose() {layerGroup=Item Override}
  ]}
 ItemFinalPoseOverride {nodes [
   Input Pose() {name=InFinalPose}
   Output Pose() {layerGroup=Item Override}
  ]}
 ItemFullBodyAimOffsets {nodes [
   Input Pose!() -> (Pose, FullBody_AimOffsetAlpha:double, Aim_Yaw:double, Aim_Pitch:double) {name=InPoseFullBodyAimOffsets}
   Input Pose!() -> (Pose) {name=InPosePreFullBodyAimOffsets}
   Output Pose() {layerGroup=Item Override}
  ]}
 ItemFullBodyOverride {nodes [
   Input Pose() {name=InPoseFullBody}
   Output Pose() {layerGroup=Item Override}
  ]}
 ItemIdleOverride {nodes [
   Input Pose() {name=InPose_Idle}
   Output Pose() {layerGroup=Item Override}
  ]}
 ItemJetPackJumpAdditive {nodes [
   n1 = Get CurrentAnimSet() -> (CurrentAnimSet:Fort Anim Input Player Anim Asset Structure)
   n2 = "Blend Poses (EFortWeaponCoreAnimation)"!(
    BlendPose_0:Pose Link Structure=Sequence Player!(Sequence:Anim Sequence Base Object Reference=Break Fort Anim Input Player Anim Asset!(FortAnimInput_PlayerAnimAsset:"Fort Anim Input Player Anim Asset Structure (by ref)"=@n1) -> (
      JetPackJumpAdditve:Anim Sequence Object Reference
      FullBodyJumpAdditive:Anim Sequence Object Reference
      LandingAdditiveOverride:Anim Sequence Object Reference
      TurnInPlaceLeftAdditiveOverride:Anim Sequence Object Reference
      TurnInPlaceRightAdditiveOverride:Anim Sequence Object Reference
      bForceKneeSliding:bool=~
     ) {is=title K2Node_BreakStruct}) -> (Pose) {bLoopAnimation=false}
    BlendPose_1:Pose Link Structure=Blend Poses by bool(
     BlendPose_0="Sequence Player (additive) 'JetpackApex_ARDS_Loop_Additive'"() {sequence=/Game/Animation/Game/MainPlayer/Locomotion/Medium/Male/Jump/AssaultRifle/JetpackApex_ARDS_Loop_Additive}
     BlendPose_1="Sequence Player (additive) 'JetpackApex_ARDS_Loop_Additive'"() {sequence=&JetpackApex_ARDS_Loop_Additive}
     BlendTime_1=Get ADSToNonADSBlendTime() -> (ADSToNonADSBlendTime:float)
     bActiveValue=Get bIsTargeting()
    ) {blendType=Linear}
    BlendTime_0:float=0
    BlendTime_1:float=0
    ActiveEnumValue:Byte=Get WeaponCoreAnim!() -> (WeaponCoreAnim)
   ) -> (Pose:Pose Link Structure) {
    is=title AnimGraphNode_BlendListByEnum
    blendType=Linear
    blendPose [{linkId=-1, sourceLinkId=-1}, {linkId=-1, sourceLinkId=-1}]
    blendTime [0.1,0.1]
   }
   Output Pose(Result=@n2) {layerGroup=Item Override}
  ]}
 ItemJetPackStartAdditive {nodes [Output Pose() {layerGroup=Item Override}]}
 ItemJumpLoopAdditive {nodes [
   Input Pose() {name=InPoseItemJumpLoopAdditive}
   Output Pose() {layerGroup=Item Override}
  ]}
 ItemJumpOverride {nodes [
   Input Pose() {name=InPoseJump}
   Output Pose() {layerGroup=Item Override}
  ]}
 ItemJumpUpAdditive {nodes [
   Input Pose() {name=InPoseItemJumpUpAdditive}
   Output Pose() {layerGroup=Item Override}
  ]}
 ItemPreAimOffsets {nodes [
   Input Pose() {name=InPosePreAimOffsets}
   Output Pose() {layerGroup=Item Override}
  ]}
 ItemPreIK {nodes [
   Input Pose() {name=InPose_PreIK}
   Output Pose() {layerGroup=Item Override}
  ]}
 ItemRightHandIK {nodes [
   Input Pose() {name=InPosePreRightHandIK}
   Input Pose() {name=InPosePostRightHandIK}
   Output Pose() {layerGroup=Item Override}
  ]}
 ItemSkeletalControl {nodes [
   Input Pose()
   Output Pose() {layerGroup=Item Override}
  ]}
 ItemSwimJumpFallLoop {nodes [Output Pose() {layerGroup=Item Override}]}
 ItemSwimJumpSurfaceEnd {nodes [Output Pose() {layerGroup=Item Override}]}
 ItemSwimJumpSurfaceLoop {nodes [Output Pose() {layerGroup=Item Override}]}
 ItemUpperBody {nodes [
   Input Pose!() -> (Pose, InputParam:Fort Anim Input Adjusted Aim Structure) {name=InPoseUpperBody}
   Output Pose() {layerGroup=Item Override}
  ]}
 ItemZipLineStartAdditive {nodes [
   n1 = Get CurrentAnimSet()
   n2 = "Blend Poses (EFortWeaponCoreAnimation)"(
    BlendPose_0=Sequence Player(Sequence=Break Fort Anim Input Player Anim Asset!(FortAnimInput_PlayerAnimAsset=@n1) -> (
      ZipLineStartAdditve:Anim Sequence Object Reference
      FullBodyJumpAdditive
      LandingAdditiveOverride
      TurnInPlaceLeftAdditiveOverride
      TurnInPlaceRightAdditiveOverride
      bForceKneeSliding=~
     )) {bLoopAnimation=false}
    BlendPose_1=Blend Poses by bool(
     BlendPose_0="Sequence Player (additive) 'ZipLine_Entry_ARDS_Additive'"() {bLoopAnimation=false, sequence=/Game/Animation/Game/MainPlayer/Combat/Gadgets/Medium/Male/ZipLine/ZipLine_Entry_ARDS_Additive}
     BlendPose_1="Sequence Player (additive) 'ZipLine_Entry_ARDS_Additive'"() {bLoopAnimation=false, sequence=&ZipLine_Entry_ARDS_Additive}
     BlendTime_1=Get ADSToNonADSBlendTime()
     bActiveValue=Get bIsTargeting()
    ) {blendType=Linear}
    BlendTime_0=0
    BlendTime_1=0
    ActiveEnumValue=Get WeaponCoreAnim()
   ) {blendType=Linear, blendPose [{linkId=-1, sourceLinkId=-1}, {linkId=-1, sourceLinkId=-1}], blendTime [0.1,0.1]}
   Output Pose(Result=@n2) {layerGroup=Item Override}
  ]}
]
variables [PivotAnimBlendTime {type=number, default=0.1}]
